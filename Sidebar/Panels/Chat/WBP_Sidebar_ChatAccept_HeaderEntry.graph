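# Sidebar chat list header entry for a group chat invite, with Accept, Decline and Report buttons.
# Accept and Decline send responses through the conversation viewmodel; WasAcceptClicked selects the spinner and whether success pushes the message list or pops navigation.
# Helper functions derive success and failure from async results, showing errors and toggling Accept; Report opens a modal.
# Selection activates the widget, with custom navigation and last-focused-button restoration.
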WidgetBlueprint /Sidebar/Panels/Chat/WBP_Sidebar_ChatAccept_HeaderEntry
graphs [
 BP_GetDesiredFocusTarget {nodes [
   Get Desired Focus Target()
   > n1 = Get!(execute:Exec) -> (then:Exec, else:Exec, LastFocusedButton:User Widget Object Reference)
   > Return Node!(execute, ReturnValue:Widget Object Reference=@n1)
   n2 = Branch(
    execute=@n1.else
    Condition=Get bIsEnabled!(self:Widget Object Reference=Get Button_AcceptInvite() -> (Button_AcceptInvite:WBP Sidebar Button Generic Object Reference)) -> (bIsEnabled:bool)
   )
   > Return Node(ReturnValue=Get Button_AcceptInvite())
   Return Node(execute=@n2.else, ReturnValue=Get Button_DeclineInvite() -> (Button_DeclineInvite:WBP Sidebar Button Generic Object Reference))
  ]}
 ConvertJoinedMembersToSocialUserForAvatar {nodes [
   n1 = ConvertJoinedMembersToSocialUserForAvatar!() -> (then, JoinedMembers:Array of Fort Social User Viewmodel Object References) {is=title K2Node_FunctionEntry}
   > Clear(TargetArray:Array of Fort Social User Object References=Get JoinedMembersSocialUsers() -> (JoinedMembersSocialUsers:Array of Fort Social User Object References))
   > n2 = For Each Loop(Array:Array of Fort Social User Viewmodel Object References=@n1) -> (Array Element:Fort Social User Viewmodel Object Reference)
   > Add(
    TargetArray:Array of Fort Social User Object References=Get JoinedMembersSocialUsers()
    NewItem:"Fort Social User Object Reference (by ref)"=GetSocialUser!(self:Fort Social User Viewmodel Object Reference=@n2) -> (ReturnValue:Fort Social User Object Reference)
   ) {is=title K2Node_CallArrayFunction}
   Return Node!(execute=@n2.Completed, SocialUsers:Array of Fort Social User Object References=Get JoinedMembersSocialUsers())
  ]}
 CreateReportModal {nodes [
   CreateReportModal() {is=title K2Node_FunctionEntry}
   > n1 = Create Widget!(
    execute:Exec
    Class:User Widget Class Reference=/FortniteCoreUI/UI/Social/Chat/TextChat/Dialogs/WBP_Dialog_VM_TextChat_Report.WBP_Dialog_VM_TextChat_Report_C
    OwningPlayer:Player Controller Object Reference
   ) -> (then:Exec, ReturnValue:WBP Dialog VM Text Chat Report Object Reference)
   > SetReportedConversationVM!(
    execute
    self:WBP Dialog VM Text Chat Report Object Reference=@n1
    ReportedConversationVM:Fort Text Chat Conversation VM Object Reference=Get FortTextChatGroupConversationVM() -> (FortTextChatGroupConversationVM:Fort Group Conversation Viewmodel Object Reference)
   ) -> (then)
   > TryPushWidgetAsModalPanel!(execute, PushWidget:User Widget Object Reference=@n1) -> (then)
   > AddTagToInputMode!(
    execute
    self:Enhanced Input Subsystem Interface Interface=Get EnhancedInputLocalPlayerSubsystem() -> (ReturnValue:Enhanced Input Local Player Subsystem Object Reference) {is=title K2Node_GetSubsystem}
    TagToAdd:"Gameplay Tag Structure (by ref)"=MakeLiteralGameplayTag!(Value:Gameplay Tag Structure=(TagName="UI.TextChat.Disabled")) -> (ReturnValue:Gameplay Tag Structure)
    Options:"Modify Context Options Structure (by ref)"=(bIgnoreAllPressedKeysUntilRelease=True,bForceImmediately=False,bNotifyUserSettings=False)
   ) -> (then)
  ]}
 EventGraph {nodes [
   "On Clicked (Button_DeclineInvite)"!() -> (OutputDelegate, then, Button:Common Button Base Object Reference) {is=title K2Node_ComponentBoundEvent}
   > Set WasAcceptClicked(WasAcceptClicked:bool=false) -> (Output_Get:bool)
   > RequestRespondToInvite!(execute, self:Fort Group Conversation Viewmodel Object Reference=Get FortTextChatGroupConversationVM(), bAcceptInvite:bool=false) -> (then)
   "On Clicked (Button_AcceptInvite)"!() -> (OutputDelegate, then, Button:Common Button Base Object Reference) {is=title K2Node_ComponentBoundEvent}
   > Set WasAcceptClicked(WasAcceptClicked=true)
   > RequestRespondToInvite(self=Get FortTextChatGroupConversationVM(), bAcceptInvite=true)
   "On Clicked (Button_ReportInvite)"!() -> (OutputDelegate, then, Button:Common Button Base Object Reference) {is=title K2Node_ComponentBoundEvent}
   > CreateReportModal() {is=title K2Node_CallFunction}
   Event PreConstruct()
   > "SetText (Text)"(self=Get ErrorMessage() -> (ErrorMessage:Common Text Object Reference))
   n1 = Event OnListItemObjectSet()
   > n2 = Cast To FortTextChatGroupConversationVM!(execute, Object=@n1.ListItemObject) -> (then, CastFailed, AsFort Group Conversation Viewmodel:Fort Group Conversation Viewmodel Object Reference) {is=title K2Node_DynamicCast}
   > Set FortTextChatGroupConversationVM(FortTextChatGroupConversationVM:Fort Group Conversation Viewmodel Object Reference=@n2) -> (Output_Get:Fort Group Conversation Viewmodel Object Reference)
   > SetDividerVisibility()
   n3 = Event On Item Selection Changed!() -> (OutputDelegate, then, bIsSelected:bool) {is=title K2Node_Event}
   > n4 = Branch(Condition=@n3.bIsSelected)
   > ActivateWidget()
   DeactivateWidget(execute=@n4.else)
  ]}
 GetShowFailureMessage {nodes [
   n1 = GetShowFailureMessage!() -> (then, InProgress:bool=~, actionresult:ELastAsyncActionResult Enum=None, lastResultMsg:Text) {is=title K2Node_FunctionEntry}
   n2 = TextIsEmpty(InText=@n1.lastResultMsg)
   > n3 = Branch(Condition=AND Boolean!(
     A=NOT Boolean(A=@n1)
     B="Equal (Enum)"(A:ELastAsyncActionResult Enum=@n1.actionresult, B:ELastAsyncActionResult Enum=Failed)
     C:bool=NOT Boolean(A=@n2)
    ) -> (ReturnValue))
   > Return Node!(execute, MessageText:Text=Get lastResultMsg() -> (lastResultMsg:Text))
   Return Node(execute=@n3.else)
  ]}
 GetSuccessResult {nodes [
   n1 = GetSuccessResult!() -> (then, InProgress:bool=~, ActionResult:ELastAsyncActionResult Enum=None) {is=title K2Node_FunctionEntry}
   > n2 = Branch(Condition=AND Boolean!(A=NOT Boolean(A=@n1), B="Equal (Enum)"(A=@n1.ActionResult, B=Success)) -> (ReturnValue))
   > Return Node!(execute, ShouldExecuteSuccessAction:bool=true)
   Return Node(execute=@n2.else, ShouldExecuteSuccessAction=false)
  ]}
 OnNavDownFromAccept {nodes [
   OnNavDownFromAccept!() -> (then, Navigation:EUINavigation Enum) {is=title K2Node_FunctionEntry}
   > Return Node!(execute, ReturnValue:Widget Object Reference=Get Button_DeclineInvite())
  ]}
 OnNavDownFromDecline {nodes [
   OnNavDownFromDecline!() -> (then, Navigation:EUINavigation Enum) {is=title K2Node_FunctionEntry}
   > Return Node(ReturnValue=Get Button_ReportInvite() -> (Button_ReportInvite:WBP Sidebar Button Generic Object Reference))
  ]}
 OnNavDownFromRestriction {nodes [
   OnNavDownFromRestriction!() -> (then, Navigation:EUINavigation Enum) {is=title K2Node_FunctionEntry}
   > Return Node(ReturnValue=Get Button_DeclineInvite())
  ]}
 OnNavUpFromAccept {nodes [
   OnNavUpFromAccept!() -> (then, Navigation:EUINavigation Enum) {is=title K2Node_FunctionEntry}
   n1 = Get WBP_Sidebar_TextChat_Restrictions() -> (WBP_Sidebar_TextChat_Restrictions:WBP Sidebar Text Chat Restrictions Object Reference)
   > n2 = Branch(Condition=IsVisible(self=Get Button_Restriction!(self:WBP Sidebar Text Chat Restrictions Object Reference=@n1) -> (Button_Restriction:WBP Sidebar Button Generic Object Reference)))
   > Return Node(ReturnValue=Get Button_Restriction(self=Get WBP_Sidebar_TextChat_Restrictions()))
   Return Node(execute=@n2.else, ReturnValue=Get Button_AcceptInvite())
  ]}
 OnNavUpFromDecline {nodes [
   OnNavUpFromDecline!() -> (then, Navigation:EUINavigation Enum) {is=title K2Node_FunctionEntry}
   n1 = Get WBP_Sidebar_TextChat_Restrictions()
   > n2 = Branch(Condition=IsVisible(self=Get Button_Restriction(self=@n1)))
   > Return Node(ReturnValue=Get Button_Restriction(self=Get WBP_Sidebar_TextChat_Restrictions()))
   n3 = Get Button_AcceptInvite()
   Branch(execute=@n2.else, Condition=Get bIsEnabled(self=@n3))
   > Return Node(ReturnValue=@n3)
  ]}
 OnNavUpFromReport {nodes [
   OnNavUpFromReport!() -> (then, Navigation:EUINavigation Enum) {is=title K2Node_FunctionEntry}
   > Return Node(ReturnValue=Get Button_DeclineInvite())
  ]}
 OnNavigationSaveLastFocus {nodes [
   n1 = OnNavigationSaveLastFocus!() -> (then, Type:EUINavigation Enum, OldFocusedWidget:Widget Object Reference, NewFocusedWidget:Widget Object Reference) {is=title K2Node_FunctionEntry}
   n2 = GetOuterUserWidget!(Widget:Widget Object Reference=@n1.NewFocusedWidget) -> (ReturnValue:User Widget Object Reference)
   > Cast To WBP_Sidebar_ButtonGeneric!(execute, Object=@n2) -> (then, CastFailed, AsWBP Sidebar Button Generic:WBP Sidebar Button Generic Object Reference) {is=title K2Node_DynamicCast}
   > Set LastFocusedButton(LastFocusedButton:User Widget Object Reference=@n2) -> (Output_Get:User Widget Object Reference)
  ]}
 ResetConversationLimitVisuals {nodes [
   n1 = ResetConversationLimitVisuals!() -> (then, Loading:bool=~) {is=title K2Node_FunctionEntry}
   > n2 = Branch(Condition=@n1)
   SetIsEnabled(execute=@n2.else, self=Get Button_AcceptInvite(), bInIsEnabled=true)
   > SetVisibility(self=Get HB_Error() -> (HB_Error:Horizontal Box Object Reference), InVisibility=Collapsed)
  ]}
 SetConversationLimitExceededVisuals {nodes [
   n1 = SetConversationLimitExceededVisuals!() -> (then, Last Result Message:Text) {is=title K2Node_FunctionEntry}
   > Set Text(Text=@n1, self=Get ErrorMessage())
   n2 = TextIsEmpty(InText=@n1)
   > SetVisibility(self=Get HB_Error(), InVisibility=Select(Option 0=SelfHitTestInvisible, Option 1=Collapsed, Index=@n2))
   > SetIsEnabled(self=Get Button_AcceptInvite(), bInIsEnabled=@n2)
   > RequestRefreshFocus(self:Common Activatable Widget Object Reference)
  ]}
 SetDividerVisibility {nodes [
   SetDividerVisibility() {is=title K2Node_FunctionEntry}
   n1 = Get FortTextChatGroupConversationVM()
   n2 = Is Empty!(TargetArray:Array of Fort Social User Viewmodel Object References=Get InvitedMembers!(self:Fort Text Chat Conversation VM Object Reference=@n1) -> (InvitedMembers:Array of Fort Social User Viewmodel Object References)) -> (ReturnValue:bool)
   n3 = Is Empty(TargetArray=Get JoinedMembers!(self:Fort Text Chat Conversation VM Object Reference=@n1) -> (JoinedMembers:Array of Fort Social User Viewmodel Object References))
   > SetVisibility(
    self=Get Divider() -> (Divider:Image Object Reference)
    InVisibility=Select(Option 0=SelfHitTestInvisible, Option 1=Collapsed, Index=AND Boolean(A=@n2, B=@n3))
   )
  ]}
 SetInviterUserInfo {nodes [
   SetInviterUserInfo!() -> (then, InviterUser:Fort Social User Object Reference) {is=title K2Node_FunctionEntry}
   > n1 = Get!(execute) -> (then, else, InviterUser:Fort Social User Object Reference)
   > "SetText (Text)"(
    self=Get Text_GroupInvite() -> (Text_GroupInvite:Common Text Object Reference)
    InText=Format Text!(
     Format="LOCTEXT(\"{name} invited you to the group\")"
     name:String=GetNicknameOrDisplayName!(self:Fort Social User Object Reference=@n1) -> (ReturnValue:String)
    ) -> (Result)
   )
  ]}
 SetSpinnerVisibility {nodes [
   n1 = SetSpinnerVisibility!() -> (then, IsActionInProgress:bool=~) {is=title K2Node_FunctionEntry}
   > SetVisibility(
    self=[
     Get TEMP_DeclineSpinner() -> (TEMP_DeclineSpinner:WBP UIKit Throbber Object Reference)
     Get TEMP_AcceptSpinner() -> (TEMP_AcceptSpinner:WBP UIKit Throbber Object Reference)
    ]
    InVisibility=Collapsed
   )
   > Branch(Condition=@n1)
   > n2 = Branch(Condition=Get WasAcceptClicked() -> (WasAcceptClicked:bool))
   > SetVisibility(self=Get TEMP_AcceptSpinner(), InVisibility=SelfHitTestInvisible)
   SetVisibility(execute=@n2.else, self=Get TEMP_DeclineSpinner(), InVisibility=SelfHitTestInvisible)
  ]}
 SetSuccessAction {nodes [
   n1 = SetSuccessAction!() -> (then, ShouldExecuteSuccessAction:bool=~) {is=title K2Node_FunctionEntry}
   > Branch(Condition=@n1)
   > n2 = Branch(Condition=Get WasAcceptClicked())
   > n3 = Construct Object from Class!(
    execute
    Class=/Sidebar/Shared/Navigation/BP_NavigationPoint.BP_NavigationPoint_C
    self
    SocialUser:Fort Social User Object Reference
    ConversationId:Name=Get ConversationId!(self:Fort Text Chat Conversation VM Object Reference=Get FortTextChatGroupConversationVM()) -> (ConversationId:Name=None)
    ConfirmationVM:Sidebar Confirmation Panel VM Object Reference
    ReservedSpots:int=0
    SocialItemVM:Fort Social Item VM Object Reference
    NavigationTag:Gameplay Tag Structure=(TagName="UI.StateTree.TransitionTag.Sidebar.Chat.MessageList")
   ) -> (then, ReturnValue:BP Navigation Point Object Reference)
   > PushNavigation!(
    execute
    self:Fort Sidebar Navigation VM Object Reference=Get FortSidebarNavigationVM() -> (FortSidebarNavigationVM:Fort Sidebar Navigation VM Object Reference)
    NavigationPoint:Fort Sidebar Navigation Point Object Reference=@n3
   ) -> (then)
   PopNavigation(execute=@n2.else, self:Fort Sidebar Navigation VM Object Reference=Get FortSidebarNavigationVM())
   > SetVisibility(self=Get HB_Error(), InVisibility=Collapsed)
   n4 = Get WBP_Sidebar_TextChat_Restrictions()
   > SetIsEnabled(
    self=Get Button_AcceptInvite()
    bInIsEnabled=NOT Boolean(A=Get NotifyIsRestrictionWidgetVisible!(self:WBP Sidebar Text Chat Restrictions Object Reference=@n4) -> (NotifyIsRestrictionWidgetVisible:bool))
   )
  ]}
]
variables [
 LastFocusedButton {type=UserWidget, default=null}
 WasAcceptClicked {type=boolean}
]
widgets {
 info {
  inheritedWidgetCount=0
  namedSlotCount=2
  parentClass=&CommonActivatableWidget
  rootWidgetClass=&VerticalBox
  widgetCount=15
 }
 widgets [
  VerticalBox_15 {
   widgetClassPath=&VerticalBox
   widget=VerticalBox_15
   uIComponents [{
     component=&self:UIComponentWidgetBlueprintExtension_0.ComponentContainer.NavigationUIComponent_0
     componentClassPath=/Script/UMG.NavigationUIComponent
    }]
  }
  WBP_Profile_Header {
   widgetClassPath=/Sidebar/Panels/Social/Profile/WBP_Profile_Header
   bIsVariable=true
   parent=VerticalBox_15
   widget=WBP_Profile_Header
   properties {showPlatformNames=false}
   slot {class=VerticalBoxSlot, properties {padding=<ltrb 18 72 18 0>}}
  }
  Text_GroupInvite {
   widgetClassPath=&CommonTextBlock
   bIsVariable=true
   parent=VerticalBox_15
   widget=Text_GroupInvite
   properties {
    autoWrapText=true
    bAutoCollapseWithEmptyText=true
    bIsScrollingEnabled=false
    justification=Center
    mobileFontSizeMultiplier=1.2
    scrollStyle=/FortUILibrary/Styles/Typography/Foundation/TextScrollStyle-NoFade.TextScrollStyle-NoFade_C
    text="{name} invited you to the group"
    textOverflowPolicy=Ellipsis
    colorAndOpacity {colorUseRule=UseColor_Specified, specifiedColor=<rgba 0.9301 1 0.0103 1>}
    font {
     bForceMonospaced=false
     bMaterialIsStencil=false
     fontMaterial=null
     fontObject=&HeadingNow
     letterSpacing=0
     monospacedWidth=1
     size=15.75
     skewAmount=0
     typefaceFontName=75 Medium
     outlineSettings {
      bApplyOutlineToDropShadows=false
      bMiteredCorners=false
      bSeparateFillAlpha=false
      outlineMaterial=null
      outlineSize=0
      outlineColor=<rgba 0 0 0 1>
     }
    }
    shadowOffset=<0 0>
   }
   slot {class=VerticalBoxSlot, properties {padding=<ltrb 18 0 18 40>}}
  }
  HB_Error {
   widgetClassPath=&HorizontalBox
   bIsVariable=true
   parent=VerticalBox_15
   widget=HB_Error
   properties {visibility=Collapsed}
   slot {class=VerticalBoxSlot, properties {padding=<ltrb 18 0 18 8>}}
  }
  WarningIcon {
   widgetClassPath=&Image
   bIsVariable=true
   parent=HB_Error
   widget=WarningIcon
   properties {
    visibility=SelfHitTestInvisible
    brush {
     bIsDynamicallyLoaded=false
     drawAs=Image
     imageType=FullColor
     mirroring=NoMirror
     resourceName=None
     resourceObject=/FortUILibrary/Textures/MaterialInstances/MI_UIKit_Icon_AgeGatedAlert
     tiling=NoTile
     imageSize=<24 24>
     margin=<ltrb 0 0 0 0>
     outlineSettings {
      bUseBrushTransparency=false
      roundingType=HalfHeightRadius
      width=0
      color {colorUseRule=UseColor_Specified, specifiedColor=<rgba 0 0 0 0>}
      cornerRadii=<0 0 0 0>
     }
     tintColor {colorUseRule=UseColor_Specified, specifiedColor=<rgba 1 1 1 1>}
     uVRegion {bIsValid=false, max=<0 0>, min=<0 0>}
    }
    colorAndOpacity=<rgba 1 0.147 0.1329 1>
   }
   slot {class=HorizontalBoxSlot, properties {horizontalAlignment=HAlign_Left, verticalAlignment=VAlign_Center, padding=<ltrb 0 2 12 6>}}
  }
  ErrorMessage {
   widgetClassPath=&CommonTextBlock
   bIsVariable=true
   parent=HB_Error
   widget=ErrorMessage
   properties {
    autoWrapText=true
    bAutoCollapseWithEmptyText=true
    bIsScrollingEnabled=false
    mobileFontSizeMultiplier=1.2
    scrollStyle=&TextScrollStyle-NoFade_C
    text=Max 200 groups reached. Leave one to create a new group.
    textOverflowPolicy=Ellipsis
    colorAndOpacity {colorUseRule=UseColor_Specified, specifiedColor=<rgba 1 0.147 0.1329 1>}
    font {
     bForceMonospaced=false
     bMaterialIsStencil=false
     fontMaterial=null
     fontObject=&HeadingNow
     letterSpacing=0
     monospacedWidth=1
     size=13.5
     skewAmount=0
     typefaceFontName=74 Regular
     outlineSettings {
      bApplyOutlineToDropShadows=false
      bMiteredCorners=false
      bSeparateFillAlpha=false
      outlineMaterial=null
      outlineSize=0
      outlineColor=<rgba 0 0 0 1>
     }
    }
    shadowOffset=<0 0>
   }
   slot {class=HorizontalBoxSlot, properties {size {sizeRule=Fill, value=1}}}
  }
  WBP_Sidebar_TextChat_Restrictions {
   widgetClassPath=&/WBP_Sidebar_TextChat_Restrictions
   bIsVariable=true
   parent=VerticalBox_15
   widget=WBP_Sidebar_TextChat_Restrictions
   properties {
    navigation=&self:WidgetTree.WBP_Sidebar_TextChat_Restrictions.WidgetNavigation_0
    restrictionWidgetPadding=<ltrb 18 0 18 8>
   }
   slot {class=VerticalBoxSlot, properties {verticalAlignment=VAlign_Top}}
  }
  Button_AcceptInvite {
   widgetClassPath=/Sidebar/Buttons/WBP_Sidebar_ButtonGeneric
   bIsVariable=true
   parent=VerticalBox_15
   widget=Button_AcceptInvite
   properties {
    minHeightOverride=60
    navigation=&self:WidgetTree.Button_AcceptInvite.WidgetNavigation_0
    text=Accept
   }
   slot {class=VerticalBoxSlot, properties {padding=<ltrb 18 18 18 12>}}
  }
  Overlay_ProgressAccept {widgetClassPath=&Overlay, namedSlotHost=Button_AcceptInvite, widget=Overlay_ProgressAccept}
  TEMP_AcceptSpinner {
   widgetClassPath=/FortUILibrary/UIKit/Throbber/WBP_UIKit_Throbber
   bIsVariable=true
   parent=Overlay_ProgressAccept
   widget=TEMP_AcceptSpinner
   properties {visibility=Collapsed}
   slot {class=OverlaySlot, properties {verticalAlignment=VAlign_Center, padding=<ltrb 8 0 0 0>}}
  }
  Button_DeclineInvite {
   widgetClassPath=&WBP_Sidebar_ButtonGeneric
   bIsVariable=true
   parent=VerticalBox_15
   widget=Button_DeclineInvite
   properties {
    minHeightOverride=60
    navigation=&self:WidgetTree.Button_DeclineInvite.WidgetNavigation_0
    secondText=Error Message
    text=Decline
    secondFont {
     bForceMonospaced=false
     bMaterialIsStencil=false
     fontMaterial=/FortUILibrary/UIKit/Buttons/MaterialInstances/CTAButtons/MI_UIKit_TextColor_RegularSecondary
     fontObject=&HeadingNow
     letterSpacing=48
     monospacedWidth=1
     size=13.5
     skewAmount=0
     typefaceFontName=75 Medium
     outlineSettings {
      bApplyOutlineToDropShadows=false
      bMiteredCorners=false
      bSeparateFillAlpha=false
      outlineMaterial=null
      outlineSize=0
      outlineColor=<rgba 0 0 0 1>
     }
    }
    secondTextPadding=<ltrb 0 0 0 0>
   }
   slot {class=VerticalBoxSlot, properties {padding=<ltrb 18 0 18 12>}}
  }
  Overlay_ProgressDecline {widgetClassPath=&Overlay, namedSlotHost=Button_DeclineInvite, widget=Overlay_ProgressDecline}
  TEMP_DeclineSpinner {
   widgetClassPath=&WBP_UIKit_Throbber
   bIsVariable=true
   parent=Overlay_ProgressDecline
   widget=TEMP_DeclineSpinner
   properties {visibility=Collapsed}
   slot {class=OverlaySlot, properties {verticalAlignment=VAlign_Center, padding=<ltrb 8 0 0 0>}}
  }
  Button_ReportInvite {
   widgetClassPath=&WBP_Sidebar_ButtonGeneric
   bIsVariable=true
   parent=VerticalBox_15
   widget=Button_ReportInvite
   properties {
    buttonType=Ghost
    minHeightOverride=60
    navigation=&self:WidgetTree.Button_ReportInvite.WidgetNavigation_0
    text=Report invite
   }
   slot {class=VerticalBoxSlot, properties {padding=<ltrb 18 0 18 18>}}
  }
  Divider {
   widgetClassPath=&Image
   bIsVariable=true
   parent=VerticalBox_15
   widget=Divider
   properties {brush {
     bIsDynamicallyLoaded=false
     drawAs=Image
     imageType=FullColor
     mirroring=NoMirror
     resourceName=None
     resourceObject=/Sidebar/Materials/MI_Sidebar_Divider
     tiling=NoTile
     imageSize=<2 2>
     margin=<ltrb 0 0 0 0>
     outlineSettings {
      bUseBrushTransparency=false
      roundingType=HalfHeightRadius
      width=0
      color {colorUseRule=UseColor_Specified, specifiedColor=<rgba 0 0 0 0>}
      cornerRadii=<0 0 0 0>
     }
     tintColor {colorUseRule=UseColor_Specified, specifiedColor=<rgba 1 1 1 1>}
     uVRegion {bIsValid=false, max=<0 0>, min=<0 0>}
    }}
   slot {class=VerticalBoxSlot, properties {padding=<ltrb 0 36 0 18>}}
  }
 ]
}
