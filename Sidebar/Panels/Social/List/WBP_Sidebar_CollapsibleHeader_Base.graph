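# Base sidebar button for collapsible section headers in social list panels, showing a title, icon, member count and arrow.
# As a tree list entry, it rotates the arrow on expansion changes, sounds on expand, and snaps to the collapsed pose when bound to a collapsed item.
# Functions derive count text from child viewmodels, with a party variant factoring in unknown users.
# PreConstruct adjusts image and button sizes for compact UI.
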WidgetBlueprint /Sidebar/Panels/Social/List/WBP_Sidebar_CollapsibleHeader_Base
animations [Anim_RotateArrow {
  binding_count=1
  length_seconds=0.15
  object=&self:Anim_RotateArrow
  sequence {bindings [ArrowCollapse {
     id=#1
     possessed_class=Image
     tracks [{
       type=MovieScene2DTransformTrack
       sections [{type=MovieScene2DTransformSection, end_frame=3, start_frame=0, channels [Angle {type=MovieSceneScriptingFloatChannel, keys [[0,90],[3,0]]}]}]
      }]
    }]}
 }]
graphs [
 EventGraph {nodes [
   n1 = Event PreConstruct()
   > Parent: PreConstruct(IsDesignTime=@n1.IsDesignTime)
   > SetCompactUIStyle(self:WBP Sidebar Button Base Object Reference) {is=title K2Node_CallFunction}
   n2 = Event OnListItemObjectSet()
   > Parent: OnListItemObjectSet!(execute, ListItemObject:Object Reference=@n2.ListItemObject) -> (then) {is=title K2Node_CallParentFunction}
   > n3 = Branch(Condition=IsListItemExpanded!(UserListEntry:User List Entry Interface) -> (ReturnValue:bool))
   n4 = Get Anim_RotateArrow() -> (Anim_RotateArrow:Widget Animation Object Reference)
   PlayAnimation(execute=@n3.else, InAnimation=@n4, StartAtTime=GetEndTime(self=@n4))
   n5 = Event On Item Expansion Changed!() -> (OutputDelegate, then, bIsExpanded:bool) {is=title K2Node_Event}
   > n6 = Branch(Condition=@n5.bIsExpanded)
   n7 = Get Anim_RotateArrow()
   > PlayAnimationReverse(InAnimation=@n7)
   > PlaySound2D(Sound=/Game/Sounds/Fort_UI_Sounds/FrontEnd/New_FrontEnd/Fortnite_UI_MainMenu_Tab_Hover_02_Cue)
   PlayAnimationForward(execute=@n6.else, InAnimation=@n7)
  ]}
 SetChildItemCount {nodes [
   n1 = SetChildItemCount!() -> (then, ChildTreeArray:Array of MVVM Base Viewmodel Object References) {is=title K2Node_FunctionEntry}
   > "SetText (Text)"(
    self=Get Txt_PlayerCount() -> (Txt_PlayerCount:Common Text Object Reference)
    InText="To Text (Integer)"(Value=Length(TargetArray:Array of MVVM Base Viewmodel Object References=@n1))
   )
  ]}
 SetCompactUIStyle {nodes [
   SetCompactUIStyle() {is=title K2Node_FunctionEntry}
   > Branch(Condition=IsUsingCompactUI())
   > SetDesiredSizeOverride!(
    execute
    self:Image Object Reference=[Get Icon() -> (Icon:Image Object Reference), Get ArrowCollapse() -> (ArrowCollapse:Image Object Reference)]
    DesiredSize:Vector 2D Structure=(X=24,Y=24)
   ) -> (then)
   > SetMinDimensions!(execute, self:Common Button Base Object Reference, InMinWidth:int=64, InMinHeight:int=64) -> (then)
  ]}
 SetEpicPartyMembersCount {nodes [
   n1 = SetEpicPartyMembersCount!() -> (then, ChildTreeArray:Array of MVVM Base Viewmodel Object References, Num Unknown Users:int) {is=title K2Node_FunctionEntry}
   > n2 = Branch(Condition="integer > integer"(A=@n1.Num Unknown Users))
   n3 = int + int(A=Length(TargetArray=@n1), B=@n1.Num Unknown Users)
   > Return Node!(execute, Text:Text="To Text (Integer)"(Value=int - int(A=@n3)))
   Return Node(execute=@n2.else, Text=Get Text(self:Text Object Reference=Get Txt_PlayerCount()))
  ]}
]
variables [
 IconMaterial {type=MaterialInstance, default=/FortUILibrary/UIKit/Buttons/TabButtons/_MaterialInstances/MI_UI_TopNav_Icon}
 IconTexture {type=Texture2D, default=/FortUILibrary/Textures/Icons/T_UI_Icon_UIKit_CurrentPlayerCount_SDF}
]
widgets {
 info {
  inheritedWidgetCount=0
  namedSlotCount=1
  parentClass=/Sidebar/Buttons/WBP_Sidebar_Button_Base.WBP_Sidebar_Button_Base_C
  rootWidgetClass=null
  widgetCount=5
 }
 widgets [
  HorizontalBox_0 {widgetClassPath=&HorizontalBox, widget=HorizontalBox_0, properties {bOverride_Cursor=true}}
  Txt_Header {
   widgetClassPath=&CommonTextBlock
   bIsVariable=true
   parent=HorizontalBox_0
   widget=Txt_Header
   properties {
    clipping=ClipToBounds
    mobileFontSizeMultiplier=1.37
    text=Header Name
    textOverflowPolicy=Ellipsis
    font {
     bForceMonospaced=false
     bMaterialIsStencil=false
     fontMaterial=/Sidebar/Materials/Fonts/MI_Sidebar_FontColor_Secondary
     fontObject=&HeadingNow
     letterSpacing=10
     monospacedWidth=1
     size=14.25
     skewAmount=0
     typefaceFontName=75 Medium
     outlineSettings {
      bApplyOutlineToDropShadows=false
      bMiteredCorners=false
      bSeparateFillAlpha=false
      outlineMaterial=/FortUILibrary/Systems/ModularButton/MaterialInstances/MI_UIKit_TextOutline
      outlineSize=0
      outlineColor=<rgba 0 0 0 1>
     }
    }
    shadowOffset=<0 0>
   }
   slot {class=HorizontalBoxSlot, properties {verticalAlignment=VAlign_Center, padding=<ltrb 18 12 0 12>, size {sizeRule=Fill, value=1}}}
  }
  Icon {
   widgetClassPath=&Image
   bIsVariable=true
   parent=HorizontalBox_0
   widget=Icon
   properties {brush {
     bIsDynamicallyLoaded=false
     drawAs=Image
     imageType=FullColor
     mirroring=NoMirror
     resourceName=None
     resourceObject=/Sidebar/Materials/Icons/MI_Sidebar_Icon_Party
     tiling=NoTile
     imageSize=<24 24>
     margin=<ltrb 0 0 0 0>
     outlineSettings {
      bUseBrushTransparency=false
      roundingType=HalfHeightRadius
      width=0
      color {colorUseRule=UseColor_Specified, specifiedColor=<rgba 0 0 0 0>}
      cornerRadii=<0 0 0 0>
     }
     tintColor {colorUseRule=UseColor_Specified, specifiedColor=<rgba 1 1 1 1>}
     uVRegion {bIsValid=false, max=<0 0>, min=<0 0>}
    }}
   slot {class=HorizontalBoxSlot, properties {verticalAlignment=VAlign_Center, padding=<ltrb 0 0 4 0>}}
  }
  Txt_PlayerCount {
   widgetClassPath=&CommonTextBlock
   bIsVariable=true
   parent=HorizontalBox_0
   widget=Txt_PlayerCount
   properties {
    mobileFontSizeMultiplier=1.37
    text=0
    font {
     bForceMonospaced=true
     bMaterialIsStencil=false
     fontMaterial=&MI_Sidebar_FontColor_Secondary
     fontObject=&HeadingNow
     letterSpacing=10
     monospacedWidth=1
     size=14.25
     skewAmount=0
     typefaceFontName=75 Medium
     outlineSettings {
      bApplyOutlineToDropShadows=false
      bMiteredCorners=false
      bSeparateFillAlpha=false
      outlineMaterial=&MI_UIKit_TextOutline
      outlineSize=0
      outlineColor=<rgba 0 0 0 1>
     }
    }
    shadowOffset=<0 0>
   }
   slot {class=HorizontalBoxSlot, properties {verticalAlignment=VAlign_Center, padding=<ltrb 0 0 8 0>}}
  }
  ArrowCollapse {
   widgetClassPath=&Image
   bIsVariable=true
   parent=HorizontalBox_0
   widget=ArrowCollapse
   properties {
    brush {
     bIsDynamicallyLoaded=false
     drawAs=Image
     imageType=FullColor
     mirroring=NoMirror
     resourceName=None
     resourceObject=/Sidebar/Materials/Icons/MI_Sidebar_Icon_Arrow
     tiling=NoTile
     imageSize=<16 16>
     margin=<ltrb 0 0 0 0>
     outlineSettings {
      bUseBrushTransparency=false
      roundingType=HalfHeightRadius
      width=0
      color {colorUseRule=UseColor_Specified, specifiedColor=<rgba 0 0 0 0>}
      cornerRadii=<0 0 0 0>
     }
     tintColor {colorUseRule=UseColor_Specified, specifiedColor=<rgba 1 1 1 1>}
     uVRegion {bIsValid=false, max=<0 0>, min=<0 0>}
    }
    renderTransform {angle=90, scale=<1 1>, shear=<0 0>, translation=<0 0>}
   }
   slot {class=HorizontalBoxSlot, properties {verticalAlignment=VAlign_Center, padding=<ltrb 0 0 18 0>}}
  }
 ]
}
