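# Sidebar social list header entry widget, a child of WBP_Sidebar_CollapsibleHeader_Base that adds no widgets of its own.
# When the list assigns its item, it runs the parent handler, then casts the item to BP_Sidebar_CustomHeaderEntryVM.
# A successful cast stores the item as the widget's custom header view model, and a failed cast does nothing further.
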WidgetBlueprint /Sidebar/Panels/Social/List/WBP_Sidebar_CustomHeader
graphs [EventGraph {nodes [
   n1 = Event OnListItemObjectSet()
   > Parent: OnListItemObjectSet!(execute, ListItemObject:Object Reference=@n1.ListItemObject) -> (then) {is=title K2Node_CallParentFunction}
   > n2 = Cast To BP_Sidebar_CustomHeaderEntryVM!(execute, Object=@n1.ListItemObject) -> (then, CastFailed, AsBP Sidebar Custom Header Entry VM:BP Sidebar Custom Header Entry VM Object Reference) {is=title K2Node_DynamicCast}
   > Set BP_Sidebar_CustomHeaderEntryVM(BP_Sidebar_CustomHeaderEntryVM:BP Sidebar Custom Header Entry VM Object Reference=@n2) -> (Output_Get:BP Sidebar Custom Header Entry VM Object Reference)
  ]}]
variables []
widgets {
 info {
  inheritedWidgetCount=0
  parentClass=&/WBP_Sidebar_CollapsibleHeader_Base.WBP_Sidebar_CollapsibleHeader_Base_C
  rootWidgetClass=null
  widgetCount=0
 }
 widgets []
}
