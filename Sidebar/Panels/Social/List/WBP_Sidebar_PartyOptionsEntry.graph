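# Sidebar social-list entry that offers a Leave party, End Local Play or Disconnect button.
# Takes a leave-party entry view model from its list item and broadcasts ActionName through LeavePartyClicked on click.
# CF_ conversion functions pick the action from splitscreen state and player index, map it to a label, and derive visibilities.
# For splitscreen players in zone, these visibilities collapse the button and show a return-to-lobby hint.
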WidgetBlueprint /Sidebar/Panels/Social/List/WBP_Sidebar_PartyOptionsEntry
graphs [
 CF_ActionNameToText {nodes [
   n1 = CF_ActionNameToText!() -> (then, Action name:Name) {is=title K2Node_FunctionEntry}
   > n2 = Switch on Name!(execute:Exec, Selection:Name=@n1) -> (Default:Exec, Disconnect:Exec, EndLocalPlay:Exec, LeaveParty:Exec) {is=title K2Node_SwitchName, pinNames [Disconnect,EndLocalPlay,LeaveParty]}
   Return Node!(execute=[@n2.Default, @n2.LeaveParty], Text:Text="LOCTEXT(\"Leave party\")")
   Return Node(execute=@n2.Disconnect, Text="LOCTEXT(\"Disconnect\")")
   Return Node(execute=@n2.EndLocalPlay, Text="LOCTEXT(\"End Local Play\")")
  ]}
 CF_LeavePartyButtonToVisibility {nodes [
   n1 = CF_LeavePartyButtonToVisibility!() -> (then, in Splitscreen:bool=~, Can Leave Party:bool=~) {is=title K2Node_FunctionEntry}
   n2 = AND Boolean(A=IsInZone!() -> (ReturnValue:bool), B=@n1)
   > Return Node!(
    execute
    Visibility:ESlateVisibility Enum=Select(Option 0=Collapsed, Option 1=Select(Option 0=Visible, Option 1=Collapsed, Index=@n2), Index=OR Boolean(A=@n1, B=@n1.Can Leave Party))
   )
  ]}
 CF_LeaveTextToVisibility {nodes [
   n1 = CF_LeaveTextToVisibility!() -> (then, in Splitscreen:bool=~, Can Leave Party:bool=~) {is=title K2Node_FunctionEntry}
   n2 = IsInZone()
   > Return Node(Visibility=Select(Option 0=Collapsed, Option 1=HitTestInvisible, Index=AND Boolean(A=@n1, B=@n2)))
  ]}
 CF_LocalPlayerCountToName {nodes [
   n1 = CF_LocalPlayerCountToName!() -> (then, Player Index:int, isSplitscreenLeave:bool=~, Can leave party:bool=~) {is=title K2Node_FunctionEntry}
   n2 = AND Boolean(A=@n1.isSplitscreenLeave, B=IsSplitscreenV2Enabled!(self:Fort Runtime Options Object Reference=GetRuntimeOptions()) -> (ReturnValue:bool))
   n3 = "Equal (Integer)"(A=Get Player Index() -> (Player Index:int=0))
   > Return Node!(
    execute
    Action Name:Name=SelectName!(A:Name=LeaveParty, B:Name=SelectName(A=EndLocalPlay, B=Disconnect, bPickA=@n3) -> (ReturnValue=None), bPickA:bool=NOT Boolean(A=@n2)) -> (ReturnValue:Name=None)
   )
  ]}
 EventGraph {nodes [
   "On Clicked (LeavePartyButton)"!() -> (OutputDelegate, then, Button:Common Button Base Object Reference) {is=title K2Node_ComponentBoundEvent}
   > Call Leave Party Clicked!(execute:Exec, self:WBP Sidebar Party Options Entry Object Reference, Action:Name=Get ActionName() -> (ActionName:Name=None)) -> (then:Exec) {is=title K2Node_CallDelegate}
   n1 = Event OnListItemObjectSet()
   > n2 = Cast To BP_SidebarLeavePartyEntryVM!(execute, Object=@n1.ListItemObject) -> (then, CastFailed, AsBP Sidebar Leave Party Entry VM:BP Sidebar Leave Party Entry VM Object Reference) {is=title K2Node_DynamicCast}
   > Set BP_SidebarLeavePartyEntryVM(BP_SidebarLeavePartyEntryVM:BP Sidebar Leave Party Entry VM Object Reference=@n2) -> (Output_Get:BP Sidebar Leave Party Entry VM Object Reference)
   Event PreConstruct()
   > SetVisibility(
    self=Get Divider!(self:WBP Sidebar Party And Game Backplate Object Reference=Get Backplate_PartyAndGame() -> (Backplate_PartyAndGame:WBP Sidebar Party And Game Backplate Object Reference)) -> (Divider:Image Object Reference)
    InVisibility=SelfHitTestInvisible
   )
  ]}
 LeavePartyClicked {nodes [LeavePartyClicked!() -> (then, Action:Name) {is=title K2Node_FunctionEntry}]}
]
variables [
 ActionName {type=string, default=None}
 LeavePartyClicked {}
]
widgets {
 info {inheritedWidgetCount=0, parentClass=&CommonUserWidget, rootWidgetClass=&Overlay, widgetCount=6}
 widgets [
  Overlay_33 {widgetClassPath=&Overlay, widget=Overlay_33}
  Backplate_PartyAndGame {
   widgetClassPath=/Sidebar/Buttons/Blocks/WBP_Sidebar_PartyAndGame_Backplate
   bIsVariable=true
   parent=Overlay_33
   widget=Backplate_PartyAndGame
   properties {showFade=false}
   slot {class=OverlaySlot, properties {horizontalAlignment=HAlign_Fill, verticalAlignment=VAlign_Fill, padding=<ltrb -18 0 -22 0>}}
  }
  LeavePartyButton {
   widgetClassPath=/Sidebar/Buttons/WBP_Sidebar_ButtonGeneric
   bIsVariable=true
   parent=Overlay_33
   widget=LeavePartyButton
   properties {minHeight=0, minHeightOverride=60, text=Leave party}
   slot {class=OverlaySlot, properties {horizontalAlignment=HAlign_Fill, verticalAlignment=VAlign_Center, padding=<ltrb 18 15 18 18>}}
  }
  LeaveInZone {
   widgetClassPath=&HorizontalBox
   bIsVariable=true
   parent=Overlay_33
   widget=LeaveInZone
   slot {class=OverlaySlot, properties {horizontalAlignment=HAlign_Center, verticalAlignment=VAlign_Fill, padding=<ltrb 18 0 18 12>}}
  }
  Icon_Voice_Status {
   widgetClassPath=&Image
   bIsVariable=true
   parent=LeaveInZone
   widget=Icon_Voice_Status
   properties {brush {
     bIsDynamicallyLoaded=false
     drawAs=Image
     imageType=FullColor
     mirroring=NoMirror
     resourceName=None
     resourceObject=/FortUILibrary/Textures/MaterialInstances/MI_UIKit_Icon_Controller
     tiling=NoTile
     imageSize=<32 32>
     margin=<ltrb 0 0 0 0>
     outlineSettings {
      bUseBrushTransparency=false
      roundingType=HalfHeightRadius
      width=0
      color {colorUseRule=UseColor_Specified, specifiedColor=<rgba 0 0 0 0>}
      cornerRadii=<0 0 0 0>
     }
     tintColor {colorUseRule=UseColor_Specified, specifiedColor=<rgba 0.5776 0.5776 0.5776 1>}
     uVRegion {bIsValid=false, max=<0 0>, min=<0 0>}
    }}
   slot {class=HorizontalBoxSlot, properties {horizontalAlignment=HAlign_Left, verticalAlignment=VAlign_Center, padding=<ltrb 0 0 10 0>}}
  }
  TextBlock {
   widgetClassPath=&UEFN_TextBlock
   bIsVariable=true
   parent=LeaveInZone
   widget=TextBlock
   properties {
    autoWrapText=true
    text="To end local play, return to lobby then open this menu"
    visibility=SelfHitTestInvisible
    colorAndOpacity {colorUseRule=UseColor_Specified, specifiedColor=<rgba 0.5776 0.5776 0.5776 1>}
    font {
     bForceMonospaced=false
     bMaterialIsStencil=false
     fontMaterial=null
     fontObject=&HeadingNow
     letterSpacing=0
     monospacedWidth=1
     size=12
     skewAmount=0
     typefaceFontName=74 Regular
     outlineSettings {
      bApplyOutlineToDropShadows=false
      bMiteredCorners=false
      bSeparateFillAlpha=false
      outlineMaterial=null
      outlineSize=0
      outlineColor=<rgba 0 0 0 1>
     }
    }
   }
   slot {class=HorizontalBoxSlot, properties {verticalAlignment=VAlign_Center, size {sizeRule=Fill, value=1}}}
  }
 ]
}
