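# Sidebar social-list widget that shows a party's pending invites as avatars beside a faded "{NameList} invited" line.
# SetPendingInvites passes the invited users to the party avatar list (capped at three), then joins their nicknames or display names with commas into the localized text.
# The text wraps and ends in an ellipsis on overflow within a fixed-height box.
# On Construct the widget becomes hit-test-invisible and collapses the backplate's divider.
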WidgetBlueprint /Sidebar/Panels/Social/List/WBP_Sidebar_PartyPendingInvites
graphs [
 EventGraph {nodes [
   Event Construct()
   > Parent: Construct()
   > SetVisibility(InVisibility=HitTestInvisible)
   > SetVisibility(
    self=Get Divider!(self:WBP Sidebar Party And Game Backplate Object Reference=Get Backplate_PartyAndGame() -> (Backplate_PartyAndGame:WBP Sidebar Party And Game Backplate Object Reference)) -> (Divider:Image Object Reference)
    InVisibility=Collapsed
   )
  ]}
 SetPendingInvites {nodes [
   n1 = SetPendingInvites!() -> (then, Invites:Array of Fort Social User Object References) {is=title K2Node_FunctionEntry}
   > AddPendingAvatars!(
    execute
    self:WBP Sidebar Party Avatar List Object Reference=Get WBP_Sidebar_PartyAvatarList() -> (WBP_Sidebar_PartyAvatarList:WBP Sidebar Party Avatar List Object Reference)
    Pending Party Members:Array of Fort Social User Object References=@n1
   ) -> (then)
   > n2 = For Each Loop(Array:Array of Fort Social User Object References=@n1) -> (Array Element:Fort Social User Object Reference)
   Set Text(
    execute=@n2.Completed
    Text=Format Text!(Format="LOCTEXT(\"{NameList} invited\")", NameList:String=Get NameList() -> (NameList:String)) -> (Result)
    self=Get Text_PendingInvites() -> (Text_PendingInvites:Common Text Object Reference)
   )
   n3 = IsEmpty(InString=Get NameList())
   Set NameList(
    execute=@n2.LoopBody
    NameList:String=Append!(
     A=Get NameList()
     B=SelectString(B=", ", bPickA=@n3)
     C:String=GetNicknameOrDisplayName!(self:Fort Social User Object Reference=@n2) -> (ReturnValue:String)
    ) -> (ReturnValue)
   ) -> (Output_Get:String)
  ]}
]
variables []
widgets {
 info {inheritedWidgetCount=0, parentClass=&CommonUserWidget, rootWidgetClass=&Overlay, widgetCount=6}
 widgets [
  Overlay_0 {widgetClassPath=&Overlay, widget=Overlay_0}
  Backplate_PartyAndGame {
   widgetClassPath=/Sidebar/Buttons/Blocks/WBP_Sidebar_PartyAndGame_Backplate
   bIsVariable=true
   parent=Overlay_0
   widget=Backplate_PartyAndGame
   properties {showFade=false}
   slot {class=OverlaySlot, properties {horizontalAlignment=HAlign_Fill, verticalAlignment=VAlign_Fill, padding=<ltrb -22 0 -22 -10>}}
  }
  HorizontalBox_0 {
   widgetClassPath=&HorizontalBox
   parent=Overlay_0
   widget=HorizontalBox_0
   slot {class=OverlaySlot, properties {padding=<ltrb 16 10 0 0>}}
  }
  WBP_Sidebar_PartyAvatarList {
   widgetClassPath=/Sidebar/Panels/Chat/WBP_Sidebar_PartyAvatarList
   bIsVariable=true
   parent=HorizontalBox_0
   widget=WBP_Sidebar_PartyAvatarList
   properties {maximumAvatarsAllowed=3}
   slot {class=HorizontalBoxSlot}
  }
  SizeBox_0 {
   widgetClassPath=&SizeBox
   parent=HorizontalBox_0
   widget=SizeBox_0
   properties {bOverride_HeightOverride=true, heightOverride=46}
   slot {class=HorizontalBoxSlot, properties {verticalAlignment=VAlign_Center, padding=<ltrb 8 0 0 0>}}
  }
  Text_PendingInvites {
   widgetClassPath=&CommonTextBlock
   bIsVariable=true
   parent=SizeBox_0
   widget=Text_PendingInvites
   properties {
    bAutoCollapseWithEmptyText=true
    bIsScrollingEnabled=false
    clipping=ClipToBounds
    mobileFontSizeMultiplier=1.4
    scrollStyle=&TextScrollStyle-Base_C
    text="[PH] pending invites [PH] pending invites [PH] pending invites[PH] pending invites [PH] pending invites [PH] pending invites"
    textOverflowPolicy=MultilineEllipsis
    wrapTextAt=400
    colorAndOpacity {colorUseRule=UseColor_Specified, specifiedColor=<rgba 1 1 1 0.5>}
    font {
     bForceMonospaced=false
     bMaterialIsStencil=false
     fontMaterial=/Sidebar/Materials/Fonts/MI_Sidebar_FontColor_Secondary
     fontObject=&HeadingNow
     letterSpacing=0
     monospacedWidth=1
     size=12.75
     skewAmount=0
     typefaceFontName=74 Regular
     outlineSettings {
      bApplyOutlineToDropShadows=false
      bMiteredCorners=false
      bSeparateFillAlpha=false
      outlineMaterial=/FortUILibrary/Systems/ModularButton/MaterialInstances/MI_UIKit_TextOutline
      outlineSize=0
      outlineColor=<rgba 0 0 0 1>
     }
    }
    shadowOffset=<0 0>
   }
   slot {class=SizeBoxSlot, properties {verticalAlignment=VAlign_Center}}
  }
 ]
}
