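# Activatable widget for the sidebar social header that holds a single icon button for adding friends.
# Hovering expands the button with an "Add friends" label and unhovering clears it, each also firing the widget's own hover dispatchers.
# Clicking builds a navigation point tagged for the AddFriend state-tree transition and pushes it to the sidebar navigation view model.
# The button is the desired focus target; the context-menu and voice-chat variables are unused in these graphs.
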WidgetBlueprint /Sidebar/Panels/Social/List/WBP_Sidebar_SocialHeader_AddFriends
graphs [
 BP_GetDesiredFocusTarget {nodes [
   Get Desired Focus Target()
   > Return Node!(execute, ReturnValue:Widget Object Reference=Get Button_AddFriends() -> (Button_AddFriends:WBP Sidebar Button Generic Object Reference))
  ]}
 EventGraph {nodes [
   "On Hovered (Button_AddFriends)"!() -> (OutputDelegate, then, Button:Common Button Base Object Reference) {is=title K2Node_ComponentBoundEvent}
   > SetText!(execute, self:WBP UIKit Button Generic Base Object Reference=Get Button_AddFriends(), Text:Text="LOCTEXT(\"Add friends\")") -> (then)
   > Call On Hovered!(execute:Exec, self:WBP Sidebar Social Header Add Friends Object Reference) -> (then:Exec) {is=title K2Node_CallDelegate}
   "On Unhovered (Button_AddFriends)"!() -> (OutputDelegate, then, Button:Common Button Base Object Reference) {is=title K2Node_ComponentBoundEvent}
   > SetText(self=Get Button_AddFriends())
   > Call On Unhovered!(execute:Exec, self:WBP Sidebar Social Header Add Friends Object Reference) -> (then:Exec) {is=title K2Node_CallDelegate}
   "On Clicked (Button_AddFriends)"!() -> (OutputDelegate, then, Button:Common Button Base Object Reference) {is=title K2Node_ComponentBoundEvent}
   > n1 = Construct Object from Class!(
    execute
    Class=/Sidebar/Shared/Navigation/BP_NavigationPoint.BP_NavigationPoint_C
    self
    SocialUser:Fort Social User Object Reference
    ConversationId:Name=None
    ConfirmationVM:Sidebar Confirmation Panel VM Object Reference
    ReservedSpots:int=0
    SocialItemVM:Fort Social Item VM Object Reference
    NavigationTag:Gameplay Tag Structure=(TagName="UI.StateTree.TransitionTag.Sidebar.AddFriend")
   ) -> (then, ReturnValue:BP Navigation Point Object Reference)
   > PushNavigation!(
    execute
    self:Fort Sidebar Navigation VM Object Reference=Get FortSidebarNavigationVM() -> (FortSidebarNavigationVM:Fort Sidebar Navigation VM Object Reference)
    NavigationPoint:Fort Sidebar Navigation Point Object Reference=@n1
   ) -> (then)
  ]}
 on Hovered {nodes [on Hovered() {is=title K2Node_FunctionEntry}]}
 on Unhovered {nodes [on Unhovered() {is=title K2Node_FunctionEntry}]}
]
variables [
 ContextMenuMuteName {type=string, default=MuteUnmute}
 ContextMenuOpenChatName {type=string, default=OpenTextChat}
 ContextMenuReportName {type=string, default=ReportConvo}
 IconSizeWithArrow {type=Vector2D, default=<54 36>}
 IconSizeWithoutArrow {type=Vector2D, default=<36 36>}
 IsPTTKey {type=boolean}
 isSocialBanned {type=boolean}
 IsUsingEpicParties {type=boolean}
 IsVoiceChatMethodPTT {type=boolean}
 Last Focused Widget {type=CommonButtonBase, default=null}
 on Hovered {}
 on Unhovered {}
 ReportMenuActionContext {type=UIKitContextMenuActionContext, default=null}
]
widgets {
 info {inheritedWidgetCount=0, parentClass=&CommonActivatableWidget, rootWidgetClass=&GridPanel, widgetCount=4}
 widgets [
  GridPanel_0 {widgetClassPath=&GridPanel, widget=GridPanel_0}
  HB_Buttons {
   widgetClassPath=&HorizontalBox
   parent=GridPanel_0
   widget=HB_Buttons
   properties {bOverride_Cursor=true}
   slot {class=GridSlot, properties {layer=1, verticalAlignment=VAlign_Center, padding=<ltrb 24 0 0 0>}}
  }
  Button_AddFriends {
   widgetClassPath=/Sidebar/Buttons/WBP_Sidebar_ButtonGeneric
   bIsVariable=true
   parent=HB_Buttons
   widget=Button_AddFriends
   properties {
    icon=/Sidebar/Materials/Icons/Textures/T_UI_Icon_Sidebar_AddFriend_MSDF
    iconMaterial=/Sidebar/Materials/Icons/MI_Sidebar_Icon_PartyOptions
    minHeight=54
    minWidth=60
    text=""
    useDisabledVisualsOnLocked=true
    useIconBlock=true
    buttonPadding=<ltrb 0 0 0 0>
    font {
     bForceMonospaced=false
     bMaterialIsStencil=false
     fontMaterial=/Sidebar/Materials/Fonts/MI_Sidebar_FontColor_Secondary
     fontObject=&HeadingNow
     letterSpacing=10
     monospacedWidth=1
     size=13.5
     skewAmount=0
     typefaceFontName=74 Regular
     outlineSettings {
      bApplyOutlineToDropShadows=false
      bMiteredCorners=false
      bSeparateFillAlpha=false
      outlineMaterial=/FortUILibrary/Systems/ModularButton/MaterialInstances/MI_UIKit_TextOutline
      outlineSize=0
      outlineColor=<rgba 0 0 0 1>
     }
    }
    iconPadding=<ltrb 8 0 8 0>
    textPadding=<ltrb 0 0 8 0>
   }
   slot {class=HorizontalBoxSlot, properties {verticalAlignment=VAlign_Bottom, padding=<ltrb 0 0 18 0>}}
  }
  Spacer_163 {widgetClassPath=&Spacer, parent=HB_Buttons, widget=Spacer_163, properties {size=<0 64>}, slot {class=HorizontalBoxSlot}}
 ]
}
