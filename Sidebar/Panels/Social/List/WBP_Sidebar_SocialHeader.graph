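# Social list header entry for the sidebar, extending the collapsible header base and enabling list auto-update on its header view model until the entry is released.
# For the UnifiedFriends type it creates an Add Friends widget in an overlay, hiding icon, player count and arrow; other types remove it.
# Hover events and custom left/right navigation functions pass highlight between header and button, so hovering the button unhighlights the header.
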WidgetBlueprint /Sidebar/Panels/Social/List/WBP_Sidebar_SocialHeader
graphs [
 AddFriendsButton {nodes [
   n1 = AddFriendsButton!() -> (then, Header Type:ESocialHeaderItemType Enum=Unknown) {is=title K2Node_FunctionEntry}
   > n2 = Branch(Condition="Equal (Enum)"(A:ESocialHeaderItemType Enum=@n1, B:ESocialHeaderItemType Enum=UnifiedFriends))
   > n3 = Get!(execute:Exec) -> (then:Exec, else:Exec, AddFriends:WBP Sidebar Social Header Add Friends Object Reference)
   SetVisibility(
    execute=@n3.else
    self=[
     Get Icon() -> (Icon:Image Object Reference)
     Get Txt_PlayerCount() -> (Txt_PlayerCount:Common Text Object Reference)
     Get ArrowCollapse() -> (ArrowCollapse:Image Object Reference)
    ]
    InVisibility=Collapsed
   )
   > n4 = Create Widget!(
    execute:Exec
    Class:User Widget Class Reference=&/WBP_Sidebar_SocialHeader_AddFriends.WBP_Sidebar_SocialHeader_AddFriends_C
    OwningPlayer:Player Controller Object Reference=GetOwningPlayer()
   ) -> (then:Exec, ReturnValue:WBP Sidebar Social Header Add Friends Object Reference)
   > n5 = Set AddFriends(AddFriends:WBP Sidebar Social Header Add Friends Object Reference=@n4) -> (Output_Get:WBP Sidebar Social Header Add Friends Object Reference)
   > AddChild(self=Get Overlay_Controls() -> (Overlay_Controls:Overlay Object Reference), Content=@n5)
   > SetNavigationRuleCustomBoundary!(
    execute
    self:Widget Object Reference=Get AddFriends() -> (AddFriends:WBP Sidebar Social Header Add Friends Object Reference)
    Direction:EUINavigation Enum=Left
    InCustomDelegate:Delegate=Create Event()
   ) -> (then)
   n6 = Create Event()
   n7 = Get AddFriends()
   > Bind Event to On Hovered(self:WBP Sidebar Social Header Add Friends Object Reference=@n7, Delegate=@n6) {is=title K2Node_AddDelegate}
   > SetVisibility(self=Get Overlay_Controls(), InVisibility=SelfHitTestInvisible)
   > SetPadding!(
    execute
    self:User Widget Object Reference=Get Background() -> (Background:WBP Sidebar Block Background Button Object Reference)
    InPadding_Left:float=0
    InPadding_Top:float=10
    InPadding_Right:float=0
    InPadding_Bottom:float=0
   ) -> (then)
   n8 = Get(execute=@n2.else)
   > RemoveChild(self=Get Overlay_Controls(), Content=@n8)
   > Unbind Event from On Hovered(self:WBP Sidebar Social Header Add Friends Object Reference=@n7, Delegate=@n6) {is=title K2Node_RemoveDelegate}
   > Set AddFriends()
   > SetVisibility(self=[Get Icon(), Get Txt_PlayerCount(), Get ArrowCollapse()], InVisibility=SelfHitTestInvisible)
   > SetVisibility(self=Get Overlay_Controls(), InVisibility=Collapsed)
   > SetPadding(self=Get Background(), InPadding_Left=0, InPadding_Top=0, InPadding_Right=0, InPadding_Bottom=0)
  ]}
 EventGraph {nodes [
   On AddFriends Hovered() {is=title K2Node_CustomEvent}
   > Parent: On Unhovered() {is=title K2Node_CallParentFunction}
   On AddFriends Unhovered() {is=title K2Node_CustomEvent}
   > Parent: On Hovered() {is=title K2Node_CallParentFunction}
   n1 = Event OnListItemObjectSet()
   > Parent: OnListItemObjectSet!(execute, ListItemObject:Object Reference=@n1.ListItemObject) -> (then) {is=title K2Node_CallParentFunction}
   > n2 = Cast To FortSocialHeaderItemVM!(execute, Object=@n1.ListItemObject) -> (then, CastFailed, AsFort Social Header Item VM:Fort Social Header Item VM Object Reference) {is=title K2Node_DynamicCast}
   > Set FortSocialHeaderItemVM(FortSocialHeaderItemVM:Fort Social Header Item VM Object Reference=@n2) -> (Output_Get:Fort Social Header Item VM Object Reference)
   > SetAllowListAutoUpdate!(
    execute
    self:Fort Social Header Item VM Object Reference=Get FortSocialHeaderItemVM() -> (FortSocialHeaderItemVM:Fort Social Header Item VM Object Reference)
    bIsEnabled:bool=true
   ) -> (then)
   Event On Entry Released() {is=title K2Node_Event}
   > SetAllowListAutoUpdate(self=Get FortSocialHeaderItemVM(), bIsEnabled=false)
   Event On Hovered()
   > n3 = Branch(Condition=HasAnyChildren!(self:Panel Widget Object Reference=Get Overlay_Controls()) -> (ReturnValue:bool))
   n4 = Get AddFriends()
   > n5 = Branch(Condition=IsHovered(self=Get Button_AddFriends!(self:WBP Sidebar Social Header Add Friends Object Reference=@n4) -> (Button_AddFriends:WBP Sidebar Button Generic Object Reference)))
   Parent: On Hovered(execute=@n5.else)
   Parent: On Hovered(execute=@n3.else)
  ]}
 NavigateLeftFromAddFriendsButton {nodes [
   NavigateLeftFromAddFriendsButton!() -> (then, Navigation:EUINavigation Enum) {is=title K2Node_FunctionEntry}
   > DeactivateWidget(self=Get AddFriends())
   > Parent: On Hovered()
   > Return Node!(execute, ReturnValue:Widget Object Reference=Self-Reference())
  ]}
 NavigaterightToAddFriendsButton {nodes [
   NavigaterightToAddFriendsButton!() -> (then, Navigation:EUINavigation Enum) {is=title K2Node_FunctionEntry}
   > n1 = Get()
   > Parent: On Unhovered()
   > Return Node(ReturnValue=Get Button_AddFriends(self=@n1))
  ]}
]
variables [AddFriends {type=WBP_Sidebar_SocialHeader_AddFriends_C, default=null}]
widgets {
 info {
  inheritedWidgetCount=0
  namedSlotCount=1
  parentClass=&/WBP_Sidebar_CollapsibleHeader_Base.WBP_Sidebar_CollapsibleHeader_Base_C
  rootWidgetClass=null
  widgetCount=1
 }
 widgets [Overlay_Controls {
   widgetClassPath=&Overlay
   bIsVariable=true
   widget=Overlay_Controls
   properties {navigation=&self:WidgetTree.Overlay_Controls.WidgetNavigation_0, visibility=Collapsed}
  }]
}
