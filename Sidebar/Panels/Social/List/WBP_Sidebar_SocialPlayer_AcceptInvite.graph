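# Sidebar social player widget, a CommonActivatableWidget with an Accept button and an icon-only Decline button for responding to an invite.
# Button clicks broadcast the on Accept Invite and on Decline Invite dispatchers, leaving the response logic to listeners.
# Hovering either button or entering the widget broadcasts on Hovered, while mouse leave broadcasts on Unhovered.
# DoCustomNavigation always returns Button_Decline, and the context menu and voice chat variables go unused in these graphs.
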WidgetBlueprint /Sidebar/Panels/Social/List/WBP_Sidebar_SocialPlayer_AcceptInvite
graphs [
 DoCustomNavigation {nodes [
   DoCustomNavigation!() -> (then, Navigation:EUINavigation Enum) {is=title K2Node_FunctionEntry}
   > Return Node!(execute, ReturnValue:Widget Object Reference=Get Button_Decline() -> (Button_Decline:WBP Sidebar Button Generic Object Reference))
  ]}
 EventGraph {nodes [
   n1 = "On Hovered (Button_Accept)"!() -> (OutputDelegate, then, Button:Common Button Base Object Reference) {is=title K2Node_ComponentBoundEvent}
   Call On Hovered!(execute:Exec=[@n1.then, @n2.then, @n3.then], self:WBP Sidebar Social Player Accept Invite Object Reference) -> (then:Exec) {is=title K2Node_CallDelegate}
   n2 = "On Hovered (Button_Decline)"!() -> (OutputDelegate, then, Button:Common Button Base Object Reference) {is=title K2Node_ComponentBoundEvent}
   "On Clicked (Button_Decline)"!() -> (OutputDelegate, then, Button:Common Button Base Object Reference) {is=title K2Node_ComponentBoundEvent}
   > Call On Decline Invite!(execute:Exec, self:WBP Sidebar Social Player Accept Invite Object Reference) -> (then:Exec) {is=title K2Node_CallDelegate}
   "On Clicked (Button_Accept)"!() -> (OutputDelegate, then, Button:Common Button Base Object Reference) {is=title K2Node_ComponentBoundEvent}
   > Call On Accept Invite!(execute:Exec, self:WBP Sidebar Social Player Accept Invite Object Reference) -> (then:Exec) {is=title K2Node_CallDelegate}
   n3 = Event OnMouseEnter()
   Event OnMouseLeave!() -> (OutputDelegate, then, MouseEvent:"Pointer Event Structure (by ref)") {is=title K2Node_Event}
   > Call On Unhovered!(execute:Exec, self:WBP Sidebar Social Player Accept Invite Object Reference) -> (then:Exec) {is=title K2Node_CallDelegate}
  ]}
 on Accept Invite {nodes [on Accept Invite() {is=title K2Node_FunctionEntry}]}
 on Decline Invite {nodes [on Decline Invite() {is=title K2Node_FunctionEntry}]}
 on Hovered {nodes [on Hovered() {is=title K2Node_FunctionEntry}]}
 on Unhovered {nodes [on Unhovered() {is=title K2Node_FunctionEntry}]}
]
variables [
 ContextMenuMuteName {type=string, default=MuteUnmute}
 ContextMenuOpenChatName {type=string, default=OpenTextChat}
 ContextMenuReportName {type=string, default=ReportConvo}
 IconSizeWithArrow {type=Vector2D, default=<54 36>}
 IconSizeWithoutArrow {type=Vector2D, default=<36 36>}
 IsPTTKey {type=boolean}
 isSocialBanned {type=boolean}
 IsUsingEpicParties {type=boolean}
 IsVoiceChatMethodPTT {type=boolean}
 Last Focused Widget {type=CommonButtonBase, default=null}
 on Accept Invite {}
 on Decline Invite {}
 on Hovered {}
 on Unhovered {}
 ReportMenuActionContext {type=UIKitContextMenuActionContext, default=null}
]
widgets {
 info {inheritedWidgetCount=0, parentClass=&CommonActivatableWidget, rootWidgetClass=&GridPanel, widgetCount=4}
 widgets [
  GridPanel_0 {widgetClassPath=&GridPanel, widget=GridPanel_0}
  HB_Buttons {
   widgetClassPath=&HorizontalBox
   parent=GridPanel_0
   widget=HB_Buttons
   properties {bOverride_Cursor=true}
   slot {class=GridSlot, properties {layer=1, verticalAlignment=VAlign_Center}}
  }
  Button_Accept {
   widgetClassPath=/Sidebar/Buttons/WBP_Sidebar_ButtonGeneric
   bIsVariable=true
   parent=HB_Buttons
   widget=Button_Accept
   properties {
    icon=/Sidebar/Materials/Icons/Textures/T_UI_Icon_Sidebar_Headphones_MSDF
    iconMaterial=/Sidebar/Materials/Icons/MI_Sidebar_Icon_PartyOptions
    minHeight=54
    minWidth=54
    navigation=&self:WidgetTree.Button_Accept.WidgetNavigation_0
    text=Accept
    useDisabledVisualsOnLocked=true
    buttonPadding=<ltrb 0 0 0 0>
    font {
     bForceMonospaced=false
     bMaterialIsStencil=false
     fontMaterial=/FortUILibrary/UIKit/Buttons/MaterialInstances/CTAButtons/MI_UIKit_TextColor_Quiet
     fontObject=&HeadingNow
     letterSpacing=10
     monospacedWidth=1
     size=15
     skewAmount=0
     typefaceFontName=74 Regular
     outlineSettings {
      bApplyOutlineToDropShadows=false
      bMiteredCorners=false
      bSeparateFillAlpha=false
      outlineMaterial=/FortUILibrary/Systems/ModularButton/MaterialInstances/MI_UIKit_TextOutline
      outlineSize=0
      outlineColor=<rgba 0 0 0 1>
     }
    }
    textPadding=<ltrb 12 0 12 0>
   }
   slot {class=HorizontalBoxSlot, properties {verticalAlignment=VAlign_Center, padding=<ltrb 8 0 12 0>}}
  }
  Button_Decline {
   widgetClassPath=&WBP_Sidebar_ButtonGeneric
   bIsVariable=true
   parent=HB_Buttons
   widget=Button_Decline
   properties {
    icon=/FortniteCoreUI/Frontend/StoreMain/Materials/Textures/T_UI_Icon_XClose
    iconMaterial=&MI_Sidebar_Icon_PartyOptions
    minHeight=54
    minWidth=54
    navigation=&self:WidgetTree.Button_Decline.WidgetNavigation_0
    useDisabledVisualsOnLocked=true
    useIconBlock=true
    useTextBlock=false
    buttonPadding=<ltrb 0 0 0 0>
    font {
     bForceMonospaced=false
     bMaterialIsStencil=false
     fontMaterial=&MI_UIKit_TextColor_Quiet
     fontObject=&HeadingNow
     letterSpacing=10
     monospacedWidth=1
     size=15
     skewAmount=0
     typefaceFontName=74 Regular
     outlineSettings {
      bApplyOutlineToDropShadows=false
      bMiteredCorners=false
      bSeparateFillAlpha=false
      outlineMaterial=&MI_UIKit_TextOutline
      outlineSize=0
      outlineColor=<rgba 0 0 0 1>
     }
    }
    iconSize=<18 18>
   }
   slot {class=HorizontalBoxSlot, properties {verticalAlignment=VAlign_Center, padding=<ltrb 0 0 12 0>}}
  }
 ]
}
