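# Activatable sidebar widget that shows an "Add" friend button with an icon for the social player list.
# It re-broadcasts the button's click, hover and unhover as its own event dispatchers, with mouse enter and leave on the widget also firing hover and unhover.
# The button is the desired focus target; the graphs contain no friend-request logic and leave the declared chat, voice and context-menu variables unused.
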WidgetBlueprint /Sidebar/Panels/Social/List/WBP_Sidebar_SocialPlayer_AddFriend
graphs [
 BP_GetDesiredFocusTarget {nodes [
   Get Desired Focus Target()
   > Return Node!(execute, ReturnValue:Widget Object Reference=Get Button_AddFriend() -> (Button_AddFriend:WBP Sidebar Button Generic Object Reference))
  ]}
 EventGraph {nodes [
   n1 = "On Hovered (Button_AddFriend)"!() -> (OutputDelegate, then, Button:Common Button Base Object Reference) {is=title K2Node_ComponentBoundEvent}
   Call On Hovered!(execute:Exec=[@n1.then, @n3.then], self:WBP Sidebar Social Player Add Friend Object Reference) -> (then:Exec) {is=title K2Node_CallDelegate}
   n2 = "On Unhovered (Button_AddFriend)"!() -> (OutputDelegate, then, Button:Common Button Base Object Reference) {is=title K2Node_ComponentBoundEvent}
   Call On Unhovered!(execute:Exec=[@n2.then, @n4.then], self:WBP Sidebar Social Player Add Friend Object Reference) -> (then:Exec) {is=title K2Node_CallDelegate}
   "On Clicked (Button_AddFriend)"!() -> (OutputDelegate, then, Button:Common Button Base Object Reference) {is=title K2Node_ComponentBoundEvent}
   > Call On Clicked!(execute:Exec, self:WBP Sidebar Social Player Add Friend Object Reference) -> (then:Exec) {is=title K2Node_CallDelegate}
   n3 = Event OnMouseEnter()
   n4 = Event OnMouseLeave!() -> (OutputDelegate, then, MouseEvent:"Pointer Event Structure (by ref)") {is=title K2Node_Event}
  ]}
 on Clicked {nodes [on Clicked() {is=title K2Node_FunctionEntry}]}
 on Hovered {nodes [on Hovered() {is=title K2Node_FunctionEntry}]}
 on Unhovered {nodes [on Unhovered() {is=title K2Node_FunctionEntry}]}
]
variables [
 ContextMenuMuteName {type=string, default=MuteUnmute}
 ContextMenuOpenChatName {type=string, default=OpenTextChat}
 ContextMenuReportName {type=string, default=ReportConvo}
 IconSizeWithArrow {type=Vector2D, default=<54 36>}
 IconSizeWithoutArrow {type=Vector2D, default=<36 36>}
 IsPTTKey {type=boolean}
 isSocialBanned {type=boolean}
 IsUsingEpicParties {type=boolean}
 IsVoiceChatMethodPTT {type=boolean}
 Last Focused Widget {type=CommonButtonBase, default=null}
 on Clicked {}
 on Hovered {}
 on Unhovered {}
 ReportMenuActionContext {type=UIKitContextMenuActionContext, default=null}
]
widgets {
 info {inheritedWidgetCount=0, parentClass=&CommonActivatableWidget, rootWidgetClass=&GridPanel, widgetCount=2}
 widgets [
  GridPanel_0 {widgetClassPath=&GridPanel, widget=GridPanel_0}
  Button_AddFriend {
   widgetClassPath=/Sidebar/Buttons/WBP_Sidebar_ButtonGeneric
   bIsVariable=true
   parent=GridPanel_0
   widget=Button_AddFriend
   properties {
    icon=/Sidebar/Materials/Icons/Textures/T_UI_Icon_Sidebar_AddFriend_MSDF
    iconMaterial=/Sidebar/Materials/Icons/MI_Sidebar_Icon_PartyOptions
    minHeight=54
    minWidth=60
    text=Add
    useDisabledVisualsOnLocked=true
    useIconBlock=true
    buttonPadding=<ltrb 0 0 0 0>
    font {
     bForceMonospaced=false
     bMaterialIsStencil=false
     fontMaterial=/Sidebar/Materials/Fonts/MI_Sidebar_FontColor_Secondary
     fontObject=&HeadingNow
     letterSpacing=10
     monospacedWidth=1
     size=13.5
     skewAmount=0
     typefaceFontName=74 Regular
     outlineSettings {
      bApplyOutlineToDropShadows=false
      bMiteredCorners=false
      bSeparateFillAlpha=false
      outlineMaterial=/FortUILibrary/Systems/ModularButton/MaterialInstances/MI_UIKit_TextOutline
      outlineSize=0
      outlineColor=<rgba 0 0 0 1>
     }
    }
    iconPadding=<ltrb 8 0 8 0>
    textPadding=<ltrb 0 0 12 0>
   }
   slot {class=GridSlot, properties {verticalAlignment=VAlign_Center, padding=<ltrb 12 0 12 0>}}
  }
 ]
}
