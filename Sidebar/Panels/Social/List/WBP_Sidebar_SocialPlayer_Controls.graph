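# Sidebar widget for the local player's voice controls: mic, headphones (deafen) and more-options buttons bound to VoiceChannelManagementVM.
# Mute and deafen changes swap icons; push-to-talk changes mic click method and shows the bound key.
# Touch with HUD mic support forces open-mic behavior.
# A context menu offers text chat, communication settings, voice-mode switching and reporting, gating text chat by portrait compact UI.
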WidgetBlueprint /Sidebar/Panels/Social/List/WBP_Sidebar_SocialPlayer_Controls
graphs [
 AddReportConversationContextMenuItem {nodes [
   AddReportConversationContextMenuItem!() -> (then, IsPartyChannel:bool=~) {is=title K2Node_FunctionEntry}
   > RemoveInteraction!(
    execute
    self:UIKit Context Menu Action Context Object Reference=Get ReportMenuActionContext() -> (ReportMenuActionContext:UIKit Context Menu Action Context Object Reference)
    InteractionNameID:Name=Get ContextMenuReportName() -> (ContextMenuReportName:Name=None)
   ) -> (then)
   > AddInteraction!(
    execute
    self:UIKit Context Menu Action Context Object Reference=Get ReportMenuActionContext()
    InActionInfo_DisplayName:Text="LOCTEXT(\"Report chat\")"
    InActionInfo_IconTexture:Texture 2D Soft Object Reference
    InActionInfo_IconMaterial:Material Instance Soft Object Reference
    InActionInfo_SortPriority:int=0
    InActionInfo_bIsAvailable:bool=True
    InActionInfo_bIsEnabled:bool=True
    InActionInfo_bIsSelected:bool=False
    InActionInfo_bAddDivider:bool=False
    InActionInfo_bPositionDividerAfter:bool=False
    InActionInfo_bCloseMenuOnSelect:bool=False
    InActionInfo_InteractionNameID:Name=Get ContextMenuReportName() -> (ContextMenuReportName=None)
    InActionInfo_Interaction:Data Table Row Handle Structure=(DataTable=None,RowName="")
    InActionInfo_PressedDelegate:Delegate="(null).None"
    InActionInfo_ReleasedDelegate:Delegate="(null).None"
    InInteractionDelegate:"Delegate (by ref)"=Create Event()
   ) -> (then)
  ]}
 BP_GetDesiredFocusTarget {nodes [
   Get Desired Focus Target()
   > n1 = Get!(execute:Exec) -> (then:Exec, else:Exec, Last Focused Widget:Common Button Base Object Reference)
   > Return Node!(execute, ReturnValue:Widget Object Reference=@n1)
   Return Node(execute=@n1.else, ReturnValue=Get Button_Mic() -> (Button_Mic:WBP Sidebar Button Generic Object Reference))
  ]}
 CF_StatusToEnabled {nodes [
   n1 = CF_StatusToEnabled!() -> (then, Status:EVoiceChatPartyClientActiveState Enum=NoClientsActiveInParty, isSocialBanned:bool=~) {is=title K2Node_FunctionEntry}
   > Return Node!(
    execute
    Enabled:bool=OR Boolean(
     A="Not Equal (Enum)"!(A:EVoiceChatPartyClientActiveState Enum=@n1, B:EVoiceChatPartyClientActiveState Enum=ThisClientActiveInParty) -> (ReturnValue:bool)
     B=@n1.isSocialBanned
    )
   )
  ]}
 EventGraph {nodes [
   "On Clicked (Button_MoreOptions)"!() -> (OutputDelegate, then, Button:Common Button Base Object Reference) {is=title K2Node_ComponentBoundEvent}
   > Open!(
    execute
    self:Menu Anchor Object Reference=Get Anchor!(self:UIKit Context Menu Anchor Object Reference=Get Report_ContextMenu() -> (Report_ContextMenu:WBP UIKit Context Menu Anchor Button Object Reference)) -> (Anchor:Menu Anchor Object Reference)
    bFocusMenu:bool=true
   ) -> (then)
   "On Clicked (Button_Headphones)"!() -> (OutputDelegate, then, Button:Common Button Base Object Reference) {is=title K2Node_ComponentBoundEvent}
   n1 = Get VoiceChannelManagementVM() -> (VoiceChannelManagementVM:Voice Channel Management VM Object Reference)
   > SetSelfIsDeafened!(
    execute
    self:Voice Channel Management VM Object Reference=@n1
    bDeafenSelf:bool=NOT Boolean(A=GetSelfIsDeafened!(self:Voice Channel Management VM Object Reference=@n1) -> (ReturnValue:bool))
   ) -> (then)
   "On Context Menu Closed (Report_ContextMenu)"() {is=title K2Node_ComponentBoundEvent}
   > RequestRefreshFocus(self:Common Activatable Widget Object Reference)
   > DeactivateWidget()
   n2 = "On Hovered (Button_Mic)"!() -> (OutputDelegate, then, Button:Common Button Base Object Reference) {is=title K2Node_ComponentBoundEvent}
   > Set Last Focused Widget(Last Focused Widget:Common Button Base Object Reference=@n2.Button) -> (Output_Get:Common Button Base Object Reference)
   n3 = "On Hovered (Button_Headphones)"!() -> (OutputDelegate, then, Button:Common Button Base Object Reference) {is=title K2Node_ComponentBoundEvent}
   > Set Last Focused Widget(Last Focused Widget=@n3.Button)
   n4 = "On Hovered (Button_MoreOptions)"!() -> (OutputDelegate, then, Button:Common Button Base Object Reference) {is=title K2Node_ComponentBoundEvent}
   > Set Last Focused Widget(Last Focused Widget=@n4.Button)
   "On Clicked (Button_Mic)"!() -> (OutputDelegate, then, Button:Common Button Base Object Reference) {is=title K2Node_ComponentBoundEvent}
   n5 = Get VoiceChannelManagementVM()
   > Branch(Condition=GetAllowPTTOnClickOrTouch!(self:Voice Channel Management VM Object Reference=@n5) -> (ReturnValue:bool))
   n6 = Get IsVoiceChatMethodPTT() -> (IsVoiceChatMethodPTT:bool)
   n7 = GetSelfIsMuted!(self:Voice Channel Management VM Object Reference=@n5) -> (ReturnValue:bool)
   > SetSelfIsMuted!(execute, self:Voice Channel Management VM Object Reference=@n5, bMuteSelf:bool=AND Boolean(A=NOT Boolean(A=@n6), B=NOT Boolean(A=@n7))) -> (then)
   "On Context Menu Opened (Report_ContextMenu)"() {is=title K2Node_ComponentBoundEvent}
   > ActivateWidget()
   OnOpenChatEvent!() -> (OutputDelegate, then, InteractionNameID:Name) {is=title K2Node_CustomEvent}
   n8 = Get IsUsingEpicParties() -> (IsUsingEpicParties:bool)
   n9 = Get VoiceChannelManagementVM()
   > MoveTextChatToCompactUIAndCloseSidebar!(
    execute
    self:Fort Sidebar VM Object Reference=Get FortSidebarVM() -> (FortSidebarVM:Fort Sidebar VM Object Reference)
    bShowParty:bool="Equal (Enum)"(
     A:ESidebarVoiceChannelType Enum=Get ActiveChannel!(self:Voice Channel Management VM Object Reference=@n9) -> (ActiveChannel:ESidebarVoiceChannelType Enum=None)
     B:ESidebarVoiceChannelType Enum=Select(Option 0:ESidebarVoiceChannelType Enum=Party, Option 1:ESidebarVoiceChannelType Enum=EpicParty, Index=@n8) -> (ReturnValue:ESidebarVoiceChannelType Enum)
    )
   ) -> (then)
   OnCommunicationSettingsContextMenu!() -> (OutputDelegate, then, InteractionNameID:Name) {is=title K2Node_CustomEvent}
   > RequestCloseSidebar(self:Fort Sidebar VM Object Reference=Get FortSidebarVM())
   > n10 = Branch(Condition=IsSaveTheWorldUI!() -> (IsSaveTheWorld:bool))
   > n11 = Delay(Duration=0.2)
   OpenSettingsToVoiceChat!(execute=[@n11.then, @n10.else], OwningWidget:Common User Widget Object Reference=Self-Reference()) -> (then)
   SetVoiceChatMethodToOpenMic!() -> (OutputDelegate, then, InteractionNameID:Name) {is=title K2Node_CustomEvent}
   > SetVoiceChatMethod!(execute, self:Voice Channel Management VM Object Reference=Get VoiceChannelManagementVM(), ChatMethod:EFortVoiceChatMethod Enum=OpenMic) -> (then)
   SetVoiceChatMethodToPTT!() -> (OutputDelegate, then, InteractionNameID:Name) {is=title K2Node_CustomEvent}
   > SetVoiceChatMethod(self=Get VoiceChannelManagementVM(), ChatMethod=PTT)
   OnReportContextMenu!() -> (OutputDelegate, then, InteractionNameID:Name) {is=title K2Node_CustomEvent}
   > ReportCurrentChannel(self:Voice Channel Management VM Object Reference=Get VoiceChannelManagementVM())
   OrientationChanged!() -> (OutputDelegate, then, inScreenOrientation:EScreenOrientation Enum=Unknown) {is=title K2Node_CustomEvent}
   > UpdateOpenTextChatContextOption()
   Button Mic On Released!() -> (OutputDelegate, then, Button:WBP UIKit Button Generic Base Object Reference) {is=title K2Node_CustomEvent}
   n12 = Get VoiceChannelManagementVM()
   > Branch(Condition=AND Boolean(A=GetAllowPTTOnClickOrTouch(self=@n12), B=Get IsVoiceChatMethodPTT()))
   > SetSelfIsMuted(self=@n12, bMuteSelf=true)
   Event Destruct()
   > n13 = Cast To PlatformGameInstance!(execute, Object=GetGameInstance!() -> (ReturnValue:Game Instance Object Reference)) -> (then, CastFailed, AsPlatform Game Instance:Platform Game Instance Object Reference) {is=title K2Node_DynamicCast}
   n14 = Create Event()
   > Unbind Event from Application Received Screen Orientation Changed Notification Delegate(self:Platform Game Instance Object Reference=@n13, Delegate=@n14) {is=title K2Node_RemoveDelegate}
   n15 = Create Event()
   n16 = Get Button_Mic()
   > Unbind Event from On Released(self:WBP UIKit Button Generic Base Object Reference=@n16, Delegate=@n15) {is=title K2Node_RemoveDelegate}
   Event OnInitialized()
   > SetupReportContextMenu()
   > UpdateOpenTextChatContextOption()
   Event Construct()
   > Initialize(self:Voice Channel Management VM Object Reference=Get VoiceChannelManagementVM())
   > SetVisibility(InVisibility=Visible)
   > UpdatePTTKey()
   > n17 = GetConsoleVariableBoolValue!(execute, VariableName:String=fort.UseEpicParties) -> (then, ReturnValue:bool)
   > Set IsUsingEpicParties(IsUsingEpicParties:bool=@n17) -> (Output_Get:bool)
   > n18 = Cast To PlatformGameInstance(Object=GetGameInstance())
   > Bind Event to Application Received Screen Orientation Changed Notification Delegate(self:Platform Game Instance Object Reference=@n18, Delegate=@n14) {is=title K2Node_AddDelegate}
   > Bind Event to On Released(self:WBP UIKit Button Generic Base Object Reference=@n16, Delegate=@n15) {is=title K2Node_AddDelegate}
   Event OnMouseEnter()
   > Call On Hovered!(execute:Exec, self:WBP Sidebar Social Player Controls Object Reference) -> (then:Exec) {is=title K2Node_CallDelegate}
   Event OnMouseLeave!() -> (OutputDelegate, then, MouseEvent:"Pointer Event Structure (by ref)") {is=title K2Node_Event}
   > Call On Unhovered!(execute:Exec, self:WBP Sidebar Social Player Controls Object Reference) -> (then:Exec) {is=title K2Node_CallDelegate}
  ]}
 NavigteLeftFromMoreOptions {nodes [
   NavigteLeftFromMoreOptions!() -> (then, Navigation:EUINavigation Enum) {is=title K2Node_FunctionEntry}
   n1 = Get Button_Headphones() -> (Button_Headphones:WBP Sidebar Button Generic Object Reference)
   > Branch(Condition=IsVisible(self=@n1))
   > Return Node!(execute, ReturnValue:Widget Object Reference=@n1)
  ]}
 OnActiveChannelChanged {nodes [
   n1 = OnActiveChannelChanged!() -> (then, ActiveChannel:ESidebarVoiceChannelType Enum=None) {is=title K2Node_FunctionEntry}
   n2 = Get IsUsingEpicParties()
   > AddReportConversationContextMenuItem!(execute, self, IsPartyChannel:bool="Equal (Enum)"(A=@n1, B=Select(Option 0=Party, Option 1=EpicParty, Index=@n2))) -> (then) {is=title K2Node_CallFunction}
  ]}
 OnDeafenChanged {nodes [
   n1 = OnDeafenChanged!() -> (then, IsDeafen:bool=~) {is=title K2Node_FunctionEntry}
   n2 = Get Button_Headphones()
   > Set ButtonType!(
    execute
    ButtonType:E_UI_Sidebar_Button_Type Enum=Select(Option 0:E_UI_Sidebar_Button_Type Enum=NewEnumerator0, Option 1:E_UI_Sidebar_Button_Type Enum=NewEnumerator6, Index=@n1) -> (ReturnValue:E_UI_Sidebar_Button_Type Enum)
    self:WBP Sidebar Button Generic Object Reference=@n2
   ) -> (then, Output_Get:E_UI_Sidebar_Button_Type Enum=NewEnumerator0)
   > InitializeButtonVisuals(self:WBP Sidebar Button Generic Object Reference=@n2)
   > SetIconMaterial!(
    execute
    self:WBP UIKit Button Generic Base Object Reference=Get Button_Headphones()
    Material Instance:Material Instance Object Reference=Select(
     Option 0:Material Instance Object Reference=/Sidebar/Materials/Icons/MI_Sidebar_Icon_Base
     Option 1:Material Instance Object Reference=/Sidebar/Materials/Icons/MI_Sidebar_Icon_Alert
     Index=Get IsDeafen() -> (IsDeafen:bool)
    ) -> (ReturnValue:Material Instance Object Reference)
   ) -> (then)
   n3 = Get Button_Headphones()
   > SetIcon!(
    execute
    self:WBP UIKit Button Generic Base Object Reference=@n3
    Icon:Object Reference=Select(
     Option 0:Object Reference=/Sidebar/Materials/Icons/Textures/T_UI_Icon_Sidebar_Headphones_MSDF
     Option 1:Object Reference=/Sidebar/Materials/Icons/Textures/T_UI_Icon_Sidebar_HeadphonesSlash_MSDF
     Index=Get IsDeafen()
    ) -> (ReturnValue:Object Reference)
   ) -> (then)
   > SetIconSize!(
    execute
    self:WBP UIKit Button Generic Base Object Reference=@n3
    Icon:Vector 2D Structure=Get IconSizeWithoutArrow() -> (IconSizeWithoutArrow:Vector 2D Structure)
   ) -> (then)
   n4 = Get IsDeafen()
   n5 = Get IsVoiceChatMethodPTT()
   > SetIsEnabled(self=Get Button_Mic(), bInIsEnabled=NOT Boolean(A=AND Boolean(A=@n4, B=@n5)))
  ]}
 OnGetVoiceChatMethod {nodes [
   n1 = OnGetVoiceChatMethod!() -> (then, VoiceChatMethod:EFortVoiceChatMethod Enum=OpenMic) {is=title K2Node_FunctionEntry}
   > Set ForceOpenMic(ForceOpenMic:bool=ShouldAlwaysActAsOpenMic!(self) -> (ReturnValue:bool)) -> (Output_Get:bool)
   n2 = Get ForceOpenMic() -> (ForceOpenMic:bool)
   > Set IsVoiceChatMethodPTT(IsVoiceChatMethodPTT:bool=AND Boolean(A="Equal (Enum)"(A:EFortVoiceChatMethod Enum=@n1, B:EFortVoiceChatMethod Enum=PTT), B=NOT Boolean(A=@n2))) -> (Output_Get:bool)
   n3 = Get IsVoiceChatMethodPTT()
   n4 = Get Button_Mic()
   > SetPadding!(
    execute
    self:User Widget Object Reference=Get Block_Icon!(self:WBP UIKit Button Generic Base Object Reference=@n4) -> (Block_Icon:WBP UIKit Block Image Object Reference)
    InPadding:Margin Structure=Make Margin(Left=0, Top=0, Right=SelectFloat(A=24, B=0, bPickA=@n3), Bottom=0)
   ) -> (then)
   > SetClickMethod!(
    execute
    self:Common Button Base Object Reference=@n4
    InClickMethod:EButtonClickMethod Enum=Select(Option 0:EButtonClickMethod Enum=DownAndUp, Option 1:EButtonClickMethod Enum=MouseDown, Index=Get IsVoiceChatMethodPTT()) -> (ReturnValue:EButtonClickMethod Enum)
   ) -> (then)
   > UpdatePTTKey()
   > UpdateToggleVoiceMethodContextOptions()
  ]}
 OnMutedChanged {nodes [
   OnMutedChanged!() -> (then, IsSelfMuted:bool=~) {is=title K2Node_FunctionEntry}
   n1 = Get Button_Mic()
   > Set ButtonType(
    ButtonType=Select(
     Option 0:E_UI_Sidebar_Button_Type Enum=NewEnumerator0
     Option 1:E_UI_Sidebar_Button_Type Enum=NewEnumerator6
     Index=Get IsSelfMuted() -> (IsSelfMuted:bool)
    ) -> (ReturnValue:E_UI_Sidebar_Button_Type Enum)
    self=@n1
   ) -> (Output_Get=NewEnumerator0)
   > InitializeButtonVisuals(self=@n1)
   > SetIconMaterial(
    self=@n1
    Material Instance=Select(
     Option 0:Material Instance Object Reference=&MI_Sidebar_Icon_Base
     Option 1:Material Instance Object Reference=&MI_Sidebar_Icon_Alert
     Index=Get IsSelfMuted()
    ) -> (ReturnValue:Material Instance Object Reference)
   )
   n2 = Get Button_Mic()
   > SetIcon(
    self=@n2
    Icon=Select(
     Option 0:Object Reference=/Sidebar/Materials/Icons/Textures/T_UI_Icon_Sidebar_Mic_MSDF
     Option 1:Object Reference=/Sidebar/Materials/Icons/Textures/T_UI_Icon_Sidebar_MicSlash_MSDF
     Index=Get IsSelfMuted()
    ) -> (ReturnValue:Object Reference)
   )
   > SetIconSize(self=@n2, Icon=Get IconSizeWithoutArrow())
   n3 = Get Button_Mic()
   > RequestInstantTransition!(
    execute
    self:UIKit Block Object Reference=Get Block_Icon(self=@n3)
    State:EUIKitBlockInstantTransitionState Enum=Select(
     Option 0:EUIKitBlockInstantTransitionState Enum=InstantDisabled
     Option 1:EUIKitBlockInstantTransitionState Enum=InstantEnabled
     Index=GetIsEnabled!(self:Widget Object Reference=@n3) -> (ReturnValue:bool)
    ) -> (ReturnValue:EUIKitBlockInstantTransitionState Enum)
   ) -> (then)
  ]}
 SetupReportContextMenu {nodes [
   SetupReportContextMenu() {is=title K2Node_FunctionEntry}
   > n1 = Construct Object from Class(Class=/Script/FortniteUI.UIKitContextMenuActionContext) -> (ReturnValue:UIKit Context Menu Action Context Object Reference)
   > n2 = Set ReportMenuActionContext(ReportMenuActionContext:UIKit Context Menu Action Context Object Reference=@n1) -> (Output_Get:UIKit Context Menu Action Context Object Reference)
   > SetContext!(
    execute
    self:WBP UIKit Context Menu Anchor Button Base Object Reference=Get Report_ContextMenu()
    InteractionContext_Data:UIKit Context Menu Action Context Object Reference=@n2
   ) -> (then)
   n3 = Get ReportMenuActionContext()
   > AddInteraction(
    self=@n3
    InActionInfo_DisplayName="LOCTEXT(\"Open text chat\")"
    InActionInfo_SortPriority=18
    InActionInfo_bIsAvailable=True
    InActionInfo_bIsEnabled=True
    InActionInfo_bIsSelected=False
    InActionInfo_bAddDivider=False
    InActionInfo_bPositionDividerAfter=False
    InActionInfo_bCloseMenuOnSelect=False
    InActionInfo_InteractionNameID=Get ContextMenuOpenChatName() -> (ContextMenuOpenChatName:Name=None)
    InActionInfo_Interaction=(DataTable=None,RowName="")
    InActionInfo_PressedDelegate="(null).None"
    InActionInfo_ReleasedDelegate="(null).None"
    InInteractionDelegate=Create Event()
   )
   > AddInteraction(
    self=@n3
    InActionInfo_DisplayName="LOCTEXT(\"Communication settings\")"
    InActionInfo_SortPriority=20
    InActionInfo_bIsAvailable=True
    InActionInfo_bIsEnabled=True
    InActionInfo_bIsSelected=False
    InActionInfo_bAddDivider=False
    InActionInfo_bPositionDividerAfter=False
    InActionInfo_bCloseMenuOnSelect=False
    InActionInfo_InteractionNameID=None
    InActionInfo_Interaction=(DataTable=None,RowName="")
    InActionInfo_PressedDelegate="(null).None"
    InActionInfo_ReleasedDelegate="(null).None"
    InInteractionDelegate=Create Event()
   )
   > AddInteraction(
    self=@n3
    InActionInfo_DisplayName="LOCTEXT(\"Switch to open mic\")"
    InActionInfo_SortPriority=2
    InActionInfo_bIsAvailable=false
    InActionInfo_bIsEnabled=True
    InActionInfo_bIsSelected=False
    InActionInfo_bAddDivider=False
    InActionInfo_bPositionDividerAfter=False
    InActionInfo_bCloseMenuOnSelect=False
    InActionInfo_InteractionNameID=SetOpenMic
    InActionInfo_Interaction=(DataTable=None,RowName="")
    InActionInfo_PressedDelegate="(null).None"
    InActionInfo_ReleasedDelegate="(null).None"
    InInteractionDelegate=Create Event()
   )
   > AddInteraction(
    self=@n3
    InActionInfo_DisplayName="LOCTEXT(\"Switch to push to talk\")"
    InActionInfo_SortPriority=2
    InActionInfo_bIsAvailable=false
    InActionInfo_bIsEnabled=True
    InActionInfo_bIsSelected=False
    InActionInfo_bAddDivider=False
    InActionInfo_bPositionDividerAfter=False
    InActionInfo_bCloseMenuOnSelect=False
    InActionInfo_InteractionNameID=SetPTT
    InActionInfo_Interaction=(DataTable=None,RowName="")
    InActionInfo_PressedDelegate="(null).None"
    InActionInfo_ReleasedDelegate="(null).None"
    InInteractionDelegate=Create Event()
   )
  ]}
 ShouldAlwaysActAsOpenMic {nodes [
   ShouldAlwaysActAsOpenMic() {is=title K2Node_FunctionEntry}
   n1 = GetOwningPlayer()
   n2 = GetRuntimeOptions()
   > Return Node(ReturnValue:bool=AND Boolean(
     A=Get bSupportsTouchHudMicButton!(self:Fort Runtime Options Object Reference=@n2) -> (bSupportsTouchHudMicButton:bool)
     B=IsUsingInputType!(InputType:ECommonInputType Enum=Touch, PlayerController:Player Controller Object Reference=@n1) -> (ReturnValue:bool)
    ))
  ]}
 UpdateOpenTextChatContextOption {nodes [
   UpdateOpenTextChatContextOption() {is=title K2Node_FunctionEntry}
   > n1 = GetConsoleVariableBoolValue(VariableName=PilgrimInverse.UsePortraitMode)
   n2 = IsUsingCompactUI()
   n3 = "Equal (Enum)"(
    A:EDisplayOrientationType Enum=DisplayOrientationFromScreenOrientation!(ScreenOrientation:EScreenOrientation Enum=GetDeviceOrientation!() -> (ReturnValue:EScreenOrientation Enum=Unknown)) -> (DisplayOrientation:EDisplayOrientationType Enum=Landscape)
    B:EDisplayOrientationType Enum=Portrait
   )
   > SetInteractionAvailability!(
    execute
    self:UIKit Context Menu Action Context Object Reference=Get ReportMenuActionContext()
    InteractionNameID:Name=Get ContextMenuOpenChatName() -> (ContextMenuOpenChatName=None)
    bInAvailability:bool=NOT Boolean(A=AND Boolean!(A=@n1, B=@n3, C:bool=@n2) -> (ReturnValue))
   ) -> (then)
  ]}
 UpdatePTTKey {nodes [
   UpdatePTTKey() {is=title K2Node_FunctionEntry}
   > n1 = Branch(Condition=Get IsVoiceChatMethodPTT())
   n2 = Get PTT_Key() -> (PTT_Key:Image Object Reference)
   > SetVisibility(self=@n2, InVisibility=SelfHitTestInvisible)
   > SetButtonPadding!(
    execute
    self:WBP UIKit Button Generic Base Object Reference=Get Button_Mic()
    Padding_Left:float=10
    Padding_Top:float=0
    Padding_Right:float=12
    Padding_Bottom:float=0
   ) -> (then)
   n3 = Get PushToTalkKeybindingSettingVM!(self:Voice Channel Management VM Object Reference=Get VoiceChannelManagementVM()) -> (PushToTalkKeybindingSettingVM:Fort Social Setting Keybind VM Object Reference)
   n4 = Get FortGlobalUIContext() -> (ReturnValue:Fort Global UIContext Object Reference) {is=title K2Node_GetSubsystem}
   > n5 = Set CurrentKey(CurrentKey:Key Structure=GetKeyForAction!(
     self:Fort Global UIContext Object Reference=@n4
     Action:Name=GetActionName!(self:Fort Social Setting Keybind VM Object Reference=@n3) -> (ReturnValue:Name=None)
     bForceGamepadKey:bool=false
     PresetNameOverride:Name=None
    ) -> (ReturnValue:Key Structure)) -> (Output_Get:Key Structure)
   > n6 = Set IsPTTKey(IsPTTKey:bool=Is Valid Key!(Key:"Key Structure (by ref)"=@n5) -> (ReturnValue:bool)) -> (Output_Get:bool)
   > SetBrush(
    self=Get PTT_Key()
    InBrush=Select(
     Option 0:"Slate Brush Structure (by ref)"=Make Slate Brush!(
      TintColor_SpecifiedColor:Linear Color Structure=(R=0.9301,G=1,B=0.0103,A=1)
      TintColor_ColorUseRule:ESlateColorStylingMode Enum
      DrawAs:ESlateBrushDrawType Enum=Image
      Tiling:ESlateBrushTileType Enum=NoTile
      Mirroring:ESlateBrushMirrorType Enum=NoMirror
      ImageSize_X:double=24
      ImageSize_Y:double=24
      Margin:Margin Structure=(Left=0,Top=0,Right=0,Bottom=0)
      ResourceObject:Object Reference=/FortUILibrary/Textures/MaterialInstances/MI_UIKit_Icon_Alert
      OutlineSettings:Slate Brush Outline Settings Structure=(CornerRadii=(X=0,Y=0,Z=0,W=0),Color=(SpecifiedColor=(R=0,G=0,B=0,A=0),ColorUseRule=UseColor_Specified),Width=0,RoundingType=HalfHeightRadius,bUseBrushTransparency=False)
     ) -> (SlateBrush:Slate Brush Structure) {is=title K2Node_MakeStruct}
     Option 1:"Slate Brush Structure (by ref)"=GetBrushForKey!(self:Fort Global UIContext Object Reference=@n4, Key:Key Structure=@n5) -> (Brush:Slate Brush Structure, ReturnValue:bool)
     Index=@n6
    ) -> (ReturnValue:"Slate Brush Structure (by ref)")
   )
   SetVisibility(execute=@n1.else, self=@n2, InVisibility=Collapsed)
   > SetButtonPadding!(execute, self=Get Button_Mic(), Padding:Margin Structure) -> (then)
  ]}
 UpdateToggleVoiceMethodContextOptions {nodes [
   UpdateToggleVoiceMethodContextOptions() {is=title K2Node_FunctionEntry}
   n1 = Get IsVoiceChatMethodPTT()
   n2 = Get ReportMenuActionContext()
   > SetInteractionAvailability(self=@n2, InteractionNameID=Select(Option 0:Name=SetOpenMic, Option 1:Name=SetPTT, Index=@n1) -> (ReturnValue:Name), bInAvailability=false)
   > SetInteractionAvailability(
    self=@n2
    InteractionNameID=Select(Option 0=SetPTT, Option 1=SetOpenMic, Index=@n1)
    bInAvailability=NOT Boolean(A=ShouldAlwaysActAsOpenMic() {is=title K2Node_CallFunction})
   )
  ]}
 on Hovered {nodes [on Hovered() {is=title K2Node_FunctionEntry}]}
 on Unhovered {nodes [on Unhovered() {is=title K2Node_FunctionEntry}]}
]
variables [
 ContextMenuMuteName {type=string, default=MuteUnmute}
 ContextMenuOpenChatName {type=string, default=OpenTextChat}
 ContextMenuReportName {type=string, default=ReportConvo}
 IconSizeWithArrow {type=Vector2D, default=<54 36>}
 IconSizeWithoutArrow {type=Vector2D, default=<36 36>}
 IsPTTKey {type=boolean}
 isSocialBanned {type=boolean}
 IsUsingEpicParties {type=boolean}
 IsVoiceChatMethodPTT {type=boolean}
 Last Focused Widget {type=CommonButtonBase, default=null}
 on Hovered {}
 on Unhovered {}
 ReportMenuActionContext {type=UIKitContextMenuActionContext, default=null}
]
widgets {
 info {
  inheritedWidgetCount=0
  namedSlotCount=2
  parentClass=&CommonActivatableWidget
  rootWidgetClass=&GridPanel
  widgetCount=9
 }
 widgets [
  GridPanel_0 {widgetClassPath=&GridPanel, widget=GridPanel_0}
  HB_Buttons {
   widgetClassPath=&HorizontalBox
   parent=GridPanel_0
   widget=HB_Buttons
   properties {bOverride_Cursor=true}
   slot {class=GridSlot, properties {layer=1, verticalAlignment=VAlign_Center}}
  }
  Button_Mic {
   widgetClassPath=/Sidebar/Buttons/WBP_Sidebar_ButtonGeneric
   bIsVariable=true
   parent=HB_Buttons
   widget=Button_Mic
   properties {
    icon=&T_UI_Icon_Sidebar_Mic_MSDF
    iconMaterial=/Sidebar/Materials/Icons/MI_Sidebar_Icon_PartyOptions
    minHeight=54
    minWidth=54
    navigation=&self:WidgetTree.Button_Mic.WidgetNavigation_0
    useDisabledVisualsOnLocked=true
    useIconBlock=true
    useTextBlock=false
    buttonPadding=<ltrb 0 0 0 0>
    font {
     bForceMonospaced=false
     bMaterialIsStencil=false
     fontMaterial=/FortUILibrary/UIKit/Buttons/MaterialInstances/CTAButtons/MI_UIKit_TextColor_Quiet
     fontObject=&HeadingNow
     letterSpacing=10
     monospacedWidth=1
     size=15
     skewAmount=0
     typefaceFontName=74 Regular
     outlineSettings {
      bApplyOutlineToDropShadows=false
      bMiteredCorners=false
      bSeparateFillAlpha=false
      outlineMaterial=/FortUILibrary/Systems/ModularButton/MaterialInstances/MI_UIKit_TextOutline
      outlineSize=0
      outlineColor=<rgba 0 0 0 1>
     }
    }
   }
   slot {class=HorizontalBoxSlot, properties {verticalAlignment=VAlign_Center, padding=<ltrb 0 0 6 0>}}
  }
  Overlay_28 {widgetClassPath=&Overlay, namedSlotHost=Button_Mic, widget=Overlay_28}
  PTT_Key {
   widgetClassPath=&Image
   bIsVariable=true
   parent=Overlay_28
   widget=PTT_Key
   properties {brush {
     bIsDynamicallyLoaded=false
     drawAs=Image
     imageType=FullColor
     mirroring=NoMirror
     resourceName=None
     resourceObject=null
     tiling=NoTile
     imageSize=<28 28>
     margin=<ltrb 0 0 0 0>
     outlineSettings {
      bUseBrushTransparency=false
      roundingType=HalfHeightRadius
      width=0
      color {colorUseRule=UseColor_Specified, specifiedColor=<rgba 0 0 0 0>}
      cornerRadii=<0 0 0 0>
     }
     tintColor {colorUseRule=UseColor_Specified, specifiedColor=<rgba 1 1 1 1>}
     uVRegion {bIsValid=false, max=<0 0>, min=<0 0>}
    }}
   slot {class=OverlaySlot, properties {horizontalAlignment=HAlign_Center, verticalAlignment=VAlign_Center, padding=<ltrb 20 0 0 0>}}
  }
  Button_Headphones {
   widgetClassPath=&WBP_Sidebar_ButtonGeneric
   bIsVariable=true
   parent=HB_Buttons
   widget=Button_Headphones
   properties {
    icon=&T_UI_Icon_Sidebar_Headphones_MSDF
    iconMaterial=&MI_Sidebar_Icon_PartyOptions
    minHeight=54
    minWidth=54
    navigation=&self:WidgetTree.Button_Headphones.WidgetNavigation_0
    useDisabledVisualsOnLocked=true
    useIconBlock=true
    useTextBlock=false
    buttonPadding=<ltrb 0 0 0 0>
    font {
     bForceMonospaced=false
     bMaterialIsStencil=false
     fontMaterial=&MI_UIKit_TextColor_Quiet
     fontObject=&HeadingNow
     letterSpacing=10
     monospacedWidth=1
     size=15
     skewAmount=0
     typefaceFontName=74 Regular
     outlineSettings {
      bApplyOutlineToDropShadows=false
      bMiteredCorners=false
      bSeparateFillAlpha=false
      outlineMaterial=&MI_UIKit_TextOutline
      outlineSize=0
      outlineColor=<rgba 0 0 0 1>
     }
    }
   }
   slot {class=HorizontalBoxSlot, properties {verticalAlignment=VAlign_Center, padding=<ltrb 0 0 6 0>}}
  }
  Button_MoreOptions {
   widgetClassPath=&WBP_Sidebar_ButtonGeneric
   bIsVariable=true
   parent=HB_Buttons
   widget=Button_MoreOptions
   properties {
    icon=/Sidebar/Materials/Icons/Textures/T_UI_Icon_Sidebar_ContextMenu_MSDF
    iconMaterial=&MI_Sidebar_Icon_PartyOptions
    minHeight=54
    minWidth=54
    navigation=&self:WidgetTree.Button_MoreOptions.WidgetNavigation_0
    useIconBlock=true
    useTextBlock=false
    buttonPadding=<ltrb 0 0 0 0>
    font {
     bForceMonospaced=false
     bMaterialIsStencil=false
     fontMaterial=&MI_UIKit_TextColor_Quiet
     fontObject=&HeadingNow
     letterSpacing=10
     monospacedWidth=1
     size=15
     skewAmount=0
     typefaceFontName=74 Regular
     outlineSettings {
      bApplyOutlineToDropShadows=false
      bMiteredCorners=false
      bSeparateFillAlpha=false
      outlineMaterial=&MI_UIKit_TextOutline
      outlineSize=0
      outlineColor=<rgba 0 0 0 1>
     }
    }
   }
   slot {class=HorizontalBoxSlot, properties {verticalAlignment=VAlign_Center, padding=<ltrb 0 0 12 0>}}
  }
  Overlay_1 {widgetClassPath=&Overlay, namedSlotHost=Button_MoreOptions, widget=Overlay_1}
  Report_ContextMenu {
   widgetClassPath=/FortUILibrary/UIKit/ContextMenu/WBP_UIKit_ContextMenuAnchorButton
   bIsVariable=true
   parent=Overlay_1
   widget=Report_ContextMenu
   properties {
    button Background Material=null
    button Outline Material=null
    buttonClass=/Game/UI/Frontend/Lobby/Nameplates/WBP_EmptyButton_ContextMenu.WBP_EmptyButton_ContextMenu_C
    buttonText=""
    closeMenuOnContextClicked=true
    menuPlacement=MenuPlacement_MenuLeft
    min Height=1
    priority=1
    button Padding=<ltrb 0 0 0 0>
   }
   slot {class=OverlaySlot, properties {horizontalAlignment=HAlign_Center, verticalAlignment=VAlign_Bottom}}
  }
 ]
}
