WidgetBlueprint /Sidebar/Panels/Social/List/WBP_Sidebar_YouIndicator
graphs [
 EventGraph {nodes [
   Event PreConstruct()
   > SetCornerRadius!(
    execute
    self:WBP Sidebar Block Backplate Object Reference=Get Backplate() -> (Backplate:WBP Sidebar Block Backplate Object Reference)
    Radius In Pixels:bool=false
    Corner Radius:double=1
   ) -> (then)
   > SetBackplateColor!(
    execute
    self:WBP Sidebar Block Backplate Object Reference=Get Backplate()
    Backplate Switch:E_UI_Sidebar_Button_Backplate_Color Enum=NewEnumerator5
   ) -> (then)
   > SetLabelText!(execute, self, Label Text:Text=Get LabelText() -> (LabelText:Text)) -> (then)
   > SetLabelColor!(execute, self, InColorAndOpacity:Slate Color Structure=Get LabelColor() -> (LabelColor:Slate Color Structure)) -> (then)
   > SetCompactUIStyle() {is=title K2Node_CallFunction}
  ]}
 SetCompactUIStyle {nodes [
   SetCompactUIStyle() {is=title K2Node_FunctionEntry}
   > Branch(Condition=IsUsingCompactUI())
   > SetHeightOverride(self=Get SB_Backplate() -> (SB_Backplate:Size Box Object Reference), InHeightOverride=22)
  ]}
 SetLabelColor {nodes [
   n1 = SetLabelColor!() -> (then, InColorAndOpacity:Slate Color Structure) {is=title K2Node_FunctionEntry}
   > SetColorAndOpacity(self=Get Label() -> (Label:Common Text Object Reference), InColorAndOpacity=@n1)
  ]}
 SetLabelText {nodes [
   n1 = SetLabelText!() -> (then, Label Text:Text) {is=title K2Node_FunctionEntry}
   > "SetText (Text)"(self=Get Label(), InText=@n1)
  ]}
]
variables [
 LabelColor {type=SlateColor, default {colorUseRule=UseColor_Specified, specifiedColor=<rgba 0.0048 0.004 0.0116 1>}}
 LabelText {type=string, default=YOU}
]
widgets {
 info {inheritedWidgetCount=0, parentClass=&CommonUserWidget, rootWidgetClass=&Overlay, widgetCount=4}
 widgets [
  Overlay_30 {widgetClassPath=&Overlay, widget=Overlay_30}
  SB_Backplate {
   widgetClassPath=&SizeBox
   bIsVariable=true
   parent=Overlay_30
   widget=SB_Backplate
   properties {bOverride_HeightOverride=true, heightOverride=19}
   slot {class=OverlaySlot, properties {horizontalAlignment=HAlign_Fill, verticalAlignment=VAlign_Center}}
  }
  Backplate {
   widgetClassPath=/Sidebar/Buttons/Blocks/WBP_Sidebar_Block_Backplate
   bIsVariable=true
   parent=SB_Backplate
   widget=Backplate
   slot {class=SizeBoxSlot}
  }
  Label {
   widgetClassPath=&CommonTextBlock
   bIsVariable=true
   parent=Overlay_30
   widget=Label
   properties {
    mobileFontSizeMultiplier=1.25
    text=YOU
    font {
     bForceMonospaced=false
     bMaterialIsStencil=false
     fontMaterial=null
     fontObject=&HeadingNow
     letterSpacing=0
     monospacedWidth=1
     size=10.5
     skewAmount=0
     typefaceFontName=76 Bold
     outlineSettings {
      bApplyOutlineToDropShadows=false
      bMiteredCorners=false
      bSeparateFillAlpha=false
      outlineMaterial=null
      outlineSize=0
      outlineColor=<rgba 0 0 0 1>
     }
    }
    shadowOffset=<0 0>
   }
   slot {class=OverlaySlot, properties {horizontalAlignment=HAlign_Center, verticalAlignment=VAlign_Center, padding=<ltrb 4 0 4 0>}}
  }
 ]
}
