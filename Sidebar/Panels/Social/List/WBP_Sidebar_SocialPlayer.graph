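# WBP_Sidebar_SocialPlayer is the sidebar social list row for a player, friend-request group, party or conversation, showing name, presence, leader crown and voice status.
# Setting the list item initializes view models, creating some lazily, with an optional CVar-gated immediate binding refresh.
# Conversion functions derive text, icons and visibility from that state.
# Interactions fill a right-click or touch-hold context menu, raising delegates for nickname edit, removal and blocking, executing others directly.
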WidgetBlueprint /Sidebar/Panels/Social/List/WBP_Sidebar_SocialPlayer
graphs [
 AddVMForLeaderIcon {nodes [
   n1 = AddVMForLeaderIcon!() -> (then, SocialUser:Fort Social User Object Reference) {is=title K2Node_FunctionEntry}
   n2 = Get HeaderItemType!(self:Fort Social Item VM Object Reference=Get FortSocialItemVM() -> (FortSocialItemVM:Fort Social Item VM Object Reference)) -> (HeaderItemType:ESocialHeaderItemType Enum=Unknown)
   > n3 = Branch(Condition=OR Boolean(
     A="Equal (Enum)"(A:ESocialHeaderItemType Enum=@n2, B:ESocialHeaderItemType Enum=PartyMembers)
     B="Equal (Enum)"(A=@n2, B=NativeEpicPartyMembers)
    ))
   > n4 = Get!(execute:Exec) -> (then:Exec, else:Exec, FortSocialUserPartyViewModel:Fort Social User Party View Model Object Reference)
   SetSocialUser!(
    execute=[@n4.then, @n6.then]
    self:Fort Social User Viewmodel Object Reference=Get FortSocialUserPartyViewModel() -> (FortSocialUserPartyViewModel:Fort Social User Party View Model Object Reference)
    InSocialUser:Fort Social User Object Reference=@n1
   ) -> (then)
   n5 = Construct Object from Class(execute=@n4.else, Class=/Script/FortniteUI.FortSocialUserPartyViewModel) -> (ReturnValue:Fort Social User Party View Model Object Reference)
   > n6 = Set FortSocialUserPartyViewModel(FortSocialUserPartyViewModel:Fort Social User Party View Model Object Reference=@n5) -> (Output_Get:Fort Social User Party View Model Object Reference)
   SetVisibility(execute=@n3.else, self=Get LeaderCrown() -> (LeaderCrown:Image Object Reference), InVisibility=Collapsed)
  ]}
 AddVMForLocalPlayerJoinability {nodes [
   n1 = AddVMForLocalPlayerJoinability!() -> (then, IsLocalPlayer:bool=~) {is=title K2Node_FunctionEntry}
   > Branch(Condition=@n1)
   > n2 = Get!(execute) -> (then, else, FortSocialSettingDiscreteVM_IsSelfJoinable:Fort Social Setting Discrete VM Object Reference)
   n3 = GetSettingStringId!(execute=[@n2.then, @n5.then], SettingValueId:EFortSocialSettingId Enum=PartyPrivacyJoinabilityID) -> (then, ReturnValue:String)
   > SetSettingId!(
    execute
    self:Fort Setting Value VM Object Reference=Get FortSocialSettingDiscreteVM_IsSelfJoinable() -> (FortSocialSettingDiscreteVM_IsSelfJoinable:Fort Social Setting Discrete VM Object Reference)
    InSettingId:String=@n3
   ) -> (then)
   n4 = Construct Object from Class(execute=@n2.else, Class=/Script/FortniteUI.FortSocialSettingDiscreteVM) -> (ReturnValue:Fort Social Setting Discrete VM Object Reference)
   > n5 = Set FortSocialSettingDiscreteVM_IsSelfJoinable(FortSocialSettingDiscreteVM_IsSelfJoinable:Fort Social Setting Discrete VM Object Reference=@n4) -> (Output_Get:Fort Social Setting Discrete VM Object Reference)
  ]}
 AddVMForVoiceIndicators {nodes [
   AddVMForVoiceIndicators() {is=title K2Node_FunctionEntry}
   > n1 = Sequence()
   > n2 = Get!(execute) -> (then, else, FortSocialUserVoiceChatStatusVM:Fort Social User Voice Chat Status View Model Object Reference)
   n3 = Construct Object from Class(execute=@n2.else, Class=/Script/FortniteUI.FortSocialUserVoiceChatStatusVM) -> (ReturnValue:Fort Social User Voice Chat Status View Model Object Reference)
   n4 = Set FortSocialUserVoiceChatStatusVM(execute=[@n3.then, @n2.then], FortSocialUserVoiceChatStatusVM:Fort Social User Voice Chat Status View Model Object Reference=@n3) -> (Output_Get:Fort Social User Voice Chat Status View Model Object Reference)
   > SetPlayerNetIdFromFortUser!(
    execute
    self:Fort Social User Voice Chat Status View Model Object Reference=@n4
    InSocialUser:Fort Social User Object Reference=Get TargetUser!(self:Fort Social Item VM Object Reference=Get FortSocialItemVM()) -> (TargetUser:Fort Social User Weak object)
   ) -> (then)
   > SetChannelType!(
    execute
    self:Fort Social User Voice Chat Status View Model Object Reference=@n4
    InChannelType:ESidebarVoiceChannelType Enum=Get ActiveChannel!(self:Voice Channel Management VM Object Reference=Get VoiceChannelManagementVM() -> (VoiceChannelManagementVM:Voice Channel Management VM Object Reference)) -> (ActiveChannel:ESidebarVoiceChannelType Enum=None)
   ) -> (then)
   n5 = AND Boolean(
    A=Get isLocalUserForBlueprint() -> (isLocalUserForBlueprint:bool)
    B=NOT Boolean(A=Get isUpdatedPartyControlsEnabled() -> (isUpdatedPartyControlsEnabled:bool))
   )
   n6 = ShouldShowVoiceIcons!() -> (ReturnValue:bool)
   SetVisibility(
    execute=@n1.then_1
    self=Get LocalUser_VoiceStatusIcons() -> (LocalUser_VoiceStatusIcons:Horizontal Box Object Reference)
    InVisibility=Select(Option 0=Collapsed, Option 1=Select(Option 0=Collapsed, Option 1=SelfHitTestInvisible, Index=@n5), Index=@n6)
   )
   n7 = Get isLocalUserForBlueprint()
   > SetVisibility(
    self=Get OtherPlayer_VoiceStatusIcons() -> (OtherPlayer_VoiceStatusIcons:Horizontal Box Object Reference)
    InVisibility=Select(Option 0=Collapsed, Option 1=Select(Option 0=SelfHitTestInvisible, Option 1=Collapsed, Index=@n7), Index=@n6)
   )
  ]}
 AllowVoiceChatSettingIndexChanged {nodes [
   n1 = AllowVoiceChatSettingIndexChanged!() -> (then, Setting Index:int) {is=title K2Node_FunctionEntry}
   > Set AllowVoiceChatSettingIndex(AllowVoiceChatSettingIndex:int=@n1) -> (Output_Get:int=0)
   > UpdateVisibility(self:Fort Social User Voice Chat Status View Model Object Reference=Get FortSocialUserVoiceChatStatusVM() -> (FortSocialUserVoiceChatStatusVM:Fort Social User Voice Chat Status View Model Object Reference))
  ]}
 BP_OnHoldTriggered {nodes [
   n1 = On Hold Triggered!() -> (then, CurrentInputType:ECommonInputType Enum) {is=title K2Node_FunctionEntry}
   > n2 = Switch on ECommonInputType!(execute, Selection:ECommonInputType Enum=@n1) -> (MouseAndKeyboard:Exec, Gamepad:Exec, Touch:Exec, Count:Exec) {is=title K2Node_SwitchEnum}
   SetIsSelected!(execute=@n2.Touch, self:Common Button Base Object Reference, InSelected:bool=true, bGiveClickFeedback:bool=false) -> (then)
   > ActivateInteractionList(self:Fort Social Item VM Object Reference=Get FortSocialItemVM())
   > PopulateContextMenu!(execute, self, Interactions:Array of Fort Social Interaction VM Object References) -> (then)
   > n3 = Branch(Condition=HasInteractionItems!(self:UIKit Context Menu Action Context Object Reference=Get ContextMenuActionContext() -> (ContextMenuActionContext:UIKit Context Menu Action Context Object Reference)) -> (ReturnValue:bool))
   > ToggleOpen!(
    execute
    self:Menu Anchor Object Reference=Get Anchor!(self:UIKit Context Menu Anchor Object Reference=Get PlayerAction_ContextMenu() -> (PlayerAction_ContextMenu:WBP UIKit Context Menu Anchor Button Object Reference)) -> (Anchor:Menu Anchor Object Reference)
    bFocusOnOpen:bool=false
   ) -> (then)
   > Return Node!(execute, ReturnValue:Event Reply Structure=Handled())
   Return Node(execute=[@n3.else, @n2.Gamepad, @n2.MouseAndKeyboard], ReturnValue=Unhandled())
  ]}
 Can Execute Interaction {nodes [
   n1 = Can Execute Interaction!() -> (then, Interaction Type:ESocialInteractionType Enum=AcceptEpicPartyInvite) {is=title K2Node_FunctionEntry}
   > Set isInteractionAvailable(isInteractionAvailable:bool=false) -> (Output_Get:bool)
   > n2 = "For Each Loop (Map)"!(
    execute:Exec
    MapPin:Map of Names to Fort Social Interaction VM Object References=Get InteractionsMap() -> (InteractionsMap:Map of Names to Fort Social Interaction VM Object References)
    BreakPin:Exec
   ) -> (then:Exec, KeyPin:Name, ValuePin:Fort Social Interaction VM Object Reference, CompletedPin:Exec)
   > Branch(Condition="Equal (Enum)"(
     A:ESocialInteractionType Enum=Get InteractionType!(self:Fort Social Interaction VM Object Reference=@n2.ValuePin) -> (InteractionType:ESocialInteractionType Enum=AcceptEpicPartyInvite)
     B:ESocialInteractionType Enum=@n1
    ))
   > Set isInteractionAvailable(isInteractionAvailable=true)
   > Set Execute Interaction(Execute Interaction:Fort Social Interaction VM Object Reference=@n2.ValuePin) -> (Output_Get:Fort Social Interaction VM Object Reference)
   Return Node!(
    execute=@n2.CompletedPin
    Can Execute:bool=Get isInteractionAvailable() -> (isInteractionAvailable:bool)
    Interaction:Fort Social Interaction VM Object Reference=Get Execute Interaction() -> (Execute Interaction:Fort Social Interaction VM Object Reference)
   )
  ]}
 ConstructContext {nodes [
   Inputs()
   > n1 = Get!(execute) -> (then, else, ContextMenuActionContext:UIKit Context Menu Action Context Object Reference)
   Outputs!(then:Exec=[@n4.then, @n1.then])
   n2 = Construct Object from Class(execute=@n1.else, Class=/Script/FortniteUI.UIKitContextMenuActionContext) -> (ReturnValue:UIKit Context Menu Action Context Object Reference)
   > n3 = Set ContextMenuActionContext(ContextMenuActionContext:UIKit Context Menu Action Context Object Reference=@n2) -> (Output_Get:UIKit Context Menu Action Context Object Reference)
   > n4 = SetContext!(
    execute
    self:WBP UIKit Context Menu Anchor Button Base Object Reference=Get PlayerAction_ContextMenu()
    InteractionContext_Data:UIKit Context Menu Action Context Object Reference=@n3
   ) -> (then)
  ]}
 Convert_HeaderTypeToMessageText {nodes [
   n1 = Convert_HeaderTypeToMessageText!() -> (then, Header Type:ESocialHeaderItemType Enum=Unknown, Presence:Text, Last Interaction Info:Text) {is=title K2Node_FunctionEntry}
   > n2 = Branch(Condition=Get isFriendRequestEntry() -> (isFriendRequestEntry:bool))
   > n3 = Get!(execute, self:Fort Social Item VM Object Reference=Get FortSocialItemVM()) -> (then, else, FriendRequests:Fort Friend Requests VM Object Reference)
   n4 = GetNumRequesterUsers!(self:Fort Friend Requests VM Object Reference=@n3) -> (ReturnValue:int=0)
   > n5 = Branch(Condition="integer > integer"(A=@n4, B=1))
   n6 = GetRequesterUsersDisplayNames!(self:Fort Friend Requests VM Object Reference=@n3) -> (ReturnValue:Text)
   n7 = "To String (Text)"(InText=@n6)
   > Return Node!(
    execute
    Text:Text=Format Text!(
     Format="LOCTEXT(\"{Nickname} + {N} others\")"
     Nickname:String=Split(SourceString=@n7, InStr=",", SearchDir=FromStart)
     N:int=int - int(A=@n4)
    ) -> (Result)
   )
   Return Node(execute=@n5.else, Text=@n6)
   n8 = Get Header Type() -> (Header Type:ESocialHeaderItemType Enum=Unknown)
   n9 = OR Boolean!(
    A="Equal (Enum)"(A:ESocialHeaderItemType Enum=@n8, B:ESocialHeaderItemType Enum=FriendInvites)
    B="Equal (Enum)"(A=@n8, B=SuggestedFriends)
    C:bool="Equal (Enum)"(A=@n8, B=RecentPlayers)
   ) -> (ReturnValue)
   Return Node(
    execute=@n2.else
    Text=Select(
     Option 0:Text=Select(Option 0=@n1.Presence, Option 1=@n1.Last Interaction Info, Index=@n9)
     Option 1:Text
     Index="Equal (Enum)"(A=@n8, B=NativeEpicPartyInvites)
    ) -> (ReturnValue:Text)
   )
  ]}
 Convert_HeaderTypeToMutualFriendsTextVisibility {nodes [
   n1 = Convert_HeaderTypeToMutualFriendsTextVisibility!() -> (then, Header Type:ESocialHeaderItemType Enum=Unknown, Mutual Friends:Text) {is=title K2Node_FunctionEntry}
   n2 = NOT Boolean(A=TextIsEmpty(InText=Get Mutual Friends() -> (Mutual Friends:Text)))
   n3 = "Equal (Enum)"(A=@n1, B=FriendInvites)
   > Return Node!(
    execute
    Text Visibility:ESlateVisibility Enum=Select(Option 0:ESlateVisibility Enum=Collapsed, Option 1:ESlateVisibility Enum=SelfHitTestInvisible, Index=AND Boolean(A=@n3, B=@n2)) -> (ReturnValue:ESlateVisibility Enum)
   )
  ]}
 Convert_VoiceActivityIcon {nodes [
   Convert_VoiceActivityIcon!() -> (then, Muted:bool=~, Active:bool=~, is Friend:bool=~) {is=title K2Node_FunctionEntry}
   n1 = NOT Boolean(A=Get is Friend() -> (is Friend:bool))
   n2 = "Equal (Integer)"(A=Get AllowVoiceChatSettingIndex() -> (AllowVoiceChatSettingIndex:int=0), B=1)
   n3 = Select(
    Option 0:Material Interface Object Reference=/Sidebar/Materials/Icons/MI_Sidebar_Icon_VCOff
    Option 1:Material Interface Object Reference
    Index=Get Active() -> (Active:bool)
   ) -> (ReturnValue:Material Interface Object Reference)
   n4 = Get Muted() -> (Muted:bool)
   > Return Node!(
    execute
    Icon:Material Interface Object Reference=Select(
     Option 0=Select(Option 0=@n3, Option 1=/Sidebar/Materials/Icons/MI_Sidebar_Icon_VCMuted, Index=@n4)
     Option 1=/Sidebar/Materials/Icons/MI_Sidebar_Icon_VCRestricted
     Index=AND Boolean(A=@n2, B=@n1)
    )
   )
  ]}
 Convert_VoiceActivityVisibility {nodes [
   Convert_VoiceActivityVisibility!() -> (then, Muted:bool=~, Active:bool=~, is Friend:bool=~) {is=title K2Node_FunctionEntry}
   n1 = NOT Boolean(A=Get is Friend())
   n2 = "Equal (Integer)"(A=Get AllowVoiceChatSettingIndex() -> (AllowVoiceChatSettingIndex=0), B=1)
   n3 = Select(Option 0:ESlateVisibility Enum=SelfHitTestInvisible, Option 1:ESlateVisibility Enum=Collapsed, Index=Get Active()) -> (ReturnValue:ESlateVisibility Enum)
   n4 = Get Muted()
   > Return Node!(
    execute
    visibility:ESlateVisibility Enum=Select(Option 0=Select(Option 0=@n3, Option 1=SelfHitTestInvisible, Index=@n4), Option 1=SelfHitTestInvisible, Index=AND Boolean(A=@n2, B=@n1))
   )
  ]}
 Convert_VoiceChannelOpacity {nodes [
   n1 = Convert_VoiceChannelOpacity!() -> (then, Listening:bool=~, Active:bool=~) {is=title K2Node_FunctionEntry}
   n2 = NOT Boolean(A=@n1)
   n3 = NOT Boolean(A=Get isLocalUserForBlueprint())
   > Return Node!(execute, Opacity:double=SelectFloat(A=0.65, B=1, bPickA=AND Boolean!(A=@n3, B=@n2, C:bool=@n1.Active) -> (ReturnValue)))
  ]}
 Convert_VoiceChannelText {nodes [
   Convert_VoiceChannelText!() -> (then, Active:bool=~, Listening:bool=~, Muted:bool=~, Friend:bool=~, Local User Channel:ESidebarVoiceChannelType Enum=None) {is=title K2Node_FunctionEntry}
   n1 = Get Muted()
   n2 = Get Listening() -> (Listening:bool)
   n3 = AND Boolean!(
    A="Equal (Integer)"(A=Get AllowVoiceChatSettingIndex() -> (AllowVoiceChatSettingIndex=0), B=1)
    B=NOT Boolean(A=Get Friend() -> (Friend:bool))
   ) -> (ReturnValue)
   > n4 = Branch(Condition=AND Boolean!(A=Get Active(), B=NOT Boolean(A=@n2), C:bool=NOT Boolean(A=@n1), D:bool=NOT Boolean(A=@n3)) -> (ReturnValue))
   > Return Node!(
    execute
    Text:Text=Select!(
     None:Text
     Party:Text="LOCTEXT(\"In game chat\")"
     Game:Text="LOCTEXT(\"In party chat\")"
     EpicParty:Text
     Index:ESidebarVoiceChannelType Enum=Get Local User Channel() -> (Local User Channel:ESidebarVoiceChannelType Enum=None)
    ) -> (ReturnValue:Text)
   )
   Return Node(execute=@n4.else)
  ]}
 Convert_VoiceIndicatorVisibility {nodes [
   n1 = Convert_VoiceIndicatorVisibility!() -> (then, Is Talking:bool=~, Is Active:bool=~) {is=title K2Node_FunctionEntry}
   > Return Node!(
    execute
    Visibility:ESlateVisibility Enum=Select!(
     Option 0:ESlateVisibility Enum=Collapsed
     Option 1:ESlateVisibility Enum=SelfHitTestInvisible
     Index:bool=AND Boolean!(A=@n1, B=@n1.Is Active) -> (ReturnValue)
    ) -> (ReturnValue:ESlateVisibility Enum)
   )
  ]}
 EventGraph {nodes [
   "On Hovered (WBP_Social_User)"() {is=title K2Node_ComponentBoundEvent}
   > Branch(Condition=HasAnyChildren!(self:Panel Widget Object Reference=Get Overlay_Controls() -> (Overlay_Controls:Overlay Object Reference)) -> (ReturnValue:bool))
   > Parent: On Hovered() {is=title K2Node_CallParentFunction}
   n1 = "On Unhovered (WBP_Social_User)"() {is=title K2Node_ComponentBoundEvent}
   Branch(execute=[@n1.then, @n7.then, @n8.then, @n12.then], Condition=HasAnyChildren(self=Get Overlay_Controls()))
   > Parent: On Unhovered() {is=title K2Node_CallParentFunction}
   n2 = Context Menu Interaction OnClicked!() -> (OutputDelegate, then, InteractionNameID:Name) {is=title K2Node_CustomEvent}
   n3 = Find(TargetMap:Map of Names to Fort Social Interaction VM Object References=Get InteractionsMap(), Key:"Name (by ref)"=@n2.InteractionNameID) -> (Value:Fort Social Interaction VM Object Reference)
   > Branch(Condition=@n3.ReturnValue)
   > n4 = Branch(Condition="Equal (Enum)"(
     A:ESocialInteractionType Enum=Get InteractionType(self=@n3) -> (InteractionType=AcceptEpicPartyInvite)
     B:ESocialInteractionType Enum=EditNickname
    ))
   > Call On Edit Nickname!(
    execute:Exec
    self:WBP Sidebar Social Player Object Reference
    SocialUser:Fort Social User Object Reference=Get TargetUser(self=Get FortSocialItemVM())
   ) -> (then:Exec) {is=title K2Node_CallDelegate}
   n5 = Branch(execute=@n4.else, Condition="Equal (Enum)"(A=Get InteractionType(self=@n3) -> (InteractionType=AcceptEpicPartyInvite), B=RemoveFriend))
   > Call On Remove Friend!(
    execute:Exec
    self:WBP Sidebar Social Player Object Reference
    SocialUser:Fort Social User Object Reference=Get TargetUser(self=Get FortSocialItemVM())
    Interaction:Fort Social Interaction VM Object Reference=@n3
   ) -> (then:Exec) {is=title K2Node_CallDelegate}
   n6 = Branch(execute=@n5.else, Condition="Equal (Enum)"(A=Get InteractionType(self=@n3) -> (InteractionType=AcceptEpicPartyInvite), B=Block))
   > Call On Block Player!(
    execute:Exec
    self:WBP Sidebar Social Player Object Reference
    SocialUser:Fort Social User Object Reference=Get TargetUser(self=Get FortSocialItemVM())
    Interaction:Fort Social Interaction VM Object Reference=@n3
   ) -> (then:Exec) {is=title K2Node_CallDelegate}
   ExecuteInteraction(execute=@n6.else, self:Fort Social Interaction VM Object Reference=@n3)
   n7 = On ControlsHovered() {is=title K2Node_CustomEvent}
   n8 = On AddFriend Hovered() {is=title K2Node_CustomEvent}
   On Add Friend Click() {is=title K2Node_CustomEvent}
   > n9 = Can Execute Interaction!(execute, self, Interaction Type:ESocialInteractionType Enum=AddFriend) -> (then, Can Execute:bool, Interaction:Fort Social Interaction VM Object Reference) {is=title K2Node_CallFunction}
   > Branch(Condition=@n9)
   > n10 = ExecuteInteraction(self=@n9.Interaction)
   n11 = FindParentWidgetOfType!(
    execute=[@n16.then, @n10.then, @n14.then]
    StartingWidget:Widget Object Reference=Self-Reference()
    Type:Widget Class Reference=/Sidebar/Panels/WBP_AddFriends_Panel.WBP_AddFriends_Panel_C
   ) -> (then, ReturnValue:WBP Add Friends Panel Object Reference)
   > RequestRefreshFocus(self:Common Activatable Widget Object Reference=@n11)
   n12 = On Accept Invite Hovered() {is=title K2Node_CustomEvent}
   On Accept Friend Invite() {is=title K2Node_CustomEvent}
   > n13 = Can Execute Interaction(Interaction Type=AcceptFriendInvite)
   > Branch(Condition=@n13)
   > n14 = ExecuteInteraction(self=@n13.Interaction)
   On Decline Friend Invite() {is=title K2Node_CustomEvent}
   > n15 = Can Execute Interaction(Interaction Type=DeclineFriendInvite)
   > Branch(Condition=@n15)
   > n16 = ExecuteInteraction(self=@n15.Interaction)
   n17 = Event OnListItemObjectSet()
   > Set isFriendRequest!(
    execute
    isFriendRequest:bool=false
    self:WBP Sidebar Player Avatar Object Reference=Get PlayerAvatar!(self:WBP Sidebar User Object Reference=Get WBP_Social_User() -> (WBP_Social_User:WBP Sidebar User Object Reference)) -> (PlayerAvatar:WBP Sidebar Player Avatar Object Reference)
   ) -> (then, Output_Get:bool)
   > SetFriendRequestsVisuals!(execute, self, Request Count:int=0) -> (then)
   > n18 = Cast To FortSocialItemVM!(execute, Object=@n17.ListItemObject) -> (then, CastFailed, AsFort Social Item VM:Fort Social Item VM Object Reference) {is=title K2Node_DynamicCast}
   > n19 = Set FortSocialItemVM(FortSocialItemVM:Fort Social Item VM Object Reference=@n18) -> (Output_Get:Fort Social Item VM Object Reference)
   > InitializeVMs!(execute, self, FortSocialItemVM:Fort Social Item VM Object Reference=@n19) -> (then)
   > Branch(Condition="Not Equal (Enum)"!(A:ESocialHeaderItemType Enum=Get HeaderItemType(self=@n18) -> (HeaderItemType=Unknown), B:ESocialHeaderItemType Enum=OfflineFriends) -> (ReturnValue:bool))
   > ActivateInteractionList(self=Get FortSocialItemVM())
   > SetPartyDividerVisibility!(execute, self, isVisible:bool=Get bIsChecked!(self:Fort Social Item VM Object Reference=@n19) -> (bIsChecked:bool)) -> (then)
   n20 = Cast To FortSocialUserViewModel!(execute=@n18.CastFailed, Object=@n17.ListItemObject) -> (then, CastFailed, AsFort SocialUser Viewmodel:Fort Social User Viewmodel Object Reference) {is=title K2Node_DynamicCast}
   > n21 = Construct Object from Class(Class=/Script/SidebarUI.FortSocialItemVM) -> (ReturnValue:Fort Social Item VM Object Reference)
   > Init!(
    execute
    self:Fort Social Item VM Object Reference=@n21
    InTargetUser:Fort Social User Object Reference=GetSocialUser!(self:Fort Social User Viewmodel Object Reference=@n20) -> (ReturnValue:Fort Social User Object Reference)
   ) -> (then)
   > n22 = Set FortSocialItemVM(FortSocialItemVM=@n21)
   > InitializeVMs(FortSocialItemVM=@n22)
   Event On Hovered()
   > n23 = Branch(Condition=HasAnyChildren(self=Get Overlay_Controls()))
   Parent: On Hovered(execute=@n23.else)
   n24 = Event PreConstruct()
   > Parent: PreConstruct(IsDesignTime=@n24.IsDesignTime)
   > Construct Context!(execute:Exec) -> (then:Exec) {is=title K2Node_MacroInstance}
   > Set Arrow Icon!(NewParam:Exec) -> (NewParam2:Exec) {is=title K2Node_MacroInstance}
   > SetCompactUIStyle(self:WBP Sidebar Button Base Object Reference) {is=title K2Node_CallFunction}
   > Branch(Condition=@n24.IsDesignTime)
   > Update You Indicator And Controls!(execute, self, UseIndicator:bool=true) -> (then)
   Event On Unhovered()
   > n25 = Branch(Condition=HasAnyChildren(self=Get Overlay_Controls()))
   Parent: On Unhovered(execute=@n25.else)
   > Set isSocialPlayerFocused(isSocialPlayerFocused:bool=false) -> (Output_Get:bool)
   Event Construct()
   n26 = Get PlayerPlaytimeVM() -> (PlayerPlaytimeVM:Player Playtime View Model Object Reference)
   > Init!(execute, self:Player Playtime View Model Object Reference=@n26) -> (then)
   > SetListenToRefresh!(execute, self:Player Playtime View Model Object Reference=@n26, bListenToRefresh:bool=true) -> (then)
   > n27 = GetConsoleVariableBoolValue!(execute, VariableName:String=Fort.UseEpicParties) -> (then, ReturnValue:bool)
   > Set isEpicParty(isEpicParty:bool=@n27) -> (Output_Get:bool)
   > n28 = GetConsoleVariableBoolValue(VariableName=Sidebar.EnableUpdatedPartyControls)
   > Set isUpdatedPartyControlsEnabled(isUpdatedPartyControlsEnabled:bool=@n28) -> (Output_Get:bool)
   Event Destruct()
   > SetListenToRefresh(self=Get PlayerPlaytimeVM(), bListenToRefresh=false)
  ]}
 ForceUpdateAllBindingsOnListItemObjectSet {nodes [
   ForceUpdateAllBindingsOnListItemObjectSet() {is=title K2Node_FunctionEntry}
   > n1 = GetConsoleVariableBoolValue(VariableName=Sidebar.ImmediatelySetListItemData)
   > Branch(Condition=@n1)
   > n2 = Get!(execute) -> (then, else, FortSocialUserNameViewModel:Fort Social User Name Viewmodel Object Reference)
   > n3 = Get!(execute, self:Fort Social Item VM Object Reference=Get FortSocialItemVM()) -> (then, else, FriendRequests:Fort Friend Requests VM Object Reference)
   n4 = Break Fort Social User Name Package!(FortSocialUserNamePackage:"Fort Social User Name Package Structure (by ref)"=Get SocialUserNames!(self:Fort Social User Name Viewmodel Object Reference=@n2) -> (SocialUserNames:Fort Social User Name Package Structure)) -> (DisplayName:String, Nickname:String, PlatformIconTag:String) {is=title K2Node_BreakStruct}
   n5 = Get FortSocialItemVM()
   n6 = Get FriendRequests!(self:Fort Social Item VM Object Reference=@n5) -> (FriendRequests:Fort Friend Requests VM Object Reference)
   n7 = Get FortTextChatGroupConversationVM() -> (FortTextChatGroupConversationVM:Fort Group Conversation Viewmodel Object Reference)
   n8 = IsValid(Object=@n7)
   n9 = Get ConversationName!(self:Fort Text Chat Conversation VM Object Reference=@n7) -> (ConversationName:Text)
   n10 = Get WBP_Social_User()
   n11 = Get RichText_Name!(self:WBP Sidebar User Object Reference=@n10) -> (RichText_Name:Common Rich Text Block Object Reference)
   > SetText(
    self=@n11
    InText=GetNameFromSidebarUser!(
     self
     NickName:String=@n4.Nickname
     PlayerName:String=@n4
     PlatformIconTag:String=@n4.PlatformIconTag
     ConversationName:Text=Select(Option 0:Text, Option 1:Text=@n9, Index=@n8) -> (ReturnValue:Text)
     DisplayNameForJoinableParty:Text=GetRelatedUsersDisplayNames!(self:Fort Social Item VM Object Reference=@n5) -> (ReturnValue:Text)
     NumUnknownUsers:int=GetNumUnknownUsers!(self:Fort Social Item VM Object Reference=@n5) -> (ReturnValue:int=0)
     NumRequesters:int=GetNumRequesterUsers(self=@n6) -> (ReturnValue=0)
    ) -> (FinalNameText:Text)
   )
   n12 = Get FriendRequests(self=Get FortSocialItemVM())
   > SetFriendRequestsVisuals(Request Count=GetNumRequesterUsers(self=@n12) -> (ReturnValue=0))
   > n13 = SetFriendRequestsAvatars!(
    execute
    self
    Requesters:Array of Fort Social User Object References=GetValidRequesterUsers!(self:Fort Friend Requests VM Object Reference=@n12) -> (ReturnValue:Array of Fort Social User Object References)
   ) -> (then)
   n14 = Get!(execute=[@n28.then, @n13.then, @n2.else]) -> (then, else, FortSocialUserInfoViewModel:Fort Social User Info View Model Object Reference)
   n15 = Get MutualFriends_Text!(self:WBP Sidebar User Object Reference=@n10) -> (MutualFriends_Text:Common Text Object Reference)
   n16 = Get MutualFriendsText!(self:Fort Social User Info View Model Object Reference=@n14) -> (MutualFriendsText:Text)
   > "SetText (Text)"(self=@n15, InText=@n16)
   > n17 = Get!(execute) -> (then, else, FortSocialItemVM:Fort Social Item VM Object Reference)
   n18 = Get HeaderItemType(self=@n17) -> (HeaderItemType=Unknown)
   > n19 = SetVisibility(
    self=@n15
    InVisibility=Convert_HeaderTypeToMutualFriendsTextVisibility!(self, Header Type:ESocialHeaderItemType Enum=@n18, Mutual Friends:Text=@n16) -> (Text Visibility:ESlateVisibility Enum=Visible) {is=title K2Node_CallFunction}
   )
   n20 = Get HB_Joinable() -> (HB_Joinable:Horizontal Box Object Reference)
   n21 = Get FortSocialUserInfoViewModel() -> (FortSocialUserInfoViewModel:Fort Social User Info View Model Object Reference)
   n22 = Get UserRichPresenceText!(self:Fort Social User Info View Model Object Reference=@n21) -> (UserRichPresenceText:Text)
   n23 = Get LastInteractionInfo!(self:Fort Social User Info View Model Object Reference=@n21) -> (LastInteractionInfo:Text)
   "SetText (Text)"(
    execute=[@n19.then, @n17.else, @n14.else]
    self=Get Text_Message!(self:WBP Sidebar User Object Reference=@n10) -> (Text_Message:Common Text Object Reference)
    InText=SelectText(
     B=Convert_HeaderTypeToMessageText!(self, Header Type:ESocialHeaderItemType Enum=@n18, Presence:Text=@n22, Last Interaction Info:Text=@n23) -> (Text:Text) {is=title K2Node_CallFunction}
     bPickA=IsVisible(self=@n20)
    )
   )
   > ForceUpdateAllAvatarBindings(self:WBP Sidebar Player Avatar Object Reference=Get PlayerAvatar(self=@n10))
   n24 = Break Fort Social User Name Package(FortSocialUserNamePackage=Get SocialUserNames(self=@n2))
   n25 = Get FortTextChatGroupConversationVM()
   n26 = IsValid(Object=@n25)
   n27 = Get ConversationName(self=@n25)
   n28 = SetText(
    execute=@n3.else
    self=@n11
    InText=GetNameFromSidebarUser(
     NickName=@n24.Nickname
     PlayerName=@n24
     PlatformIconTag=@n24.PlatformIconTag
     ConversationName=Select(Option 1=@n27, Index=@n26)
     NumUnknownUsers=0
     NumRequesters=0
    )
   )
  ]}
 GetLeaderIconVisibility {nodes [
   n1 = GetLeaderIconVisibility!() -> (then, IsLeader:bool=~, IsEpicPartyLeader:bool=~) {is=title K2Node_FunctionEntry}
   n2 = AND Boolean!(
    A=@n1.IsEpicPartyLeader
    B=Get isEpicParty() -> (isEpicParty:bool)
    C:bool=ShowEpicPartyLeader!(self:Fort Runtime Options Object Reference=GetRuntimeOptions()) -> (ReturnValue:bool)
   ) -> (ReturnValue)
   n3 = GetRuntimeOptions()
   n4 = AND Boolean(
    A=NOT Boolean(A=HideFNPartyLeaderFromUI!(self:Fort Runtime Options Object Reference=@n3) -> (ReturnValue:bool))
    B=@n1
    C=NOT Boolean(A=Get isEpicParty())
   )
   > Return Node(Visibility=Select(
     Option 0:ESlateVisibility Enum=Collapsed
     Option 1:ESlateVisibility Enum=SelfHitTestInvisible
     Index=OR Boolean!(A=@n4, B=@n2) -> (ReturnValue)
    ) -> (ReturnValue:ESlateVisibility Enum))
  ]}
 GetNameFromSidebarUser {nodes [
   n1 = GetNameFromSidebarUser!() -> (
    then
    NickName:String
    PlayerName:String
    PlatformIconTag:String
    ConversationName:Text
    DisplayNameForJoinableParty:Text
    NumUnknownUsers:int
    NumRequesters:int
   ) {is=title K2Node_FunctionEntry}
   > n2 = Branch(Condition="integer > integer"(A=Get NumRequesters() -> (NumRequesters:int=0)))
   > Return Node!(execute, FinalNameText:Text="LOCTEXT(\"Friend requests\")")
   n3 = Branch(execute=@n2.else, Condition=TextIsEmpty(InText=@n1.ConversationName))
   > n4 = Get()
   n5 = "Equal (Enum)"(
    A:ESocialHeaderItemType Enum=Get HeaderItemType(self=@n4) -> (HeaderItemType=Unknown)
    B:ESocialHeaderItemType Enum=JoinableNativeEpicParties
   )
   n6 = GetNumUnknownUsers(self=@n4) -> (ReturnValue=0)
   > n7 = Branch(Condition=AND Boolean!(A="integer > integer"(A=@n6), B=@n5) -> (ReturnValue))
   n8 = GetRelatedUsersDisplayNames(self=@n4)
   > Return Node(FinalNameText=Format Text!(
     Format="INVTEXT(\"{Name} <SecondaryNameStyle>{more}</>\")"
     Name:String="To String (Text)"(InText=@n8)
     more:Text=Format Text!(Format="LOCTEXT(\"+{N} more\")", N:int=@n6) -> (Result)
    ) -> (Result))
   n9 = GetConcatenatedNameAndNicknameWithIcon!(
    self:WBP Sidebar User Object Reference=Get WBP_Social_User()
    NickName:String=Get NickName() -> (NickName:String)
    PlayerName:String=Get PlayerName() -> (PlayerName:String)
    PlatformIconTag:String=Get PlatformIconTag() -> (PlatformIconTag:String)
   ) -> (FinalNameText:Text)
   Return Node(execute=@n7.else, FinalNameText=SelectText(A=GetRelatedUsersDisplayNames(self=@n4), B=@n9, bPickA=@n5))
   Return Node(execute=@n4.else, FinalNameText=@n9)
   Return Node(execute=@n3.else, FinalNameText=@n1.ConversationName)
  ]}
 GetOtherPlayerInvitedYouTextVisibility {nodes [
   n1 = GetOtherPlayerInvitedYouTextVisibility!() -> (
    then
    TargetUser:Fort Social User Object Reference
    OtherPlayerInteractions:Array of Fort Social Interaction VM Object References
    Social Header:ESocialHeaderItemType Enum=Unknown
   ) {is=title K2Node_FunctionEntry}
   > n2 = Is Valid(InputObject=@n1)
   n3 = Branch(
    execute=@n2.Is Not Valid
    Condition=ShouldShowOtherPlayerInvitedYouText!(
     self
     TargetUser:Fort Social User Object Reference
     OtherPlayerInteractions:Array of Fort Social Interaction VM Object References=@n1.OtherPlayerInteractions
     Social Header:ESocialHeaderItemType Enum=@n1.Social Header
    ) -> (ShouldShow:bool)
   )
   > Return Node!(execute, IsJoinableVisible:ESlateVisibility Enum=SelfHitTestInvisible)
   Return Node(execute=@n3.else, IsJoinableVisible=Collapsed)
   n4 = IsLocalUserForBlueprint!(self:Fort Social User Object Reference=@n1) -> (ReturnValue:bool)
   n5 = OR Boolean(A="Equal (Enum)"(A=@n1.Social Header, B=NativeEpicPartyInvites), B="Equal (Enum)"(A=@n1.Social Header, B=RequestsToJoinEpicParties))
   n6 = "Equal (Enum)"(A=@n1.Social Header, B=JoinableParties)
   n7 = Get isEpicParty()
   n8 = Branch(
    execute=@n2.Is Valid
    Condition=AND Boolean!(
     A=ShouldShowOtherPlayerInvitedYouText(TargetUser=@n1, OtherPlayerInteractions=@n1.OtherPlayerInteractions, Social Header=@n1.Social Header)
     B=NOT Boolean(A=@n4)
     C:bool=Select(Option 0:bool=@n6, Option 1:bool=@n5, Index=@n7) -> (ReturnValue:bool)
    ) -> (ReturnValue)
   )
   > Return Node(IsJoinableVisible=SelfHitTestInvisible)
   Return Node(execute=@n8.else, IsJoinableVisible=Collapsed)
  ]}
 InitializeVMs {nodes [
   n1 = InitializeVMs!() -> (then, FortSocialItemVM:Fort Social Item VM Object Reference) {is=title K2Node_FunctionEntry}
   > Set FortTextChatGroupConversationVM(FortTextChatGroupConversationVM:Fort Group Conversation Viewmodel Object Reference=Get TargetConversation!(self:Fort Social Item VM Object Reference=@n1) -> (TargetConversation:Fort Group Conversation Viewmodel Object Reference)) -> (Output_Get:Fort Group Conversation Viewmodel Object Reference)
   > InitializeWithFortSocialItemVM!(
    execute
    self:WBP Sidebar Player Avatar Object Reference=Get PlayerAvatar(self=Get WBP_Social_User())
    FortSocialItemVM:Fort Social Item VM Object Reference=Get FortSocialItemVM()
   ) -> (then)
   > n2 = GetConsoleVariableBoolValue(VariableName=Fort.Rollback.PlatformIconDisplayRuleToAlways)
   > n3 = Branch(Condition=@n2)
   > n4 = Set PlatformTagDisplayRule!(
    execute
    PlatformTagDisplayRule:EFortSocialPlatformTagDisplayRule Enum=Always
    self:Fort Social User Name Viewmodel Object Reference=Get FortSocialUserNameViewModel() -> (FortSocialUserNameViewModel:Fort Social User Name Viewmodel Object Reference)
   ) -> (then, Output_Get:EFortSocialPlatformTagDisplayRule Enum=Never)
   n5 = Get!(execute=[@n4.then, @n7.then], self:Fort Social Item VM Object Reference=Get FortSocialItemVM()) -> (then, else, TargetUser:Fort Social User Weak object)
   > SetSocialUser(self=[Get FortSocialUserInfoViewModel(), Get FortSocialUserNameViewModel()], InSocialUser=@n5)
   > AddVMForLeaderIcon!(execute, self, SocialUser:Fort Social User Object Reference=@n5) -> (then) {is=title K2Node_CallFunction}
   n6 = IsLocalUserForBlueprint(self=Get TargetUser(self=Get FortSocialItemVM()))
   > Set isLocalUserForBlueprint(isLocalUserForBlueprint:bool=@n6) -> (Output_Get:bool)
   > Update You Indicator And Controls(UseIndicator=@n6)
   > AddVMForLocalPlayerJoinability!(execute, self, IsLocalPlayer:bool=@n6) -> (then) {is=title K2Node_CallFunction}
   > AddVMForVoiceIndicators() {is=title K2Node_CallFunction}
   > ForceUpdateAllBindingsOnListItemObjectSet() {is=title K2Node_CallFunction}
   SetSocialUser(execute=@n5.else, self=[Get FortSocialUserInfoViewModel(), Get FortSocialUserNameViewModel()])
   > ForceUpdateAllBindingsOnListItemObjectSet()
   n7 = Set PlatformTagDisplayRule(execute=@n3.else, PlatformTagDisplayRule=OnlyIfDifferent, self=Get FortSocialUserNameViewModel()) -> (Output_Get=Never)
  ]}
 IsContextMenuInteractionAvailable {nodes [
   n1 = IsContextMenuInteractionAvailable!() -> (then, InteractionType:ESocialInteractionType Enum=AcceptEpicPartyInvite) {is=title K2Node_FunctionEntry}
   n2 = OR Boolean!(
    A="Equal (Enum)"(A:ESocialInteractionType Enum=@n1, B:ESocialInteractionType Enum=SpectateFriendsGame)
    B="Equal (Enum)"(A=@n1, B=JoinAndSpectateFriend)
    C:bool="Equal (Enum)"(A=@n1, B=JoinMatchViaPresence)
    D:bool="Equal (Enum)"(A=@n1, B=JoinEpicPartyAndSpectateFriend)
   ) -> (ReturnValue)
   n3 = "Equal (Enum)"(
    A:EPlaytimeState Enum=Get PlayState!(self:Player Playtime View Model Object Reference=Get PlayerPlaytimeVM()) -> (PlayState:EPlaytimeState Enum=Unblocked)
    B:EPlaytimeState Enum=PlayTime_Blocked
   )
   > Return Node!(execute, IsAvailable:bool=NOT Boolean(A=AND Boolean!(A=@n2, B=@n3) -> (ReturnValue)))
  ]}
 NavigateLeftToSocialPlayer {nodes [
   NavigateLeftToSocialPlayer!() -> (then, Navigation:EUINavigation Enum) {is=title K2Node_FunctionEntry}
   > Set isSocialPlayerFocused(isSocialPlayerFocused=true)
   > Parent: On Hovered()
   > OnSidebarUserUnHovered(self:WBP Sidebar User Object Reference=Get WBP_Social_User())
   > Return Node!(execute, ReturnValue:Widget Object Reference=Self-Reference())
  ]}
 NavigateToControls {nodes [
   NavigateToControls!() -> (then, Navigation:EUINavigation Enum) {is=title K2Node_FunctionEntry}
   > Set isSocialPlayerFocused(isSocialPlayerFocused=false)
   > n1 = Get!(execute) -> (then, else, YouControlsWidget:WBP Sidebar Social Player Controls Object Reference)
   > Parent: On Unhovered()
   n2 = Get Button_Mic!(self:WBP Sidebar Social Player Controls Object Reference=@n1) -> (Button_Mic:WBP Sidebar Button Generic Object Reference)
   > n3 = Branch(Condition=IsVisible(self=@n2))
   > Return Node(ReturnValue=@n2)
   Return Node(
    execute=@n3.else
    ReturnValue=Get Button_MoreOptions!(self:WBP Sidebar Social Player Controls Object Reference=@n1) -> (Button_MoreOptions:WBP Sidebar Button Generic Object Reference)
   )
   n4 = Get!(execute=@n1.else) -> (then, else, Add Friend Button:WBP Sidebar Social Player Add Friend Object Reference)
   > Return Node(ReturnValue=@n4)
   n5 = Get!(execute=@n4.else) -> (then, else, Add Accept Invite:WBP Sidebar Social Player Accept Invite Object Reference)
   > Return Node(ReturnValue=@n5)
  ]}
 OnBlockPlayer {nodes [OnBlockPlayer!() -> (then, SocialUser:Fort Social User Object Reference, Interaction:Fort Social Interaction VM Object Reference) {is=title K2Node_FunctionEntry}]}
 OnEditNickname {nodes [OnEditNickname!() -> (then, SocialUser:Fort Social User Object Reference) {is=title K2Node_FunctionEntry}]}
 OnMouseButtonDown {nodes [
   n1 = OnMouseButtonDown!() -> (then, MyGeometry:Geometry Structure, MouseEvent:"Pointer Event Structure (by ref)") {is=title K2Node_FunctionEntry}
   > ActivateInteractionList(self=Get FortSocialItemVM())
   > PopulateContextMenu()
   n2 = Get ContextMenuActionContext()
   > n3 = Branch(Condition=AND Boolean!(
     A=Is Mouse Button Down!(Input:"Pointer Event Structure (by ref)"=@n1.MouseEvent, MouseButton:Key Structure=RightMouseButton) -> (ReturnValue:bool)
     B=HasInteractionItems(self=@n2)
     C:bool=IsValid(Object=@n2)
    ) -> (ReturnValue))
   > Open!(execute, self:Menu Anchor Object Reference=Get Anchor(self=Get PlayerAction_ContextMenu()), bFocusMenu:bool=true) -> (then)
   > Return Node(ReturnValue:Event Reply Structure=Handled())
   Return Node(execute=@n3.else, ReturnValue=Unhandled())
  ]}
 OnRemoveFriend {nodes [OnRemoveFriend!() -> (then, SocialUser:Fort Social User Object Reference, Interaction:Fort Social Interaction VM Object Reference) {is=title K2Node_FunctionEntry}]}
 PopulateContextMenu {nodes [
   PopulateContextMenu!() -> (then, Interactions:Array of Fort Social Interaction VM Object References) {is=title K2Node_FunctionEntry}
   > Clear!(execute, TargetMap:Map of Names to Fort Social Interaction VM Object References=Get InteractionsMap()) -> (then) {is=title K2Node_CallFunction}
   > RemoveAllInteractions(self:UIKit Context Menu Action Context Object Reference=Get ContextMenuActionContext())
   > n1 = For Each Loop(Array:Array of Fort Social Interaction VM Object References=Get Interactions!(self:Fort Social Item VM Object Reference=Get FortSocialItemVM()) -> (Interactions:Array of Fort Social Interaction VM Object References)) -> (Array Element:Fort Social Interaction VM Object Reference)
   n2 = Get InteractionType(self=@n1) -> (InteractionType=AcceptEpicPartyInvite)
   n3 = Enum to Name!(Enumerator:ESocialInteractionType Enum=@n2) -> (ReturnValue:Name) {is=title K2Node_GetEnumeratorName}
   > Add!(
    execute
    TargetMap:Map of Names to Fort Social Interaction VM Object References=Get InteractionsMap()
    Key:"Name (by ref)"=@n3
    Value:"Fort Social Interaction VM Object Reference (by ref)"=@n1
   ) -> (then) {is=title K2Node_CallFunction}
   > Update Control Buttons()
   > AddInteraction!(
    execute
    self:UIKit Context Menu Action Context Object Reference=Get ContextMenuActionContext()
    InActionInfo_DisplayName:Text=GetDisplayName!(self:Fort Social Interaction VM Object Reference=@n1) -> (ReturnValue:Text)
    InActionInfo_IconTexture:Texture 2D Soft Object Reference
    InActionInfo_IconMaterial:Material Instance Soft Object Reference=Select(
     Option 0:Material Instance Soft Object Reference
     Option 1:Material Instance Soft Object Reference=/Sidebar/Materials/Icons/MI_Sidebar_Icon_Crown
     Index="Equal (Enum)"(A:ESocialInteractionType Enum=@n2, B:ESocialInteractionType Enum=PromoteToEpicPartyLeader)
    ) -> (ReturnValue:Material Instance Soft Object Reference)
    InActionInfo_SortPriority:int=1
    InActionInfo_bIsAvailable:bool=IsContextMenuInteractionAvailable!(InteractionType:ESocialInteractionType Enum=@n2) -> (IsAvailable:bool) {is=title K2Node_CallFunction}
    InActionInfo_bIsEnabled:bool=True
    InActionInfo_bIsSelected:bool=False
    InActionInfo_bAddDivider:bool=False
    InActionInfo_bPositionDividerAfter:bool=False
    InActionInfo_bCloseMenuOnSelect:bool=False
    InActionInfo_InteractionNameID:Name=@n3
    InActionInfo_Interaction:Data Table Row Handle Structure=(DataTable=None,RowName="")
    InActionInfo_PressedDelegate:Delegate="(null).None"
    InActionInfo_ReleasedDelegate:Delegate="(null).None"
    InInteractionDelegate:"Delegate (by ref)"=Create Event()
   ) -> (then)
  ]}
 SetArrowIcon {nodes [
   Inputs!() -> (NewParam:Exec)
   n1 = GetDynamicMaterial!(Brush:"Slate Brush Structure (by ref)"=Get Brush!(self:Image Object Reference=Get Icon() -> (Icon:Image Object Reference)) -> (Brush:Slate Brush Structure)) -> (ReturnValue)
   > SetTextureParameterValue(self=@n1, ParameterName=Icon, Value=/FortUILibrary/Textures/Icons/T_UI_Icon_UIKit_Chevron_MSDF)
   > SetScalarParameterValue(self=@n1, ParameterName=Rotation, Value=0)
   > Outputs!(NewParam2:Exec)
  ]}
 SetCompactUIStyle {nodes [
   SetCompactUIStyle() {is=title K2Node_FunctionEntry}
   > Parent: SetCompactUIStyle() {is=title K2Node_CallParentFunction}
   > Branch(Condition=IsUsingCompactUI())
   n1 = Get PlayerAction_ContextMenu()
   n2 = SlotAsGridSlot(Widget=@n1)
   > SetVerticalAlignment(self:Grid Slot Object Reference=@n2, InVerticalAlignment=VAlign_Bottom)
   > SetHorizontalAlignment(self:Grid Slot Object Reference=@n2, InHorizontalAlignment=HAlign_Right)
   > SetPlacement!(execute, self:Menu Anchor Object Reference=Get Anchor(self=@n1), InPlacement:EMenuPlacement Enum=MenuPlacement_AboveRightAnchor) -> (then)
  ]}
 SetFriendRequestsAvatars {nodes [
   n1 = SetFriendRequestsAvatars!() -> (then, Requesters:Array of Fort Social User Object References) {is=title K2Node_FunctionEntry}
   n2 = Get PlayerAvatar(self=Get WBP_Social_User())
   > Set isFriendRequest(isFriendRequest="integer > integer"(A=Length(TargetArray:Array of Fort Social User Object References=@n1), B=1), self=@n2)
   > InitializeWithSocialUsers!(
    execute
    self:WBP Sidebar Player Avatar Object Reference=@n2
    SocialUsers:Array of Fort Social User Object References=Get Requesters() -> (Requesters:Array of Fort Social User Object References)
   ) -> (then)
  ]}
 SetFriendRequestsVisuals {nodes [
   n1 = SetFriendRequestsVisuals!() -> (then, Request Count:int) {is=title K2Node_FunctionEntry}
   n2 = "Not Equal (Integer)"(A=@n1)
   > n3 = Set with Broadcast isFriendRequestEntry!(execute, isFriendRequestEntry:bool=@n2) -> (then, Output_Get:bool)
   > SetVisibility(
    self=Get FriendRequestIndiactor() -> (FriendRequestIndiactor:Horizontal Box Object Reference)
    InVisibility=Select(Option 0:ESlateVisibility Enum=Collapsed, Option 1:ESlateVisibility Enum=HitTestInvisible, Index=@n2) -> (ReturnValue:ESlateVisibility Enum)
   )
   > OverrideCount!(
    execute
    self:WBP UIKit Status Indicator Base Object Reference=Get WBP_UIKit_StatusIndicator() -> (WBP_UIKit_StatusIndicator:WBP UIKit Status Indicator Object Reference)
    Count:int=Get Request Count() -> (Request Count:int=0)
   ) -> (then)
   > Set with Broadcast bShowMultiple!(execute, bShowMultiple:bool=@n3, self:WBP Sidebar Player Avatar Object Reference=Get PlayerAvatar(self=Get WBP_Social_User())) -> (then, Output_Get:bool)
  ]}
 SetMaxTextWidthBeforeClipping {nodes [
   SetMaxTextWidthBeforeClipping!() -> (then, Voice Chat Method:EFortVoiceChatMethod Enum=OpenMic) {is=title K2Node_FunctionEntry}
   > Branch(Condition=AND Boolean!(A=Get isLocalUserForBlueprint(), B=Get isUpdatedPartyControlsEnabled()) -> (ReturnValue))
   n1 = IsVisible(self=Get LeaderCrown())
   n2 = Get Voice Chat Method() -> (Voice Chat Method:EFortVoiceChatMethod Enum=OpenMic)
   > SetMaxTextWidthBeforeClipping!(
    execute
    self:WBP Sidebar User Object Reference=Get WBP_Social_User()
    MaxNameTextWidth:double=SelectFloat(
     A=SelectFloat(A=175, B=200, bPickA=@n1)
     B=SelectFloat(A=215, B=235, bPickA=@n1)
     bPickA="Equal (Enum)"(A:EFortVoiceChatMethod Enum=@n2, B:EFortVoiceChatMethod Enum=PTT)
    )
   ) -> (then) {is=title K2Node_CallFunction}
  ]}
 SetPartyDividerVisibility {nodes [
   n1 = SetPartyDividerVisibility!() -> (then, isVisible:bool=~) {is=title K2Node_FunctionEntry}
   > n2 = Branch(Condition=@n1)
   n3 = Get Background() -> (Background:WBP Sidebar Block Background Button Object Reference)
   > n4 = SetVisibility(
    self=Get Divider!(self:WBP Sidebar Party And Game Backplate Object Reference=Get Backplate_PartyAndGame!(self:WBP Sidebar Block Background Button Object Reference=@n3) -> (Backplate_PartyAndGame:WBP Sidebar Party And Game Backplate Object Reference)) -> (Divider:Image Object Reference)
    InVisibility=SelfHitTestInvisible
   )
   Branch(
    execute=[@n6.then, @n4.then]
    Condition=IsValid(Object=Get YouControlsWidget() -> (YouControlsWidget:WBP Sidebar Social Player Controls Object Reference))
   )
   > SetPadding!(
    execute
    self:User Widget Object Reference=Get Backplate_PartyAndGame(self=Get Background())
    InPadding_Left:float=0
    InPadding_Top:float=0
    InPadding_Right:float=-300
    InPadding_Bottom:float=0
   ) -> (then)
   n5 = Get Background()
   n6 = SetVisibility(execute=@n2.else, self=Get Divider(self=Get Backplate_PartyAndGame(self=@n5)), InVisibility=Collapsed)
  ]}
 ShoudShowJoinableText {nodes [
   n1 = ShoudShowJoinableText!() -> (then, Slate Visiblity:ESlateVisibility Enum=Visible, Header Type:ESocialHeaderItemType Enum=Unknown, isFriendRequest:bool=~) {is=title K2Node_FunctionEntry}
   > n2 = Branch(Condition=Get isFriendRequest() -> (isFriendRequest:bool))
   > Return Node!(execute, Visibility:ESlateVisibility Enum=Collapsed)
   n3 = Branch(
    execute=@n2.else
    Condition=OR Boolean!(
     A="Equal (Enum)"(A:ESocialHeaderItemType Enum=@n1.Header Type, B:ESocialHeaderItemType Enum=UnifiedFriends)
     B="Equal (Enum)"(A=@n1.Header Type, B=JoinableNativeEpicParties)
     C="Equal (Enum)"(A=@n1.Header Type, B=JoinableParties)
     D="Equal (Enum)"(A=@n1.Header Type, B=Friends)
     E:bool="Equal (Enum)"(A=@n1.Header Type, B=McpOnlineFriends)
     F:bool="Equal (Enum)"(A=@n1.Header Type, B=PlatformOnlineFriends)
     G:bool="Equal (Enum)"(A=@n1.Header Type, B=AwayFriends)
    ) -> (ReturnValue)
   )
   > Return Node(Visibility=Get Slate Visiblity() -> (Slate Visiblity:ESlateVisibility Enum=Visible))
   Return Node(execute=@n3.else, Visibility=Collapsed)
  ]}
 ShouldShowOtherPlayerInvitedYouText {nodes [
   n1 = ShouldShowOtherPlayerInvitedYouText!() -> (
    then
    TargetUser:Fort Social User Object Reference
    OtherPlayerInteractions:Array of Fort Social Interaction VM Object References
    Social Header:ESocialHeaderItemType Enum=Unknown
   ) {is=title K2Node_FunctionEntry}
   > Set Should Show(Should Show:bool=false) -> (Output_Get:bool)
   > n2 = For Each Loop(Array=@n1.OtherPlayerInteractions)
   n3 = Branch(execute=@n2.Completed, Condition=Get Should Show() -> (Should Show:bool))
   > n4 = SetVisibility(
    self=[Get DividerDot_Invited() -> (DividerDot_Invited:Image Object Reference), Get Text_Message(self=Get WBP_Social_User())]
    InVisibility=Collapsed
   )
   n5 = Get Social Header() -> (Social Header:ESocialHeaderItemType Enum=Unknown)
   "SetText (Text)"(
    execute=[@n4.then, @n3.else]
    self=Get Text_Invited() -> (Text_Invited:Common Text Object Reference)
    InText=SelectText(
     A="LOCTEXT(\"Asked to join your party!\")"
     B="LOCTEXT(\"Invited you to their party!\")"
     bPickA="Equal (Enum)"(A=@n5, B=RequestsToJoinEpicParties)
    )
   )
   > Return Node!(execute, ShouldShow:bool=Get Should Show())
   n6 = OR Boolean!(
    A="Equal (Enum)"(A=Get Social Header() -> (Social Header=Unknown), B=NativeEpicPartyInvites)
    B="Equal (Enum)"(A=Get Social Header() -> (Social Header=Unknown), B=RequestsToJoinEpicParties)
   ) -> (ReturnValue)
   n7 = "Equal (Enum)"(A=Get Social Header() -> (Social Header=Unknown), B=JoinableParties)
   n8 = Get isEpicParty()
   n9 = Get InteractionType(self=@n2) -> (InteractionType=AcceptEpicPartyInvite)
   n10 = OR Boolean!(
    A="Equal (Enum)"(A:ESocialInteractionType Enum=@n9, B:ESocialInteractionType Enum=RejectEpicPartyInvite)
    B="Equal (Enum)"(A=@n9, B=DismissEpicPartyJoinRequest)
    C:bool="Equal (Enum)"(A=@n9, B=RejectEpicPartyConversationInvite)
   ) -> (ReturnValue)
   n11 = "Equal (Enum)"(A=@n9, B=RejectPartyInvite)
   n12 = Get isEpicParty()
   Branch(
    execute=@n2.LoopBody
    Condition=AND Boolean(A=Select(Option 0:bool=@n11, Option 1:bool=@n10, Index=@n12) -> (ReturnValue:bool), B=Select(Option 0=@n7, Option 1=@n6, Index=@n8))
   )
   > Set Should Show(Should Show=true)
  ]}
 ShouldShowVoiceIcons {nodes [
   ShouldShowVoiceIcons() {is=title K2Node_FunctionEntry}
   > n1 = Get!(execute) -> (then, else, FortSocialItemVM:Fort Social Item VM Object Reference)
   n2 = Get HeaderItemType(self=@n1) -> (HeaderItemType=Unknown)
   n3 = OR Boolean!(
    A="Equal (Enum)"(A:ESocialHeaderItemType Enum=@n2, B:ESocialHeaderItemType Enum=PartyMembers)
    B="Equal (Enum)"(A=@n2, B=NativeEpicPartyMembers)
   ) -> (ReturnValue)
   n4 = Get VoiceChannelManagementVM()
   n5 = GetAvailableChannels!(self:Voice Channel Management VM Object Reference=@n4) -> (ReturnValue:Array of ESidebarVoiceChannelType Enums=None)
   n6 = Get ActiveChannel(self=@n4) -> (ActiveChannel=None)
   n7 = AND Boolean(
    A=OR Boolean(A="Equal (Enum)"(A=@n2, B=TeamMembers), B="Equal (Enum)"(A=@n2, B=GameVoiceChannels))
    B=AND Boolean(
     A=Contains Item(TargetArray:Array of ESidebarVoiceChannelType Enums=@n5, ItemToFind:"ESidebarVoiceChannelType Enum (by ref)"=Game)
     B="Equal (Enum)"(A:ESidebarVoiceChannelType Enum=@n6, B:ESidebarVoiceChannelType Enum=Game)
    )
   )
   n8 = GetEpicPartyClientStatus!(self:Voice Channel Management VM Object Reference=Get VoiceChannelManagementVM()) -> (ReturnValue:EVoiceChatPartyClientActiveState Enum=NoClientsActiveInParty)
   > Return Node!(
    execute
    ReturnValue:bool=AND Boolean(
     A=OR Boolean(A=@n3, B=@n7)
     B="Equal (Enum)"(A:EVoiceChatPartyClientActiveState Enum=@n8, B:EVoiceChatPartyClientActiveState Enum=ThisClientActiveInParty)
    )
   )
   Return Node(execute=@n1.else, ReturnValue=false)
  ]}
 Update Control Buttons {nodes [
   Update Control Buttons() {is=title K2Node_FunctionEntry}
   > Branch(Condition=Get isEpicParty())
   > n1 = Sequence()
   > n2 = Can Execute Interaction(Interaction Type=AddFriend)
   n3 = Get HeaderItemType(self=Get FortSocialItemVM()) -> (HeaderItemType=Unknown)
   n4 = "Equal (Enum)"(A:ESocialHeaderItemType Enum=@n3, B:ESocialHeaderItemType Enum=SuggestedFriends)
   n5 = "Equal (Enum)"(A=@n3, B=RecentPlayers)
   > n6 = Branch(Condition=AND Boolean(A=@n2, B=OR Boolean(A=@n4, B=@n5)))
   > n7 = Get!(execute) -> (then, else, Add Friend Button:WBP Sidebar Social Player Add Friend Object Reference)
   n8 = Create Widget!(
    execute:Exec=@n7.else
    Class:User Widget Class Reference=&/WBP_Sidebar_SocialPlayer_AddFriend.WBP_Sidebar_SocialPlayer_AddFriend_C
    OwningPlayer:Player Controller Object Reference=GetOwningPlayer()
   ) -> (then:Exec, ReturnValue:WBP Sidebar Social Player Add Friend Object Reference)
   > n9 = Set Add Friend Button(Add Friend Button:WBP Sidebar Social Player Add Friend Object Reference=@n8) -> (Output_Get:WBP Sidebar Social Player Add Friend Object Reference)
   > n10 = AddChildToOverlay!(execute, self:Overlay Object Reference=Get Overlay_Controls(), Content:Widget Object Reference=@n9) -> (then, ReturnValue:Overlay Slot Object Reference)
   > Set VerticalAlignment!(execute, VerticalAlignment:EVerticalAlignment Enum=VAlign_Center, self:Overlay Slot Object Reference=@n10) -> (then, Output_Get:EVerticalAlignment Enum=VAlign_Fill)
   > SetNavigationRuleCustomBoundary!(
    execute
    self:Widget Object Reference=Get Add Friend Button() -> (Add Friend Button:WBP Sidebar Social Player Add Friend Object Reference)
    Direction:EUINavigation Enum=Left
    InCustomDelegate:Delegate=Create Event()
   ) -> (then)
   > Bind Event to On Hovered(self:WBP Sidebar Social Player Add Friend Object Reference=@n8, Delegate=Create Event()) {is=title K2Node_AddDelegate}
   > Bind Event to On Clicked(self:WBP Sidebar Social Player Add Friend Object Reference=@n8, Delegate=Create Event()) {is=title K2Node_AddDelegate}
   n11 = Get(execute=@n6.else)
   > RemoveChild(self=Get Overlay_Controls(), Content=@n11)
   > Unbind all Events from On Hovered!(execute:Exec, self:WBP Sidebar Social Player Add Friend Object Reference=@n11) -> (then:Exec) {is=title K2Node_ClearDelegate}
   > Unbind all Events from On Clicked!(execute:Exec, self:WBP Sidebar Social Player Add Friend Object Reference=@n11) -> (then:Exec) {is=title K2Node_ClearDelegate}
   n12 = Get FortSocialItemVM()
   n13 = Branch(execute=@n1.then_1, Condition="Equal (Enum)"(A=Get HeaderItemType(self=@n12) -> (HeaderItemType=Unknown), B=FriendInvites))
   > n14 = Get!(execute) -> (then, else, Add Accept Invite:WBP Sidebar Social Player Accept Invite Object Reference)
   n15 = Create Widget(
    execute=@n14.else
    Class=&/WBP_Sidebar_SocialPlayer_AcceptInvite.WBP_Sidebar_SocialPlayer_AcceptInvite_C
    OwningPlayer=GetOwningPlayer()
   ) -> (ReturnValue:WBP Sidebar Social Player Accept Invite Object Reference)
   > n16 = Set Add Accept Invite(Add Accept Invite:WBP Sidebar Social Player Accept Invite Object Reference=@n15) -> (Output_Get:WBP Sidebar Social Player Accept Invite Object Reference)
   > n17 = AddChildToOverlay(self=Get Overlay_Controls(), Content=@n16)
   > Set VerticalAlignment(VerticalAlignment=VAlign_Center, self=@n17) -> (Output_Get=VAlign_Fill)
   > Bind Event to On Hovered(self:WBP Sidebar Social Player Accept Invite Object Reference=@n15, Delegate=Create Event())
   > Bind Event to On Accept Invite(self:WBP Sidebar Social Player Accept Invite Object Reference=@n15, Delegate=Create Event()) {is=title K2Node_AddDelegate}
   > Bind Event to On Decline Invite(self:WBP Sidebar Social Player Accept Invite Object Reference=@n15, Delegate=Create Event()) {is=title K2Node_AddDelegate}
   n18 = Get(execute=@n13.else)
   > RemoveChild(self=Get Overlay_Controls(), Content=@n18)
   > Unbind all Events from On Hovered(self:WBP Sidebar Social Player Accept Invite Object Reference=@n18)
  ]}
 Update You Indicator And Controls {nodes [
   n1 = Update You Indicator And Controls!() -> (then, UseIndicator:bool=~) {is=title K2Node_FunctionEntry}
   > n2 = Branch(Condition=@n1)
   > n3 = Sequence()
   > n4 = Get!(execute) -> (then, else, YouIndicatorWidget:Common User Widget Object Reference)
   n5 = Create Widget(execute=@n4.else, Class=&/WBP_Sidebar_YouIndicator.WBP_Sidebar_YouIndicator_C, OwningPlayer=GetOwningPlayer()) -> (ReturnValue:WBP Sidebar You Indicator Object Reference)
   > n6 = Set YouIndicatorWidget(YouIndicatorWidget:Common User Widget Object Reference=@n5) -> (Output_Get:Common User Widget Object Reference)
   > AddChildToOverlay(self=Get Overlay_YouIndicator() -> (Overlay_YouIndicator:Overlay Object Reference), Content=@n6)
   n7 = Get HeaderItemType(self=Get FortSocialItemVM()) -> (HeaderItemType=Unknown)
   n8 = "Equal (Enum)"(A=@n7, B=NativeEpicPartyMembers)
   n9 = "Equal (Enum)"(A=@n7, B=TeamMembers)
   n10 = "Equal (Enum)"(A=@n7, B=GameVoiceChannels)
   Branch(execute=@n3.then_1, Condition=AND Boolean(A=Get isUpdatedPartyControlsEnabled(), B=OR Boolean!(A=@n8, B=@n9, C:bool=@n10) -> (ReturnValue)))
   > n11 = Get!(execute) -> (then, else, YouControlsWidget:WBP Sidebar Social Player Controls Object Reference)
   n12 = Create Widget(
    execute=@n11.else
    Class=&/WBP_Sidebar_SocialPlayer_Controls.WBP_Sidebar_SocialPlayer_Controls_C
    OwningPlayer=GetOwningPlayer()
   ) -> (ReturnValue:WBP Sidebar Social Player Controls Object Reference)
   > n13 = Set YouControlsWidget(YouControlsWidget:WBP Sidebar Social Player Controls Object Reference=@n12) -> (Output_Get:WBP Sidebar Social Player Controls Object Reference)
   > n14 = AddChildToOverlay(self=Get Overlay_Controls(), Content=@n13)
   > Set VerticalAlignment(VerticalAlignment=VAlign_Center, self=@n14) -> (Output_Get=VAlign_Fill)
   > Bind Event to On Hovered(self:WBP Sidebar Social Player Controls Object Reference=@n13, Delegate=Create Event())
   > SetNavigationRuleCustomBoundary(self=@n13, Direction=Left, InCustomDelegate=Create Event())
   > SetMaxTextWidthBeforeClipping!(
    execute
    self:Self Object Reference
    Voice Chat Method:EFortVoiceChatMethod Enum=GetVoiceChatMethod!(self:Voice Channel Management VM Object Reference=Get VoiceChannelManagementVM()) -> (ReturnValue:EFortVoiceChatMethod Enum=OpenMic)
   ) -> (then)
   n15 = Sequence(execute=@n2.else)
   > Get!(execute) -> (then, else, Overlay_YouIndicator:Overlay Object Reference)
   > RemoveChild(self=Get Overlay_YouIndicator(), Content=Get YouIndicatorWidget() -> (YouIndicatorWidget:Common User Widget Object Reference))
   > Set YouIndicatorWidget()
   > SetMaxTextWidthBeforeClipping!(execute, self:WBP Sidebar User Object Reference=Get WBP_Social_User(), MaxNameTextWidth:double=435) -> (then)
   n16 = Get!(execute=@n15.then_1) -> (then, else, YouControlsWidget:WBP Sidebar Social Player Controls Object Reference)
   > Unbind Event from On Hovered(self:WBP Sidebar Social Player Controls Object Reference=@n16, Delegate=Create Event()) {is=title K2Node_RemoveDelegate}
   > RemoveChild(self=Get Overlay_Controls(), Content=@n16)
   > Set YouControlsWidget()
  ]}
 UpdateFriendRequestsBindings {nodes [
   n1 = UpdateFriendRequestsBindings!() -> (then, Request Count:int) {is=title K2Node_FunctionEntry}
   > Branch(Condition="Not Equal (Integer)"(A=@n1))
   > ForceUpdateAllBindingsOnListItemObjectSet()
  ]}
 UpdateVoiceIconsVisibility {nodes [
   UpdateVoiceIconsVisibility!() -> (then, Voice State:EVoiceChatPartyClientActiveState Enum=NoClientsActiveInParty) {is=title K2Node_FunctionEntry}
   > Get!(execute) -> (then, else, FortSocialItemVM:Fort Social Item VM Object Reference)
   > AddVMForVoiceIndicators()
  ]}
 UpdateisSocialBanned {nodes [
   n1 = UpdateisSocialBanned!() -> (then, isSocialBanned:bool=~) {is=title K2Node_FunctionEntry}
   > n2 = Get!(execute) -> (then, else, YouControlsWidget:WBP Sidebar Social Player Controls Object Reference)
   > Set with Broadcast isSocialBanned!(execute, isSocialBanned:bool=@n1, self:WBP Sidebar Social Player Controls Object Reference=@n2) -> (then, Output_Get:bool)
  ]}
]
variables [
 Add Accept Invite {type=WBP_Sidebar_SocialPlayer_AcceptInvite_C, default=null}
 Add Friend Button {type=WBP_Sidebar_SocialPlayer_AddFriend_C, default=null}
 AllowVoiceChatSettingIndex {type=integer}
 ContextMenuActionContext {type=UIKitContextMenuActionContext, default=null}
 InteractionsMap {type="map<FortSocialInteractionVM>", default {}}
 isEpicParty {type=boolean}
 isFriendRequestEntry {type=boolean}
 isLocalUserForBlueprint {type=boolean}
 isSocialPlayerFocused {type=boolean}
 isUpdatedPartyControlsEnabled {type=boolean}
 Map Value {type=FortSocialInteractionVM, default=null}
 OnBlockPlayer {}
 OnEditNickname {}
 OnRemoveFriend {}
 YouControlsWidget {type=WBP_Sidebar_SocialPlayer_Controls_C, default=null}
 YouIndicatorWidget {type=CommonUserWidget, default=null}
]
widgets {
 info {
  inheritedWidgetCount=0
  namedSlotCount=4
  parentClass=/Sidebar/Buttons/WBP_Sidebar_Button_Base.WBP_Sidebar_Button_Base_C
  rootWidgetClass=null
  widgetCount=24
 }
 widgets [
  GridPanel {widgetClassPath=&GridPanel, bIsVariable=true, widget=GridPanel, properties {columnFill [1], rowFill [1]}}
  WBP_Social_User {
   widgetClassPath=/Sidebar/Shared/WBP_Sidebar_User
   bIsVariable=true
   parent=GridPanel
   widget=WBP_Social_User
   properties {visibility=Visible, padding=<ltrb 15 2 15 2>}
   slot {class=GridSlot}
  }
  HorizontalBox_0 {widgetClassPath=&HorizontalBox, namedSlotHost=WBP_Social_User, widget=HorizontalBox_0}
  HB_Joinable {widgetClassPath=&HorizontalBox, bIsVariable=true, parent=HorizontalBox_0, widget=HB_Joinable, slot {class=HorizontalBoxSlot}}
  Text_Joinable {
   widgetClassPath=&CommonTextBlock
   bIsVariable=true
   parent=HB_Joinable
   widget=Text_Joinable
   properties {
    justification=Center
    mobileFontSizeMultiplier=1.4
    text=Joinable
    colorAndOpacity {colorUseRule=UseColor_Specified, specifiedColor=<rgba 0.1144 0.7682 0.1356 1>}
    font {
     bForceMonospaced=false
     bMaterialIsStencil=false
     fontMaterial=null
     fontObject=&HeadingNow
     letterSpacing=0
     monospacedWidth=1
     size=12.75
     skewAmount=0
     typefaceFontName=75 Medium
     outlineSettings {
      bApplyOutlineToDropShadows=false
      bMiteredCorners=false
      bSeparateFillAlpha=false
      outlineMaterial=/FortUILibrary/Systems/ModularButton/MaterialInstances/MI_UIKit_TextOutline
      outlineSize=0
      outlineColor=<rgba 0 0 0 1>
     }
    }
    shadowOffset=<0 0>
   }
   slot {class=HorizontalBoxSlot}
  }
  DividerDot_Joinable {
   widgetClassPath=&Image
   bIsVariable=true
   parent=HB_Joinable
   widget=DividerDot_Joinable
   properties {
    visibility=SelfHitTestInvisible
    brush {
     bIsDynamicallyLoaded=false
     drawAs=Image
     imageType=FullColor
     mirroring=NoMirror
     resourceName=None
     resourceObject=/FortUILibrary/Foundation/Art_ToolKit/Tools/MI_UI_Circle_ConversationBacchusConfirm
     tiling=NoTile
     imageSize=<6 6>
     margin=<ltrb 0 0 0 0>
     outlineSettings {
      bUseBrushTransparency=false
      roundingType=HalfHeightRadius
      width=0
      color {colorUseRule=UseColor_Specified, specifiedColor=<rgba 0 0 0 0>}
      cornerRadii=<0 0 0 0>
     }
     tintColor {colorUseRule=UseColor_Specified, specifiedColor=<rgba 1 1 1 1>}
     uVRegion {bIsValid=false, max=<0 0>, min=<0 0>}
    }
    colorAndOpacity=<rgba 1 1 1 0.5>
   }
   slot {class=HorizontalBoxSlot, properties {horizontalAlignment=HAlign_Center, verticalAlignment=VAlign_Center, padding=<ltrb 6 0 6 0>}}
  }
  HB_Invited {
   widgetClassPath=&HorizontalBox
   bIsVariable=true
   parent=HorizontalBox_0
   widget=HB_Invited
   properties {visibility=Collapsed}
   slot {class=HorizontalBoxSlot}
  }
  Text_Invited {
   widgetClassPath=&CommonTextBlock
   bIsVariable=true
   parent=HB_Invited
   widget=Text_Invited
   properties {
    justification=Center
    mobileFontSizeMultiplier=1.4
    text="Invited you to their party!"
    colorAndOpacity {colorUseRule=UseColor_Specified, specifiedColor=<rgba 0.9301 1 0.0103 1>}
    font {
     bForceMonospaced=false
     bMaterialIsStencil=false
     fontMaterial=null
     fontObject=&HeadingNow
     letterSpacing=0
     monospacedWidth=1
     size=12.75
     skewAmount=0
     typefaceFontName=75 Medium
     outlineSettings {
      bApplyOutlineToDropShadows=false
      bMiteredCorners=false
      bSeparateFillAlpha=false
      outlineMaterial=&MI_UIKit_TextOutline
      outlineSize=0
      outlineColor=<rgba 0 0 0 1>
     }
    }
    shadowOffset=<0 0>
   }
   slot {class=HorizontalBoxSlot}
  }
  DividerDot_Invited {
   widgetClassPath=&Image
   bIsVariable=true
   parent=HB_Invited
   widget=DividerDot_Invited
   properties {
    visibility=SelfHitTestInvisible
    brush {
     bIsDynamicallyLoaded=false
     drawAs=Image
     imageType=FullColor
     mirroring=NoMirror
     resourceName=None
     resourceObject=&MI_UI_Circle_ConversationBacchusConfirm
     tiling=NoTile
     imageSize=<6 6>
     margin=<ltrb 0 0 0 0>
     outlineSettings {
      bUseBrushTransparency=false
      roundingType=HalfHeightRadius
      width=0
      color {colorUseRule=UseColor_Specified, specifiedColor=<rgba 0 0 0 0>}
      cornerRadii=<0 0 0 0>
     }
     tintColor {colorUseRule=UseColor_Specified, specifiedColor=<rgba 1 1 1 1>}
     uVRegion {bIsValid=false, max=<0 0>, min=<0 0>}
    }
    colorAndOpacity=<rgba 1 1 1 0.5>
   }
   slot {class=HorizontalBoxSlot, properties {horizontalAlignment=HAlign_Center, verticalAlignment=VAlign_Center, padding=<ltrb 6 0 6 0>}}
  }
  HorizontalBox_LeaderAndYouFlag {widgetClassPath=&HorizontalBox, bIsVariable=true, namedSlotHost=WBP_Social_User, widget=HorizontalBox_LeaderAndYouFlag}
  LeaderCrown {
   widgetClassPath=&Image
   bIsVariable=true
   parent=HorizontalBox_LeaderAndYouFlag
   widget=LeaderCrown
   properties {
    visibility=Collapsed
    brush {
     bIsDynamicallyLoaded=false
     drawAs=Image
     imageType=FullColor
     mirroring=NoMirror
     resourceName=None
     resourceObject=&MI_Sidebar_Icon_Crown
     tiling=NoTile
     imageSize=<24 24>
     margin=<ltrb 0 0 0 0>
     outlineSettings {
      bUseBrushTransparency=false
      roundingType=HalfHeightRadius
      width=0
      color {colorUseRule=UseColor_Specified, specifiedColor=<rgba 0 0 0 0>}
      cornerRadii=<0 0 0 0>
     }
     tintColor {colorUseRule=UseColor_Specified, specifiedColor=<rgba 1 1 1 1>}
     uVRegion {bIsValid=false, max=<0 0>, min=<0 0>}
    }
    colorAndOpacity=<rgba 0.9301 1 0.0103 1>
   }
   slot {class=HorizontalBoxSlot, properties {horizontalAlignment=HAlign_Left, verticalAlignment=VAlign_Center, padding=<ltrb 0 0 8 0>}}
  }
  Overlay_YouIndicator {
   widgetClassPath=&Overlay
   bIsVariable=true
   parent=HorizontalBox_LeaderAndYouFlag
   widget=Overlay_YouIndicator
   slot {class=HorizontalBoxSlot, properties {verticalAlignment=VAlign_Center}}
  }
  OtherPlayer_VoiceStatusIcons {
   widgetClassPath=&HorizontalBox
   bIsVariable=true
   parent=GridPanel
   widget=OtherPlayer_VoiceStatusIcons
   properties {visibility=Collapsed}
   slot {class=GridSlot, properties {column=1, horizontalAlignment=HAlign_Left, verticalAlignment=VAlign_Center, padding=<ltrb 0 0 16 0>}}
  }
  OtherPlayer_VoiceStatus {
   widgetClassPath=&Image
   bIsVariable=true
   parent=OtherPlayer_VoiceStatusIcons
   widget=OtherPlayer_VoiceStatus
   properties {
    visibility=Collapsed
    brush {
     bIsDynamicallyLoaded=false
     drawAs=Image
     imageType=FullColor
     mirroring=NoMirror
     resourceName=None
     resourceObject=/VoiceRep/Materials/MI_UI_Icon_Muted
     tiling=NoTile
     imageSize=<32 32>
     margin=<ltrb 0 0 0 0>
     outlineSettings {
      bUseBrushTransparency=false
      roundingType=HalfHeightRadius
      width=0
      color {colorUseRule=UseColor_Specified, specifiedColor=<rgba 0 0 0 0>}
      cornerRadii=<0 0 0 0>
     }
     tintColor {colorUseRule=UseColor_Specified, specifiedColor=<rgba 0.5776 0.5776 0.5776 1>}
     uVRegion {bIsValid=false, max=<0 0>, min=<0 0>}
    }
   }
   slot {class=HorizontalBoxSlot}
  }
  SizeBox_0 {
   widgetClassPath=&SizeBox
   parent=OtherPlayer_VoiceStatusIcons
   widget=SizeBox_0
   properties {bOverride_MaxDesiredWidth=true, maxDesiredWidth=128}
   slot {class=HorizontalBoxSlot, properties {verticalAlignment=VAlign_Center}}
  }
  Text_Chat_Channel {
   widgetClassPath=&CommonTextBlock
   bIsVariable=true
   parent=SizeBox_0
   widget=Text_Chat_Channel
   properties {
    bAutoCollapseWithEmptyText=true
    clipping=ClipToBounds
    mobileFontSizeMultiplier=1.4
    scrollStyle=&TextScrollStyle-Base_C
    textOverflowPolicy=Ellipsis
    colorAndOpacity {colorUseRule=UseColor_Specified, specifiedColor=<rgba 0.5776 0.5776 0.5776 1>}
    font {
     bForceMonospaced=false
     bMaterialIsStencil=false
     fontMaterial=/Sidebar/Materials/Fonts/MI_Sidebar_FontColor_Secondary
     fontObject=&HeadingNow
     letterSpacing=0
     monospacedWidth=1
     size=12.75
     skewAmount=0
     typefaceFontName=74 Regular
     outlineSettings {
      bApplyOutlineToDropShadows=false
      bMiteredCorners=false
      bSeparateFillAlpha=false
      outlineMaterial=&MI_UIKit_TextOutline
      outlineSize=0
      outlineColor=<rgba 0 0 0 1>
     }
    }
    shadowOffset=<0 0>
   }
   slot {class=SizeBoxSlot, properties {horizontalAlignment=HAlign_Center, verticalAlignment=VAlign_Center}}
  }
  LocalUser_VoiceStatusIcons {
   widgetClassPath=&HorizontalBox
   bIsVariable=true
   parent=GridPanel
   widget=LocalUser_VoiceStatusIcons
   properties {visibility=Collapsed}
   slot {
    class=GridSlot
    properties {column=1, columnSpan=1, horizontalAlignment=HAlign_Left, verticalAlignment=VAlign_Center, padding=<ltrb 0 0 16 0>}
   }
  }
  LocalUser_MutedIcon {
   widgetClassPath=&Image
   bIsVariable=true
   parent=LocalUser_VoiceStatusIcons
   widget=LocalUser_MutedIcon
   properties {
    visibility=Collapsed
    brush {
     bIsDynamicallyLoaded=false
     drawAs=Image
     imageType=FullColor
     mirroring=NoMirror
     resourceName=None
     resourceObject=/Sidebar/Materials/Icons/MI_Sidebar_Icon_MicSlash
     tiling=NoTile
     imageSize=<32 32>
     margin=<ltrb 0 0 0 0>
     outlineSettings {
      bUseBrushTransparency=false
      roundingType=HalfHeightRadius
      width=0
      color {colorUseRule=UseColor_Specified, specifiedColor=<rgba 0 0 0 0>}
      cornerRadii=<0 0 0 0>
     }
     tintColor {colorUseRule=UseColor_Specified, specifiedColor=<rgba 1 1 1 1>}
     uVRegion {bIsValid=false, max=<0 0>, min=<0 0>}
    }
   }
   slot {class=HorizontalBoxSlot, properties {size {sizeRule=Fill, value=1}}}
  }
  LocalUser_DeafenedIcon {
   widgetClassPath=&Image
   bIsVariable=true
   parent=LocalUser_VoiceStatusIcons
   widget=LocalUser_DeafenedIcon
   properties {
    visibility=Collapsed
    brush {
     bIsDynamicallyLoaded=false
     drawAs=Image
     imageType=FullColor
     mirroring=NoMirror
     resourceName=None
     resourceObject=/Sidebar/Materials/Icons/MI_Sidebar_Icon_HeadphonesSlash
     tiling=NoTile
     imageSize=<32 32>
     margin=<ltrb 0 0 0 0>
     outlineSettings {
      bUseBrushTransparency=false
      roundingType=HalfHeightRadius
      width=0
      color {colorUseRule=UseColor_Specified, specifiedColor=<rgba 0 0 0 0>}
      cornerRadii=<0 0 0 0>
     }
     tintColor {colorUseRule=UseColor_Specified, specifiedColor=<rgba 1 1 1 1>}
     uVRegion {bIsValid=false, max=<0 0>, min=<0 0>}
    }
   }
   slot {class=HorizontalBoxSlot, properties {size {sizeRule=Fill, value=1}}}
  }
  FriendRequestIndiactor {
   widgetClassPath=&HorizontalBox
   bIsVariable=true
   parent=GridPanel
   widget=FriendRequestIndiactor
   properties {visibility=Collapsed}
   slot {class=GridSlot, properties {column=1, horizontalAlignment=HAlign_Right, verticalAlignment=VAlign_Center, padding=<ltrb 0 0 16 0>}}
  }
  WBP_UIKit_StatusIndicator {
   widgetClassPath=/FortUILibrary/UIKit/StatusIndicator/WBP_UIKit_StatusIndicator
   bIsVariable=true
   parent=FriendRequestIndiactor
   widget=WBP_UIKit_StatusIndicator
   properties {type=Attention, bangType=Custom, configuration=Number, visibility=HitTestInvisible}
   slot {class=HorizontalBoxSlot}
  }
  Icon {
   widgetClassPath=&Image
   bIsVariable=true
   parent=FriendRequestIndiactor
   widget=Icon
   properties {brush {
     bIsDynamicallyLoaded=false
     drawAs=Image
     imageType=FullColor
     mirroring=NoMirror
     resourceName=None
     resourceObject=/Sidebar/Materials/Icons/MI_Sidebar_Icon_Base
     tiling=NoTile
     imageSize=<24 24>
     margin=<ltrb 0 0 0 0>
     outlineSettings {
      bUseBrushTransparency=false
      roundingType=HalfHeightRadius
      width=0
      color {colorUseRule=UseColor_Specified, specifiedColor=<rgba 0 0 0 0>}
      cornerRadii=<0 0 0 0>
     }
     tintColor {colorUseRule=UseColor_Specified, specifiedColor=<rgba 1 1 1 1>}
     uVRegion {bIsValid=false, max=<0 0>, min=<0 0>}
    }}
   slot {class=HorizontalBoxSlot, properties {horizontalAlignment=HAlign_Center, verticalAlignment=VAlign_Center, padding=<ltrb 4 0 0 0>}}
  }
  PlayerAction_ContextMenu {
   widgetClassPath=/FortUILibrary/UIKit/ContextMenu/WBP_UIKit_ContextMenuAnchorButton
   bIsVariable=true
   parent=GridPanel
   widget=PlayerAction_ContextMenu
   properties {
    button Background Material=null
    button Outline Material=null
    buttonClass=/Game/UI/Frontend/Lobby/Nameplates/WBP_EmptyButton_ContextMenu.WBP_EmptyButton_ContextMenu_C
    buttonText=""
    closeMenuOnContextClicked=true
    isMenuModal=false
    menuPlacement=MenuPlacement_BelowAnchor
    min Height=1
    button Padding=<ltrb 0 0 0 0>
   }
   slot {class=GridSlot, properties {horizontalAlignment=HAlign_Center, verticalAlignment=VAlign_Bottom}}
  }
  Overlay_Controls {
   widgetClassPath=&Overlay
   bIsVariable=true
   widget=Overlay_Controls
   properties {navigation=&self:WidgetTree.Overlay_Controls.WidgetNavigation_0}
  }
 ]
}
