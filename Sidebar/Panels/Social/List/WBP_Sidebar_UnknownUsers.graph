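# Sidebar social list entry button that stands in for a group of unknown users as a "+N more" row.
# When given a list item, it casts it to BP_SidebarUnknownUsersEntryVM and formats its unknown-user count into the name text of an embedded WBP_Sidebar_User.
# It then collapses that widget's message, mutual-friends and second-line areas and hides the avatar's online status.
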WidgetBlueprint /Sidebar/Panels/Social/List/WBP_Sidebar_UnknownUsers
graphs [
 EventGraph {nodes [
   n1 = Event OnListItemObjectSet()
   > n2 = Cast To BP_SidebarUnknownUsersEntryVM!(execute, Object=@n1.ListItemObject) -> (then, CastFailed, AsBP Sidebar Unknown Users Entry VM:BP Sidebar Unknown Users Entry VM Object Reference) {is=title K2Node_DynamicCast}
   > SetText(
    self=Get RichText_Name!(self:WBP Sidebar User Object Reference=Get WBP_Social_User() -> (WBP_Social_User:WBP Sidebar User Object Reference)) -> (RichText_Name:Common Rich Text Block Object Reference)
    InText=Format Text!(
     Format="LOCTEXT(\"+{N} more\")"
     N:int=Get Num Unknown Users!(self:BP Sidebar Unknown Users Entry VM Object Reference=@n2) -> (Num Unknown Users:int=0)
    ) -> (Result)
   )
   n3 = Get WBP_Social_User()
   > SetVisibility(
    self=[
     Get Text_Message!(self:WBP Sidebar User Object Reference=@n3) -> (Text_Message:Common Text Object Reference)
     Get MutualFriends_Text!(self:WBP Sidebar User Object Reference=@n3) -> (MutualFriends_Text:Common Text Object Reference)
     Get Slot_SecondLine!(self:WBP Sidebar User Object Reference=@n3) -> (Slot_SecondLine:Named Slot Object Reference)
    ]
    InVisibility=Collapsed
   )
   > SetOnlineStatusVisibility!(
    execute
    self:WBP Sidebar Player Avatar Object Reference=Get PlayerAvatar!(self:WBP Sidebar User Object Reference=@n3) -> (PlayerAvatar:WBP Sidebar Player Avatar Object Reference)
    bHideStatus:bool=true
   ) -> (then)
  ]}
 OnBlockPlayer {nodes [OnBlockPlayer!() -> (then, SocialUser:Fort Social User Object Reference, Interaction:Fort Social Interaction VM Object Reference) {is=title K2Node_FunctionEntry}]}
 OnEditNickname {nodes [OnEditNickname!() -> (then, SocialUser:Fort Social User Object Reference) {is=title K2Node_FunctionEntry}]}
 OnRemoveFriend {nodes [OnRemoveFriend!() -> (then, SocialUser:Fort Social User Object Reference, Interaction:Fort Social Interaction VM Object Reference) {is=title K2Node_FunctionEntry}]}
]
variables [
 ContextMenuActionContext {type=UIKitContextMenuActionContext, default=null}
 InteractionsMap {type="map<FortSocialInteractionVM>", default {}}
 Is Local Player {type=boolean}
 OnBlockPlayer {}
 OnEditNickname {}
 OnRemoveFriend {}
 YouIndicatorWidget {type=CommonUserWidget, default=null}
]
widgets {
 info {
  inheritedWidgetCount=0
  namedSlotCount=1
  parentClass=/Sidebar/Buttons/WBP_Sidebar_Button_Base.WBP_Sidebar_Button_Base_C
  rootWidgetClass=null
  widgetCount=2
 }
 widgets [
  GridPanel {widgetClassPath=&GridPanel, bIsVariable=true, widget=GridPanel, properties {columnFill [1], rowFill [1]}}
  WBP_Social_User {
   widgetClassPath=/Sidebar/Shared/WBP_Sidebar_User
   bIsVariable=true
   parent=GridPanel
   widget=WBP_Social_User
   properties {padding=<ltrb 15 12 15 12>}
   slot {class=GridSlot}
  }
 ]
}
