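# Sidebar social panel widget for Party and Game voice chat: channel selection, mic and headphone controls, party joinability and chat reporting.
# Clicks and context-menu actions drive the voice channel and party view models, with mute acting as push-to-talk where allowed.
# Helper functions turn restriction and active-device state into substate messages, visibility and fix actions like settings or parental controls.
# Console variables switch Epic parties and updated controls; design-time presets preview states.
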WidgetBlueprint /Sidebar/Panels/Social/WBP_PartyAndGame_Controls
animations [Anim_OnlineStatusBackplate {
  binding_count=1
  length_seconds=0.3
  object=&self:Anim_OnlineStatusBackplate
  sequence {
   bindings [Backplate_ChannelSelector {
     id=#1
     possessed_class=Image
     tracks [{
       type=MovieSceneWidgetMaterialTrack
       sections [{
         type=MovieSceneParameterSection
         end_frame=30
         start_frame=0
         channels [Control {type=MovieSceneScriptingFloatChannel, default=0, thinned_from=3, keys [[0,0],[30,1]]}]
        }]
      }]
    }]
   display_rate {numerator=100}
  }
 }]
graphs [
 AddMuteContextMenuItem {nodes [
   n1 = AddMuteContextMenuItem!() -> (then, IsSelfMuted:bool=~) {is=title K2Node_FunctionEntry}
   > RemoveInteraction!(
    execute
    self:UIKit Context Menu Action Context Object Reference=Get MicrophoneMenuActionContext() -> (MicrophoneMenuActionContext:UIKit Context Menu Action Context Object Reference)
    InteractionNameID:Name=Get ContextMenuMuteName() -> (ContextMenuMuteName:Name=None)
   ) -> (then)
   > AddInteraction!(
    execute
    self:UIKit Context Menu Action Context Object Reference=Get MicrophoneMenuActionContext()
    InActionInfo_DisplayName:Text=SelectText(A="LOCTEXT(\"Unmute\")", B="LOCTEXT(\"Mute\")", bPickA=@n1)
    InActionInfo_IconTexture:Texture 2D Soft Object Reference
    InActionInfo_IconMaterial:Material Instance Soft Object Reference
    InActionInfo_SortPriority:int=0
    InActionInfo_bIsAvailable:bool=True
    InActionInfo_bIsEnabled:bool=True
    InActionInfo_bIsSelected:bool=False
    InActionInfo_bAddDivider:bool=False
    InActionInfo_bPositionDividerAfter:bool=False
    InActionInfo_bCloseMenuOnSelect:bool=False
    InActionInfo_InteractionNameID:Name=Get ContextMenuMuteName() -> (ContextMenuMuteName=None)
    InActionInfo_Interaction:Data Table Row Handle Structure=(DataTable=None,RowName="")
    InActionInfo_PressedDelegate:Delegate="(null).None"
    InActionInfo_ReleasedDelegate:Delegate="(null).None"
    InInteractionDelegate:"Delegate (by ref)"=Create Event()
   ) -> (then)
  ]}
 AddReportConversationContextMenuItem {nodes [
   AddReportConversationContextMenuItem!() -> (then, IsPartyChannel:bool=~) {is=title K2Node_FunctionEntry}
   > RemoveInteraction(
    self=Get ReportMenuActionContext() -> (ReportMenuActionContext:UIKit Context Menu Action Context Object Reference)
    InteractionNameID=Get ContextMenuReportName() -> (ContextMenuReportName:Name=None)
   )
   > AddInteraction(
    self=Get ReportMenuActionContext()
    InActionInfo_DisplayName="LOCTEXT(\"Report chat\")"
    InActionInfo_SortPriority=0
    InActionInfo_bIsAvailable=True
    InActionInfo_bIsEnabled=True
    InActionInfo_bIsSelected=False
    InActionInfo_bAddDivider=False
    InActionInfo_bPositionDividerAfter=False
    InActionInfo_bCloseMenuOnSelect=False
    InActionInfo_InteractionNameID=Get ContextMenuReportName() -> (ContextMenuReportName=None)
    InActionInfo_Interaction=(DataTable=None,RowName="")
    InActionInfo_PressedDelegate="(null).None"
    InActionInfo_ReleasedDelegate="(null).None"
    InInteractionDelegate=Create Event()
   )
  ]}
 BP_GetDesiredFocusTarget {nodes [
   Get Desired Focus Target()
   > n1 = Get!(execute:Exec) -> (then:Exec, else:Exec, LastFocusedButton:User Widget Object Reference)
   > n2 = Branch(Condition=IsVisible(self=@n1))
   > Return Node!(execute, ReturnValue:Widget Object Reference=@n1)
   n3 = Get VoiceChannelManagementVM() -> (VoiceChannelManagementVM:Voice Channel Management VM Object Reference)
   n4 = Get Channel_Party_Button() -> (Channel_Party_Button:WBP Sidebar Voice Channel Button Object Reference)
   Return Node(
    execute=[@n2.else, @n1.else]
    ReturnValue=Select!(
     None:Widget Object Reference=@n4
     Party:Widget Object Reference=@n4
     Game:Widget Object Reference=Get Channel_Game_Button() -> (Channel_Game_Button:WBP Sidebar Voice Channel Button Object Reference)
     EpicParty:Widget Object Reference=@n4
     Index:ESidebarVoiceChannelType Enum=Get ActiveChannel!(self:Voice Channel Management VM Object Reference=@n3) -> (ActiveChannel:ESidebarVoiceChannelType Enum=None)
    ) -> (ReturnValue:Widget Object Reference)
   )
  ]}
 CF_VoiceChatMethodTextVisibility {nodes [
   CF_VoiceChatMethodTextVisibility!() -> (then, Active Channel:ESidebarVoiceChannelType Enum=None, AvailableChannels:Array of ESidebarVoiceChannelType Enums=None) {is=title K2Node_FunctionEntry}
   n1 = "Equal (Enum)"(
    A:ESidebarVoiceChannelType Enum=Get Active Channel() -> (Active Channel:ESidebarVoiceChannelType Enum=None)
    B:ESidebarVoiceChannelType Enum=Select!(
     Option 0:ESidebarVoiceChannelType Enum=Party
     Option 1:ESidebarVoiceChannelType Enum=EpicParty
     Index:bool=Get IsUsingEpicParties() -> (IsUsingEpicParties:bool)
    ) -> (ReturnValue:ESidebarVoiceChannelType Enum)
   )
   n2 = Get AvailableChannels() -> (AvailableChannels:Array of ESidebarVoiceChannelType Enums=None)
   n3 = Get FortGlobalUIContext() -> (ReturnValue:Fort Global UIContext Object Reference) {is=title K2Node_GetSubsystem}
   n4 = Select(
    Option 0:ESlateVisibility Enum=Collapsed
    Option 1:ESlateVisibility Enum=SelfHitTestInvisible
    Index=AND Boolean(
     A=Contains Item(TargetArray:Array of ESidebarVoiceChannelType Enums=@n2, ItemToFind:"ESidebarVoiceChannelType Enum (by ref)"=Game)
     B=IsInZone!(self:Fort Global UIContext Object Reference=@n3) -> (ReturnValue:bool)
    )
   ) -> (ReturnValue:ESlateVisibility Enum)
   > Return Node!(
    execute
    Visibility:ESlateVisibility Enum=Select(
     Option 0=Select(Option 0=@n4, Option 1=SelfHitTestInvisible, Index=@n1)
     Option 1=Collapsed
     Index=ShouldAlwaysActAsOpenMic!(self) -> (ReturnValue:bool)
    )
   )
  ]}
 "CloseSidebarOpenMouse&&KeyboardSettings" {nodes [
   Inputs()
   > RequestCloseSidebar(self:Fort Sidebar VM Object Reference=Get FortSidebarVM() -> (FortSidebarVM:Fort Sidebar VM Object Reference))
   > n1 = Branch(Condition=IsSaveTheWorldUI!() -> (IsSaveTheWorld:bool))
   > n2 = Delay(Duration=0.2)
   RequestOpenSettings!(
    execute=[@n2.then, @n1.else]
    self:Fort Settings Context Object Reference=Get FortSettingsContext!() -> (ReturnValue:Fort Settings Context Object Reference) {is=title K2Node_GetBlueprintContextEx}
    ContextWidget:Common User Widget Object Reference=Self-Reference()
    SettingsTabId:Name=MouseAndKeyboardBindingsCollection
    SettingToSelect:String=LegacyKeyboardInputSetting_PushtoTalk_PushToTalk
   ) -> (then)
   > Outputs!(then:Exec)
   Select(Option 0:Name=MouseAndKeyboardBindingsCollection, Option 1:Name=KBMPresetCollection, Index=IsSaveTheWorldUI()) -> (ReturnValue:Name)
  ]}
 CloseSidebarOpenSettings {nodes [
   Inputs()
   > RequestCloseSidebar(self=Get FortSidebarVM())
   > n1 = Branch(Condition=IsSaveTheWorldUI())
   > n2 = Delay(Duration=0.2)
   OpenSettingsToVoiceChat!(execute=[@n2.then, @n1.else], OwningWidget:Common User Widget Object Reference=Self-Reference()) -> (then)
   > Outputs()
  ]}
 CloseSidebarOpenVoiceChatSettings {nodes [
   Inputs()
   > RequestCloseSidebar(self=Get FortSidebarVM())
   > n1 = Branch(Condition=IsSaveTheWorldUI())
   > n2 = Delay(Duration=0.2)
   OpenSettingsToVoiceChatMethod!(execute=[@n2.then, @n1.else], OwningWidget:Common User Widget Object Reference=Self-Reference()) -> (then)
   > Outputs()
  ]}
 DesignTimePreset {nodes [
   Inputs()
   n1 = Get IsUsingEpicParties()
   n2 = Get SidebarCurrentVoiceChatStatusEnum() -> (SidebarCurrentVoiceChatStatusEnum:E_UI_BP_SidebarVoiceChatStatus Enum=NewEnumerator0)
   > SetVisibility(
    self=Get VB_SubStates() -> (VB_SubStates:Vertical Box Object Reference)
    InVisibility=GetSidebarSubStateVisibility!(
     self
     VoiceChatStatus:E_UI_BP_SidebarVoiceChatStatus Enum=@n2
     ActiveChannel:ESidebarVoiceChannelType Enum=Select(Option 0:ESidebarVoiceChannelType Enum=Party, Option 1:ESidebarVoiceChannelType Enum=EpicParty, Index=@n1) -> (ReturnValue:ESidebarVoiceChannelType Enum)
     IsEpicParty:bool=false
    ) -> (NewParam:ESlateVisibility Enum=Visible)
   )
   > SetText!(
    execute
    self:WBP UIKit Button Generic Base Object Reference=Get Substate_Action_Button() -> (Substate_Action_Button:WBP Sidebar Button Generic Object Reference)
    Text:Text=SetSubStateActionButtonText!(self, VoiceChatStatus:E_UI_BP_SidebarVoiceChatStatus Enum=@n2) -> (ButtonText:Text)
   ) -> (then)
   K2Node_Composite_1 = DesignTimePreviewRestriction!(Index:E_UI_BP_SidebarVoiceChatStatus Enum=Get SidebarCurrentVoiceChatStatusEnum() -> (SidebarCurrentVoiceChatStatusEnum=NewEnumerator0)) -> (ReturnValue:ELocalVoiceChatRestriction Enum=Unrestricted) {is=title K2Node_Composite}
   n3 = Select(
    Option 0:EVoiceChatPartyClientActiveState Enum=ThisClientActiveInParty
    Option 1:EVoiceChatPartyClientActiveState Enum=OtherClientActiveInParty
    Index="Equal (Enum)"(
     A:E_UI_BP_SidebarVoiceChatStatus Enum=Get SidebarCurrentVoiceChatStatusEnum() -> (SidebarCurrentVoiceChatStatusEnum=NewEnumerator0)
     B:E_UI_BP_SidebarVoiceChatStatus Enum=NewEnumerator4
    )
   ) -> (ReturnValue:EVoiceChatPartyClientActiveState Enum)
   > SetText!(
    execute
    self:Rich Text Block Object Reference=Get SubState_RichMessage() -> (SubState_RichMessage:Common Rich Text Block Object Reference)
    InText:"Text (by ref)"=GetSubStateMessageText!(
     self
     ActiveClientName:Text="LOCTEXT(\"iPhone\")"
     ActiveClientPlatformTag:String=IOS
     Restrictions:ELocalVoiceChatRestriction Enum=@K2Node_Composite_1
     DeviceStatus:EVoiceChatPartyClientActiveState Enum=@n3
    ) -> (MessageText:Text)
   ) -> (then)
   > SetPlatformIcon!(execute, self, ActiveClientPlatformTag:String=IOS) -> (then)
   > SetSubStateMessageVisuals!(execute, self, VoiceChatStatus:E_UI_BP_SidebarVoiceChatStatus Enum=@n2) -> (then)
   > SetBannedStateVisibility!(
    execute
    self
    IsSocialBanned:bool="Equal (Enum)"(A=Get SidebarCurrentVoiceChatStatusEnum() -> (SidebarCurrentVoiceChatStatusEnum=NewEnumerator0), B=NewEnumerator6)
   ) -> (then)
   > Outputs()
   SetSubStateActionButtonVisibility!(self, Restriction:ELocalVoiceChatRestriction Enum=@K2Node_Composite_1, ActiveDeviceStatus:EVoiceChatPartyClientActiveState Enum=@n3) -> (ButtonVisibility:ESlateVisibility Enum=Visible)
  ]}
 DesignTimePreviewRestriction {
  path=DesignTimePreset.K2Node_Composite_1.DesignTimePreviewRestriction
  nodes [Outputs!(ReturnValue:ELocalVoiceChatRestriction Enum=Select!(
     NewEnumerator0:ELocalVoiceChatRestriction Enum=Unrestricted
     NewEnumerator4:ELocalVoiceChatRestriction Enum=Unrestricted
     NewEnumerator1:ELocalVoiceChatRestriction Enum=UserPermission
     NewEnumerator2:ELocalVoiceChatRestriction Enum=NotAcceptedModeration
     NewEnumerator3:ELocalVoiceChatRestriction Enum=ParentalPermission
     NewEnumerator6:ELocalVoiceChatRestriction Enum=SocialBan
     NewEnumerator5:ELocalVoiceChatRestriction Enum=Unrestricted
     NewEnumerator7:ELocalVoiceChatRestriction Enum=Unrestricted
     NewEnumerator8:ELocalVoiceChatRestriction Enum=Unrestricted
     Index:E_UI_BP_SidebarVoiceChatStatus Enum=Inputs!() -> (Index:E_UI_BP_SidebarVoiceChatStatus Enum=NewEnumerator0)
    ) -> (ReturnValue:ELocalVoiceChatRestriction Enum))]
 }
 EventGraph {nodes [
   "On Clicked (Channel_Party_Button)"!() -> (OutputDelegate, then, Button:Common Button Base Object Reference) {is=title K2Node_ComponentBoundEvent}
   n1 = Select!(Option 0:ESidebarVoiceChannelType Enum=Party, Option 1:ESidebarVoiceChannelType Enum=EpicParty, Index:bool=Get IsUsingEpicParties()) -> (ReturnValue:ESidebarVoiceChannelType Enum)
   n2 = Get ActiveChannel(self=Get VoiceChannelManagementVM()) -> (ActiveChannel=None)
   > Branch(Condition="Not Equal (Enum)"!(A:ESidebarVoiceChannelType Enum=@n2, B:ESidebarVoiceChannelType Enum=@n1) -> (ReturnValue:bool))
   > SetActiveChannel!(execute, self:Voice Channel Management VM Object Reference=Get VoiceChannelManagementVM(), Channel:ESidebarVoiceChannelType Enum=@n1) -> (then)
   > DeactivateWidget()
   n3 = Get Anim_OnlineStatusBackplate() -> (Anim_OnlineStatusBackplate:Widget Animation Object Reference)
   > PlayAnimationReverse(InAnimation=@n3)
   "On Clicked (Channel_Game_Button)"!() -> (OutputDelegate, then, Button:Common Button Base Object Reference) {is=title K2Node_ComponentBoundEvent}
   > Branch(Condition="Not Equal (Enum)"(A=@n2, B=Game))
   > SetActiveChannel(self=Get VoiceChannelManagementVM(), Channel=Game)
   > DeactivateWidget()
   > PlayAnimationForward(InAnimation=@n3)
   n4 = "On Button Option Clicked (Button_PartyJoinability_New)"!() -> (OutputDelegate, then, Button ID:Name=None) {is=title K2Node_ComponentBoundEvent}
   > SetIsJoinable!(
    execute
    self:Fort Social Epic Party Header Item VM Object Reference=Get FortSocialEpicPartyHeaderItemVM() -> (FortSocialEpicPartyHeaderItemVM:Fort Social Epic Party Header Item VM Object Reference)
    bInIsJoinable:bool="Equal (Name)"(A=@n4.Button ID, B=Joinable)
   ) -> (then)
   "On Clicked (Button_PartyJoinability)"!() -> (OutputDelegate, then, Button:Common Button Base Object Reference) {is=title K2Node_ComponentBoundEvent}
   n5 = Get FortSocialEpicPartyHeaderItemVM()
   > SetIsJoinable(self=@n5, bInIsJoinable=NOT Boolean(A=IsJoinable!(self:Fort Social Epic Party Header Item VM Object Reference=@n5) -> (ReturnValue:bool)))
   n6 = "On Clicked (Substate_Action_Button)"!() -> (OutputDelegate, then, Button:Common Button Base Object Reference) {is=title K2Node_ComponentBoundEvent}
   Handle Sub State Click!(execute:Exec=[@n6.then, @n15.then]) {is=title K2Node_MacroInstance}
   "On Clicked (Button_MoreOptions)"!() -> (OutputDelegate, then, Button:Common Button Base Object Reference) {is=title K2Node_ComponentBoundEvent}
   > Open!(
    execute
    self:Menu Anchor Object Reference=Get Anchor!(self:UIKit Context Menu Anchor Object Reference=Get Report_ContextMenu() -> (Report_ContextMenu:WBP UIKit Context Menu Anchor Button Object Reference)) -> (Anchor:Menu Anchor Object Reference)
    bFocusMenu:bool=true
   ) -> (then)
   n7 = "On Clicked (Button_Mic)"!() -> (OutputDelegate, then, Button:Common Button Base Object Reference) {is=title K2Node_ComponentBoundEvent}
   n8 = Get VoiceChannelManagementVM()
   Branch(
    execute=[@n7.then, @n17.then]
    Condition=GetAllowPTTOnClickOrTouch!(self:Voice Channel Management VM Object Reference=@n8) -> (ReturnValue:bool)
   )
   n9 = GetSelfIsMuted!(self:Voice Channel Management VM Object Reference=@n8) -> (ReturnValue:bool)
   n10 = Get IsVoiceChatMethodPTT() -> (IsVoiceChatMethodPTT:bool)
   > SetSelfIsMuted!(execute, self:Voice Channel Management VM Object Reference=@n8, bMuteSelf:bool=AND Boolean(A=NOT Boolean(A=@n10), B=NOT Boolean(A=@n9))) -> (then)
   > Show PTTContext Menu!(execute:Exec) -> (then:Exec) {is=title K2Node_MacroInstance}
   "On Clicked (Substate_MoreOptions_Button)"!() -> (OutputDelegate, then, Button:Common Button Base Object Reference) {is=title K2Node_ComponentBoundEvent}
   > Open(
    self=Get Anchor(self=Get Substate_MoreOptions_ContextMenu() -> (Substate_MoreOptions_ContextMenu:WBP UIKit Context Menu Anchor Button Object Reference))
    bFocusMenu=true
   )
   n11 = "On Clicked (Button_Headphones)"!() -> (OutputDelegate, then, Button:Common Button Base Object Reference) {is=title K2Node_ComponentBoundEvent}
   n12 = Get VoiceChannelManagementVM()
   n13 = GetSelfIsDeafened!(self:Voice Channel Management VM Object Reference=@n12) -> (ReturnValue:bool)
   SetSelfIsDeafened!(execute=[@n11.then, @n18.then], self:Voice Channel Management VM Object Reference=@n12, bDeafenSelf:bool=NOT Boolean(A=@n13)) -> (then)
   > SetupHeadphoneContextMenu!(execute, self, Deafen:bool=@n13) -> (then)
   "On Released (Button_Mic)"!() -> (OutputDelegate, then, Button:WBP UIKit Button Generic Base Object Reference) {is=title K2Node_ComponentBoundEvent}
   n14 = Get VoiceChannelManagementVM()
   > Branch(Condition=AND Boolean(A=GetAllowPTTOnClickOrTouch(self=@n14), B=Get IsVoiceChatMethodPTT()))
   > SetSelfIsMuted(self=@n14, bMuteSelf=true)
   n15 = "On Button Clicked (WBP_Social_Restriction_Generic_Internal)"() {is=title K2Node_ComponentBoundEvent}
   OnReportContextMenu!() -> (OutputDelegate, then, InteractionNameID:Name) {is=title K2Node_CustomEvent}
   > ReportCurrentChannel(self:Voice Channel Management VM Object Reference=Get VoiceChannelManagementVM())
   OnInputMethodChanged!() -> (OutputDelegate, then, bNewInputType:ECommonInputType Enum) {is=title K2Node_CustomEvent}
   > OnGetVoiceChatMethod!(
    execute
    self
    VoiceChatMethod:EFortVoiceChatMethod Enum=GetVoiceChatMethod!(self:Voice Channel Management VM Object Reference=Get VoiceChannelManagementVM()) -> (ReturnValue:EFortVoiceChatMethod Enum=OpenMic)
   ) -> (then)
   n16 = VoiceChatControlsCoachMarkClicked() {is=title K2Node_CustomEvent}
   > RequestTooltipUpdate(self:Fort Sidebar VM Object Reference=Get FortSidebarVM())
   SetVoiceChatMethodToOpenMic!() -> (OutputDelegate, then, InteractionNameID:Name) {is=title K2Node_CustomEvent}
   > SetVoiceChatMethod!(execute, self:Voice Channel Management VM Object Reference=Get VoiceChannelManagementVM(), ChatMethod:EFortVoiceChatMethod Enum=OpenMic) -> (then)
   OrientationChanged!() -> (OutputDelegate, then, inScreenOrientation:EScreenOrientation Enum=Unknown) {is=title K2Node_CustomEvent}
   > UpdateOpenTextChatContextOption()
   OnCommunicationSettingsContextMenu!() -> (OutputDelegate, then, InteractionNameID:Name) {is=title K2Node_CustomEvent}
   > Close Sidebar Open Settings!(execute:Exec) -> (then:Exec) {is=title K2Node_MacroInstance}
   n17 = OnToggleMuteContextMenu!() -> (OutputDelegate, then, InteractionNameID:Name) {is=title K2Node_CustomEvent}
   n18 = OnDeafenContextMenu!() -> (OutputDelegate, then, InteractionNameID:Name) {is=title K2Node_CustomEvent}
   Adjust Voice Method!() -> (OutputDelegate, then, InteractionNameID:Name) {is=title K2Node_CustomEvent}
   > Close Sidebar Open Voice Chat Settings!(execute:Exec) -> (then:Exec) {is=title K2Node_MacroInstance}
   SetVoiceChatMethodToPTT!() -> (OutputDelegate, then, InteractionNameID:Name) {is=title K2Node_CustomEvent}
   > SetVoiceChatMethod(self=Get VoiceChannelManagementVM(), ChatMethod=PTT)
   OnSetPTTKey!() -> (OutputDelegate, then, InteractionNameID:Name) {is=title K2Node_CustomEvent}
   > "Close Sidebar Open Mouse&& Keyboard Settings"!(execute:Exec) -> (then:Exec) {is=title K2Node_MacroInstance}
   OnOpenChatEvent!() -> (OutputDelegate, then, Interaction Name ID:Name) {is=title K2Node_CustomEvent}
   n19 = Get VoiceChannelManagementVM()
   > MoveTextChatToCompactUIAndCloseSidebar!(
    execute
    self:Fort Sidebar VM Object Reference=Get FortSidebarVM()
    bShowParty:bool=Is Channel Party Channel!(A:ESidebarVoiceChannelType Enum=Get ActiveChannel(self=@n19) -> (ActiveChannel=None)) -> (ReturnValue:bool=~) {is=title K2Node_MacroInstance}
   ) -> (then)
   Event Construct()
   > SetVisibility(
    self=Get Headphone_ContextMenu() -> (Headphone_ContextMenu:WBP UIKit Context Menu Anchor Button Object Reference)
    InVisibility=Select(Option 0:ESlateVisibility Enum=Collapsed, Option 1:ESlateVisibility Enum=SelfHitTestInvisible, Index=Get IsUsingEpicParties()) -> (ReturnValue:ESlateVisibility Enum)
   )
   > Initialize(self:Voice Channel Management VM Object Reference=Get VoiceChannelManagementVM())
   > Initialize!(
    execute
    self:Fort Social Header Item VM Object Reference=Get GameChatButtonInfoVM() -> (GameChatButtonInfoVM:Fort Social Header Item VM Object Reference)
    InHeaderItemType:ESocialHeaderItemType Enum=TeamMembers
   ) -> (then)
   n20 = Create Event()
   n21 = Get CommonInputSubsystem()
   > Bind Event to On Input Method Changed(self=@n21, Delegate=@n20)
   n22 = Create Event()
   n23 = Get Substate_MoreOptions_Button() -> (Substate_MoreOptions_Button:WBP Sidebar Button Generic Object Reference)
   > Assign On Visibility Changed!(execute:Exec, self:User Widget Object Reference=@n23, Delegate:"Delegate (by ref)"=@n22) -> (then:Exec) {is=title K2Node_AssignDelegate}
   > OnSubstateMoreOptionsVisibilityChanged!(
    execute
    self
    InVisibility:ESlateVisibility Enum=Get Visibility!(self:Widget Object Reference=@n23) -> (Visibility:ESlateVisibility Enum=Visible)
   ) -> (then)
   > UpdatePTTKey()
   n24 = Get Button_PartyJoinability_New() -> (Button_PartyJoinability_New:WBP Sidebar Button With Context Menu Object Reference)
   > SetGridSlotPadding!(
    execute
    Widget:Widget Object Reference=Get ArrowIcon!(self:WBP Sidebar Button With Context Menu Object Reference=@n24) -> (ArrowIcon:Image Object Reference)
    Padding_Left:float=0
    Padding_Top:float=Get ArrowIcon_TopPadding() -> (ArrowIcon_TopPadding:double)
    Padding_Right:float=12
    Padding_Bottom:float=0
   ) -> (then)
   > SetPlacement!(
    execute
    self:Menu Anchor Object Reference=Get Anchor(self=Get Sidebar_Anchor!(self:WBP Sidebar Button With Context Menu Object Reference=@n24) -> (Sidebar_Anchor:WBP Sidebar Anchor Object Reference))
    InPlacement:EMenuPlacement Enum=MenuPlacement_CenteredBelowAnchor
   ) -> (then)
   n25 = GetSelfIsDeafened(self=Get VoiceChannelManagementVM())
   n26 = Get IsVoiceChatMethodPTT()
   > SetIsEnabled(self=Get Button_Mic() -> (Button_Mic:WBP Sidebar Button Generic Object Reference), bInIsEnabled=NOT Boolean(A=AND Boolean(A=@n25, B=@n26)))
   > SetupHeadphoneContextMenu(Deafen=@n25)
   > n27 = Cast To PlatformGameInstance!(execute, Object=GetGameInstance!() -> (ReturnValue:Game Instance Object Reference)) -> (then, CastFailed, AsPlatform Game Instance:Platform Game Instance Object Reference) {is=title K2Node_DynamicCast}
   n28 = Create Event()
   > Bind Event to Application Received Screen Orientation Changed Notification Delegate(self:Platform Game Instance Object Reference=@n27, Delegate=@n28) {is=title K2Node_AddDelegate}
   > UpdateOpenTextChatContextOption()
   n29 = Event PreConstruct()
   > SetMobileStyle() {is=title K2Node_CallFunction}
   > Branch(Condition=@n29.IsDesignTime)
   > Design Time Preset!(execute:Exec) -> (then:Exec) {is=title K2Node_MacroInstance}
   > SetCoachMarkPlacement!(
    execute
    self
    CheckMarkPlacement:EFortSidebarCoachMarkPlacement Enum=Get PreviewCoachMarkPlacement() -> (PreviewCoachMarkPlacement:EFortSidebarCoachMarkPlacement Enum=Outside)
   ) -> (then)
   Event OnInitialized()
   > n30 = Construct Object from Class(Class=/Script/FortniteUI.FortSocialUserNameViewModel) -> (ReturnValue:Fort Social User Name Viewmodel Object Reference)
   > Set SocialUserVM_PartyPlayerName(SocialUserVM_PartyPlayerName:Fort Social User Name Viewmodel Object Reference=@n30) -> (Output_Get:Fort Social User Name Viewmodel Object Reference)
   > n31 = GetConsoleVariableBoolValue!(execute, VariableName:String=fort.UseEpicParties) -> (then, ReturnValue:bool)
   > Set with Broadcast IsUsingEpicParties!(execute, IsUsingEpicParties:bool=@n31) -> (then, Output_Get:bool)
   > SetupReportContextMenu()
   > SetupMicrophoneContextMenu()
   > n32 = GetConsoleVariableBoolValue(VariableName=Sidebar.EnableUpdatedPartyControls)
   > n33 = Set isUpdatedPartyControlsEnabled(isUpdatedPartyControlsEnabled:bool=@n32) -> (Output_Get:bool)
   > SetVisibility(
    self=Get HB_Controls_New() -> (HB_Controls_New:Horizontal Box Object Reference)
    InVisibility=Select(Option 0=Collapsed, Option 1=SelfHitTestInvisible, Index=@n33)
   )
   n34 = Event OnListItemObjectSet()
   > Is Valid(InputObject=@n34.ListItemObject)
   > n35 = Cast To FortSocialEpicPartyHeaderItemVM!(execute, Object=@n34.ListItemObject) -> (then, CastFailed, AsFort Social Epic Party Header Item VM:Fort Social Epic Party Header Item VM Object Reference) {is=title K2Node_DynamicCast}
   > n36 = Set FortSocialEpicPartyHeaderItemVM(FortSocialEpicPartyHeaderItemVM:Fort Social Epic Party Header Item VM Object Reference=@n35) -> (Output_Get:Fort Social Epic Party Header Item VM Object Reference)
   Set Initial Channel Selector State!(execute:Exec=[@n35.CastFailed, @n36.then]) -> (then:Exec) {is=title K2Node_MacroInstance}
   Event Destruct()
   > Unbind Event from On Input Method Changed(self=@n21, Delegate=@n20)
   > Unbind Event from On Visibility Changed(self:User Widget Object Reference=@n23, Delegate=@n22) {is=title K2Node_RemoveDelegate}
   > n37 = Cast To PlatformGameInstance(Object=GetGameInstance())
   > Unbind Event from Application Received Screen Orientation Changed Notification Delegate(self:Platform Game Instance Object Reference=@n37, Delegate=@n28) {is=title K2Node_RemoveDelegate}
   Event On Activated()
   > Branch(Condition=ShouldShowOnboarding!(self:Fort Sidebar Onboard Tooltip Widget Object Reference=Get WBP_VoiceChat_ControlsMoved_CoachMark() -> (WBP_VoiceChat_ControlsMoved_CoachMark:WBP Voice Chat Controls Moved Coach Mark Object Reference)) -> (ReturnValue:bool))
   > n38 = FindParentWidgetOfType!(execute, StartingWidget:Widget Object Reference=Self-Reference(), Type:Widget Class Reference=/Sidebar/WBP_Sidebar.WBP_Sidebar_C) -> (then, ReturnValue:WBP Sidebar Object Reference)
   > SetVisibility(
    self=Get SpacerForCoachmarks!(self:WBP Sidebar Object Reference=@n38) -> (SpacerForCoachmarks:Spacer Object Reference)
    InVisibility=SelfHitTestInvisible
   )
   n39 = Get WBP_VoiceChat_ControlsMoved_CoachMark()
   > ActivateWidget(self=@n39)
   > Bind Event to Button Other Action Clicked(
    self:WBP UIKit Coach Mark Base Object Reference=Get WBP_UIKit_CoachMark!(self:WBP Voice Chat Controls Moved Coach Mark Object Reference=@n39) -> (WBP_UIKit_CoachMark:WBP UIKit Coach Mark Object Reference)
    Delegate=@n16
   ) {is=title K2Node_AddDelegate}
   Event OnFocusLost!() -> (OutputDelegate, then, InFocusEvent:Focus Event Structure) {is=title K2Node_Event}
   > n40 = Branch(Condition=IsVisible(self=Get WBP_VoiceChat_ControlsMoved_CoachMark()))
   DeactivateWidget(execute=@n40.else)
   EmptyEvent!() -> (OutputDelegate, then, InteractionNameID:Name) {is=title K2Node_CustomEvent}
   Reroute Node(InputPin:Wildcard) -> (OutputPin:Wildcard)
  ]}
 GetFirstFocusableButtonOnBottomRow {nodes [
   GetFirstFocusableButtonOnBottomRow() {is=title K2Node_FunctionEntry}
   > n1 = Get()
   > n2 = Branch(Condition=IsVisible(self=Get Button_PartyJoinability() -> (Button_PartyJoinability:WBP Sidebar Button Generic Object Reference)))
   > Return Node!(execute, FirstFocusableButtonOnBottomRow:Widget Object Reference=Get Button_PartyJoinability())
   n3 = Branch(execute=@n2.else, Condition=GetIsEnabled!(self:Widget Object Reference=Get Button_Mic()) -> (ReturnValue:bool))
   > Return Node(FirstFocusableButtonOnBottomRow=Get Button_Mic())
   Branch(execute=@n3.else, Condition=GetIsEnabled(self=Get Button_Headphones() -> (Button_Headphones:WBP Sidebar Button Generic Object Reference)))
   > Return Node(FirstFocusableButtonOnBottomRow=Get Button_Headphones())
   n4 = Get VoiceChannelManagementVM()
   n5 = Get Channel_Party_Button()
   Return Node(
    execute=@n1.else
    FirstFocusableButtonOnBottomRow=Select!(
     None:Widget Object Reference=@n5
     Party:Widget Object Reference=@n5
     Game:Widget Object Reference=Get Channel_Game_Button()
     EpicParty:Widget Object Reference=@n5
     Index:ESidebarVoiceChannelType Enum=Get ActiveChannel(self=@n4) -> (ActiveChannel=None)
    ) -> (ReturnValue:Widget Object Reference)
   )
  ]}
 GetSidebarControlsVisibility {nodes [
   n1 = GetSidebarControlsVisibility!() -> (then, VoiceChatStatus:E_UI_BP_SidebarVoiceChatStatus Enum=NewEnumerator0, ActiveChannel:ESidebarVoiceChannelType Enum=None) {is=title K2Node_FunctionEntry}
   n2 = "Equal (Enum)"(A=@n1, B=NewEnumerator4)
   n3 = Get isUpdatedPartyControlsEnabled() -> (isUpdatedPartyControlsEnabled:bool)
   > Return Node!(
    execute
    Visibility:ESlateVisibility Enum=Select!(Option 0:ESlateVisibility Enum=SelfHitTestInvisible, Option 1:ESlateVisibility Enum=Collapsed, Index:bool=OR Boolean(A=@n2, B=@n3)) -> (ReturnValue:ESlateVisibility Enum)
   )
  ]}
 GetSidebarSubStateVisibility {nodes [
   n1 = GetSidebarSubStateVisibility!() -> (
    then
    VoiceChatStatus:E_UI_BP_SidebarVoiceChatStatus Enum=NewEnumerator0
    ActiveChannel:ESidebarVoiceChannelType Enum=None
    IsEpicParty:bool=~
   ) {is=title K2Node_FunctionEntry}
   n2 = AND Boolean(
    A="Equal (Enum)"(A=@n1, B=NewEnumerator4)
    B="Equal (Enum)"(A:ESidebarVoiceChannelType Enum=@n1.ActiveChannel, B:ESidebarVoiceChannelType Enum=Game)
   )
   n3 = "Equal (Enum)"(A:E_UI_BP_SidebarVoiceChatStatus Enum=@n1, B:E_UI_BP_SidebarVoiceChatStatus Enum=NewEnumerator0)
   > Return Node!(execute, NewParam:ESlateVisibility Enum=Select(Option 0=SelfHitTestInvisible, Option 1=Collapsed, Index=OR Boolean(A=@n3, B=@n2)))
  ]}
 GetSidebarVoiceChatStatus {nodes [
   n1 = GetSidebarVoiceChatStatus!() -> (then, Restrictions:ELocalVoiceChatRestriction Enum=Unrestricted, DeviceStatus:EVoiceChatPartyClientActiveState Enum=NoClientsActiveInParty) {is=title K2Node_FunctionEntry}
   > n2 = Switch on ELocalVoiceChatRestriction!(execute, Selection:ELocalVoiceChatRestriction Enum=@n1) -> (
    Unrestricted:Exec
    UserPermission:Exec
    ParentalPermission:Exec
    SplitscreenParentalPermission:Exec
    NotAcceptedModeration:Exec
    SplitscreenNotAcceptedModeration:Exec
    PlatformPermission:Exec
    Cabined:Exec
    SocialBan:Exec
    NonLocalReason:Exec
    MicrophonePermissionDenied:Exec
    NotAcceptedBrigadeiro:Exec
    SplitscreenNotAcceptedBrigadeiro:Exec
   ) {is=title K2Node_SwitchEnum}
   Return Node!(
    execute=[@n2.Cabined, @n2.ParentalPermission, @n2.PlatformPermission, @n2.SplitscreenParentalPermission]
    CurrentVoiceChatStatus:E_UI_BP_SidebarVoiceChatStatus Enum=NewEnumerator3
   )
   Return Node(execute=@n2.MicrophonePermissionDenied, CurrentVoiceChatStatus=NewEnumerator7)
   Return Node(execute=@n2.NonLocalReason, CurrentVoiceChatStatus=NewEnumerator5)
   Return Node(execute=[@n2.NotAcceptedBrigadeiro, @n2.SplitscreenNotAcceptedBrigadeiro], CurrentVoiceChatStatus=NewEnumerator8)
   Return Node(execute=[@n2.NotAcceptedModeration, @n2.SplitscreenNotAcceptedModeration], CurrentVoiceChatStatus=NewEnumerator2)
   Return Node(execute=@n2.SocialBan, CurrentVoiceChatStatus=NewEnumerator6)
   > n3 = Branch(Condition="Equal (Enum)"(
     A:EVoiceChatPartyClientActiveState Enum=Get DeviceStatus() -> (DeviceStatus:EVoiceChatPartyClientActiveState Enum=NoClientsActiveInParty)
     B:EVoiceChatPartyClientActiveState Enum=ThisClientActiveInParty
    ))
   > Return Node(CurrentVoiceChatStatus=NewEnumerator0)
   Return Node(execute=@n3.else, CurrentVoiceChatStatus=NewEnumerator4)
   Return Node(execute=@n2.UserPermission, CurrentVoiceChatStatus=NewEnumerator1)
  ]}
 GetSubStateMessageText {nodes [
   n1 = GetSubStateMessageText!() -> (
    then
    ActiveClientName:Text
    ActiveClientPlatformTag:String
    Restrictions:ELocalVoiceChatRestriction Enum=Unrestricted
    DeviceStatus:EVoiceChatPartyClientActiveState Enum=NoClientsActiveInParty
   ) {is=title K2Node_FunctionEntry}
   > n2 = Branch(Condition="Equal (Enum)"(A:ELocalVoiceChatRestriction Enum=@n1.Restrictions, B:ELocalVoiceChatRestriction Enum=Unrestricted))
   > n3 = Branch(Condition="Not Equal (Enum)"(
     A:EVoiceChatPartyClientActiveState Enum=Get DeviceStatus() -> (DeviceStatus=NoClientsActiveInParty)
     B:EVoiceChatPartyClientActiveState Enum=ThisClientActiveInParty
    ))
   > n4 = Branch(Condition="Equal (Enum)"(
     A:EVoiceChatPartyClientActiveState Enum=Get DeviceStatus() -> (DeviceStatus=NoClientsActiveInParty)
     B:EVoiceChatPartyClientActiveState Enum=NoClientsActiveInParty
    ))
   > Return Node!(execute, MessageText:Text="LOCTEXT(\"Voice chat not active on this device\")")
   n5 = Get ActiveClientName() -> (ActiveClientName:Text)
   n6 = Get ActiveClientName()
   n7 = Get ActiveClientPlatformTag() -> (ActiveClientPlatformTag:String)
   Return Node(
    execute=@n4.else
    MessageText=Select(
     Option 0:Text=Format Text!(Format="LOCTEXT(\"In voice chat on {clientName}\")", clientName:Text=@n6) -> (Result)
     Option 1:Text=Format Text!(Format="LOCTEXT(\"In voice chat on <{platformTag}/>\")", platformTag:String=@n7) -> (Result)
     Index=TextIsEmpty(InText=@n5)
    ) -> (ReturnValue:Text)
   )
   n8 = MakeLiteralText(Value="LOCTEXT(\"Voice chat is turned off due to the other player\\'s settings\")")
   Return Node(
    execute=[@n3.else, @n2.else]
    MessageText=Select!(
     Unrestricted:Text="LOCTEXT(\"Voice chat not active on this device\")"
     UserPermission:Text="LOCTEXT(\"Voice chat is turned off in your settings\")"
     ParentalPermission:Text=MakeLiteralText(Value="LOCTEXT(\"Your Parent or Guardian has turned off voice chat\")")
     SplitscreenParentalPermission:Text=@n8
     NotAcceptedModeration:Text=MakeLiteralText(Value="LOCTEXT(\"View the voice reporting update to use voice chat\")")
     SplitscreenNotAcceptedModeration:Text=@n8
     PlatformPermission:Text=MakeLiteralText(Value="LOCTEXT(\"You can\\'t use voice chat due to your console\\'s privacy settings\")")
     Cabined:Text=MakeLiteralText(Value="LOCTEXT(\"You need your parent\\'s permission to use voice chat\")")
     SocialBan:Text
     NonLocalReason:Text="LOCTEXT(\"Something went wrong\")"
     MicrophonePermissionDenied:Text=MakeLiteralText(Value="LOCTEXT(\"Allow microphone access to use voice chat\")")
     NotAcceptedBrigadeiro:Text="INVTEXT(\"Lorem ipsum dolor sit amet, consectetur adipiscing elit. Curabitur quis nunc sollicitudin, vestibulum mauris at, pulvinar ante. \")"
     SplitscreenNotAcceptedBrigadeiro:Text="INVTEXT(\"Lorem ipsum dolor sit amet, consectetur adipiscing elit. Curabitur quis nunc sollicitudin, vestibulum mauris at, pulvinar ante. \")"
     Index:ELocalVoiceChatRestriction Enum=Get Restrictions() -> (Restrictions:ELocalVoiceChatRestriction Enum=Unrestricted)
    ) -> (ReturnValue)
   )
  ]}
 GetSubstateMoreOptionsVisibility {nodes [
   n1 = GetSubstateMoreOptionsVisibility!() -> (then, VoiceChatStatus:E_UI_BP_SidebarVoiceChatStatus Enum=NewEnumerator0, ActiveChannel:ESidebarVoiceChannelType Enum=None) {is=title K2Node_FunctionEntry}
   n2 = "Equal (Enum)"(A:E_UI_BP_SidebarVoiceChatStatus Enum=@n1, B:E_UI_BP_SidebarVoiceChatStatus Enum=NewEnumerator4)
   n3 = "Not Equal (Enum)"(
    A:ESlateVisibility Enum=GetSidebarSubStateVisibility(VoiceChatStatus=@n1, ActiveChannel=@n1.ActiveChannel, IsEpicParty=false) -> (NewParam=Visible) {is=title K2Node_CallFunction}
    B:ESlateVisibility Enum=Collapsed
   )
   > Return Node!(
    execute
    Visibility:ESlateVisibility Enum=Select!(Option 0:ESlateVisibility Enum=Collapsed, Option 1:ESlateVisibility Enum=SelfHitTestInvisible, Index:bool=AND Boolean(A=@n2, B=@n3)) -> (ReturnValue:ESlateVisibility Enum)
   )
  ]}
 GetVoiceChatOffIconVisibility {nodes [
   n1 = GetVoiceChatOffIconVisibility!() -> (
    then
    Restrictions:ELocalVoiceChatRestriction Enum=Unrestricted
    CurrentActiveChannel:ESidebarVoiceChannelType Enum=None
    AvailableChannels:Array of ESidebarVoiceChannelType Enums=None
   ) {is=title K2Node_FunctionEntry}
   > n2 = Branch(Condition="Equal (Enum)"(A:ESidebarVoiceChannelType Enum=@n1.CurrentActiveChannel, B:ESidebarVoiceChannelType Enum=Game))
   n3 = Get FortGlobalUIContext!() -> (ReturnValue:Fort Global UIContext Object Reference) {is=title K2Node_GetBlueprintContextEx}
   > n4 = Branch(Condition=NOT Boolean(A=IsInZone(self=@n3)))
   > Return Node!(execute, OffIconVisibility:ESlateVisibility Enum=Collapsed)
   n5 = Get AvailableChannels() -> (AvailableChannels=None)
   Return Node(
    execute=@n4.else
    OffIconVisibility=Select(Option 0=SelfHitTestInvisible, Option 1=Collapsed, Index=Contains Item(TargetArray=@n5, ItemToFind=Game))
   )
   n6 = Get Restrictions() -> (Restrictions=Unrestricted)
   Return Node(
    execute=@n2.else
    OffIconVisibility=Select(
     Option 0=SelfHitTestInvisible
     Option 1=Collapsed
     Index="Equal (Enum)"(A:ELocalVoiceChatRestriction Enum=@n6, B:ELocalVoiceChatRestriction Enum=Unrestricted)
    )
   )
  ]}
 GetVoiceReportingStatusText {nodes [
   n1 = GetVoiceReportingStatusText!() -> (
    then
    Restrictions:ELocalVoiceChatRestriction Enum=Unrestricted
    IsVoiceReportingOn:bool=~
    CurrentActiveChannel:ESidebarVoiceChannelType Enum=None
    AvailableChannels:Array of ESidebarVoiceChannelType Enums=None
   ) {is=title K2Node_FunctionEntry}
   > n2 = Branch(Condition="Equal (Enum)"(A:ESidebarVoiceChannelType Enum=@n1.CurrentActiveChannel, B:ESidebarVoiceChannelType Enum=Game))
   > Set Justification!(
    execute
    Justification:ETextJustify Enum=Center
    self:Text Layout Widget Object Reference=Get VoiceReportStatusNew_Text() -> (VoiceReportStatusNew_Text:Common Text Object Reference)
   ) -> (then, Output_Get:ETextJustify Enum=Left)
   n3 = Get FortGlobalUIContext()
   > n4 = Branch(Condition=NOT Boolean(A=IsInZone(self=@n3)))
   > Return Node!(execute, VoiceReportingStatusText:Text="LOCTEXT(\"Game members will show here once you\\'re in a game\")")
   n5 = Get AvailableChannels() -> (AvailableChannels=None)
   n6 = Branch(execute=@n4.else, Condition=NOT Boolean(A=Contains Item(TargetArray=@n5, ItemToFind=Game)))
   > Return Node(VoiceReportingStatusText="LOCTEXT(\"Game voice chat not available\")")
   n7 = Branch(
    execute=[@n6.else, @n2.else]
    Condition="Equal (Enum)"(A:ELocalVoiceChatRestriction Enum=Get Restrictions() -> (Restrictions=Unrestricted), B:ELocalVoiceChatRestriction Enum=Unrestricted)
   )
   > Set Justification(Justification=Right, self=Get VoiceReportStatusNew_Text()) -> (Output_Get=Left)
   > Return Node(VoiceReportingStatusText=Select(
     Option 0:Text="LOCTEXT(\"Voice reporting off\")"
     Option 1:Text="LOCTEXT(\"Voice reporting on\")"
     Index=Get IsVoiceReportingOn() -> (IsVoiceReportingOn:bool)
    ) -> (ReturnValue:Text))
   Set Justification(execute=@n7.else, Justification=Right, self=Get VoiceReportStatusNew_Text()) -> (Output_Get=Left)
   > Return Node(VoiceReportingStatusText="LOCTEXT(\"Your voice chat is off\")")
  ]}
 HandleSubStateClick {nodes [
   Inputs!() -> (execute:Exec)
   > n1 = Switch on E_UI_BP_SidebarVoiceChatStatus!(
    execute
    Selection:E_UI_BP_SidebarVoiceChatStatus Enum=Get SidebarCurrentVoiceChatStatusEnum() -> (SidebarCurrentVoiceChatStatusEnum=NewEnumerator0)
   ) -> (
    NewEnumerator0:Exec
    NewEnumerator4:Exec
    NewEnumerator1:Exec
    NewEnumerator2:Exec
    NewEnumerator3:Exec
    NewEnumerator6:Exec
    NewEnumerator5:Exec
    NewEnumerator7:Exec
    NewEnumerator8:Exec
   ) {is=title K2Node_SwitchEnum}
   Close Sidebar Open Settings(execute=@n1.NewEnumerator1)
   RequestCloseSidebar(execute=@n1.NewEnumerator3, self=Get FortSidebarVM())
   > RequestToOpenParentalControls!(execute) -> (then)
   SetThisClientAsActive(execute=@n1.NewEnumerator4, self:Voice Channel Management VM Object Reference=Get VoiceChannelManagementVM())
   OpenMicrophonePermissionSettings(execute=@n1.NewEnumerator7, self:Voice Channel Management VM Object Reference=Get VoiceChannelManagementVM())
   Push Plaster Beach Modal!(execute:Exec=@n1.NewEnumerator8) -> (then:Exec) {is=title K2Node_MacroInstance}
   n2 = Construct Object from Class!(
    execute=@n1.NewEnumerator2
    Class=/Sidebar/Shared/Navigation/BP_NavigationPoint.BP_NavigationPoint_C
    self
    SocialUser:Fort Social User Object Reference
    ConversationId:Name=None
    ConfirmationVM:Sidebar Confirmation Panel VM Object Reference
    ReservedSpots:int=0
    SocialItemVM:Fort Social Item VM Object Reference
    NavigationTag:Gameplay Tag Structure=(TagName="UI.StateTree.TransitionTag.Sidebar.Social.VoiceReportDisclosure")
   ) -> (then, ReturnValue:BP Navigation Point Object Reference)
   > PushNavigation!(
    execute
    self:Fort Sidebar Navigation VM Object Reference=Get FortSidebarNavigationVM() -> (FortSidebarNavigationVM:Fort Sidebar Navigation VM Object Reference)
    NavigationPoint:Fort Sidebar Navigation Point Object Reference=@n2
   ) -> (then)
   Outputs!()
  ]}
 InitializeCurrentIslandSocialUser {nodes [
   n1 = InitializeCurrentIslandSocialUser!() -> (then, PartyMembers:Array of Fort Social Item VM Object References) {is=title K2Node_FunctionEntry}
   > n2 = For Each Loop with Break!(Exec:Exec, Array:Array of Fort Social Item VM Object References=@n1, Break:Exec=@n4.then) -> (LoopBody:Exec, Array Element:Fort Social Item VM Object Reference, Array Index:int, Completed:Exec)
   n3 = Get TargetUser!(self:Fort Social Item VM Object Reference=@n2) -> (TargetUser:Fort Social User Weak object)
   > Branch(Condition=IsLocalUserForBlueprint!(self:Fort Social User Object Reference=@n3) -> (ReturnValue:bool))
   > n4 = SetSocialUser!(
    execute
    self:Fort Social User Viewmodel Object Reference=Get PartyMemberCurrentIslandVM() -> (PartyMemberCurrentIslandVM:Party Member Current Island View Model Object Reference)
    InSocialUser:Fort Social User Object Reference=@n3
   ) -> (then)
  ]}
 IsChannelPartyChannel {nodes [
   n1 = Get IsUsingEpicParties()
   Outputs!(ReturnValue:bool="Equal (Enum)"(
     A:ESidebarVoiceChannelType Enum=Inputs!() -> (A:ESidebarVoiceChannelType Enum=None)
     B:ESidebarVoiceChannelType Enum=Select(Option 0:ESidebarVoiceChannelType Enum=Party, Option 1:ESidebarVoiceChannelType Enum=EpicParty, Index=@n1) -> (ReturnValue:ESidebarVoiceChannelType Enum)
    ))
  ]}
 NavigateUotoTab {nodes [
   NavigateUotoTab!() -> (then, Navigation:EUINavigation Enum) {is=title K2Node_FunctionEntry}
   > n1 = FindParentWidgetOfType(StartingWidget=Self-Reference(), Type=&WBP_Sidebar_C)
   > n2 = NavigationToTabGroup!(execute, self:WBP Sidebar Object Reference=@n1, Navigation:EUINavigation Enum=Up) -> (then, ReturnValue:Widget Object Reference)
   > Return Node!(execute, ReturnValue:Widget Object Reference=@n2)
  ]}
 OnActiveChannelChanged {nodes [
   n1 = OnActiveChannelChanged!() -> (then, ActiveChannel:ESidebarVoiceChannelType Enum=None) {is=title K2Node_FunctionEntry}
   > Branch(Condition="Not Equal (Enum)"(A:ESidebarVoiceChannelType Enum=@n1, B:ESidebarVoiceChannelType Enum=None))
   > n2 = Set IsPartyChannel(IsPartyChannel:bool=Is Channel Party Channel(A=@n1) -> (ReturnValue=~)) -> (Output_Get:bool)
   > SetChannelButtonActive!(execute, self:WBP Sidebar Voice Channel Button Object Reference=Get Channel_Party_Button(), IsCurrentlyTheActiveChannel:bool=@n2) -> (then)
   > SetChannelButtonActive(self=Get Channel_Game_Button(), IsCurrentlyTheActiveChannel=NOT Boolean(A=Get IsPartyChannel() -> (IsPartyChannel:bool)))
   n3 = Get IsUsingEpicParties()
   n4 = Get IsPartyChannel()
   > SetVisibility(
    self=[
     Get PartyJoinable_Text_Status() -> (PartyJoinable_Text_Status:Common Text Object Reference)
     Get Button_PartyJoinability_New()
     Get Button_PartyJoinability()
    ]
    InVisibility=Select(Option 0:ESlateVisibility Enum=Collapsed, Option 1:ESlateVisibility Enum=SelfHitTestInvisible, Index=AND Boolean(A=@n3, B=@n4)) -> (ReturnValue:ESlateVisibility Enum)
   )
   > AddReportConversationContextMenuItem!(execute, self, IsPartyChannel:bool=Get IsPartyChannel()) -> (then) {is=title K2Node_CallFunction}
   > Branch(Condition=Get isUpdatedPartyControlsEnabled())
   > n5 = Branch(Condition=AND Boolean(A=Get IsUsingEpicParties(), B=Get IsPartyChannel()))
   > SetNavigationRuleCustomBoundary!(
    execute
    self:Widget Object Reference=Get Grid_ChannelSelect_CustomNavBoundary() -> (Grid_ChannelSelect_CustomNavBoundary:Grid Panel Object Reference)
    Direction:EUINavigation Enum=Down
    InCustomDelegate:Delegate=Create Event()
   ) -> (then)
   > SetNavigationRuleCustomBoundary(self=[Get Substate_Action_Button(), Get Substate_MoreOptions_Button()], Direction=Down, InCustomDelegate=Create Event())
   n6 = Get Substate_Action_Button()
   SetNavigationRuleBase!(
    execute=@n5.else
    self:Widget Object Reference=[@n6, Get Substate_MoreOptions_Button()]
    Direction:EUINavigation Enum=Down
    Rule:EUINavigationRule Enum=Escape
   ) -> (then)
   > n7 = Branch(Condition=IsVisible(self=@n6))
   SetNavigationRuleBase(execute=@n7.else, self=Get Grid_ChannelSelect_CustomNavBoundary(), Direction=Down, Rule=Escape)
  ]}
 OnCurrentIslandVMSet {nodes [
   n1 = OnCurrentIslandVMSet!() -> (then, Island VM:Discover Activity View Model Object Reference) {is=title K2Node_FunctionEntry}
   > Is Valid(InputObject=@n1)
   > FetchImages(self:Fort Discover Link Entry Tile Item VM Object Reference=Get CurrentIslandVM!(self:Party Member Current Island View Model Object Reference=Get PartyMemberCurrentIslandVM()) -> (CurrentIslandVM:Discover Activity View Model Object Reference))
  ]}
 OnDeafenChanged {nodes [
   n1 = OnDeafenChanged!() -> (then, IsDeafen:bool=~) {is=title K2Node_FunctionEntry}
   n2 = Get Button_Headphones()
   > Set ButtonType!(
    execute
    ButtonType:E_UI_Sidebar_Button_Type Enum=Select(Option 0:E_UI_Sidebar_Button_Type Enum=NewEnumerator0, Option 1:E_UI_Sidebar_Button_Type Enum=NewEnumerator6, Index=@n1) -> (ReturnValue:E_UI_Sidebar_Button_Type Enum)
    self:WBP Sidebar Button Generic Object Reference=@n2
   ) -> (then, Output_Get:E_UI_Sidebar_Button_Type Enum=NewEnumerator0)
   > InitializeButtonVisuals(self:WBP Sidebar Button Generic Object Reference=@n2)
   > SetVisibility(
    self=Get Deafen_Message() -> (Deafen_Message:Common Text Object Reference)
    InVisibility=Select(Option 0:ESlateVisibility Enum=Collapsed, Option 1:ESlateVisibility Enum=SelfHitTestInvisible, Index=Get IsDeafen() -> (IsDeafen:bool)) -> (ReturnValue:ESlateVisibility Enum)
   )
   > SetIconMaterial!(
    execute
    self:WBP UIKit Button Generic Base Object Reference=Get Button_Headphones()
    Material Instance:Material Instance Object Reference=Select(
     Option 0:Material Instance Object Reference=/Sidebar/Materials/Icons/MI_Sidebar_Icon_Base
     Option 1:Material Instance Object Reference=/Sidebar/Materials/Icons/MI_Sidebar_Icon_Alert
     Index=Get IsDeafen()
    ) -> (ReturnValue:Material Instance Object Reference)
   ) -> (then)
   n3 = Get Button_Headphones()
   > SetIcon!(
    execute
    self:WBP UIKit Button Generic Base Object Reference=@n3
    Icon:Object Reference=Select(
     Option 0:Object Reference=/Sidebar/Materials/Icons/Textures/T_UI_Icon_Sidebar_Headphones_MSDF
     Option 1:Object Reference=/Sidebar/Materials/Icons/Textures/T_UI_Icon_Sidebar_HeadphonesSlash_MSDF
     Index=Get IsDeafen()
    ) -> (ReturnValue:Object Reference)
   ) -> (then)
   > SetIconSize!(
    execute
    self:WBP UIKit Button Generic Base Object Reference=@n3
    Icon:Vector 2D Structure=Get IconSizeWithoutArrow() -> (IconSizeWithoutArrow:Vector 2D Structure)
   ) -> (then)
   n4 = Get IsDeafen()
   n5 = Get IsVoiceChatMethodPTT()
   > SetIsEnabled(self=Get Button_Mic(), bInIsEnabled=NOT Boolean(A=AND Boolean(A=@n4, B=@n5)))
  ]}
 OnFocusReceived {nodes [
   OnFocusReceived!() -> (then, MyGeometry:Geometry Structure, InFocusEvent:Focus Event Structure) {is=title K2Node_FunctionEntry}
   > ActivateWidget()
   > RequestRefreshFocus(self:Common Activatable Widget Object Reference)
   > Return Node(ReturnValue:Event Reply Structure=Handled())
  ]}
 OnGetVoiceChatMethod {nodes [
   n1 = OnGetVoiceChatMethod!() -> (then, VoiceChatMethod:EFortVoiceChatMethod Enum=OpenMic) {is=title K2Node_FunctionEntry}
   > Set ForceOpenMic(ForceOpenMic:bool=ShouldAlwaysActAsOpenMic()) -> (Output_Get:bool)
   n2 = Get ForceOpenMic() -> (ForceOpenMic:bool)
   > Set IsVoiceChatMethodPTT(IsVoiceChatMethodPTT:bool=AND Boolean(A="Equal (Enum)"(A:EFortVoiceChatMethod Enum=@n1, B:EFortVoiceChatMethod Enum=PTT), B=NOT Boolean(A=@n2))) -> (Output_Get:bool)
   > Set Text(
    Text=SelectText(A="LOCTEXT(\"Push to talk\")", B="LOCTEXT(\"Open mic\")", bPickA=Get IsVoiceChatMethodPTT())
    self=Get VoiceChannelType_Text() -> (VoiceChannelType_Text:Common Text Object Reference)
   )
   n3 = Get IsVoiceChatMethodPTT()
   n4 = Get Button_Mic()
   > SetPadding!(
    execute
    self:User Widget Object Reference=Get Block_Icon!(self:WBP UIKit Button Generic Base Object Reference=@n4) -> (Block_Icon:WBP UIKit Block Image Object Reference)
    InPadding:Margin Structure=Make Margin(Left=0, Top=0, Right=SelectFloat(A=24, B=0, bPickA=@n3), Bottom=0)
   ) -> (then)
   > SetClickMethod!(
    execute
    self:Common Button Base Object Reference=@n4
    InClickMethod:EButtonClickMethod Enum=Select(Option 0:EButtonClickMethod Enum=DownAndUp, Option 1:EButtonClickMethod Enum=MouseDown, Index=Get IsVoiceChatMethodPTT()) -> (ReturnValue:EButtonClickMethod Enum)
   ) -> (then)
   n5 = Get VoiceChannelManagementVM()
   > SetVisibility(
    self=[Get VoiceChannelType_Text(), Get VoiceChannelType_Icon_DividerDot() -> (VoiceChannelType_Icon_DividerDot:Image Object Reference)]
    InVisibility=CF_VoiceChatMethodTextVisibility!(
     self
     Active Channel:ESidebarVoiceChannelType Enum=GetActiveChannel!(self:Voice Channel Management VM Object Reference=@n5) -> (ReturnValue:"ESidebarVoiceChannelType Enum (by ref)"=None)
     AvailableChannels:Array of ESidebarVoiceChannelType Enums=GetAvailableChannels!(self:Voice Channel Management VM Object Reference=@n5) -> (ReturnValue:Array of ESidebarVoiceChannelType Enums=None)
    ) -> (Visibility:ESlateVisibility Enum=Visible) {is=title K2Node_CallFunction}
   )
   > UpdatePTTKey()
   > UpdateToggleVoiceMethodContextOptions()
  ]}
 OnMouseButtonDown {nodes [
   OnMouseButtonDown!() -> (then, MyGeometry:Geometry Structure, MouseEvent:"Pointer Event Structure (by ref)") {is=title K2Node_FunctionEntry}
   > Return Node(ReturnValue=Handled())
  ]}
 OnMutedChanged {nodes [
   OnMutedChanged!() -> (then, IsSelfMuted:bool=~) {is=title K2Node_FunctionEntry}
   n1 = Get Button_Mic()
   > Set ButtonType(
    ButtonType=Select(
     Option 0:E_UI_Sidebar_Button_Type Enum=NewEnumerator0
     Option 1:E_UI_Sidebar_Button_Type Enum=NewEnumerator6
     Index=Get IsSelfMuted() -> (IsSelfMuted:bool)
    ) -> (ReturnValue:E_UI_Sidebar_Button_Type Enum)
    self=@n1
   ) -> (Output_Get=NewEnumerator0)
   > InitializeButtonVisuals(self=@n1)
   > SetIconMaterial(
    self=@n1
    Material Instance=Select(
     Option 0:Material Instance Object Reference=&MI_Sidebar_Icon_Base
     Option 1:Material Instance Object Reference=&MI_Sidebar_Icon_Alert
     Index=Get IsSelfMuted()
    ) -> (ReturnValue:Material Instance Object Reference)
   )
   n2 = Get Button_Mic()
   > SetIcon(
    self=@n2
    Icon=Select(
     Option 0:Object Reference=/Sidebar/Materials/Icons/Textures/T_UI_Icon_Sidebar_Mic_MSDF
     Option 1:Object Reference=/Sidebar/Materials/Icons/Textures/T_UI_Icon_Sidebar_MicSlash_MSDF
     Index=Get IsSelfMuted()
    ) -> (ReturnValue:Object Reference)
   )
   > SetIconSize(self=@n2, Icon=Get IconSizeWithoutArrow())
   n3 = Get Button_Mic()
   > RequestInstantTransition!(
    execute
    self:UIKit Block Object Reference=Get Block_Icon(self=@n3)
    State:EUIKitBlockInstantTransitionState Enum=Select(
     Option 0:EUIKitBlockInstantTransitionState Enum=InstantDisabled
     Option 1:EUIKitBlockInstantTransitionState Enum=InstantEnabled
     Index=GetIsEnabled(self=@n3)
    ) -> (ReturnValue:EUIKitBlockInstantTransitionState Enum)
   ) -> (then)
  ]}
 OnNavigateDownFromActionButton {nodes [
   OnNavigateDownFromActionButton!() -> (then, Navigation:EUINavigation Enum) {is=title K2Node_FunctionEntry}
   > n1 = Branch(Condition=Get isUpdatedPartyControlsEnabled())
   n2 = Get Button_PartyJoinability_New()
   > n3 = Branch(Condition="Not Equal (Enum)"(
     A:ESlateVisibility Enum=GetVisibility!(self:Widget Object Reference=@n2) -> (ReturnValue:ESlateVisibility Enum=Visible)
     B:ESlateVisibility Enum=Collapsed
    ))
   > Return Node(ReturnValue:Widget Object Reference=Get Button_PartyJoinability_New())
   SetNavigationRuleBase(execute=@n3.else, self=[Get Substate_Action_Button(), Get Substate_MoreOptions_Button()], Direction=Down, Rule=Escape)
   n4 = GetFirstFocusableButtonOnBottomRow!(execute=@n1.else, self) -> (then, FirstFocusableButtonOnBottomRow:Widget Object Reference) {is=title K2Node_CallFunction}
   > Return Node(ReturnValue=@n4)
  ]}
 OnNavigateDownFromChannelButtons {nodes [
   OnNavigateDownFromChannelButtons!() -> (then, Navigation:EUINavigation Enum) {is=title K2Node_FunctionEntry}
   n1 = Get VB_SubStates()
   n2 = Get Substate_Action_Button()
   > n3 = Branch(Condition=AND Boolean(A=IsVisible(self=@n2), B=IsVisible(self=@n1)))
   > Return Node(ReturnValue=Get Substate_Action_Button())
   n4 = Branch(execute=@n3.else, Condition=Get isUpdatedPartyControlsEnabled())
   n5 = Get Button_PartyJoinability_New()
   > Branch(Condition="Not Equal (Enum)"(A=GetVisibility(self=@n5) -> (ReturnValue=Visible), B=Collapsed))
   > Return Node(ReturnValue=Get Button_PartyJoinability_New())
   n6 = GetFirstFocusableButtonOnBottomRow(execute=@n4.else)
   > Return Node(ReturnValue=@n6)
  ]}
 OnNavigateToPartyJoinability {nodes [
   OnNavigateToPartyJoinability!() -> (then, Navigation:EUINavigation Enum) {is=title K2Node_FunctionEntry}
   > Return Node(ReturnValue=Get Button_PartyJoinability_New())
  ]}
 OnNavigateUpFromActionButton {nodes [
   OnNavigateUpFromActionButton!() -> (then, Navigation:EUINavigation Enum) {is=title K2Node_FunctionEntry}
   n1 = Get VoiceChannelManagementVM()
   n2 = Get Channel_Party_Button()
   > Return Node(ReturnValue=Select!(
     None:Widget Object Reference=@n2
     Party:Widget Object Reference=@n2
     Game:Widget Object Reference=Get Channel_Game_Button()
     EpicParty:Widget Object Reference=@n2
     Index:ESidebarVoiceChannelType Enum=Get ActiveChannel(self=@n1) -> (ActiveChannel=None)
    ) -> (ReturnValue:Widget Object Reference))
  ]}
 OnNavigateUpFromChannelButtons {nodes [
   n1 = OnNavigateUpFromChannelButtons!() -> (then, Navigation:EUINavigation Enum) {is=title K2Node_FunctionEntry}
   > n2 = Switch on EUINavigation!(execute, Selection:EUINavigation Enum=@n1) -> (Left:Exec, Right:Exec, Up:Exec, Down:Exec, Next:Exec, Previous:Exec, Invalid:Exec) {is=title K2Node_SwitchEnum}
   n3 = Get WBP_Social_Restriction_Generic_Internal() -> (WBP_Social_Restriction_Generic_Internal:WBP Social Restriction Generic Internal Object Reference)
   n4 = Branch(execute=@n2.Up, Condition=IsVisible(self=@n3))
   > Return Node(ReturnValue=@n3)
   n5 = FindParentWidgetOfType(execute=@n4.else, StartingWidget=Self-Reference(), Type=&WBP_Sidebar_C)
   > n6 = NavigationToTabGroup(self=@n5, Navigation=Up)
   > Return Node(ReturnValue=@n6)
  ]}
 OnSetLastFocusedButton {nodes [
   n1 = OnSetLastFocusedButton!() -> (then, Type:EUINavigation Enum, OldFocusedWidget:Widget Object Reference, NewFocusedWidget:Widget Object Reference) {is=title K2Node_FunctionEntry}
   n2 = GetOuterUserWidget!(Widget:Widget Object Reference=@n1.NewFocusedWidget) -> (ReturnValue:User Widget Object Reference)
   > n3 = Cast To WBP_Sidebar_Button_Base!(execute, Object=@n2) -> (then, CastFailed, AsWBP Sidebar Button Base:WBP Sidebar Button Base Object Reference) {is=title K2Node_DynamicCast}
   n4 = Branch(execute=[@n5.then, @n3.then], Condition=IsValid(Object=@n2))
   Set LastFocusedButton(execute=[@n6.then, @n4.then], LastFocusedButton:User Widget Object Reference=@n2) -> (Output_Get:User Widget Object Reference)
   n5 = Cast To WBP_Sidebar_ButtonGeneric!(execute=@n3.CastFailed, Object=@n2) -> (then, CastFailed, AsWBP Sidebar Button Generic:WBP Sidebar Button Generic Object Reference) {is=title K2Node_DynamicCast}
   n6 = Cast To WBP_Social_ButtonGeneric!(execute=@n5.CastFailed, Object=@n2) -> (then, CastFailed, AsWBP Social Button Generic:WBP Social Button Generic Object Reference) {is=title K2Node_DynamicCast}
  ]}
 OnSubstateMoreOptionsVisibilityChanged {nodes [
   n1 = OnSubstateMoreOptionsVisibilityChanged!() -> (then, InVisibility:ESlateVisibility Enum=Visible) {is=title K2Node_FunctionEntry}
   > n2 = Branch(Condition="Equal (Enum)"(A:ESlateVisibility Enum=@n1, B:ESlateVisibility Enum=Collapsed))
   n3 = Get Substate_Action_Button()
   n4 = Get Substate_MoreOptions_Button()
   > SetNavigationRuleCustomBoundary(self=[@n3, @n4], Direction=Down, InCustomDelegate=Create Event())
   > n5 = GetFirstFocusableButtonOnBottomRow()
   > Set LastFocusedButton(LastFocusedButton=Cast To UserWidget!(Object=@n5) -> (AsUser Widget:User Widget Object Reference, bSuccess:bool=~) {is=title K2Node_DynamicCast})
   > RequestRefreshFocus()
   SetNavigationRuleBase(execute=@n2.else, self=[@n3, @n4], Direction=Down, Rule=Escape)
  ]}
 SetBannedStateVisibility {nodes [
   SetBannedStateVisibility!() -> (then, IsSocialBanned:bool=~) {is=title K2Node_FunctionEntry}
   > SetVisibility(
    self=Get SocialBan_Message() -> (SocialBan_Message:WBP Social Ban Message Object Reference)
    InVisibility=Select!(
     Option 0:ESlateVisibility Enum=Collapsed
     Option 1:ESlateVisibility Enum=SelfHitTestInvisible
     Index:bool=Get IsSocialBanned() -> (IsSocialBanned:bool)
    ) -> (ReturnValue:ESlateVisibility Enum)
   )
   > SetVisibility(
    self=Get BorderTinted_OtherSubMessage() -> (BorderTinted_OtherSubMessage:Border Object Reference)
    InVisibility=Select(Option 0=SelfHitTestInvisible, Option 1=Collapsed, Index=Get IsSocialBanned())
   )
  ]}
 SetCoachMarkPlacement {nodes [
   n1 = SetCoachMarkPlacement!() -> (then, CheckMarkPlacement:EFortSidebarCoachMarkPlacement Enum=Outside) {is=title K2Node_FunctionEntry}
   > SetHorizontalAlignment(
    self:Horizontal Box Slot Object Reference=SlotAsHorizontalBoxSlot!(Widget:Widget Object Reference=Get CommonVisualAttachment_247() -> (CommonVisualAttachment_247:Common Visual Attachment Object Reference)) -> (ReturnValue:Horizontal Box Slot Object Reference)
    InHorizontalAlignment=Select!(Outside:EHorizontalAlignment Enum=HAlign_Left, Inside:EHorizontalAlignment Enum=HAlign_Fill, Index:EFortSidebarCoachMarkPlacement Enum=@n1) -> (ReturnValue:EHorizontalAlignment Enum)
   )
   n2 = Get CommonVisualAttachment_247()
   > Set!(
    execute
    ContentAnchor_X:double=Select(Outside:double=1, Inside:double=0, Index=@n1) -> (ReturnValue:double)
    ContentAnchor_Y:double=Get!(self:Common Visual Attachment Object Reference=@n2) -> (ContentAnchor_X:double, ContentAnchor_Y:double).ContentAnchor_Y
    self:Common Visual Attachment Object Reference=@n2
   ) -> (then, Output_Get:Vector 2D Structure)
   n3 = Get WBP_VoiceChat_ControlsMoved_CoachMark()
   n4 = Get!(self:User Widget Object Reference=@n3) -> (Padding_Left:float=~, Padding_Top:float=~, Padding_Right:float=~, Padding_Bottom:float=~)
   > SetPadding!(
    execute
    self=@n3
    InPadding_Left:float=@n4
    InPadding_Top:float=Select(Outside:float=-100, Inside:float=24, Index=@n1) -> (ReturnValue:float)
    InPadding_Right:float=@n4.Padding_Right
    InPadding_Bottom:float=@n4.Padding_Bottom
   ) -> (then)
   > SetCoachMarkPlacement(self:WBP Voice Chat Controls Moved Coach Mark Object Reference=Get WBP_VoiceChat_ControlsMoved_CoachMark(), CheckMarkPlacement=@n1) {is=title K2Node_CallFunction}
  ]}
 SetCurrentIslandImageLoadingState {nodes [
   SetCurrentIslandImageLoadingState!() -> (then, IslandImageState:EActivityImageLoadingState Enum=Unset) {is=title K2Node_FunctionEntry}
   n1 = Get IslandTileMaterial() -> (IslandTileMaterial:Image Object Reference)
   > Set IslandTile MID(IslandTile MID:Material Instance Dynamic Object Reference=GetDynamicMaterial!(Brush:"Slate Brush Structure (by ref)"=Get Brush!(self:Image Object Reference=@n1) -> (Brush:Slate Brush Structure)) -> (ReturnValue)) -> (Output_Get:Material Instance Dynamic Object Reference)
   > n2 = Switch on EActivityImageLoadingState!(execute, Selection:EActivityImageLoadingState Enum=Get IslandImageState() -> (IslandImageState:EActivityImageLoadingState Enum=Unset)) -> (Unset:Exec, Loading:Exec, Loaded:Exec, Error:Exec) {is=title K2Node_SwitchEnum}
   SetScalarParameterValue(
    execute=@n2.Loaded
    self=Get IslandTile MID() -> (IslandTile MID:Material Instance Dynamic Object Reference)
    ParameterName=IsLoading
    Value=0
   )
   n3 = Get PartyMemberCurrentIslandVM()
   > SetTextureParameterValue(
    self=Get IslandTile MID()
    ParameterName=Thumbnail
    Value=GetImageTexture!(self:Fort Discover Link Entry Tile Item VM Object Reference=Get CurrentIslandVM(self=@n3)) -> (ReturnValue:Texture Object Reference)
   )
   SetScalarParameterValue(execute=@n2.Loading, self=Get IslandTile MID(), ParameterName=IsLoading, Value=1)
   SetScalarParameterValue(execute=@n2.Error, self=Get IslandTile MID(), ParameterName=IsLoading, Value=0)
   > SetScalarParameterValue(self=Get IslandTile MID(), ParameterName=IsError, Value=1)
  ]}
 SetEpicPartiesPreviewEnabled {nodes [
   n1 = SetEpicPartiesPreviewEnabled!() -> (then, Interaction Name Id:Name) {is=title K2Node_FunctionEntry}
   > n2 = Cast To FortPlayerController(Object=GetOwningPlayer())
   n3 = GetClientSettings!(self:Fort Player Controller Object Reference=@n2) -> (ReturnValue:Fort Client Settings Record Object Reference)
   > Branch(Condition=IsValid(Object=@n3))
   > SetIsEpicPartiesPreviewEnabled!(
    execute
    self:Fort Client Settings Record Object Reference=@n3
    bEnable:bool="Equal (Name)"(A=@n1, B=EnableEpicPartiesPreview)
    PlayerController:Fort Player Controller Object Reference=@n2
   ) -> (then)
   > UpdateEpicPartiesPreviewInteractionAvailability()
   > RequestCloseSidebar(self=Get FortSidebarVM())
  ]}
 SetGameChannelInfo {nodes [
   n1 = SetGameChannelInfo!() -> (then, TeamMembers:Array of Fort Social Item VM Object References) {is=title K2Node_FunctionEntry}
   n2 = Get FortGlobalUIContext()
   > n3 = Branch(Condition=NOT Boolean(A=IsInZone(self=@n2)))
   > SetGroupStatus!(execute, self:WBP Sidebar Voice Channel Button Object Reference=Get Channel_Game_Button(), StatusText:Text="LOCTEXT(\"Not in a game yet\")") -> (then)
   n4 = Branch(execute=@n3.else, Condition="integer > integer"(A=Length(TargetArray:Array of Fort Social Item VM Object References=@n1), B=1))
   n5 = Get TeamMembers() -> (TeamMembers:Array of Fort Social Item VM Object References)
   > SetGroupStatus(
    self=Get Channel_Game_Button()
    StatusText=Format Text!(Format="LOCTEXT(\"{count} members\")", count:int=Length(TargetArray=@n5)) -> (Result)
   )
   SetGroupStatus(execute=@n4.else, self=Get Channel_Game_Button(), StatusText="LOCTEXT(\"Just you\")")
  ]}
 SetInitialChannelSelectorState {nodes [
   Inputs!() -> (execute:Exec)
   n1 = Get ActiveChannel(self=Get VoiceChannelManagementVM()) -> (ActiveChannel=None)
   n2 = Get Backplate_ChannelSelector() -> (Backplate_ChannelSelector:Image Object Reference)
   > SetScalarParameterValue(
    self=GetDynamicMaterial(Brush=Get Brush(self=@n2))
    ParameterName=Control
    Value=SelectFloat(A=0, B=1, bPickA=Is Channel Party Channel(A=@n1) -> (ReturnValue=~))
   )
   > Outputs!(then:Exec)
  ]}
 SetIsJoinable {nodes [
   n1 = SetIsJoinable!() -> (then, IsJoinable:bool=~) {is=title K2Node_FunctionEntry}
   > SetLocalActiveOption!(
    execute
    self:WBP Sidebar Button With Context Menu Object Reference=Get Button_PartyJoinability_New()
    Name:Name=SelectName!(A:Name=Joinable, B:Name=Invite, bPickA:bool=@n1) -> (ReturnValue:Name=None)
   ) -> (then)
  ]}
 SetIslandPresenceText {nodes [
   n1 = SetIslandPresenceText!() -> (then, presence:Text) {is=title K2Node_FunctionEntry}
   > SetGroupStatus(self=Get Channel_Game_Button(), StatusText=@n1)
  ]}
 SetMobileStyle {nodes [
   SetMobileStyle() {is=title K2Node_FunctionEntry}
   > Branch(Condition=IsUsingCompactUI())
   > n1 = Sequence!(execute) -> (then_0, then_1, then_2:Exec, then_3:Exec, then_4:Exec)
   n2 = Get SubState_RichMessage()
   SetAutoWrapText!(execute=@n1.then_2, self:Rich Text Block Object Reference=@n2, InAutoTextWrap:bool=true) -> (then)
   > SetScrollingEnabled!(execute, self:Common Rich Text Block Object Reference=@n2, bInIsScrollingEnabled:bool=false) -> (then)
   n3 = SlotAsHorizontalBoxSlot(Widget=@n2)
   > SetSize!(execute, self:Horizontal Box Slot Object Reference=@n3, InSize_Value:float=1, InSize_SizeRule:ESlateSizeRule Enum=Fill) -> (then)
   > SetVerticalAlignment(self:Horizontal Box Slot Object Reference=@n3, InVerticalAlignment=VAlign_Top)
   n4 = SlotAsVerticalBoxSlot(Widget=Get VB_SubStates())
   n5 = Get(self:Vertical Box Slot Object Reference=@n4) -> (Padding_Left=~, Padding_Top=~, Padding_Right=~, Padding_Bottom=~)
   SetPadding(
    execute=@n1.then_0
    self:Vertical Box Slot Object Reference=@n4
    InPadding_Left=@n5
    InPadding_Top=@n5.Padding_Top
    InPadding_Right=@n5.Padding_Right
    InPadding_Bottom=12
   )
   n6 = SlotAsVerticalBoxSlot(Widget=Get BorderTinted_OtherSubMessage())
   > SetHorizontalAlignment(self:Vertical Box Slot Object Reference=@n6, InHorizontalAlignment=HAlign_Fill)
   n7 = Get(self=@n6) -> (Padding_Left=~, Padding_Top=~, Padding_Right=~, Padding_Bottom=~)
   > SetPadding(self=@n6, InPadding_Left=@n7, InPadding_Top=@n7.Padding_Top, InPadding_Right=@n7.Padding_Right, InPadding_Bottom=24)
   n8 = Get Button_PartyJoinability_New()
   SetPadding!(execute=@n1.then_4, self:Horizontal Box Slot Object Reference=SlotAsHorizontalBoxSlot(Widget=@n8), InPadding:Margin Structure) -> (then)
   > SetButtonPadding!(
    execute
    self:WBP UIKit Button Generic Base Object Reference=@n8
    Padding_Left:float=12
    Padding_Top:float=18
    Padding_Right:float=56
    Padding_Bottom:float=18
   ) -> (then)
   n9 = Get ArrowIcon(self=@n8)
   > SetGridSlotAlignment!(
    execute
    Widget:Widget Object Reference=@n9
    Horizontal:EHorizontalAlignment Enum=HAlign_Right
    Vertical:EVerticalAlignment Enum=VAlign_Center
   ) -> (then)
   > SetDesiredSizeOverride!(execute, self:Image Object Reference=@n9, DesiredSize:Vector 2D Structure=(X=24,Y=24)) -> (then)
   > SetVerticalAlignment(self=SlotAsHorizontalBoxSlot(Widget=Get VoiceReportStatusNew_Text()), InVerticalAlignment=VAlign_Center)
   > Set ArrowIcon_TopPadding(ArrowIcon_TopPadding:double=0) -> (Output_Get:double)
   n10 = SlotAsHorizontalBoxSlot(Widget=Get SubState_RichPlatformIcon() -> (SubState_RichPlatformIcon:Common Rich Text Block Object Reference))
   n11 = Get(self:Horizontal Box Slot Object Reference=@n10) -> (Padding_Left=~, Padding_Top=~, Padding_Right=~, Padding_Bottom=~)
   SetPadding!(
    execute=@n1.then_1
    self=@n10
    InPadding_Left:float=@n11
    InPadding_Top:float=4
    InPadding_Right:float=8
    InPadding_Bottom:float=@n11.Padding_Bottom
   ) -> (then)
   > SetVerticalAlignment(self=@n10, InVerticalAlignment=VAlign_Top)
   > SetPadding(
    self=SlotAsHorizontalBoxSlot(Widget=Get WarningIcon() -> (WarningIcon:Image Object Reference))
    InPadding_Left=0
    InPadding_Top=3
    InPadding_Right=8
    InPadding_Bottom=6
   )
   SetDesiredSizeOverride(execute=@n1.then_3, self=Get OnlyYouImage() -> (OnlyYouImage:Image Object Reference), DesiredSize=(X=64,Y=64))
   > SetDesiredSizeOverride(self=Get IslandTileMaterial(), DesiredSize=(X=64,Y=64))
   > Event Set Size!(
    execute
    self:WBP Sidebar Player Avatar Object Reference=Get PlayerAvatar() -> (PlayerAvatar:WBP Sidebar Player Avatar Object Reference)
    NewSize:Vector 2D Structure=(X=64,Y=64)
   ) -> (then)
   > Set IconSizeWithArrow(IconSizeWithArrow:Vector 2D Structure=(X=72,Y=48)) -> (Output_Get:Vector 2D Structure)
   > Set IconSizeWithoutArrow(IconSizeWithoutArrow:Vector 2D Structure=(X=48,Y=48)) -> (Output_Get:Vector 2D Structure)
   n12 = Get IconSizeWithoutArrow()
   > SetIconSize(
    self=[
     Get Button_MoreOptions() -> (Button_MoreOptions:WBP Sidebar Button Generic Object Reference)
     Get Substate_MoreOptions_Button()
     Get Button_PartyJoinability()
    ]
    Icon=@n12
   )
   > SetIconSize(
    self=[Get Button_Mic(), Get Button_Headphones()]
    Icon=Select!(
     Option 0:Vector 2D Structure=@n12
     Option 1:Vector 2D Structure=Get IconSizeWithArrow() -> (IconSizeWithArrow:Vector 2D Structure)
     Index:bool=Get IsUsingEpicParties()
    ) -> (ReturnValue:Vector 2D Structure)
   )
  ]}
 SetPartyAvatars {nodes [
   n1 = SetPartyAvatars!() -> (then, PartyMembers:Array of Fort Social Item VM Object References) {is=title K2Node_FunctionEntry}
   > Clear(TargetArray:Array of Fort Social User Object References=Get PartySocialUsers() -> (PartySocialUsers:Array of Fort Social User Object References))
   > n2 = For Each Loop(Array:Array of Fort Social Item VM Object References=@n1) -> (Array Element:Fort Social Item VM Object Reference)
   InitializeWithSocialUsers!(
    execute=@n2.Completed
    self:WBP Sidebar Player Avatar Object Reference=Get PlayerAvatar()
    SocialUsers:Array of Fort Social User Object References=Get PartySocialUsers()
   ) -> (then)
   n3 = Get TargetUser(self=@n2)
   n4 = Branch(execute=@n2.LoopBody, Condition=IsLocalUserForBlueprint(self=@n3))
   > Insert!(
    execute
    TargetArray:Array of Fort Social User Object References=Get PartySocialUsers()
    NewItem:"Fort Social User Object Reference (by ref)"=@n3
    Index:int=0
   ) -> (then) {is=title K2Node_CallArrayFunction}
   Add(
    execute=@n4.else
    TargetArray:Array of Fort Social User Object References=Get PartySocialUsers()
    NewItem:"Fort Social User Object Reference (by ref)"=@n3
   ) {is=title K2Node_CallArrayFunction}
  ]}
 SetPartyChannelInfo {nodes [
   n1 = SetPartyChannelInfo!() -> (then, PartyMembers:Array of Fort Social Item VM Object References) {is=title K2Node_FunctionEntry}
   > n2 = Branch(Condition="integer > integer"(A=Length(TargetArray=@n1), B=1))
   n3 = Get SwitcherPartyIcon() -> (SwitcherPartyIcon:Common Visibility Switcher Object Reference)
   > SetActiveWidget(self=@n3, Widget=Get PlayerAvatar())
   n4 = Get PartyMembers() -> (PartyMembers:Array of Fort Social Item VM Object References)
   > SetGroupStatus(self=Get Channel_Party_Button(), StatusText=Format Text(Format="LOCTEXT(\"{count} members\")", count=Length(TargetArray=@n4)))
   > SetPartyAvatars!(execute, self, PartyMembers:Array of Fort Social Item VM Object References=@n4) -> (then) {is=title K2Node_CallFunction}
   SetActiveWidget(execute=@n2.else, self=@n3, Widget=Get OnlyYouImage())
   > SetGroupStatus(self=Get Channel_Party_Button(), StatusText="LOCTEXT(\"Just you\")")
  ]}
 SetPartyJoinabilityStatus {nodes [
   n1 = SetPartyJoinabilityStatus!() -> (then, IsJoinable:bool=~) {is=title K2Node_FunctionEntry}
   > n2 = Branch(Condition=@n1)
   n3 = Get Button_PartyJoinability()
   > SetIcon(self=@n3, Icon=/Sidebar/Materials/Icons/Textures/T_UI_Icon_Sidebar_Unlocked_MSDF)
   n4 = Get PartyJoinable_Text_Status()
   > "SetText (Text)"(self=@n4, InText="LOCTEXT(\"Joinable\")")
   SetIcon(execute=@n2.else, self=@n3, Icon=/Sidebar/Materials/Icons/Textures/T_UI_Icon_Sidebar_Locked_MSDF)
   > "SetText (Text)"(self=@n4, InText="LOCTEXT(\"Invite only\")")
  ]}
 SetPartyNames {nodes [
   n1 = SetPartyNames!() -> (then, PartyMembers:Array of Fort Social Item VM Object References) {is=title K2Node_FunctionEntry}
   > n2 = For Each Loop(Array=@n1)
   n3 = Get TargetUser(self=@n2)
   > Branch(Condition=NOT Boolean(A=IsLocalUserForBlueprint(self=@n3)))
   n4 = Get SocialUserVM_PartyPlayerName() -> (SocialUserVM_PartyPlayerName:Fort Social User Name Viewmodel Object Reference)
   > SetSocialUser(self=@n4, InSocialUser=@n3)
   > n5 = Branch(Condition=TextIsEmpty(InText=Get PlayerNamesString() -> (PlayerNamesString:Text)))
   n6 = Break Fort Social User Name Package!(FortSocialUserNamePackage:"Fort Social User Name Package Structure (by ref)"=Get SocialUserNames!(self:Fort Social User Name Viewmodel Object Reference=@n4) -> (SocialUserNames:Fort Social User Name Package Structure)) -> (DisplayName:String, Nickname:String, PlatformIconTag:String) {is=title K2Node_BreakStruct}
   > Set PlayerNamesString(PlayerNamesString:Text="To Text (String)"(InString=@n6)) -> (Output_Get:Text)
   Set PlayerNamesString(
    execute=@n5.else
    PlayerNamesString=Format Text!(Format="LOCTEXT(\"{current string}, {nextPlayer}\")", current string:Text=Get PlayerNamesString(), nextPlayer:String=@n6) -> (Result)
   )
   SetGroupStatus(
    execute=@n2.Completed
    self=Get Channel_Party_Button()
    StatusText=Format Text!(Format="LOCTEXT(\"{current string}, You\")", current string=Get PlayerNamesString()) -> (Result)
   )
  ]}
 SetPlatformIcon {nodes [
   SetPlatformIcon!() -> (then, ActiveClientPlatformTag:String) {is=title K2Node_FunctionEntry}
   > n1 = Get Data Table Row RichTextIconSet!(
    execute:Exec
    DataTable:Data Table Object Reference=/FortUILibrary/Foundation/RichTextData/RichTextIconSet
    RowName:Name=String To Name(InString=Get ActiveClientPlatformTag())
   ) -> (then:Exec, RowNotFound:Exec, ReturnValue:Rich Text Icon Data Structure) {is=title K2Node_GetDataTableRow}
   > SetText(
    self=Get SubState_RichPlatformIcon()
    InText=Format Text!(Format="LOCTEXT(\"<{platformTag}/>\")", platformTag:String=Get ActiveClientPlatformTag()) -> (Result)
   )
   SetText(execute=@n1.RowNotFound, self=Get SubState_RichPlatformIcon())
  ]}
 SetSubStateActionButtonText {nodes [
   n1 = SetSubStateActionButtonText!() -> (then, VoiceChatStatus:E_UI_BP_SidebarVoiceChatStatus Enum=NewEnumerator0) {is=title K2Node_FunctionEntry}
   > Return Node!(
    execute
    ButtonText:Text=Select!(
     NewEnumerator0:Text
     NewEnumerator4:Text="LOCTEXT(\"Chat here instead\")"
     NewEnumerator1:Text="LOCTEXT(\"Open communication settings\")"
     NewEnumerator2:Text="LOCTEXT(\"View update\")"
     NewEnumerator3:Text="LOCTEXT(\"Open parental controls\")"
     NewEnumerator6:Text
     NewEnumerator5:Text
     NewEnumerator7:Text="LOCTEXT(\"Go to settings\")"
     NewEnumerator8:Text
     Index:E_UI_BP_SidebarVoiceChatStatus Enum=@n1
    ) -> (ReturnValue:Text)
   )
  ]}
 SetSubStateActionButtonVisibility {nodes [
   n1 = SetSubStateActionButtonVisibility!() -> (
    then
    Restriction:ELocalVoiceChatRestriction Enum=Unrestricted
    ActiveDeviceStatus:EVoiceChatPartyClientActiveState Enum=NoClientsActiveInParty
   ) {is=title K2Node_FunctionEntry}
   > n2 = Switch on ELocalVoiceChatRestriction(Selection=@n1)
   Return Node!(execute=[@n2.NonLocalReason, @n2.PlatformPermission, @n2.SocialBan], ButtonVisibility:ESlateVisibility Enum=Collapsed)
   n3 = Get FortGlobalUIContext() {is=title K2Node_GetSubsystem}
   Return Node(
    execute=[@n2.Cabined, @n2.ParentalPermission, @n2.SplitscreenParentalPermission]
    ButtonVisibility=Select!(Option 0:ESlateVisibility Enum=SelfHitTestInvisible, Option 1:ESlateVisibility Enum=Collapsed, Index:bool=IsInZone(self=@n3)) -> (ReturnValue:ESlateVisibility Enum)
   )
   Return Node(
    execute=[@n2.MicrophonePermissionDenied, @n2.NotAcceptedModeration, @n2.SplitscreenNotAcceptedModeration, @n2.UserPermission]
    ButtonVisibility=SelfHitTestInvisible
   )
   n4 = Get ActiveDeviceStatus() -> (ActiveDeviceStatus:EVoiceChatPartyClientActiveState Enum=NoClientsActiveInParty)
   > Return Node(ButtonVisibility=Select(
     Option 0=Collapsed
     Option 1=SelfHitTestInvisible
     Index="Not Equal (Enum)"(A:EVoiceChatPartyClientActiveState Enum=@n4, B:EVoiceChatPartyClientActiveState Enum=ThisClientActiveInParty)
    ))
  ]}
 SetSubStateMessageVisuals {nodes [
   n1 = SetSubStateMessageVisuals!() -> (then, VoiceChatStatus:E_UI_BP_SidebarVoiceChatStatus Enum=NewEnumerator0) {is=title K2Node_FunctionEntry}
   n2 = "Equal (Enum)"(A:E_UI_BP_SidebarVoiceChatStatus Enum=@n1, B:E_UI_BP_SidebarVoiceChatStatus Enum=NewEnumerator4)
   > SetHorizontalAlignment(
    self:Border Object Reference=Get BorderTinted_OtherSubMessage()
    InHorizontalAlignment=Select(Option 0:EHorizontalAlignment Enum=HAlign_Fill, Option 1:EHorizontalAlignment Enum=HAlign_Center, Index=@n2) -> (ReturnValue:EHorizontalAlignment Enum)
   )
   > SetContentColorAndOpacity!(
    execute
    self:Border Object Reference=Get BorderTinted_OtherSubMessage()
    InContentColorAndOpacity:Linear Color Structure=SelectColor(A=(R=0.0395,G=1,B=0.2307,A=1), B=(R=0.5776,G=0.5776,B=0.5776,A=1), bPickA=@n2)
   ) -> (then)
   > SetVisibility(
    self=Get SubState_RichPlatformIcon()
    InVisibility=Select(Option 0:ESlateVisibility Enum=Collapsed, Option 1:ESlateVisibility Enum=SelfHitTestInvisible, Index=@n2) -> (ReturnValue:ESlateVisibility Enum)
   )
   > SetVisibility(self=Get WarningIcon(), InVisibility=Select(Option 0=SelfHitTestInvisible, Option 1=Collapsed, Index=@n2))
   n3 = SelectColor(
    A=Get WarningColor() -> (WarningColor:Linear Color Structure)
    B=Get DefaultColor() -> (DefaultColor:Linear Color Structure)
    bPickA="Equal (Enum)"(A=Get VoiceChatStatus() -> (VoiceChatStatus:E_UI_BP_SidebarVoiceChatStatus Enum=NewEnumerator0), B=NewEnumerator7)
   )
   > SetDefaultColorAndOpacity!(
    execute
    self:Rich Text Block Object Reference=Get SubState_RichMessage()
    InColorAndOpacity:Slate Color Structure="To SlateColor (Color)"!(Value:Linear Color Structure=@n3) -> (ReturnValue:Slate Color Structure)
   ) -> (then)
   > SetColorAndOpacity(self:Image Object Reference=Get WarningIcon(), InColorAndOpacity:Linear Color Structure=@n3)
  ]}
 SetupHeadphoneContextMenu {nodes [
   SetupHeadphoneContextMenu!() -> (then, Deafen:bool=~) {is=title K2Node_FunctionEntry}
   > Branch(Condition=Get IsUsingEpicParties())
   > RemoveAllInteractions(self:UIKit Context Menu Action Context Object Reference=Get HeadphoneMenuActionContext() -> (HeadphoneMenuActionContext:UIKit Context Menu Action Context Object Reference))
   > n1 = Construct Object from Class!(execute, Class=/Script/FortniteUI.UIKitContextMenuActionContext, self) -> (then, ReturnValue:UIKit Context Menu Action Context Object Reference)
   > n2 = Set HeadphoneMenuActionContext(HeadphoneMenuActionContext:UIKit Context Menu Action Context Object Reference=@n1) -> (Output_Get:UIKit Context Menu Action Context Object Reference)
   > SetContext!(
    execute
    self:WBP UIKit Context Menu Anchor Button Base Object Reference=Get Headphone_ContextMenu()
    InteractionContext_Data:UIKit Context Menu Action Context Object Reference=@n2
   ) -> (then)
   > AddInteraction(
    self=Get HeadphoneMenuActionContext()
    InActionInfo_DisplayName=SelectText(A="LOCTEXT(\"Undeafen\")", B="LOCTEXT(\"Deafen\")", bPickA=Get Deafen() -> (Deafen:bool))
    InActionInfo_SortPriority=20
    InActionInfo_bIsAvailable=True
    InActionInfo_bIsEnabled=True
    InActionInfo_bIsSelected=False
    InActionInfo_bAddDivider=False
    InActionInfo_bPositionDividerAfter=False
    InActionInfo_bCloseMenuOnSelect=False
    InActionInfo_InteractionNameID=None
    InActionInfo_Interaction=(DataTable=None,RowName="")
    InActionInfo_PressedDelegate="(null).None"
    InActionInfo_ReleasedDelegate="(null).None"
    InInteractionDelegate=Create Event()
   )
  ]}
 SetupMicrophoneContextMenu {nodes [
   SetupMicrophoneContextMenu() {is=title K2Node_FunctionEntry}
   > n1 = Construct Object from Class(Class=&UIKitContextMenuActionContext)
   > n2 = Set MicrophoneMenuActionContext(MicrophoneMenuActionContext:UIKit Context Menu Action Context Object Reference=@n1) -> (Output_Get:UIKit Context Menu Action Context Object Reference)
   > SetContext(
    self=Get Microphone_ContextMenu() -> (Microphone_ContextMenu:WBP UIKit Context Menu Anchor Button Object Reference)
    InteractionContext_Data=@n2
   )
   n3 = Get MicrophoneMenuActionContext()
   > AddInteraction(
    self=@n3
    InActionInfo_DisplayName="LOCTEXT(\"No push to talk button set\")"
    InActionInfo_IconTexture=/FortniteCoreUI/UI/Login/Art/IconAlert
    InActionInfo_SortPriority=1
    InActionInfo_bIsAvailable=True
    InActionInfo_bIsEnabled=false
    InActionInfo_bIsSelected=False
    InActionInfo_bAddDivider=False
    InActionInfo_bPositionDividerAfter=False
    InActionInfo_bCloseMenuOnSelect=False
    InActionInfo_InteractionNameID=None
    InActionInfo_Interaction=(DataTable=None,RowName="")
    InActionInfo_PressedDelegate="(null).None"
    InActionInfo_ReleasedDelegate="(null).None"
    InInteractionDelegate=Create Event()
   )
   > AddInteraction(
    self=@n3
    InActionInfo_DisplayName="LOCTEXT(\"Switch to open mic\")"
    InActionInfo_SortPriority=2
    InActionInfo_bIsAvailable=True
    InActionInfo_bIsEnabled=True
    InActionInfo_bIsSelected=False
    InActionInfo_bAddDivider=False
    InActionInfo_bPositionDividerAfter=False
    InActionInfo_bCloseMenuOnSelect=False
    InActionInfo_InteractionNameID=None
    InActionInfo_Interaction=(DataTable=None,RowName="")
    InActionInfo_PressedDelegate="(null).None"
    InActionInfo_ReleasedDelegate="(null).None"
    InInteractionDelegate=Create Event()
   )
   > AddInteraction(
    self=@n3
    InActionInfo_DisplayName="LOCTEXT(\"Set push to talk button\")"
    InActionInfo_SortPriority=3
    InActionInfo_bIsAvailable=True
    InActionInfo_bIsEnabled=True
    InActionInfo_bIsSelected=False
    InActionInfo_bAddDivider=False
    InActionInfo_bPositionDividerAfter=False
    InActionInfo_bCloseMenuOnSelect=False
    InActionInfo_InteractionNameID=None
    InActionInfo_Interaction=(DataTable=None,RowName="")
    InActionInfo_PressedDelegate="(null).None"
    InActionInfo_ReleasedDelegate="(null).None"
    InInteractionDelegate=Create Event()
   )
  ]}
 SetupReportContextMenu {nodes [
   SetupReportContextMenu() {is=title K2Node_FunctionEntry}
   > n1 = Construct Object from Class(Class=&UIKitContextMenuActionContext)
   > n2 = Set ReportMenuActionContext(ReportMenuActionContext:UIKit Context Menu Action Context Object Reference=@n1) -> (Output_Get:UIKit Context Menu Action Context Object Reference)
   > SetContext(self=[Get Report_ContextMenu(), Get Substate_MoreOptions_ContextMenu()], InteractionContext_Data=@n2)
   n3 = Get ReportMenuActionContext()
   > AddInteraction(
    self=@n3
    InActionInfo_DisplayName="LOCTEXT(\"Open text chat\")"
    InActionInfo_SortPriority=18
    InActionInfo_bIsAvailable=True
    InActionInfo_bIsEnabled=True
    InActionInfo_bIsSelected=False
    InActionInfo_bAddDivider=False
    InActionInfo_bPositionDividerAfter=False
    InActionInfo_bCloseMenuOnSelect=False
    InActionInfo_InteractionNameID=Get ContextMenuOpenChatName() -> (ContextMenuOpenChatName:Name=None)
    InActionInfo_Interaction=(DataTable=None,RowName="")
    InActionInfo_PressedDelegate="(null).None"
    InActionInfo_ReleasedDelegate="(null).None"
    InInteractionDelegate=Create Event()
   )
   > AddInteraction(
    self=@n3
    InActionInfo_DisplayName="LOCTEXT(\"Communication settings\")"
    InActionInfo_SortPriority=20
    InActionInfo_bIsAvailable=True
    InActionInfo_bIsEnabled=True
    InActionInfo_bIsSelected=False
    InActionInfo_bAddDivider=False
    InActionInfo_bPositionDividerAfter=False
    InActionInfo_bCloseMenuOnSelect=False
    InActionInfo_InteractionNameID=None
    InActionInfo_Interaction=(DataTable=None,RowName="")
    InActionInfo_PressedDelegate="(null).None"
    InActionInfo_ReleasedDelegate="(null).None"
    InInteractionDelegate=Create Event()
   )
   > AddInteraction(
    self=@n3
    InActionInfo_DisplayName="LOCTEXT(\"Switch to open mic\")"
    InActionInfo_SortPriority=2
    InActionInfo_bIsAvailable=false
    InActionInfo_bIsEnabled=True
    InActionInfo_bIsSelected=False
    InActionInfo_bAddDivider=False
    InActionInfo_bPositionDividerAfter=False
    InActionInfo_bCloseMenuOnSelect=False
    InActionInfo_InteractionNameID=SetOpenMic
    InActionInfo_Interaction=(DataTable=None,RowName="")
    InActionInfo_PressedDelegate="(null).None"
    InActionInfo_ReleasedDelegate="(null).None"
    InInteractionDelegate=Create Event()
   )
   > AddInteraction(
    self=@n3
    InActionInfo_DisplayName="LOCTEXT(\"Switch to push to talk\")"
    InActionInfo_SortPriority=2
    InActionInfo_bIsAvailable=false
    InActionInfo_bIsEnabled=True
    InActionInfo_bIsSelected=False
    InActionInfo_bAddDivider=False
    InActionInfo_bPositionDividerAfter=False
    InActionInfo_bCloseMenuOnSelect=False
    InActionInfo_InteractionNameID=SetPTT
    InActionInfo_Interaction=(DataTable=None,RowName="")
    InActionInfo_PressedDelegate="(null).None"
    InActionInfo_ReleasedDelegate="(null).None"
    InInteractionDelegate=Create Event()
   )
   > n4 = Cast To FortPlayerController(Object=GetOwningPlayer())
   n5 = GetClientSettings(self=@n4)
   > Branch(Condition=IsValid(Object=@n5))
   > Branch(Condition=GetIsEpicPartiesPreviewAvailable!(self:Fort Client Settings Record Object Reference=@n5) -> (ReturnValue:bool))
   > AddInteraction(
    self=@n3
    InActionInfo_DisplayName="LOCTEXT(\"Enable Epic Parties Preview\")"
    InActionInfo_SortPriority=99
    InActionInfo_bIsAvailable=false
    InActionInfo_bIsEnabled=True
    InActionInfo_bIsSelected=False
    InActionInfo_bAddDivider=true
    InActionInfo_bPositionDividerAfter=False
    InActionInfo_bCloseMenuOnSelect=False
    InActionInfo_InteractionNameID=EnableEpicPartiesPreview
    InActionInfo_Interaction=(DataTable=None,RowName="")
    InActionInfo_PressedDelegate="(null).None"
    InActionInfo_ReleasedDelegate="(null).None"
    InInteractionDelegate=Create Event()
   )
   > AddInteraction(
    self=@n3
    InActionInfo_DisplayName="LOCTEXT(\"Disable Epic Parties Preview\")"
    InActionInfo_SortPriority=99
    InActionInfo_bIsAvailable=false
    InActionInfo_bIsEnabled=True
    InActionInfo_bIsSelected=False
    InActionInfo_bAddDivider=true
    InActionInfo_bPositionDividerAfter=False
    InActionInfo_bCloseMenuOnSelect=False
    InActionInfo_InteractionNameID=DisableEpicPartiesPreview
    InActionInfo_Interaction=(DataTable=None,RowName="")
    InActionInfo_PressedDelegate="(null).None"
    InActionInfo_ReleasedDelegate="(null).None"
    InInteractionDelegate=Create Event()
   )
   > UpdateEpicPartiesPreviewInteractionAvailability()
  ]}
 ShouldAlwaysActAsOpenMic {nodes [
   ShouldAlwaysActAsOpenMic() {is=title K2Node_FunctionEntry}
   n1 = GetOwningPlayer()
   n2 = GetRuntimeOptions()
   > Return Node!(
    execute
    ReturnValue:bool=AND Boolean(
     A=Get bSupportsTouchHudMicButton!(self:Fort Runtime Options Object Reference=@n2) -> (bSupportsTouchHudMicButton:bool)
     B=IsUsingInputType!(InputType:ECommonInputType Enum=Touch, PlayerController:Player Controller Object Reference=@n1) -> (ReturnValue:bool)
    )
   )
  ]}
 ShowPTTContextMenu {nodes [
   Inputs()
   n1 = Get IsPTTKey() -> (IsPTTKey:bool)
   > n2 = Branch(Condition=AND Boolean(A=Get IsVoiceChatMethodPTT(), B=NOT Boolean(A=@n1)))
   > n3 = Open(self=Get Anchor(self=Get Microphone_ContextMenu()), bFocusMenu=true)
   Outputs(then=[@n3.then, @n2.else])
  ]}
 UpdateEpicPartiesPreviewInteractionAvailability {nodes [
   UpdateEpicPartiesPreviewInteractionAvailability() {is=title K2Node_FunctionEntry}
   > n1 = Cast To FortPlayerController(Object=GetOwningPlayer())
   n2 = GetClientSettings(self=@n1)
   > Branch(Condition=IsValid(Object=@n2))
   > n3 = Branch(Condition=GetIsEpicPartiesPreviewAvailable(self=@n2))
   n4 = GetIsEpicPartiesPreviewEnabled!(self:Fort Client Settings Record Object Reference=@n2) -> (ReturnValue:bool)
   > SetVisibility(
    self=Get VB_EpicPartyPreviewEnabled() -> (VB_EpicPartyPreviewEnabled:Vertical Box Object Reference)
    InVisibility=Select(Option 0=Collapsed, Option 1=HitTestInvisible, Index=@n4)
   )
   n5 = Get ReportMenuActionContext()
   > SetInteractionAvailability!(
    execute
    self:UIKit Context Menu Action Context Object Reference=@n5
    InteractionNameID:Name=Select(Option 0:Name=DisableEpicPartiesPreview, Option 1:Name=EnableEpicPartiesPreview, Index=@n4) -> (ReturnValue:Name)
    bInAvailability:bool=false
   ) -> (then)
   > SetInteractionAvailability(self=@n5, InteractionNameID=Select(Option 0=EnableEpicPartiesPreview, Option 1=DisableEpicPartiesPreview, Index=@n4), bInAvailability=true)
   SetVisibility(execute=@n3.else, self=Get VB_EpicPartyPreviewEnabled(), InVisibility=Collapsed)
   n6 = Get ReportMenuActionContext()
   > SetInteractionAvailability(self=@n6, InteractionNameID=DisableEpicPartiesPreview, bInAvailability=false)
   > SetInteractionAvailability(self=@n6, InteractionNameID=EnableEpicPartiesPreview, bInAvailability=false)
  ]}
 UpdateOpenTextChatContextOption {nodes [
   UpdateOpenTextChatContextOption() {is=title K2Node_FunctionEntry}
   > n1 = GetConsoleVariableBoolValue(VariableName=PilgrimInverse.UsePortraitMode)
   n2 = IsUsingCompactUI()
   n3 = "Equal (Enum)"(
    A:EDisplayOrientationType Enum=DisplayOrientationFromScreenOrientation!(ScreenOrientation:EScreenOrientation Enum=GetDeviceOrientation!() -> (ReturnValue:EScreenOrientation Enum=Unknown)) -> (DisplayOrientation:EDisplayOrientationType Enum=Landscape)
    B:EDisplayOrientationType Enum=Portrait
   )
   > SetInteractionAvailability(
    self=Get ReportMenuActionContext()
    InteractionNameID=Get ContextMenuOpenChatName() -> (ContextMenuOpenChatName=None)
    bInAvailability=NOT Boolean(A=AND Boolean!(A=@n1, B=@n3, C:bool=@n2) -> (ReturnValue))
   )
  ]}
 UpdatePTTKey {nodes [
   UpdatePTTKey() {is=title K2Node_FunctionEntry}
   > n1 = Branch(Condition=Get IsVoiceChatMethodPTT())
   n2 = Get PTT_Key() -> (PTT_Key:Image Object Reference)
   > SetVisibility(self=@n2, InVisibility=SelfHitTestInvisible)
   n3 = Get PushToTalkKeybindingSettingVM!(self:Voice Channel Management VM Object Reference=Get VoiceChannelManagementVM()) -> (PushToTalkKeybindingSettingVM:Fort Social Setting Keybind VM Object Reference)
   n4 = Get FortGlobalUIContext()
   > n5 = Set CurrentKey(CurrentKey:Key Structure=GetKeyForAction!(
     self:Fort Global UIContext Object Reference=@n4
     Action:Name=GetActionName!(self:Fort Social Setting Keybind VM Object Reference=@n3) -> (ReturnValue:Name=None)
     bForceGamepadKey:bool=false
     PresetNameOverride:Name=None
    ) -> (ReturnValue:Key Structure)) -> (Output_Get:Key Structure)
   > n6 = Set IsPTTKey(IsPTTKey:bool=Is Valid Key!(Key:"Key Structure (by ref)"=@n5) -> (ReturnValue:bool)) -> (Output_Get:bool)
   > SetBrush(
    self=Get PTT_Key()
    InBrush=Select(
     Option 0:"Slate Brush Structure (by ref)"=Make Slate Brush!(
      TintColor_SpecifiedColor:Linear Color Structure=(R=0.9301,G=1,B=0.0103,A=1)
      TintColor_ColorUseRule:ESlateColorStylingMode Enum
      DrawAs:ESlateBrushDrawType Enum=Image
      Tiling:ESlateBrushTileType Enum=NoTile
      Mirroring:ESlateBrushMirrorType Enum=NoMirror
      ImageSize_X:double=24
      ImageSize_Y:double=24
      Margin:Margin Structure=(Left=0,Top=0,Right=0,Bottom=0)
      ResourceObject:Object Reference=/FortUILibrary/Textures/MaterialInstances/MI_UIKit_Icon_Alert
      OutlineSettings:Slate Brush Outline Settings Structure=(CornerRadii=(X=0,Y=0,Z=0,W=0),Color=(SpecifiedColor=(R=0,G=0,B=0,A=0),ColorUseRule=UseColor_Specified),Width=0,RoundingType=HalfHeightRadius,bUseBrushTransparency=False)
     ) -> (SlateBrush:Slate Brush Structure) {is=title K2Node_MakeStruct}
     Option 1:"Slate Brush Structure (by ref)"=GetBrushForKey!(self:Fort Global UIContext Object Reference=@n4, Key:Key Structure=@n5) -> (Brush:Slate Brush Structure, ReturnValue:bool)
     Index=@n6
    ) -> (ReturnValue:"Slate Brush Structure (by ref)")
   )
   SetVisibility(execute=@n1.else, self=@n2, InVisibility=Collapsed)
  ]}
 UpdateToggleVoiceMethodContextOptions {nodes [
   UpdateToggleVoiceMethodContextOptions() {is=title K2Node_FunctionEntry}
   n1 = Get IsVoiceChatMethodPTT()
   n2 = Get ReportMenuActionContext()
   > SetInteractionAvailability(self=@n2, InteractionNameID=Select(Option 0:Name=SetOpenMic, Option 1:Name=SetPTT, Index=@n1) -> (ReturnValue:Name), bInAvailability=false)
   > SetInteractionAvailability(
    self=@n2
    InteractionNameID=Select(Option 0=SetPTT, Option 1=SetOpenMic, Index=@n1)
    bInAvailability=NOT Boolean(A=ShouldAlwaysActAsOpenMic() {is=title K2Node_CallFunction})
   )
  ]}
]
variables [
 ArrowIcon_TopPadding {type=number, default=5}
 ContextMenuMuteName {type=string, default=MuteUnmute}
 ContextMenuOpenChatName {type=string, default=OpenTextChat}
 ContextMenuReportName {type=string, default=ReportConvo}
 HeadphoneMenuActionContext {type=UIKitContextMenuActionContext, default=null}
 IconSizeWithArrow {type=Vector2D, default=<54 36>}
 IconSizeWithoutArrow {type=Vector2D, default=<36 36>}
 In Color and Opacity {type=SlateColor, default {colorUseRule=UseColor_Specified, specifiedColor=<rgba 1 0 1 1>}}
 IslandTile MID {type=MaterialInstanceDynamic}
 IsPTTKey {type=boolean}
 isUpdatedPartyControlsEnabled {type=boolean}
 IsUsingEpicParties {type=boolean}
 IsVoiceChatMethodPTT {type=boolean}
 LastFocusedButton {type=UserWidget, default=null}
 MicrophoneMenuActionContext {type=UIKitContextMenuActionContext, default=null}
 PreviewCoachMarkPlacement {type=EFortSidebarCoachMarkPlacement, default=Outside}
 ReportMenuActionContext {type=UIKitContextMenuActionContext, default=null}
 SidebarCurrentVoiceChatStatusEnum {type=E_UI_BP_SidebarVoiceChatStatus, default=ParentalControls}
 SocialUserVM_PartyPlayerName {type=FortSocialUserNameViewModel, default=null}
 VoiceReportingStatusText {type=string, default=""}
]
widgets {
 info {
  inheritedWidgetCount=0
  namedSlotCount=6
  parentClass=&CommonActivatableWidget
  rootWidgetClass=&Overlay
  widgetCount=59
 }
 widgets [
  OverlayRoot {
   widgetClassPath=&Overlay
   widget=OverlayRoot
   uIComponents [{
     component=&self:UIComponentWidgetBlueprintExtension_0.ComponentContainer.NavigationUIComponent_0
     componentClassPath=/Script/UMG.NavigationUIComponent
    }]
  }
  PartyAndGame_Backplate {
   widgetClassPath=/Sidebar/Buttons/Blocks/WBP_Sidebar_PartyAndGame_Backplate
   bIsVariable=true
   parent=OverlayRoot
   widget=PartyAndGame_Backplate
   slot {class=OverlaySlot, properties {horizontalAlignment=HAlign_Fill, verticalAlignment=VAlign_Fill, padding=<ltrb -18 -6 -28 -10>}}
  }
  VerticalBox {
   widgetClassPath=&VerticalBox
   parent=OverlayRoot
   widget=VerticalBox
   properties {bOverride_Cursor=true}
   slot {class=OverlaySlot, properties {horizontalAlignment=HAlign_Fill, verticalAlignment=VAlign_Fill, padding=<ltrb 18 6 18 0>}}
  }
  WBP_Social_Restriction_Generic_Internal {
   widgetClassPath=/FortniteCoreUI/UI/Social/Restrictions/WBP_Social_Restriction_Generic_Internal
   bIsVariable=true
   parent=VerticalBox
   widget=WBP_Social_Restriction_Generic_Internal
   properties {
    navigation=&self:WidgetTree.WBP_Social_Restriction_Generic_Internal.WidgetNavigation_0
    visibility=Collapsed
   }
   slot {class=VerticalBoxSlot, properties {padding=<ltrb 0 4 0 12>}}
  }
  Grid_ChannelSelect_CustomNavBoundary {
   widgetClassPath=&GridPanel
   bIsVariable=true
   parent=VerticalBox
   widget=Grid_ChannelSelect_CustomNavBoundary
   properties {
    navigation=&self:WidgetTree.Grid_ChannelSelect_CustomNavBoundary.WidgetNavigation_2
    columnFill [0.5,0.5]
    rowFill [1]
   }
   slot {class=VerticalBoxSlot, properties {padding=<ltrb 0 6 0 6>}}
  }
  Channel_Party_Button {
   widgetClassPath=&/WBP_Sidebar_VoiceChannel_Button
   bIsVariable=true
   parent=Grid_ChannelSelect_CustomNavBoundary
   widget=Channel_Party_Button
   properties {
    cornerRadius=18
    groupLabelText=Party chat
    holdData=&CommonUI_HoldData_Base_C
    navigation=&self:WidgetTree.Channel_Party_Button.WidgetNavigation_0
   }
   slot {class=GridSlot, properties {layer=1, padding=<ltrb 0 0 3 0>}}
  }
  SwitcherPartyIcon {widgetClassPath=&CommonVisibilitySwitcher, bIsVariable=true, namedSlotHost=Channel_Party_Button, widget=SwitcherPartyIcon}
  OnlyYouImage {
   widgetClassPath=&Image
   bIsVariable=true
   parent=SwitcherPartyIcon
   widget=OnlyYouImage
   properties {brush {
     bIsDynamicallyLoaded=false
     drawAs=Image
     imageType=FullColor
     mirroring=NoMirror
     resourceName=None
     resourceObject=/Sidebar/Materials/Icons/M_UI_Sidebar_SoloParty
     tiling=NoTile
     imageSize=<48 48>
     margin=<ltrb 0 0 0 0>
     outlineSettings {
      bUseBrushTransparency=false
      roundingType=HalfHeightRadius
      width=0
      color {colorUseRule=UseColor_Specified, specifiedColor=<rgba 0 0 0 0>}
      cornerRadii=<0 0 0 0>
     }
     tintColor {colorUseRule=UseColor_Specified, specifiedColor=<rgba 1 1 1 1>}
     uVRegion {bIsValid=false, max=<0 0>, min=<0 0>}
    }}
   slot {class=CommonVisibilitySwitcherSlot, properties {horizontalAlignment=HAlign_Left, verticalAlignment=VAlign_Center}}
  }
  PlayerAvatar {
   widgetClassPath=/Sidebar/Panels/WBP_Sidebar_PlayerAvatar
   bIsVariable=true
   parent=SwitcherPartyIcon
   widget=PlayerAvatar
   properties {avatarSize=<48 48>}
   slot {class=CommonVisibilitySwitcherSlot, properties {verticalAlignment=VAlign_Center}}
  }
  Channel_Game_Button {
   widgetClassPath=&WBP_Sidebar_VoiceChannel_Button
   bIsVariable=true
   parent=Grid_ChannelSelect_CustomNavBoundary
   widget=Channel_Game_Button
   properties {
    cornerRadius=18
    groupLabelText=Game chat
    holdData=&CommonUI_HoldData_Base_C
    navigation=&self:WidgetTree.Channel_Game_Button.WidgetNavigation_0
   }
   slot {class=GridSlot, properties {column=1, layer=2, padding=<ltrb 3 0 0 0>}}
  }
  IslandTileMaterial {
   widgetClassPath=&Image
   bIsVariable=true
   namedSlotHost=Channel_Game_Button
   widget=IslandTileMaterial
   properties {brush {
     bIsDynamicallyLoaded=false
     drawAs=Image
     imageType=FullColor
     mirroring=NoMirror
     resourceName=None
     resourceObject=/Sidebar/Materials/MI_Sidebar_IslandTile
     tiling=NoTile
     imageSize=<48 48>
     margin=<ltrb 0 0 0 0>
     outlineSettings {
      bUseBrushTransparency=false
      roundingType=HalfHeightRadius
      width=0
      color {colorUseRule=UseColor_Specified, specifiedColor=<rgba 0 0 0 0>}
      cornerRadii=<0 0 0 0>
     }
     tintColor {colorUseRule=UseColor_Specified, specifiedColor=<rgba 1 1 1 1>}
     uVRegion {bIsValid=false, max=<0 0>, min=<0 0>}
    }}
  }
  Backplate_ChannelSelector {
   widgetClassPath=&Image
   bIsVariable=true
   parent=Grid_ChannelSelect_CustomNavBoundary
   widget=Backplate_ChannelSelector
   properties {brush {
     bIsDynamicallyLoaded=false
     drawAs=Image
     imageType=FullColor
     mirroring=NoMirror
     resourceName=None
     resourceObject=/Sidebar/Materials/MI_Sidebar_TabSwitcherBackplate_PartyAndGame
     tiling=NoTile
     imageSize=<32 32>
     margin=<ltrb 0 0 0 0>
     outlineSettings {
      bUseBrushTransparency=false
      roundingType=HalfHeightRadius
      width=0
      color {colorUseRule=UseColor_Specified, specifiedColor=<rgba 0 0 0 0>}
      cornerRadii=<0 0 0 0>
     }
     tintColor {colorUseRule=UseColor_Specified, specifiedColor=<rgba 1 1 1 1>}
     uVRegion {bIsValid=false, max=<0 0>, min=<0 0>}
    }}
   slot {class=GridSlot, properties {columnSpan=2, padding=<ltrb -6 -6 -6 -6>}}
  }
  SocialBan_Message {
   widgetClassPath=/FortniteCoreUI/UI/Social/SocialBan/WBP_SocialBan_Message
   bIsVariable=true
   parent=VerticalBox
   widget=SocialBan_Message
   slot {class=VerticalBoxSlot, properties {horizontalAlignment=HAlign_Center, padding=<ltrb 0 12 0 12>}}
  }
  VB_SubStates {
   widgetClassPath=&VerticalBox
   bIsVariable=true
   parent=VerticalBox
   widget=VB_SubStates
   properties {visibility=Collapsed}
   slot {class=VerticalBoxSlot, properties {padding=<ltrb 0 12 0 0>}}
  }
  BorderTinted_OtherSubMessage {
   widgetClassPath=&Border
   bIsVariable=true
   parent=VB_SubStates
   widget=BorderTinted_OtherSubMessage
   properties {
    background {
     bIsDynamicallyLoaded=false
     drawAs=Image
     imageType=NoImage
     mirroring=NoMirror
     resourceName=None
     resourceObject=null
     tiling=NoTile
     imageSize=<32 32>
     margin=<ltrb 0 0 0 0>
     outlineSettings {
      bUseBrushTransparency=false
      roundingType=HalfHeightRadius
      width=0
      color {colorUseRule=UseColor_Specified, specifiedColor=<rgba 0 0 0 0>}
      cornerRadii=<0 0 0 0>
     }
     tintColor {colorUseRule=UseColor_Specified, specifiedColor=<rgba 1 1 1 0>}
     uVRegion {bIsValid=false, max=<0 0>, min=<0 0>}
    }
    brushColor=<rgba 1 1 1 0>
    contentColorAndOpacity=<rgba 0 0.599 0.0009 1>
    padding=<ltrb 0 0 0 0>
   }
   slot {class=VerticalBoxSlot, properties {padding=<ltrb 0 0 0 12>, size {sizeRule=Fill, value=1}}}
  }
  HorizontalBox {
   widgetClassPath=&HorizontalBox
   parent=BorderTinted_OtherSubMessage
   widget=HorizontalBox
   slot {class=BorderSlot, properties {padding=<ltrb 0 0 0 0>}}
  }
  WarningIcon {
   widgetClassPath=&Image
   bIsVariable=true
   parent=HorizontalBox
   widget=WarningIcon
   properties {
    visibility=SelfHitTestInvisible
    brush {
     bIsDynamicallyLoaded=false
     drawAs=Image
     imageType=FullColor
     mirroring=NoMirror
     resourceName=None
     resourceObject=&MI_UIKit_Icon_Alert
     tiling=NoTile
     imageSize=<24 24>
     margin=<ltrb 0 0 0 0>
     outlineSettings {
      bUseBrushTransparency=false
      roundingType=HalfHeightRadius
      width=0
      color {colorUseRule=UseColor_Specified, specifiedColor=<rgba 0 0 0 0>}
      cornerRadii=<0 0 0 0>
     }
     tintColor {colorUseRule=UseColor_Specified, specifiedColor=<rgba 1 1 1 1>}
     uVRegion {bIsValid=false, max=<0 0>, min=<0 0>}
    }
   }
   slot {class=HorizontalBoxSlot, properties {verticalAlignment=VAlign_Top, padding=<ltrb 0 0 8 6>}}
  }
  SubState_RichPlatformIcon {
   widgetClassPath=&CommonRichTextBlock
   bIsVariable=true
   parent=HorizontalBox
   widget=SubState_RichPlatformIcon
   properties {
    bOverrideDefaultStyle=true
    clipping=ClipToBounds
    mobileTextBlockScale=1.5
    text="<mobile/>"
    wrapTextAt=500
    defaultTextStyleOverride {
     overflowPolicy=Clip
     transformPolicy=None
     colorAndOpacity {colorUseRule=UseColor_Specified, specifiedColor=<rgba 1 1 1 1>}
     font {
      bForceMonospaced=false
      bMaterialIsStencil=false
      fontMaterial=/Sidebar/Materials/Fonts/MI_Sidebar_FontColor_Secondary
      fontObject=&HeadingNow
      letterSpacing=10
      monospacedWidth=1
      size=12
      skewAmount=0
      typefaceFontName=74 Regular
      outlineSettings {
       bApplyOutlineToDropShadows=false
       bMiteredCorners=false
       bSeparateFillAlpha=false
       outlineMaterial=null
       outlineSize=0
       outlineColor=<rgba 0 0 0 1>
      }
     }
     highlightColor {colorUseRule=UseColor_Specified, specifiedColor=<rgba 0 0 0 1>}
     highlightShape {
      bIsDynamicallyLoaded=false
      drawAs=Image
      imageType=NoImage
      mirroring=NoMirror
      resourceName=None
      resourceObject=null
      tiling=NoTile
      imageSize=<32 32>
      margin=<ltrb 0 0 0 0>
      outlineSettings {
       bUseBrushTransparency=false
       roundingType=HalfHeightRadius
       width=0
       color {colorUseRule=UseColor_Specified, specifiedColor=<rgba 0 0 0 0>}
       cornerRadii=<0 0 0 0>
      }
      tintColor {colorUseRule=UseColor_Specified, specifiedColor=<rgba 1 1 1 1>}
      uVRegion {bIsValid=false, max=<0 0>, min=<0 0>}
     }
     selectedBackgroundColor {colorUseRule=UseColor_Specified, specifiedColor=<rgba 1 1 1 1>}
     shadowColorAndOpacity=<rgba 0 0 0 1>
     shadowOffset=<0 0>
     strikeBrush {
      bIsDynamicallyLoaded=false
      drawAs=Image
      imageType=NoImage
      mirroring=NoMirror
      resourceName=None
      resourceObject=null
      tiling=NoTile
      imageSize=<32 32>
      margin=<ltrb 0 0 0 0>
      outlineSettings {
       bUseBrushTransparency=false
       roundingType=HalfHeightRadius
       width=0
       color {colorUseRule=UseColor_Specified, specifiedColor=<rgba 0 0 0 0>}
       cornerRadii=<0 0 0 0>
      }
      tintColor {colorUseRule=UseColor_Specified, specifiedColor=<rgba 1 1 1 1>}
      uVRegion {bIsValid=false, max=<0 0>, min=<0 0>}
     }
     underlineBrush {
      bIsDynamicallyLoaded=false
      drawAs=Image
      imageType=NoImage
      mirroring=NoMirror
      resourceName=None
      resourceObject=null
      tiling=NoTile
      imageSize=<32 32>
      margin=<ltrb 0 0 0 0>
      outlineSettings {
       bUseBrushTransparency=false
       roundingType=HalfHeightRadius
       width=0
       color {colorUseRule=UseColor_Specified, specifiedColor=<rgba 0 0 0 0>}
       cornerRadii=<0 0 0 0>
      }
      tintColor {colorUseRule=UseColor_Specified, specifiedColor=<rgba 1 1 1 1>}
      uVRegion {bIsValid=false, max=<0 0>, min=<0 0>}
     }
    }
   }
   slot {class=HorizontalBoxSlot, properties {horizontalAlignment=HAlign_Center, verticalAlignment=VAlign_Center, padding=<ltrb 0 0 4 0>}}
  }
  SubState_RichMessage {
   widgetClassPath=&CommonRichTextBlock
   bIsVariable=true
   parent=HorizontalBox
   widget=SubState_RichMessage
   properties {
    bOverrideDefaultStyle=true
    clipping=ClipToBounds
    inlineIconDisplayMode=TextOnly
    mobileTextBlockScale=1.33
    scrollStyle=&TextScrollStyle-Base_C
    text=Your Parent or Guardian has turned off voic qwdasd asdas das dasd as as d asasde chat
    defaultTextStyleOverride {
     overflowPolicy=Clip
     transformPolicy=None
     colorAndOpacity {colorUseRule=UseColor_Specified, specifiedColor=<rgba 1 1 1 1>}
     font {
      bForceMonospaced=false
      bMaterialIsStencil=false
      fontMaterial=&MI_Sidebar_FontColor_Secondary
      fontObject=&HeadingNow
      letterSpacing=10
      monospacedWidth=1
      size=13.5
      skewAmount=0
      typefaceFontName=74 Regular
      outlineSettings {
       bApplyOutlineToDropShadows=false
       bMiteredCorners=false
       bSeparateFillAlpha=false
       outlineMaterial=null
       outlineSize=0
       outlineColor=<rgba 0 0 0 1>
      }
     }
     highlightColor {colorUseRule=UseColor_Specified, specifiedColor=<rgba 0 0 0 1>}
     highlightShape {
      bIsDynamicallyLoaded=false
      drawAs=Image
      imageType=NoImage
      mirroring=NoMirror
      resourceName=None
      resourceObject=null
      tiling=NoTile
      imageSize=<32 32>
      margin=<ltrb 0 0 0 0>
      outlineSettings {
       bUseBrushTransparency=false
       roundingType=HalfHeightRadius
       width=0
       color {colorUseRule=UseColor_Specified, specifiedColor=<rgba 0 0 0 0>}
       cornerRadii=<0 0 0 0>
      }
      tintColor {colorUseRule=UseColor_Specified, specifiedColor=<rgba 1 1 1 1>}
      uVRegion {bIsValid=false, max=<0 0>, min=<0 0>}
     }
     selectedBackgroundColor {colorUseRule=UseColor_Specified, specifiedColor=<rgba 1 1 1 1>}
     shadowColorAndOpacity=<rgba 0 0 0 1>
     shadowOffset=<0 0>
     strikeBrush {
      bIsDynamicallyLoaded=false
      drawAs=Image
      imageType=NoImage
      mirroring=NoMirror
      resourceName=None
      resourceObject=null
      tiling=NoTile
      imageSize=<32 32>
      margin=<ltrb 0 0 0 0>
      outlineSettings {
       bUseBrushTransparency=false
       roundingType=HalfHeightRadius
       width=0
       color {colorUseRule=UseColor_Specified, specifiedColor=<rgba 0 0 0 0>}
       cornerRadii=<0 0 0 0>
      }
      tintColor {colorUseRule=UseColor_Specified, specifiedColor=<rgba 1 1 1 1>}
      uVRegion {bIsValid=false, max=<0 0>, min=<0 0>}
     }
     underlineBrush {
      bIsDynamicallyLoaded=false
      drawAs=Image
      imageType=NoImage
      mirroring=NoMirror
      resourceName=None
      resourceObject=null
      tiling=NoTile
      imageSize=<32 32>
      margin=<ltrb 0 0 0 0>
      outlineSettings {
       bUseBrushTransparency=false
       roundingType=HalfHeightRadius
       width=0
       color {colorUseRule=UseColor_Specified, specifiedColor=<rgba 0 0 0 0>}
       cornerRadii=<0 0 0 0>
      }
      tintColor {colorUseRule=UseColor_Specified, specifiedColor=<rgba 1 1 1 1>}
      uVRegion {bIsValid=false, max=<0 0>, min=<0 0>}
     }
    }
   }
   slot {class=HorizontalBoxSlot, properties {size {sizeRule=Fill, value=1}}}
  }
  HorizontalBox_0 {
   widgetClassPath=&HorizontalBox
   parent=VB_SubStates
   widget=HorizontalBox_0
   slot {class=VerticalBoxSlot, properties {padding=<ltrb 0 0 0 12>}}
  }
  Substate_Action_Button {
   widgetClassPath=/Sidebar/Buttons/WBP_Sidebar_ButtonGeneric
   bIsVariable=true
   parent=HorizontalBox_0
   widget=Substate_Action_Button
   properties {
    minHeightOverride=60
    navigation=&self:WidgetTree.Substate_Action_Button.WidgetNavigation_0
    text=Sub State Action
   }
   slot {class=HorizontalBoxSlot, properties {size {sizeRule=Fill, value=1}}}
  }
  Substate_MoreOptions_Button {
   widgetClassPath=&WBP_Sidebar_ButtonGeneric
   bIsVariable=true
   parent=HorizontalBox_0
   widget=Substate_MoreOptions_Button
   properties {
    icon=/Sidebar/Materials/Icons/Textures/T_UI_Icon_Sidebar_ContextMenu_MSDF
    iconMaterial=/Sidebar/Materials/Icons/MI_Sidebar_Icon_PartyOptions
    navigation=&self:WidgetTree.Substate_MoreOptions_Button.WidgetNavigation_0
    useIconBlock=true
    useTextBlock=false
    buttonPadding=<ltrb 24 12 24 12>
    font {
     bForceMonospaced=false
     bMaterialIsStencil=false
     fontMaterial=/FortUILibrary/UIKit/Buttons/MaterialInstances/CTAButtons/MI_UIKit_TextColor_Quiet
     fontObject=&HeadingNow
     letterSpacing=10
     monospacedWidth=1
     size=15
     skewAmount=0
     typefaceFontName=74 Regular
     outlineSettings {
      bApplyOutlineToDropShadows=false
      bMiteredCorners=false
      bSeparateFillAlpha=false
      outlineMaterial=/FortUILibrary/Systems/ModularButton/MaterialInstances/MI_UIKit_TextOutline
      outlineSize=0
      outlineColor=<rgba 0 0 0 1>
     }
    }
    iconSize=<36 36>
   }
   slot {class=HorizontalBoxSlot, properties {padding=<ltrb 12 0 0 0>}}
  }
  Overlay_2 {widgetClassPath=&Overlay, namedSlotHost=Substate_MoreOptions_Button, widget=Overlay_2}
  Substate_MoreOptions_ContextMenu {
   widgetClassPath=/FortUILibrary/UIKit/ContextMenu/WBP_UIKit_ContextMenuAnchorButton
   bIsVariable=true
   parent=Overlay_2
   widget=Substate_MoreOptions_ContextMenu
   properties {
    button Background Material=null
    button Outline Material=null
    buttonClass=/Game/UI/Frontend/Lobby/Nameplates/WBP_EmptyButton_ContextMenu.WBP_EmptyButton_ContextMenu_C
    buttonText=""
    closeMenuOnContextClicked=true
    isMenuModal=false
    menuPlacement=MenuPlacement_MenuLeft
    min Height=1
    button Padding=<ltrb 0 0 0 0>
   }
   slot {class=OverlaySlot, properties {horizontalAlignment=HAlign_Center, verticalAlignment=VAlign_Bottom}}
  }
  VB_EpicPartyPreviewEnabled {
   widgetClassPath=&VerticalBox
   bIsVariable=true
   parent=VerticalBox
   widget=VB_EpicPartyPreviewEnabled
   properties {visibility=Collapsed}
   slot {class=VerticalBoxSlot, properties {padding=<ltrb 0 12 0 0>}}
  }
  BorderTinted_OtherSubMessage_1 {
   widgetClassPath=&Border
   bIsVariable=true
   parent=VB_EpicPartyPreviewEnabled
   widget=BorderTinted_OtherSubMessage_1
   properties {
    background {
     bIsDynamicallyLoaded=false
     drawAs=Image
     imageType=NoImage
     mirroring=NoMirror
     resourceName=None
     resourceObject=null
     tiling=NoTile
     imageSize=<32 32>
     margin=<ltrb 0 0 0 0>
     outlineSettings {
      bUseBrushTransparency=false
      roundingType=HalfHeightRadius
      width=0
      color {colorUseRule=UseColor_Specified, specifiedColor=<rgba 0 0 0 0>}
      cornerRadii=<0 0 0 0>
     }
     tintColor {colorUseRule=UseColor_Specified, specifiedColor=<rgba 1 1 1 0>}
     uVRegion {bIsValid=false, max=<0 0>, min=<0 0>}
    }
    brushColor=<rgba 1 1 1 0>
    contentColorAndOpacity=<rgba 0.599 0 0.4445 1>
    padding=<ltrb 0 0 0 0>
   }
   slot {class=VerticalBoxSlot, properties {padding=<ltrb 0 0 0 12>, size {sizeRule=Fill, value=1}}}
  }
  HorizontalBox_1 {
   widgetClassPath=&HorizontalBox
   parent=BorderTinted_OtherSubMessage_1
   widget=HorizontalBox_1
   slot {class=BorderSlot, properties {padding=<ltrb 0 0 0 0>}}
  }
  WarningIcon_1 {
   widgetClassPath=&Image
   bIsVariable=true
   parent=HorizontalBox_1
   widget=WarningIcon_1
   properties {
    visibility=SelfHitTestInvisible
    brush {
     bIsDynamicallyLoaded=false
     drawAs=Image
     imageType=FullColor
     mirroring=NoMirror
     resourceName=None
     resourceObject=&MI_UIKit_Icon_Alert
     tiling=NoTile
     imageSize=<24 24>
     margin=<ltrb 0 0 0 0>
     outlineSettings {
      bUseBrushTransparency=false
      roundingType=HalfHeightRadius
      width=0
      color {colorUseRule=UseColor_Specified, specifiedColor=<rgba 0 0 0 0>}
      cornerRadii=<0 0 0 0>
     }
     tintColor {colorUseRule=UseColor_Specified, specifiedColor=<rgba 1 1 1 1>}
     uVRegion {bIsValid=false, max=<0 0>, min=<0 0>}
    }
   }
   slot {class=HorizontalBoxSlot, properties {verticalAlignment=VAlign_Top, padding=<ltrb 0 0 8 6>}}
  }
  SubState_RichMessage_1 {
   widgetClassPath=&CommonRichTextBlock
   bIsVariable=true
   parent=HorizontalBox_1
   widget=SubState_RichMessage_1
   properties {
    bOverrideDefaultStyle=true
    clipping=ClipToBounds
    inlineIconDisplayMode=TextOnly
    mobileTextBlockScale=1.33
    scrollStyle=&TextScrollStyle-Base_C
    text=Epic Parties Preview Active
    defaultTextStyleOverride {
     overflowPolicy=Clip
     transformPolicy=None
     colorAndOpacity {colorUseRule=UseColor_Specified, specifiedColor=<rgba 1 1 1 1>}
     font {
      bForceMonospaced=false
      bMaterialIsStencil=false
      fontMaterial=&MI_Sidebar_FontColor_Secondary
      fontObject=&HeadingNow
      letterSpacing=10
      monospacedWidth=1
      size=13.5
      skewAmount=0
      typefaceFontName=74 Regular
      outlineSettings {
       bApplyOutlineToDropShadows=false
       bMiteredCorners=false
       bSeparateFillAlpha=false
       outlineMaterial=null
       outlineSize=0
       outlineColor=<rgba 0 0 0 1>
      }
     }
     highlightColor {colorUseRule=UseColor_Specified, specifiedColor=<rgba 0 0 0 1>}
     highlightShape {
      bIsDynamicallyLoaded=false
      drawAs=Image
      imageType=NoImage
      mirroring=NoMirror
      resourceName=None
      resourceObject=null
      tiling=NoTile
      imageSize=<32 32>
      margin=<ltrb 0 0 0 0>
      outlineSettings {
       bUseBrushTransparency=false
       roundingType=HalfHeightRadius
       width=0
       color {colorUseRule=UseColor_Specified, specifiedColor=<rgba 0 0 0 0>}
       cornerRadii=<0 0 0 0>
      }
      tintColor {colorUseRule=UseColor_Specified, specifiedColor=<rgba 1 1 1 1>}
      uVRegion {bIsValid=false, max=<0 0>, min=<0 0>}
     }
     selectedBackgroundColor {colorUseRule=UseColor_Specified, specifiedColor=<rgba 1 1 1 1>}
     shadowColorAndOpacity=<rgba 0 0 0 1>
     shadowOffset=<0 0>
     strikeBrush {
      bIsDynamicallyLoaded=false
      drawAs=Image
      imageType=NoImage
      mirroring=NoMirror
      resourceName=None
      resourceObject=null
      tiling=NoTile
      imageSize=<32 32>
      margin=<ltrb 0 0 0 0>
      outlineSettings {
       bUseBrushTransparency=false
       roundingType=HalfHeightRadius
       width=0
       color {colorUseRule=UseColor_Specified, specifiedColor=<rgba 0 0 0 0>}
       cornerRadii=<0 0 0 0>
      }
      tintColor {colorUseRule=UseColor_Specified, specifiedColor=<rgba 1 1 1 1>}
      uVRegion {bIsValid=false, max=<0 0>, min=<0 0>}
     }
     underlineBrush {
      bIsDynamicallyLoaded=false
      drawAs=Image
      imageType=NoImage
      mirroring=NoMirror
      resourceName=None
      resourceObject=null
      tiling=NoTile
      imageSize=<32 32>
      margin=<ltrb 0 0 0 0>
      outlineSettings {
       bUseBrushTransparency=false
       roundingType=HalfHeightRadius
       width=0
       color {colorUseRule=UseColor_Specified, specifiedColor=<rgba 0 0 0 0>}
       cornerRadii=<0 0 0 0>
      }
      tintColor {colorUseRule=UseColor_Specified, specifiedColor=<rgba 1 1 1 1>}
      uVRegion {bIsValid=false, max=<0 0>, min=<0 0>}
     }
    }
   }
   slot {class=HorizontalBoxSlot, properties {size {sizeRule=Fill, value=1}}}
  }
  BorderTinted_OtherSubMessage_2 {
   widgetClassPath=&Border
   bIsVariable=true
   parent=VB_EpicPartyPreviewEnabled
   widget=BorderTinted_OtherSubMessage_2
   properties {
    background {
     bIsDynamicallyLoaded=false
     drawAs=Image
     imageType=NoImage
     mirroring=NoMirror
     resourceName=None
     resourceObject=null
     tiling=NoTile
     imageSize=<32 32>
     margin=<ltrb 0 0 0 0>
     outlineSettings {
      bUseBrushTransparency=false
      roundingType=HalfHeightRadius
      width=0
      color {colorUseRule=UseColor_Specified, specifiedColor=<rgba 0 0 0 0>}
      cornerRadii=<0 0 0 0>
     }
     tintColor {colorUseRule=UseColor_Specified, specifiedColor=<rgba 1 1 1 0>}
     uVRegion {bIsValid=false, max=<0 0>, min=<0 0>}
    }
    brushColor=<rgba 1 1 1 0>
    contentColorAndOpacity=<rgba 0.599 0 0.4445 1>
    padding=<ltrb 0 0 0 0>
   }
   slot {class=VerticalBoxSlot, properties {padding=<ltrb 0 0 0 12>, size {sizeRule=Fill, value=1}}}
  }
  HorizontalBox_3 {
   widgetClassPath=&HorizontalBox
   parent=BorderTinted_OtherSubMessage_2
   widget=HorizontalBox_3
   slot {class=BorderSlot, properties {padding=<ltrb 0 0 0 0>}}
  }
  WarningIcon_3 {
   widgetClassPath=&Image
   bIsVariable=true
   parent=HorizontalBox_3
   widget=WarningIcon_3
   properties {
    visibility=SelfHitTestInvisible
    brush {
     bIsDynamicallyLoaded=false
     drawAs=Image
     imageType=FullColor
     mirroring=NoMirror
     resourceName=None
     resourceObject=&MI_UIKit_Icon_Alert
     tiling=NoTile
     imageSize=<24 24>
     margin=<ltrb 0 0 0 0>
     outlineSettings {
      bUseBrushTransparency=false
      roundingType=HalfHeightRadius
      width=0
      color {colorUseRule=UseColor_Specified, specifiedColor=<rgba 0 0 0 0>}
      cornerRadii=<0 0 0 0>
     }
     tintColor {colorUseRule=UseColor_Specified, specifiedColor=<rgba 1 1 1 1>}
     uVRegion {bIsValid=false, max=<0 0>, min=<0 0>}
    }
   }
   slot {class=HorizontalBoxSlot, properties {verticalAlignment=VAlign_Top, padding=<ltrb 0 0 8 6>}}
  }
  SubState_RichMessage_3 {
   widgetClassPath=&CommonRichTextBlock
   bIsVariable=true
   parent=HorizontalBox_3
   widget=SubState_RichMessage_3
   properties {
    bOverrideDefaultStyle=true
    clipping=ClipToBounds
    inlineIconDisplayMode=TextOnly
    mobileTextBlockScale=1.33
    scrollStyle=&TextScrollStyle-Base_C
    text=Issues playing with friends? Disable via ... menu below
    defaultTextStyleOverride {
     overflowPolicy=Clip
     transformPolicy=None
     colorAndOpacity {colorUseRule=UseColor_Specified, specifiedColor=<rgba 1 1 1 1>}
     font {
      bForceMonospaced=false
      bMaterialIsStencil=false
      fontMaterial=&MI_Sidebar_FontColor_Secondary
      fontObject=&HeadingNow
      letterSpacing=10
      monospacedWidth=1
      size=13.5
      skewAmount=0
      typefaceFontName=74 Regular
      outlineSettings {
       bApplyOutlineToDropShadows=false
       bMiteredCorners=false
       bSeparateFillAlpha=false
       outlineMaterial=null
       outlineSize=0
       outlineColor=<rgba 0 0 0 1>
      }
     }
     highlightColor {colorUseRule=UseColor_Specified, specifiedColor=<rgba 0 0 0 1>}
     highlightShape {
      bIsDynamicallyLoaded=false
      drawAs=Image
      imageType=NoImage
      mirroring=NoMirror
      resourceName=None
      resourceObject=null
      tiling=NoTile
      imageSize=<32 32>
      margin=<ltrb 0 0 0 0>
      outlineSettings {
       bUseBrushTransparency=false
       roundingType=HalfHeightRadius
       width=0
       color {colorUseRule=UseColor_Specified, specifiedColor=<rgba 0 0 0 0>}
       cornerRadii=<0 0 0 0>
      }
      tintColor {colorUseRule=UseColor_Specified, specifiedColor=<rgba 1 1 1 1>}
      uVRegion {bIsValid=false, max=<0 0>, min=<0 0>}
     }
     selectedBackgroundColor {colorUseRule=UseColor_Specified, specifiedColor=<rgba 1 1 1 1>}
     shadowColorAndOpacity=<rgba 0 0 0 1>
     shadowOffset=<0 0>
     strikeBrush {
      bIsDynamicallyLoaded=false
      drawAs=Image
      imageType=NoImage
      mirroring=NoMirror
      resourceName=None
      resourceObject=null
      tiling=NoTile
      imageSize=<32 32>
      margin=<ltrb 0 0 0 0>
      outlineSettings {
       bUseBrushTransparency=false
       roundingType=HalfHeightRadius
       width=0
       color {colorUseRule=UseColor_Specified, specifiedColor=<rgba 0 0 0 0>}
       cornerRadii=<0 0 0 0>
      }
      tintColor {colorUseRule=UseColor_Specified, specifiedColor=<rgba 1 1 1 1>}
      uVRegion {bIsValid=false, max=<0 0>, min=<0 0>}
     }
     underlineBrush {
      bIsDynamicallyLoaded=false
      drawAs=Image
      imageType=NoImage
      mirroring=NoMirror
      resourceName=None
      resourceObject=null
      tiling=NoTile
      imageSize=<32 32>
      margin=<ltrb 0 0 0 0>
      outlineSettings {
       bUseBrushTransparency=false
       roundingType=HalfHeightRadius
       width=0
       color {colorUseRule=UseColor_Specified, specifiedColor=<rgba 0 0 0 0>}
       cornerRadii=<0 0 0 0>
      }
      tintColor {colorUseRule=UseColor_Specified, specifiedColor=<rgba 1 1 1 1>}
      uVRegion {bIsValid=false, max=<0 0>, min=<0 0>}
     }
    }
   }
   slot {class=HorizontalBoxSlot, properties {size {sizeRule=Fill, value=1}}}
  }
  CommonVisualAttachment_247 {
   widgetClassPath=&CommonVisualAttachment
   bIsVariable=true
   parent=VerticalBox
   widget=CommonVisualAttachment_247
   properties {contentAnchor=<1 -0.5>}
   slot {class=VerticalBoxSlot, properties {horizontalAlignment=HAlign_Left}}
  }
  WBP_VoiceChat_ControlsMoved_CoachMark {
   widgetClassPath=/FortniteCoreUI/UI/Social/Chat/TextChat/Coachmarks/WBP_VoiceChat_ControlsMoved_CoachMark
   bIsVariable=true
   parent=CommonVisualAttachment_247
   widget=WBP_VoiceChat_ControlsMoved_CoachMark
   properties {
    bIsFocusable=true
    navigation=&self:WidgetTree.WBP_VoiceChat_ControlsMoved_CoachMark.WidgetNavigation_0
    padding=<ltrb 0 -100 0 0>
   }
   slot {class=SizeBoxSlot, properties {padding=<ltrb 0 0 60 0>}}
  }
  VB_Controls {
   widgetClassPath=&VerticalBox
   bIsVariable=true
   parent=VerticalBox
   widget=VB_Controls
   slot {class=VerticalBoxSlot, properties {padding=<ltrb 0 12 0 0>}}
  }
  HB_Buttons {
   widgetClassPath=&HorizontalBox
   parent=VB_Controls
   widget=HB_Buttons
   properties {bOverride_Cursor=true}
   slot {class=VerticalBoxSlot, properties {padding=<ltrb 0 8 0 8>}}
  }
  Button_PartyJoinability {
   widgetClassPath=&WBP_Sidebar_ButtonGeneric
   bIsVariable=true
   parent=HB_Buttons
   widget=Button_PartyJoinability
   properties {
    icon=&T_UI_Icon_Sidebar_Locked_MSDF
    iconMaterial=&MI_Sidebar_Icon_PartyOptions
    navigation=&self:WidgetTree.Button_PartyJoinability.WidgetNavigation_0
    useIconBlock=true
    useTextBlock=false
    buttonPadding=<ltrb 24 12 24 12>
    font {
     bForceMonospaced=false
     bMaterialIsStencil=false
     fontMaterial=&MI_UIKit_TextColor_Quiet
     fontObject=&HeadingNow
     letterSpacing=10
     monospacedWidth=1
     size=15
     skewAmount=0
     typefaceFontName=74 Regular
     outlineSettings {
      bApplyOutlineToDropShadows=false
      bMiteredCorners=false
      bSeparateFillAlpha=false
      outlineMaterial=&MI_UIKit_TextOutline
      outlineSize=0
      outlineColor=<rgba 0 0 0 1>
     }
    }
    iconSize=<36 36>
   }
   slot {class=HorizontalBoxSlot, properties {padding=<ltrb 0 0 12 0>, size {sizeRule=Fill, value=1}}}
  }
  Button_Mic {
   widgetClassPath=&WBP_Sidebar_ButtonGeneric
   bIsVariable=true
   parent=HB_Buttons
   widget=Button_Mic
   properties {
    icon=&T_UI_Icon_Sidebar_Mic_MSDF
    iconMaterial=&MI_Sidebar_Icon_PartyOptions
    navigation=&self:WidgetTree.Button_Mic.WidgetNavigation_0
    useIconBlock=true
    useTextBlock=false
    buttonPadding=<ltrb 24 12 24 12>
    font {
     bForceMonospaced=false
     bMaterialIsStencil=false
     fontMaterial=&MI_UIKit_TextColor_Quiet
     fontObject=&HeadingNow
     letterSpacing=10
     monospacedWidth=1
     size=15
     skewAmount=0
     typefaceFontName=74 Regular
     outlineSettings {
      bApplyOutlineToDropShadows=false
      bMiteredCorners=false
      bSeparateFillAlpha=false
      outlineMaterial=&MI_UIKit_TextOutline
      outlineSize=0
      outlineColor=<rgba 0 0 0 1>
     }
    }
    iconSize=<36 36>
   }
   slot {class=HorizontalBoxSlot, properties {padding=<ltrb 0 0 12 0>, size {sizeRule=Fill, value=1}}}
  }
  Overlay_28 {widgetClassPath=&Overlay, namedSlotHost=Button_Mic, widget=Overlay_28}
  Microphone_ContextMenu {
   widgetClassPath=&WBP_UIKit_ContextMenuAnchorButton
   bIsVariable=true
   parent=Overlay_28
   widget=Microphone_ContextMenu
   properties {
    button Background Material=null
    button Outline Material=null
    buttonClass=&WBP_EmptyButton_ContextMenu_C
    buttonText=""
    closeMenuOnContextClicked=true
    isMenuModal=false
    menuPlacement=MenuPlacement_BelowAnchor
    min Height=1
    button Padding=<ltrb 0 0 0 0>
   }
   slot {class=OverlaySlot, properties {horizontalAlignment=HAlign_Center, verticalAlignment=VAlign_Bottom}}
  }
  PTT_Key {
   widgetClassPath=&Image
   bIsVariable=true
   parent=Overlay_28
   widget=PTT_Key
   properties {brush {
     bIsDynamicallyLoaded=false
     drawAs=Image
     imageType=FullColor
     mirroring=NoMirror
     resourceName=None
     resourceObject=null
     tiling=NoTile
     imageSize=<28 28>
     margin=<ltrb 0 0 0 0>
     outlineSettings {
      bUseBrushTransparency=false
      roundingType=HalfHeightRadius
      width=0
      color {colorUseRule=UseColor_Specified, specifiedColor=<rgba 0 0 0 0>}
      cornerRadii=<0 0 0 0>
     }
     tintColor {colorUseRule=UseColor_Specified, specifiedColor=<rgba 1 1 1 1>}
     uVRegion {bIsValid=false, max=<0 0>, min=<0 0>}
    }}
   slot {class=OverlaySlot, properties {horizontalAlignment=HAlign_Center, verticalAlignment=VAlign_Center, padding=<ltrb 24 0 0 0>}}
  }
  Button_Headphones {
   widgetClassPath=&WBP_Sidebar_ButtonGeneric
   bIsVariable=true
   parent=HB_Buttons
   widget=Button_Headphones
   properties {
    icon=&T_UI_Icon_Sidebar_Headphones_MSDF
    iconMaterial=&MI_Sidebar_Icon_PartyOptions
    navigation=&self:WidgetTree.Button_Headphones.WidgetNavigation_0
    useIconBlock=true
    useTextBlock=false
    buttonPadding=<ltrb 24 12 24 12>
    font {
     bForceMonospaced=false
     bMaterialIsStencil=false
     fontMaterial=&MI_UIKit_TextColor_Quiet
     fontObject=&HeadingNow
     letterSpacing=10
     monospacedWidth=1
     size=15
     skewAmount=0
     typefaceFontName=74 Regular
     outlineSettings {
      bApplyOutlineToDropShadows=false
      bMiteredCorners=false
      bSeparateFillAlpha=false
      outlineMaterial=&MI_UIKit_TextOutline
      outlineSize=0
      outlineColor=<rgba 0 0 0 1>
     }
    }
    iconSize=<36 36>
   }
   slot {class=HorizontalBoxSlot, properties {padding=<ltrb 0 0 12 0>, size {sizeRule=Fill, value=1}}}
  }
  Overlay {widgetClassPath=&Overlay, namedSlotHost=Button_Headphones, widget=Overlay}
  Headphone_ContextMenu {
   widgetClassPath=&WBP_UIKit_ContextMenuAnchorButton
   bIsVariable=true
   parent=Overlay
   widget=Headphone_ContextMenu
   properties {
    button Background Material=null
    button Outline Material=null
    buttonClass=&WBP_EmptyButton_ContextMenu_C
    buttonText=""
    closeMenuOnContextClicked=true
    isMenuModal=false
    menuPlacement=MenuPlacement_BelowAnchor
    min Height=1
    button Padding=<ltrb 0 0 0 0>
   }
   slot {class=OverlaySlot, properties {horizontalAlignment=HAlign_Center, verticalAlignment=VAlign_Bottom}}
  }
  Button_MoreOptions {
   widgetClassPath=&WBP_Sidebar_ButtonGeneric
   bIsVariable=true
   parent=HB_Buttons
   widget=Button_MoreOptions
   properties {
    icon=&T_UI_Icon_Sidebar_ContextMenu_MSDF
    iconMaterial=&MI_Sidebar_Icon_PartyOptions
    navigation=&self:WidgetTree.Button_MoreOptions.WidgetNavigation_0
    useIconBlock=true
    useTextBlock=false
    buttonPadding=<ltrb 24 12 24 12>
    font {
     bForceMonospaced=false
     bMaterialIsStencil=false
     fontMaterial=&MI_UIKit_TextColor_Quiet
     fontObject=&HeadingNow
     letterSpacing=10
     monospacedWidth=1
     size=15
     skewAmount=0
     typefaceFontName=74 Regular
     outlineSettings {
      bApplyOutlineToDropShadows=false
      bMiteredCorners=false
      bSeparateFillAlpha=false
      outlineMaterial=&MI_UIKit_TextOutline
      outlineSize=0
      outlineColor=<rgba 0 0 0 1>
     }
    }
    iconSize=<36 36>
   }
   slot {class=HorizontalBoxSlot, properties {size {sizeRule=Fill, value=1}}}
  }
  Overlay_1 {widgetClassPath=&Overlay, namedSlotHost=Button_MoreOptions, widget=Overlay_1}
  Report_ContextMenu {
   widgetClassPath=&WBP_UIKit_ContextMenuAnchorButton
   bIsVariable=true
   parent=Overlay_1
   widget=Report_ContextMenu
   properties {
    button Background Material=null
    button Outline Material=null
    buttonClass=&WBP_EmptyButton_ContextMenu_C
    buttonText=""
    closeMenuOnContextClicked=true
    isMenuModal=false
    menuPlacement=MenuPlacement_MenuLeft
    min Height=1
    button Padding=<ltrb 0 0 0 0>
   }
   slot {class=OverlaySlot, properties {horizontalAlignment=HAlign_Center, verticalAlignment=VAlign_Bottom}}
  }
  Deafen_Message {
   widgetClassPath=&CommonTextBlock
   bIsVariable=true
   parent=VB_Controls
   widget=Deafen_Message
   properties {
    autoWrapText=true
    justification=Center
    mobileFontSizeMultiplier=1.41
    text="You can't hear others, and your mic is muted"
    colorAndOpacity {colorUseRule=UseColor_Specified, specifiedColor=<rgba 1 0.147 0.1329 1>}
    font {
     bForceMonospaced=false
     bMaterialIsStencil=false
     fontMaterial=null
     fontObject=&HeadingNow
     letterSpacing=0
     monospacedWidth=1
     size=12.75
     skewAmount=0
     typefaceFontName=74 Regular
     outlineSettings {
      bApplyOutlineToDropShadows=false
      bMiteredCorners=false
      bSeparateFillAlpha=false
      outlineMaterial=null
      outlineSize=0
      outlineColor=<rgba 0 0 0 1>
     }
    }
    shadowOffset=<0 0>
   }
   slot {class=VerticalBoxSlot, properties {padding=<ltrb 0 10 0 6>}}
  }
  WB_ReportingStatus {
   widgetClassPath=&WrapBox
   parent=VB_Controls
   widget=WB_ReportingStatus
   properties {bOverride_Cursor=true, clipping=ClipToBounds, horizontalAlignment=HAlign_Center}
   slot {class=VerticalBoxSlot, properties {verticalAlignment=VAlign_Top, padding=<ltrb 4 8 4 15>}}
  }
  PartyJoinable_Text_Status {
   widgetClassPath=&CommonTextBlock
   bIsVariable=true
   parent=WB_ReportingStatus
   widget=PartyJoinable_Text_Status
   properties {
    justification=Center
    mobileFontSizeMultiplier=1.41
    text=Invite only
    font {
     bForceMonospaced=false
     bMaterialIsStencil=false
     fontMaterial=&MI_Sidebar_FontColor_Secondary
     fontObject=&HeadingNow
     letterSpacing=10
     monospacedWidth=1
     size=12.75
     skewAmount=0
     typefaceFontName=74 Regular
     outlineSettings {
      bApplyOutlineToDropShadows=false
      bMiteredCorners=false
      bSeparateFillAlpha=false
      outlineMaterial=null
      outlineSize=0
      outlineColor=<rgba 0 0 0 1>
     }
    }
    shadowOffset=<0 0>
   }
   slot {class=WrapBoxSlot, properties {horizontalAlignment=HAlign_Left, verticalAlignment=VAlign_Center}}
  }
  PartyJoinable_Icon_DividerDot {
   widgetClassPath=&Image
   bIsVariable=true
   parent=WB_ReportingStatus
   widget=PartyJoinable_Icon_DividerDot
   properties {
    visibility=SelfHitTestInvisible
    brush {
     bIsDynamicallyLoaded=false
     drawAs=Image
     imageType=FullColor
     mirroring=NoMirror
     resourceName=None
     resourceObject=/FortUILibrary/Foundation/Art_ToolKit/Tools/MI_UI_Circle_ConversationBacchusConfirm
     tiling=NoTile
     imageSize=<6 6>
     margin=<ltrb 0 0 0 0>
     outlineSettings {
      bUseBrushTransparency=false
      roundingType=HalfHeightRadius
      width=0
      color {colorUseRule=UseColor_Specified, specifiedColor=<rgba 0 0 0 0>}
      cornerRadii=<0 0 0 0>
     }
     tintColor {colorUseRule=UseColor_Specified, specifiedColor=<rgba 1 1 1 1>}
     uVRegion {bIsValid=false, max=<0 0>, min=<0 0>}
    }
    colorAndOpacity=<rgba 0.5776 0.5776 0.5776 1>
   }
   slot {class=WrapBoxSlot, properties {horizontalAlignment=HAlign_Left, verticalAlignment=VAlign_Center, padding=<ltrb 8 3 8 3>}}
  }
  VoiceChannelType_Text {
   widgetClassPath=&CommonTextBlock
   bIsVariable=true
   parent=WB_ReportingStatus
   widget=VoiceChannelType_Text
   properties {
    justification=Center
    mobileFontSizeMultiplier=1.41
    text=Open mic
    font {
     bForceMonospaced=false
     bMaterialIsStencil=false
     fontMaterial=&MI_Sidebar_FontColor_Secondary
     fontObject=&HeadingNow
     letterSpacing=10
     monospacedWidth=1
     size=12.75
     skewAmount=0
     typefaceFontName=74 Regular
     outlineSettings {
      bApplyOutlineToDropShadows=false
      bMiteredCorners=false
      bSeparateFillAlpha=false
      outlineMaterial=null
      outlineSize=0
      outlineColor=<rgba 0 0 0 1>
     }
    }
    shadowOffset=<0 0>
   }
   slot {class=WrapBoxSlot, properties {horizontalAlignment=HAlign_Left, verticalAlignment=VAlign_Center}}
  }
  VoiceChannelType_Icon_DividerDot {
   widgetClassPath=&Image
   bIsVariable=true
   parent=WB_ReportingStatus
   widget=VoiceChannelType_Icon_DividerDot
   properties {
    visibility=SelfHitTestInvisible
    brush {
     bIsDynamicallyLoaded=false
     drawAs=Image
     imageType=FullColor
     mirroring=NoMirror
     resourceName=None
     resourceObject=&MI_UI_Circle_ConversationBacchusConfirm
     tiling=NoTile
     imageSize=<6 6>
     margin=<ltrb 0 0 0 0>
     outlineSettings {
      bUseBrushTransparency=false
      roundingType=HalfHeightRadius
      width=0
      color {colorUseRule=UseColor_Specified, specifiedColor=<rgba 0 0 0 0>}
      cornerRadii=<0 0 0 0>
     }
     tintColor {colorUseRule=UseColor_Specified, specifiedColor=<rgba 1 1 1 1>}
     uVRegion {bIsValid=false, max=<0 0>, min=<0 0>}
    }
    colorAndOpacity=<rgba 0.5776 0.5776 0.5776 1>
   }
   slot {class=WrapBoxSlot, properties {horizontalAlignment=HAlign_Left, verticalAlignment=VAlign_Center, padding=<ltrb 8 3 8 3>}}
  }
  VoiceChatOff_Icon {
   widgetClassPath=&Image
   bIsVariable=true
   parent=WB_ReportingStatus
   widget=VoiceChatOff_Icon
   properties {
    visibility=SelfHitTestInvisible
    brush {
     bIsDynamicallyLoaded=false
     drawAs=Image
     imageType=FullColor
     mirroring=NoMirror
     resourceName=None
     resourceObject=/FortUILibrary/Textures/MaterialInstances/MI_UIKit_Icon_VoiceChatMuted
     tiling=NoTile
     imageSize=<24 24>
     margin=<ltrb 0 0 0 0>
     outlineSettings {
      bUseBrushTransparency=false
      roundingType=HalfHeightRadius
      width=0
      color {colorUseRule=UseColor_Specified, specifiedColor=<rgba 0 0 0 0>}
      cornerRadii=<0 0 0 0>
     }
     tintColor {colorUseRule=UseColor_Specified, specifiedColor=<rgba 1 1 1 1>}
     uVRegion {bIsValid=false, max=<0 0>, min=<0 0>}
    }
    colorAndOpacity=<rgba 0.5776 0.5776 0.5776 1>
   }
   slot {class=WrapBoxSlot, properties {horizontalAlignment=HAlign_Left, verticalAlignment=VAlign_Center, padding=<ltrb 0 0 3 0>}}
  }
  VoiceReportStatus_Text {
   widgetClassPath=&CommonTextBlock
   bIsVariable=true
   parent=WB_ReportingStatus
   widget=VoiceReportStatus_Text
   properties {
    bIsVolatile=true
    justification=Center
    mobileFontSizeMultiplier=1.41
    scrollStyle=&TextScrollStyle-Base_C
    text="Voice reporting {status}"
    textOverflowPolicy=Ellipsis
    wrapTextAt=550
    font {
     bForceMonospaced=false
     bMaterialIsStencil=false
     fontMaterial=&MI_Sidebar_FontColor_Secondary
     fontObject=&HeadingNow
     letterSpacing=10
     monospacedWidth=1
     size=12.75
     skewAmount=0
     typefaceFontName=74 Regular
     outlineSettings {
      bApplyOutlineToDropShadows=false
      bMiteredCorners=false
      bSeparateFillAlpha=false
      outlineMaterial=null
      outlineSize=0
      outlineColor=<rgba 0 0 0 1>
     }
    }
    shadowOffset=<0 0>
   }
   slot {class=WrapBoxSlot, properties {verticalAlignment=VAlign_Center}}
  }
  HB_Controls_New {
   widgetClassPath=&HorizontalBox
   bIsVariable=true
   parent=VerticalBox
   widget=HB_Controls_New
   properties {bOverride_Cursor=true}
   slot {class=VerticalBoxSlot, properties {padding=<ltrb 0 18 0 0>}}
  }
  Button_PartyJoinability_New {
   widgetClassPath=/Sidebar/Buttons/WBP_Sidebar_ButtonWithContextMenu
   bIsVariable=true
   parent=HB_Controls_New
   widget=Button_PartyJoinability_New
   properties {
    bNavigateToNextWidgetOnDisable=true
    minHeight=0
    mobileMinHeight=64
    optionName=Party:
    override Compact Style=true
    secondText=Value
    useLocalOptions=true
    buttonPadding=<ltrb 8 2 36 2>
    font {
     bForceMonospaced=false
     bMaterialIsStencil=false
     fontMaterial=&MI_Sidebar_FontColor_Secondary
     fontObject=&HeadingNow
     letterSpacing=10
     monospacedWidth=1
     size=13.5
     skewAmount=0
     typefaceFontName=75 Medium
     outlineSettings {
      bApplyOutlineToDropShadows=false
      bMiteredCorners=false
      bSeparateFillAlpha=false
      outlineMaterial=&MI_UIKit_TextOutline
      outlineSize=0
      outlineColor=<rgba 0 0 0 1>
     }
    }
    options {invite=Invite only, joinable=Joinable}
    secondFont {
     bForceMonospaced=false
     bMaterialIsStencil=false
     fontMaterial=&MI_Sidebar_FontColor_Secondary
     fontObject=&HeadingNow
     letterSpacing=10
     monospacedWidth=1
     size=13.5
     skewAmount=0
     typefaceFontName=75 Medium
     outlineSettings {
      bApplyOutlineToDropShadows=false
      bMiteredCorners=false
      bSeparateFillAlpha=false
      outlineMaterial=&MI_UIKit_TextOutline
      outlineSize=0
      outlineColor=<rgba 0 0 0 1>
     }
    }
    textPadding=<ltrb 0 0 8 0>
   }
   slot {class=HorizontalBoxSlot, properties {horizontalAlignment=HAlign_Left, verticalAlignment=VAlign_Top, padding=<ltrb 0 -2 0 0>}}
  }
  VoiceReportStatusNew_Text {
   widgetClassPath=&CommonTextBlock
   bIsVariable=true
   parent=HB_Controls_New
   widget=VoiceReportStatusNew_Text
   properties {
    autoWrapText=true
    bIsVolatile=true
    justification=Center
    mobileFontSizeMultiplier=1.33
    scrollStyle=&TextScrollStyle-Base_C
    text="Reporting {status}"
    textOverflowPolicy=Ellipsis
    font {
     bForceMonospaced=false
     bMaterialIsStencil=false
     fontMaterial=&MI_Sidebar_FontColor_Secondary
     fontObject=&HeadingNow
     letterSpacing=10
     monospacedWidth=1
     size=13.5
     skewAmount=0
     typefaceFontName=75 Medium
     outlineSettings {
      bApplyOutlineToDropShadows=false
      bMiteredCorners=false
      bSeparateFillAlpha=false
      outlineMaterial=null
      outlineSize=0
      outlineColor=<rgba 0 0 0 1>
     }
    }
    shadowOffset=<0 0>
   }
   slot {class=HorizontalBoxSlot, properties {verticalAlignment=VAlign_Top, size {sizeRule=Fill, value=1}}}
  }
 ]
}
