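# Sidebar social tree list, a FortPolymorphicTreeView showing players under social headers, plus party controls, pending invites and leave-party rows.
# SetTreeViewItems builds top-level rows, the children callback fills each header, and GetDesiredEntryClassForItem maps view model types to entry widgets.
# Entry clicks, hovers and player actions are re-broadcast through dispatchers.
# A console variable enables gamepad handling that reselects the last focused user by name and can defer pooled-entry regeneration.
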WidgetBlueprint /Sidebar/Panels/Social/List/WBP_Sidebar_SocialTreeList
graphs [
 EventGraph {nodes [
   n1 = "On Entry Initialized (TreeView)"!() -> (OutputDelegate, then, Item:Object Reference, Widget:User Widget Object Reference) {is=title K2Node_ComponentBoundEvent}
   > n2 = Cast To WBP_Sidebar_PartyOptionsEntry!(execute, Object=@n1.Widget) -> (then, CastFailed, AsWBP Sidebar Party Options Entry:WBP Sidebar Party Options Entry Object Reference) {is=title K2Node_DynamicCast}
   n3 = Create Event()
   > n4 = Bind Event to Leave Party Clicked(self:WBP Sidebar Party Options Entry Object Reference=@n2, Delegate=@n3) {is=title K2Node_AddDelegate}
   n5 = Cast To WBP_Sidebar_SocialPlayer!(execute=[@n4.then, @n2.CastFailed], Object=@n1.Widget) -> (then, CastFailed, AsWBP Sidebar Social Player:WBP Sidebar Social Player Object Reference) {is=title K2Node_DynamicCast}
   n6 = Create Event()
   > Bind Event to On Edit Nickname(self:WBP Sidebar Social Player Object Reference=@n5, Delegate=@n6) {is=title K2Node_AddDelegate}
   n7 = Create Event()
   > Bind Event to On Remove Friend(self:WBP Sidebar Social Player Object Reference=@n5, Delegate=@n7) {is=title K2Node_AddDelegate}
   n8 = Create Event()
   > Bind Event to On Block Player(self:WBP Sidebar Social Player Object Reference=@n5, Delegate=@n8) {is=title K2Node_AddDelegate}
   n9 = "On Entry Released (TreeView)"!() -> (OutputDelegate, then, Widget:User Widget Object Reference) {is=title K2Node_ComponentBoundEvent}
   > n10 = Cast To WBP_Sidebar_PartyOptionsEntry(Object=@n9.Widget)
   > n11 = Unbind Event from Leave Party Clicked(self:WBP Sidebar Party Options Entry Object Reference=@n10, Delegate=@n3) {is=title K2Node_RemoveDelegate}
   n12 = Cast To WBP_Sidebar_SocialPlayer(execute=[@n11.then, @n10.CastFailed], Object=@n9.Widget)
   > Unbind Event from On Edit Nickname(self:WBP Sidebar Social Player Object Reference=@n12, Delegate=@n6) {is=title K2Node_RemoveDelegate}
   > Unbind Event from On Remove Friend(self:WBP Sidebar Social Player Object Reference=@n12, Delegate=@n7) {is=title K2Node_RemoveDelegate}
   > Unbind Event from On Block Player(self:WBP Sidebar Social Player Object Reference=@n12, Delegate=@n8) {is=title K2Node_RemoveDelegate}
   n13 = "On Item Is Hovered Changed (TreeView)"!() -> (OutputDelegate, then, Item:Object Reference, bIsHovered:bool) {is=title K2Node_ComponentBoundEvent}
   > Branch(Condition=@n13.bIsHovered)
   n14 = Get TreeView() -> (TreeView:Tree View Object Reference)
   > Set Selected Item!(execute, self:List View Object Reference=@n14, Item:Object Reference=@n13.Item) -> (then) {is=title K2Node_CallFunction}
   > Branch(Condition=AND Boolean(A=Get DisableListSortDuringScroll() -> (DisableListSortDuringScroll:bool), B=Get IsSocialTab() -> (IsSocialTab:bool)))
   n15 = GetIndexForItem!(self:List View Object Reference=@n14, Item:Object Reference=@n13.Item) -> (ReturnValue:int=0)
   > Call On Item Hovered!(execute:Exec, self:WBP Sidebar Social Tree List Object Reference, Index:int=@n15) -> (then:Exec) {is=title K2Node_CallDelegate}
   > "Set Party Controls is  Hovered"("Party Controls is  Hovered":bool="Equal (Integer)"(A=@n15)) -> (Output_Get:bool)
   > n16 = Cast To FortSocialItemVM!(execute, Object=@n13.Item) -> (then, CastFailed, AsFort Social Item VM:Fort Social Item VM Object Reference) {is=title K2Node_DynamicCast}
   > Set Last focused user(Last focused user:String=GetNicknameOrDisplayName!(self:Fort Social User Object Reference=Get TargetUser!(self:Fort Social Item VM Object Reference=@n16) -> (TargetUser:Fort Social User Weak object)) -> (ReturnValue:String)) -> (Output_Get:String)
   Set Last focused user(execute=@n16.CastFailed)
   n17 = "On Item Clicked (TreeView)"!() -> (OutputDelegate, then, Item:Object Reference) {is=title K2Node_ComponentBoundEvent}
   > n18 = Cast To FortSocialItemVM(Object=@n17.Item)
   > Call Item Clicked!(execute:Exec, self:WBP Sidebar Social Tree List Object Reference, ItemVM:Fort Social Item VM Object Reference=@n18) -> (then:Exec) {is=title K2Node_CallDelegate}
   n19 = Cast To FortSocialUserViewModel!(execute=@n18.CastFailed, Object=@n17.Item) -> (then, CastFailed, AsFort SocialUser Viewmodel:Fort Social User Viewmodel Object Reference) {is=title K2Node_DynamicCast}
   > n20 = Construct Object from Class(Class=/Script/SidebarUI.FortSocialItemVM) -> (ReturnValue:Fort Social Item VM Object Reference)
   > Init!(
    execute
    self:Fort Social Item VM Object Reference=@n20
    InTargetUser:Fort Social User Object Reference=GetSocialUser!(self:Fort Social User Viewmodel Object Reference=@n19) -> (ReturnValue:Fort Social User Object Reference)
   ) -> (then)
   > Call Item Clicked(ItemVM=@n20)
   n21 = On Item EditNicknameAction!() -> (OutputDelegate, then, SocialUser:Fort Social User Object Reference) {is=title K2Node_CustomEvent}
   > Call Item Edit Nickname!(execute:Exec, self:WBP Sidebar Social Tree List Object Reference, SocialUser:Fort Social User Object Reference=@n21.SocialUser) -> (then:Exec) {is=title K2Node_CallDelegate}
   n22 = OnHandleInputMethodChanged!() -> (OutputDelegate, then, bNewInputType:ECommonInputType Enum) {is=title K2Node_CustomEvent}
   > SetReturnFocusToSelection!(
    execute
    self:List View Object Reference=Get TreeView()
    bEnabled:bool=Select!(MouseAndKeyboard:bool=false, Gamepad:bool=true, Touch:bool=false, Count:bool=false, Index:ECommonInputType Enum=@n22.bNewInputType) -> (ReturnValue:bool)
   ) -> (then)
   n23 = OnBlockPlayerAction!() -> (OutputDelegate, then, SocialUser:Fort Social User Object Reference, Interaction:Fort Social Interaction VM Object Reference) {is=title K2Node_CustomEvent}
   > Call Item Block Player!(
    execute:Exec
    self:WBP Sidebar Social Tree List Object Reference
    SocialUser:Fort Social User Object Reference=@n23.SocialUser
    Interaction:Fort Social Interaction VM Object Reference=@n23.Interaction
   ) -> (then:Exec) {is=title K2Node_CallDelegate}
   n24 = On Item RemoveFriendAction!() -> (OutputDelegate, then, SocialUser:Fort Social User Object Reference, Interaction:Fort Social Interaction VM Object Reference) {is=title K2Node_CustomEvent}
   > Call Item Remove Friend!(
    execute:Exec
    self:WBP Sidebar Social Tree List Object Reference
    SocialUser:Fort Social User Object Reference=@n24.SocialUser
    Interaction:Fort Social Interaction VM Object Reference=@n24.Interaction
   ) -> (then:Exec) {is=title K2Node_CallDelegate}
   n25 = OnLeavePartyClicked!() -> (OutputDelegate, then, Action:Name) {is=title K2Node_CustomEvent}
   > Call Leave Party Clicked!(execute:Exec, self:WBP Sidebar Social Tree List Object Reference, Action:Name=@n25.Action) -> (then:Exec) {is=title K2Node_CallDelegate}
   Event Destruct()
   n26 = Create Event()
   n27 = Get CommonInputSubsystem()
   > Unbind Event from On Input Method Changed(self=@n27, Delegate=@n26)
   Event Construct()
   > Bind Event to On Input Method Changed(self=@n27, Delegate=@n26)
   > OnHandleInputMethodChanged!(execute, self, bNewInputType:ECommonInputType Enum=GetCurrentInputType(self=@n27)) -> (then) {is=title K2Node_CallFunction}
   > n28 = GetConsoleVariableBoolValue!(execute, VariableName:String=Sidebar.DisableListSortDuringScroll) -> (then, ReturnValue:bool)
   > Set DisableListSortDuringScroll(DisableListSortDuringScroll:bool=@n28) -> (Output_Get:bool)
   Event PreConstruct()
   > SetCompactUIStyle() {is=title K2Node_CallFunction}
   Event OnInitialized()
   > n29 = Construct Object from Class(Class=/Sidebar/Panels/Social/List/Data/BP_SidebarLeavePartyEntryVM.BP_SidebarLeavePartyEntryVM_C) -> (ReturnValue:BP Sidebar Leave Party Entry VM Object Reference)
   > Set LeavePartyViewModel(LeavePartyViewModel:BP Sidebar Leave Party Entry VM Object Reference=@n29) -> (Output_Get:BP Sidebar Leave Party Entry VM Object Reference)
   > n30 = Construct Object from Class(Class=/Sidebar/Panels/Social/List/Data/BP_SidebarPendingInvitesEntryVM.BP_SidebarPendingInvitesEntryVM_C) -> (ReturnValue:BP Sidebar Pending Invites Entry VM Object Reference)
   > Set PendingInvitesVM(PendingInvitesVM:BP Sidebar Pending Invites Entry VM Object Reference=@n30) -> (Output_Get:BP Sidebar Pending Invites Entry VM Object Reference)
  ]}
 GetDesiredEntryClassForItem {nodes [
   n1 = GetDesiredEntryClassForItem!() -> (then, Item:Object Reference) {is=title K2Node_FunctionEntry}
   > Is Valid(InputObject=@n1)
   > n2 = Cast To FortTextChatGroupConversationVM!(execute, Object=@n1) -> (then, CastFailed, AsFort Group Conversation Viewmodel:Fort Group Conversation Viewmodel Object Reference) {is=title K2Node_DynamicCast}
   > Return Node!(
    execute
    ReturnValue:User Widget Class Reference=/Sidebar/Panels/Chat/WBP_Sidebar_ChatAccept_HeaderEntry.WBP_Sidebar_ChatAccept_HeaderEntry_C
   )
   n3 = Cast To BP_Sidebar_CustomHeaderEntryVM!(execute=@n2.CastFailed, Object=@n1) -> (then, CastFailed, AsBP Sidebar Custom Header Entry VM:BP Sidebar Custom Header Entry VM Object Reference) {is=title K2Node_DynamicCast}
   > Return Node(ReturnValue=&/WBP_Sidebar_CustomHeader.WBP_Sidebar_CustomHeader_C)
   n4 = Cast To BP_SidebarLeavePartyEntryVM!(execute=@n3.CastFailed, Object=@n1) -> (then, CastFailed, AsBP Sidebar Leave Party Entry VM:BP Sidebar Leave Party Entry VM Object Reference) {is=title K2Node_DynamicCast}
   > Return Node(ReturnValue=&/WBP_Sidebar_PartyOptionsEntry.WBP_Sidebar_PartyOptionsEntry_C)
   n5 = Cast To BP_SidebarPartyControlsEntryVM!(execute=@n4.CastFailed, Object=@n1) -> (then, CastFailed, AsBP Sidebar Party Controls Entry VM:BP Sidebar Party Controls Entry VM Object Reference) {is=title K2Node_DynamicCast}
   > Return Node(ReturnValue=/Sidebar/Panels/Social/WBP_PartyAndGame_Controls.WBP_PartyAndGame_Controls_C)
   n6 = Cast To BP_SidebarUnknownUsersEntryVM!(execute=@n5.CastFailed, Object=@n1) -> (then, CastFailed, AsBP Sidebar Unknown Users Entry VM:BP Sidebar Unknown Users Entry VM Object Reference) {is=title K2Node_DynamicCast}
   > Return Node(ReturnValue=&/WBP_Sidebar_UnknownUsers.WBP_Sidebar_UnknownUsers_C)
   n7 = Cast To BP_SidebarPendingInvitesEntryVM!(execute=@n6.CastFailed, Object=@n1) -> (then, CastFailed, AsBP Sidebar Pending Invites Entry VM:BP Sidebar Pending Invites Entry VM Object Reference) {is=title K2Node_DynamicCast}
   > Return Node(ReturnValue=&/WBP_Sidebar_PartyPendingInvites.WBP_Sidebar_PartyPendingInvites_C)
   n8 = Cast To FortSocialHeaderItemVM!(execute=@n7.CastFailed, Object=@n1) -> (then, CastFailed, AsFort Social Header Item VM:Fort Social Header Item VM Object Reference) {is=title K2Node_DynamicCast}
   > Return Node(ReturnValue=&/WBP_Sidebar_SocialHeader.WBP_Sidebar_SocialHeader_C)
   Cast To FortSocialItemVM(execute=@n8.CastFailed, Object=@n1)
   > Return Node(ReturnValue=&/WBP_Sidebar_SocialPlayer.WBP_Sidebar_SocialPlayer_C)
  ]}
 GetScrollOffset {nodes [
   GetScrollOffset() {is=title K2Node_FunctionEntry}
   > Return Node!(
    execute
    Scroll Offset:float=GetScrollOffset!(self:List View Base Object Reference=Get TreeView()) -> (ReturnValue:float) {is=title K2Node_CallFunction}
   )
  ]}
 Has Game Voice Channel Header {nodes [
   Inputs!() -> (Exec:Exec)
   > Set HasGameVoiceChannelHeader(HasGameVoiceChannelHeader:bool=false) -> (Output_Get:bool)
   > n1 = For Each Loop with Break!(
    Exec:Exec
    Array:Array of Fort Social Header Item VM Object References=Get OuterArray() -> (OuterArray:Array of Fort Social Header Item VM Object References)
    Break:Exec=@n3.then
   ) -> (LoopBody:Exec, Array Element:Fort Social Header Item VM Object Reference, Array Index:int, Completed:Exec)
   Outputs!(Out:Exec=@n1.Completed)
   n2 = Get HeaderItemType!(self:Fort Social Header Item VM Object Reference=@n1) -> (HeaderItemType:ESocialHeaderItemType Enum=Unknown)
   > Branch(Condition=AND Boolean(
     A="Equal (Enum)"(A:ESocialHeaderItemType Enum=@n2, B:ESocialHeaderItemType Enum=GameVoiceChannels)
     B=ShouldShowHeader!(self:Fort Social Header Item VM Object Reference=@n1) -> (ReturnValue:bool)
    ))
   > n3 = Set HasGameVoiceChannelHeader(HasGameVoiceChannelHeader=true)
  ]}
 ItemBlockPlayer {nodes [ItemBlockPlayer!() -> (then, SocialUser:Fort Social User Object Reference, Interaction:Fort Social Interaction VM Object Reference) {is=title K2Node_FunctionEntry}]}
 ItemClicked {nodes [ItemClicked!() -> (then, ItemVM:Fort Social Item VM Object Reference) {is=title K2Node_FunctionEntry}]}
 ItemEditNickname {nodes [ItemEditNickname!() -> (then, SocialUser:Fort Social User Object Reference) {is=title K2Node_FunctionEntry}]}
 ItemRemoveFriend {nodes [ItemRemoveFriend!() -> (then, SocialUser:Fort Social User Object Reference, Interaction:Fort Social Interaction VM Object Reference) {is=title K2Node_FunctionEntry}]}
 LeavePartyClicked {nodes [LeavePartyClicked!() -> (then, Action:Name) {is=title K2Node_FunctionEntry}]}
 OnItemHovered {nodes [OnItemHovered!() -> (then, Index:int) {is=title K2Node_FunctionEntry}]}
 On_TreeView_BP_OnGetItemChildren {nodes [
   n1 = On_TreeView_BP_OnGetItemChildren!() -> (then, Item:Object Reference) {is=title K2Node_FunctionEntry}
   Clear(
    execute=[@n1.then, @n27.then]
    TargetArray:Array of MVVM Base Viewmodel Object References=Get ChildrenArray() -> (ChildrenArray:Array of MVVM Base Viewmodel Object References)
   )
   > n2 = Cast To BP_SidebarPartyControlsEntryVM(Object=Get Item() -> (Item:Object Reference))
   > Return Node!(execute, Children:Array of Object References)
   n3 = Cast To BP_SidebarUnknownUsersEntryVM(execute=@n2.CastFailed, Object=Get Item())
   > Return Node()
   n4 = Cast To FortSocialHeaderItemVM(execute=@n3.CastFailed, Object=Get Item())
   n5 = Get SocialItems!(self:Fort Social Header Item VM Object Reference=@n4) -> (SocialItems:Array of Fort Social Item VM Object References)
   > n6 = For Each Loop(Array:Array of Fort Social Item VM Object References=@n5) -> (Array Element:Fort Social Item VM Object Reference)
   n7 = Cast To FortSocialEpicPartyHeaderItemVM!(execute=@n6.Completed, Object=@n4) -> (then, CastFailed, AsFort Social Epic Party Header Item VM:Fort Social Epic Party Header Item VM Object Reference) {is=title K2Node_DynamicCast}
   > n8 = Branch(Condition=Get IsSocialTab())
   > n9 = Branch(Condition=HasPendingInvites!(self:Fort Social Epic Party Header Item VM Object Reference=@n7) -> (ReturnValue:bool))
   > n10 = Add(
    TargetArray:Array of MVVM Base Viewmodel Object References=Get ChildrenArray()
    NewItem:"MVVM Base Viewmodel Object Reference (by ref)"=Get PendingInvitesVM() -> (PendingInvitesVM:BP Sidebar Pending Invites Entry VM Object Reference)
   ) {is=title K2Node_CallArrayFunction}
   n11 = CanLeaveParty!(self:Fort Social Epic Party Header Item VM Object Reference=@n7) -> (ReturnValue:bool)
   n12 = LeavePartyActionWillLogout!(self:Fort Social Epic Party Header Item VM Object Reference=@n7) -> (ReturnValue:bool)
   n13 = Branch(execute=[@n10.then, @n9.else], Condition=OR Boolean(A=@n11, B=@n12))
   n14 = Get LeavePartyViewModel() -> (LeavePartyViewModel:BP Sidebar Leave Party Entry VM Object Reference)
   > Set with Broadcast isSplitScreenLeaveAction!(execute, isSplitScreenLeaveAction:bool=@n12, self:BP Sidebar Leave Party Entry VM Object Reference=@n14) -> (then, Output_Get:bool)
   > Set with Broadcast isCanLeaveParty!(execute, isCanLeaveParty:bool=@n11, self:BP Sidebar Leave Party Entry VM Object Reference=@n14) -> (then, Output_Get:bool)
   > n15 = Add(TargetArray=Get ChildrenArray(), NewItem=@n14)
   Return Node(execute=[@n15.then, @n13.else, @n8.else, @n7.CastFailed], Children=Get ChildrenArray())
   n16 = Get HeaderItemType(self=@n4) -> (HeaderItemType=Unknown)
   n17 = "Equal (Integer)"(A=Last Index!(TargetArray:Array of Fort Social Item VM Object References=@n5) -> (ReturnValue=0), B=@n6.Array Index)
   n18 = Branch(execute=@n6.LoopBody, Condition=AND Boolean(A="Equal (Enum)"(A=@n16, B=TeamMembers), B=@n17))
   > n19 = Set bIsChecked!(
    execute
    bIsChecked:bool=NOT Boolean(A=Get HasGameVoiceChannelHeader() -> (HasGameVoiceChannelHeader:bool))
    self:Fort Social Item VM Object Reference=@n6
   ) -> (then, Output_Get:bool)
   Add(execute=[@n25.then, @n24.else, @n19.then], TargetArray=Get ChildrenArray(), NewItem=@n6)
   > n20 = Get!(execute:Exec) -> (then:Exec, else:Exec, FortInputSystemVM:Fort Input System VM Object Reference)
   > Branch(Condition="Equal (Enum)"(
     A:ECommonInputType Enum=Get CommonInputType!(self:Fort Input System VM Object Reference=@n20) -> (CommonInputType:ECommonInputType Enum)
     B:ECommonInputType Enum=Gamepad
    ))
   n21 = GetNicknameOrDisplayName(self=Get TargetUser(self=@n6))
   n22 = Get Last focused user() -> (Last focused user:String)
   > Branch(Condition=AND Boolean(A="Equal Exactly (String)"(A=@n22, B=@n21), B=Get DisableListSortDuringScroll()))
   n23 = Get TreeView()
   > Set Selected Item(self=@n23, Item=@n6)
   > Navigate To Item!(execute, self:List View Object Reference=@n23, Item:Object Reference=@n6) -> (then)
   > Set Selected Item(self=Get TreeView(), Item=@n6)
   n24 = Branch(
    execute=@n18.else
    Condition=AND Boolean(A="Equal (Enum)"(A:ESocialHeaderItemType Enum=@n16, B:ESocialHeaderItemType Enum=GameVoiceChannels), B=@n17)
   )
   > n25 = Set bIsChecked(bIsChecked=true, self=@n6)
   n26 = Cast To BP_Sidebar_CustomHeaderEntryVM(execute=@n4.CastFailed, Object=Get Item())
   > Return Node(Children=Get ChildrenItems!(self:BP Sidebar Custom Header Entry VM Object Reference=@n26) -> (ChildrenItems:Array of MVVM Base Viewmodel Object References))
   n27 = Set Focused(Focused:bool=false) -> (Output_Get:bool)
  ]}
 SendItemClickedUserToPanel {nodes [SendItemClickedUserToPanel!() -> (then, SocialTargetUser:Fort Social User Object Reference) {is=title K2Node_FunctionEntry}]}
 SetCompactUIStyle {nodes [
   SetCompactUIStyle() {is=title K2Node_FunctionEntry}
   > Branch(Condition=IsUsingCompactUI())
   > Branch(Condition=IsUsingInputType!(InputType:ECommonInputType Enum=Touch, PlayerController:Player Controller Object Reference=GetOwningPlayer()) -> (ReturnValue:bool))
   > SetAllowOverScroll!(execute, self:List View Base Object Reference=Get TreeView(), NewAllowOverscroll:bool=false) -> (then)
  ]}
 SetScrollOffset {nodes [
   n1 = SetScrollOffset!() -> (then, ScrollOffset:float=~) {is=title K2Node_FunctionEntry}
   > SetScrollOffset!(execute, self:List View Base Object Reference=Get TreeView(), InScrollOffset:float=@n1) -> (then) {is=title K2Node_CallFunction}
  ]}
 SetTreeViewItems {nodes [
   SetTreeViewItems!() -> (then, OuterArray:Array of Fort Social Header Item VM Object References) {is=title K2Node_FunctionEntry}
   > n1 = Sequence!(execute) -> (then_0, then_1, then_2:Exec)
   n2 = Get HasAddedVoiceControlsEntry() -> (HasAddedVoiceControlsEntry:bool)
   > Branch(Condition=AND Boolean(A=Get IsSocialTab(), B=NOT Boolean(A=@n2)))
   > ClearListItems(self=Get TreeView())
   > n3 = Construct Object from Class(Class=/Sidebar/Panels/Social/List/Data/BP_SidebarPartyControlsEntryVM.BP_SidebarPartyControlsEntryVM_C) -> (ReturnValue:BP Sidebar Party Controls Entry VM Object Reference)
   > AddItem!(execute, self:List View Object Reference=Get TreeView(), Item:Object Reference=@n3) -> (then)
   > Set HasAddedVoiceControlsEntry(HasAddedVoiceControlsEntry:bool=true) -> (Output_Get:bool)
   Set EpicPartyHeaderItemViewModel(execute=@n1.then_1, EpicPartyHeaderItemViewModel:Fort Social Epic Party Header Item VM Object Reference) -> (Output_Get:Fort Social Epic Party Header Item VM Object Reference)
   > n4 = Reverse for Each Loop!(
    Exec:Exec
    Array:Array of Object References=GetListItems!(self:List View Object Reference=Get TreeView()) -> (ReturnValue:Array of Object References)
   ) -> (LoopBody:Exec, ArrayIndex:int, ArrayElement:Object Reference, Completed:Exec) {is=title K2Node_MacroInstance}
   > n5 = Cast To BP_SidebarPartyControlsEntryVM(Object=@n4.ArrayElement)
   > Reroute Node()
   RemoveItem!(execute=@n5.CastFailed, self:List View Object Reference=Get TreeView(), Item:Object Reference=@n4.ArrayElement) -> (then)
   n6 = For Each Loop(Exec=@n1.then_2, Array:Array of Fort Social Header Item VM Object References=Get OuterArray()) -> (Array Element:Fort Social Header Item VM Object Reference)
   ExpandAll(execute=@n6.Completed, self:Tree View Object Reference=Get TreeView())
   > n7 = GetConsoleVariableBoolValue(VariableName=SocialUI.ReuseSocialItemVMs)
   n8 = GetRuntimeOptions()
   > Branch(Condition=OR Boolean(A=@n7, B=Get bUsePoolingForSocialVMs!(self:Fort Runtime Options Object Reference=@n8) -> (bUsePoolingForSocialVMs:bool)))
   n9 = "Get Party Controls is  Hovered"() -> ("Party Controls is  Hovered":bool)
   n10 = Get CommonInputType(self=Get FortInputSystemVM() -> (FortInputSystemVM:Fort Input System VM Object Reference))
   > n11 = Branch(Condition=AND Boolean!(
     A="Equal (Enum)"(A:ECommonInputType Enum=@n10, B:ECommonInputType Enum=Gamepad)
     B=Get DisableListSortDuringScroll()
     C:bool=NOT Boolean(A=@n9)
    ) -> (ReturnValue))
   > Set Regenerate On Activated(Regenerate On Activated:bool=true) -> (Output_Get:bool)
   RegenerateAllEntries(execute=@n11.else, self:Fort Polymorphic View Object Reference)
   Branch(execute=@n6.LoopBody, Condition=ShouldShowHeader(self=@n6))
   n12 = Get HeaderItemType(self=@n6) -> (HeaderItemType=Unknown)
   n13 = "Equal (Enum)"(A:ESocialHeaderItemType Enum=@n12, B:ESocialHeaderItemType Enum=PartyMembers)
   n14 = "Equal (Enum)"(A=@n12, B=NativeEpicPartyMembers)
   n15 = Get SocialItems(self=@n6)
   n16 = Length(TargetArray:Array of Fort Social Item VM Object References=@n15)
   > n17 = Branch(Condition=AND Boolean!(A="Equal (Integer)"(A=@n16, B=1), B=OR Boolean(A=@n13, B=@n14)) -> (ReturnValue))
   > n18 = Cast To FortSocialEpicPartyHeaderItemVM(Object=@n6)
   > n19 = Branch(Condition=NOT Boolean(A=CanLeaveParty(self=@n18)))
   n20 = "Get (a copy)"(Array:Array of Fort Social Item VM Object References=@n15) -> (Output:Fort Social Item VM Object Reference)
   > Set bIsChecked(bIsChecked=true, self=@n20)
   > AddItem(self=Get TreeView(), Item=@n20)
   n21 = Set bIsChecked(execute=[@n19.else, @n18.CastFailed], bIsChecked=false, self=@n20)
   AddItem(execute=[@n21.then, @n17.else], self=Get TreeView(), Item=@n6)
   > n22 = Cast To FortSocialEpicPartyHeaderItemVM(Object=@n6)
   n23 = Get HeaderItemType(self=@n6) -> (HeaderItemType=Unknown)
   > Branch(Condition=OR Boolean(A="Equal (Enum)"(A=@n23, B=PartyMembers), B="Equal (Enum)"(A=@n23, B=NativeEpicPartyMembers)))
   > Set EpicPartyHeaderItemViewModel(EpicPartyHeaderItemViewModel=@n22)
  ]}
]
variables [
 DisableListSortDuringScroll {type=boolean}
 EpicPartyHeaderItemViewModel {type=FortSocialEpicPartyHeaderItemVM, default=null}
 HasAddedVoiceControlsEntry {type=boolean}
 HasGameVoiceChannelHeader {type=boolean}
 IsSocialTab {type=boolean}
 ItemBlockPlayer {}
 ItemClicked {}
 ItemEditNickname {}
 ItemRemoveFriend {}
 Last focused user {type=string, default=""}
 LeavePartyClicked {}
 LeavePartyViewModel {type=BP_SidebarLeavePartyEntryVM_C, default=null}
 OnItemHovered {}
 "Party Controls is  Hovered" {type=boolean}
 PendingInvitesVM {type=BP_SidebarPendingInvitesEntryVM_C, default=null}
 Regenerate On Activated {type=boolean}
 SendItemClickedUserToPanel {}
]
widgets {
 info {
  inheritedWidgetCount=1
  parentClass=/Script/FortniteUI.FortPolymorphicTreeView
  rootWidgetClass=/Script/CommonUI.CommonTreeView
  widgetCount=1
 }
 widgets [TreeView {
   widgetClassPath=&CommonTreeView
   bInherited=true
   bIsVariable=true
   widget=TreeView
   properties {
    bEnableScrollAnimation=false
    bReturnFocusToSelection=true
    clipping=Inherit
    entryWidgetClass=&WBP_Sidebar_SocialPlayer_C
    numDesignerPreviewEntries=20
    verticalEntrySpacing=10
    scrollBarStyle {
     thickness=8
     draggedThumbImage {
      bIsDynamicallyLoaded=false
      drawAs=Box
      imageType=FullColor
      mirroring=NoMirror
      resourceName=../../../Engine/Content/Slate/Common/Scrollbar_Thumb.png
      resourceObject=null
      tiling=NoTile
      imageSize=<0 0>
      margin=<ltrb 0.25 0.25 0.25 0.25>
      outlineSettings {
       bUseBrushTransparency=false
       roundingType=HalfHeightRadius
       width=0
       color {colorUseRule=UseColor_Specified, specifiedColor=<rgba 0 0 0 0>}
       cornerRadii=<0 0 0 0>
      }
      tintColor {colorUseRule=UseColor_Specified, specifiedColor=<rgba 1 1 1 1>}
      uVRegion {bIsValid=false, max=<0 0>, min=<0 0>}
     }
     horizontalBackgroundImage {
      bIsDynamicallyLoaded=false
      drawAs=NoDrawType
      imageType=NoImage
      mirroring=NoMirror
      resourceName=None
      resourceObject=null
      tiling=NoTile
      imageSize=<0 0>
      margin=<ltrb 0 0 0 0>
      outlineSettings {
       bUseBrushTransparency=false
       roundingType=HalfHeightRadius
       width=0
       color {colorUseRule=UseColor_Specified, specifiedColor=<rgba 0 0 0 0>}
       cornerRadii=<0 0 0 0>
      }
      tintColor {colorUseRule=UseColor_Specified, specifiedColor=<rgba 1 1 1 1>}
      uVRegion {bIsValid=false, max=<0 0>, min=<0 0>}
     }
     horizontalBottomSlotImage {
      bIsDynamicallyLoaded=false
      drawAs=NoDrawType
      imageType=FullColor
      mirroring=NoMirror
      resourceName=../../../Engine/Content/Slate/Common/Scrollbar_Background_Horizontal.png
      resourceObject=null
      tiling=NoTile
      imageSize=<8 8>
      margin=<ltrb 0 0 0 0>
      outlineSettings {
       bUseBrushTransparency=false
       roundingType=HalfHeightRadius
       width=0
       color {colorUseRule=UseColor_Specified, specifiedColor=<rgba 0 0 0 0>}
       cornerRadii=<0 0 0 0>
      }
      tintColor {colorUseRule=UseColor_Specified, specifiedColor=<rgba 1 1 1 1>}
      uVRegion {bIsValid=false, max=<0 0>, min=<0 0>}
     }
     horizontalTopSlotImage {
      bIsDynamicallyLoaded=false
      drawAs=NoDrawType
      imageType=FullColor
      mirroring=NoMirror
      resourceName=../../../Engine/Content/Slate/Common/Scrollbar_Background_Horizontal.png
      resourceObject=null
      tiling=NoTile
      imageSize=<8 8>
      margin=<ltrb 0 0 0 0>
      outlineSettings {
       bUseBrushTransparency=false
       roundingType=HalfHeightRadius
       width=0
       color {colorUseRule=UseColor_Specified, specifiedColor=<rgba 0 0 0 0>}
       cornerRadii=<0 0 0 0>
      }
      tintColor {colorUseRule=UseColor_Specified, specifiedColor=<rgba 1 1 1 1>}
      uVRegion {bIsValid=false, max=<0 0>, min=<0 0>}
     }
     hoveredThumbImage {
      bIsDynamicallyLoaded=false
      drawAs=Box
      imageType=FullColor
      mirroring=NoMirror
      resourceName=../../../Engine/Content/Slate/Common/Scrollbar_Thumb.png
      resourceObject=null
      tiling=NoTile
      imageSize=<0 0>
      margin=<ltrb 0.25 0.25 0.25 0.25>
      outlineSettings {
       bUseBrushTransparency=false
       roundingType=HalfHeightRadius
       width=0
       color {colorUseRule=UseColor_Specified, specifiedColor=<rgba 0 0 0 0>}
       cornerRadii=<0 0 0 0>
      }
      tintColor {colorUseRule=UseColor_Specified, specifiedColor=<rgba 1 1 1 1>}
      uVRegion {bIsValid=false, max=<0 0>, min=<0 0>}
     }
     normalThumbImage {
      bIsDynamicallyLoaded=false
      drawAs=Box
      imageType=FullColor
      mirroring=NoMirror
      resourceName=../../../Engine/Content/Slate/Common/Scrollbar_Thumb.png
      resourceObject=null
      tiling=NoTile
      imageSize=<0 0>
      margin=<ltrb 0.25 0.25 0.25 0.25>
      outlineSettings {
       bUseBrushTransparency=false
       roundingType=HalfHeightRadius
       width=0
       color {colorUseRule=UseColor_Specified, specifiedColor=<rgba 0 0 0 0>}
       cornerRadii=<0 0 0 0>
      }
      tintColor {colorUseRule=UseColor_Specified, specifiedColor=<rgba 1 1 1 1>}
      uVRegion {bIsValid=false, max=<0 0>, min=<0 0>}
     }
     verticalBackgroundImage {
      bIsDynamicallyLoaded=false
      drawAs=NoDrawType
      imageType=NoImage
      mirroring=NoMirror
      resourceName=None
      resourceObject=null
      tiling=NoTile
      imageSize=<0 0>
      margin=<ltrb 0 0 0 0>
      outlineSettings {
       bUseBrushTransparency=false
       roundingType=HalfHeightRadius
       width=0
       color {colorUseRule=UseColor_Specified, specifiedColor=<rgba 0 0 0 0>}
       cornerRadii=<0 0 0 0>
      }
      tintColor {colorUseRule=UseColor_Specified, specifiedColor=<rgba 1 1 1 1>}
      uVRegion {bIsValid=false, max=<0 0>, min=<0 0>}
     }
     verticalBottomSlotImage {
      bIsDynamicallyLoaded=false
      drawAs=NoDrawType
      imageType=FullColor
      mirroring=NoMirror
      resourceName=../../../Engine/Content/Slate/Common/Scrollbar_Background_Vertical.png
      resourceObject=null
      tiling=NoTile
      imageSize=<8 8>
      margin=<ltrb 0 0 0 0>
      outlineSettings {
       bUseBrushTransparency=false
       roundingType=HalfHeightRadius
       width=0
       color {colorUseRule=UseColor_Specified, specifiedColor=<rgba 0 0 0 0>}
       cornerRadii=<0 0 0 0>
      }
      tintColor {colorUseRule=UseColor_Specified, specifiedColor=<rgba 1 1 1 1>}
      uVRegion {bIsValid=false, max=<0 0>, min=<0 0>}
     }
     verticalTopSlotImage {
      bIsDynamicallyLoaded=false
      drawAs=NoDrawType
      imageType=FullColor
      mirroring=NoMirror
      resourceName=../../../Engine/Content/Slate/Common/Scrollbar_Background_Vertical.png
      resourceObject=null
      tiling=NoTile
      imageSize=<8 8>
      margin=<ltrb 0 0 0 0>
      outlineSettings {
       bUseBrushTransparency=false
       roundingType=HalfHeightRadius
       width=0
       color {colorUseRule=UseColor_Specified, specifiedColor=<rgba 0 0 0 0>}
       cornerRadii=<0 0 0 0>
      }
      tintColor {colorUseRule=UseColor_Specified, specifiedColor=<rgba 1 1 1 1>}
      uVRegion {bIsValid=false, max=<0 0>, min=<0 0>}
     }
    }
   }
  }]
}
